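# Control Rig for the M_MED_GuineaPig player body cosmetic that drives corrective deform bones around joints and muscles such as glutes, groin, elbows, lats, pecs, knees and wrists.
# Per-region functions weight helper bones with parent constraints or aims, then apply tuned offsets scaled by spherical pose readers or remapped joint rotation deltas.
# Right-side bones reuse left-side offsets through a mirror helper, and optional branches write each alpha to root animation attributes.
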ControlRigBlueprint /BRCosmetics/Characters/Player/Male/Medium/Bodies/M_MED_GuineaPig/Meshes/M_MED_GuineaPig_CtrlRig
graphs [
 ContainedGraph {nodes [
   n1 = RigVMFunctionEntryNode()
   n2 = /Script/RigVM.RigVMFunction_MathTransformToEulerTransform(Value=@n1.transform) {is=struct RigVMUnitNode, template="Cast::Execute(in Value,out Result)"}
   > RigVMFunctionReturnNode(
    mirrorTransform=(Rotation=(X=0,Y=0,Z=0,W=1),Translation=(X=0,Y=0,Z=0),Scale3D=(X=0,Y=0,Z=0))
    mirrorTransform.Rotation=/Script/RigVM.RigVMFunction_MathQuaternionFromRotatorV2(Value=@n2.Result.Rotation) {is=struct RigVMUnitNode, template="Cast::Execute(in Value,out Result)"}.Result
    mirrorTransform.Translation=/Script/RigVM.RigVMFunction_MathVectorScale(Factor=-1, Value=@n2.Result.Location) {is=struct RigVMUnitNode, template="Scale::Execute(in Value,in Factor,out Result)"}.Result
    mirrorTransform.Scale3D=@n2.Result.Scale
   )
  ]}
 ContainedGraph {nodes [
   n1 = RigVMFunctionEntryNode()
   n2 = &RigUnit_GetTransform(Space=LocalSpace, Item=@n1.Item)
   n3 = &RigUnit_GetTransform(Space=LocalSpace, bInitial=true, Item=@n1.Item)
   > RigVMFunctionReturnNode(OutTransform=&RigVMFunction_MathTransformToEulerTransform(Value=&RigVMFunction_MathTransformMakeRelative(Global=@n2.Transform, Parent=@n3.Transform).Local) {template="Cast::Execute(in Value,out Result)"}.Result)
  ]}
 ContainedGraph {nodes [
   n1 = RigVMFunctionEntryNode()
   > RigVMVariableNode(Value=@n1.value_struct) {variable=value_struct_var}
   > RigVMVariableNode(Value=@n1.curve_struct) {variable=curve_struct_var}
   > SequenceExecution = RigVMAggregateNode()
   n2 = &RigUnit_SphericalPoseReader(
    ActiveRegionScaleFactors=(PositiveWidth=3.5,NegativeWidth=0.5,PositiveHeight=3,NegativeHeight=3)
    ActiveRegionSize=0.1
    Debug=(bDrawDebug=false,DebugScale=25,DebugSegments=20,DebugThickness=0.25)
    DriverAxis=(X=-1,Y=0,Z=0)
    DriverItem=(Type=Bone,Name="thigh_l")
    FalloffRegionScaleFactors=(PositiveWidth=1,NegativeWidth=1,PositiveHeight=1,NegativeHeight=1)
    FalloffSize=0.45
    OptionalParentItem=(Type=Bone)
    RotationOffset=(X=90,Y=0,Z=0)
    ExecutePin=@SequenceExecution.A
    Debug.bDrawDebug=@n1.debug
   )
   > n3 = &RigUnit_SphericalPoseReader(
    ActiveRegionScaleFactors=(PositiveWidth=3.5,NegativeWidth=0.5,PositiveHeight=3,NegativeHeight=3)
    ActiveRegionSize=0.1
    Debug=(bDrawDebug=false,DebugScale=25,DebugSegments=20,DebugThickness=0.25)
    DriverAxis=(X=1,Y=0,Z=0)
    DriverItem=(Type=Bone,Name="thigh_r")
    FalloffRegionScaleFactors=(PositiveWidth=1,NegativeWidth=1,PositiveHeight=1,NegativeHeight=1)
    FalloffSize=0.45
    FlipHeightScaling=false
    FlipWidthScaling=true
    OptionalParentItem=(Type=Bone)
    RotationOffset=(X=-90,Y=0,Z=0)
    Debug.bDrawDebug=@n1.debug
   )
   n4 = RigVMVariableNode() {variable=curve_struct_var}
   n5 = &RigVMFunction_ControlFlowBranch(ExecuteContext=@n3.ExecutePin, Condition=@n4.Value.deformRig_l_thigh_up_alpha_88_#1)
   n6 = RigVMDispatchNode(BoneName=root, Name=deformRig_l_thigh_up_alpha, ExecuteContext=@n5.True, Value=@n2.OutputParam) {template="DISPATCH_RigDispatch_SetAnimAttribute(in Name,in BoneName,in Value,out Success)"}
   n7 = &RigVMFunction_ControlFlowBranch(ExecuteContext=@n6.ExecuteContext, Condition=@n4.Value.deformRig_r_thigh_up_alpha_89_#2)
   RigVMDispatchNode(BoneName=root, Name=deformRig_r_thigh_up_alpha, ExecuteContext=@n7.True, Value=@n3.OutputParam) {template="DISPATCH_RigDispatch_SetAnimAttribute(in Name,in BoneName,in Value,out Success)"}
   RigVMFunctionReturnNode(ExecuteContext=@SequenceExecution.D)
   n8 = &RigVMFunction_ControlFlowBranch(ExecuteContext=@SequenceExecution.C, Condition=@n1.glute)
   SequenceExecution_1_1 = RigVMAggregateNode(ExecuteContext=@n8.True)
   &RigUnit_ParentConstraint(
    AdvancedSettings=(RotationOrderForFilter=XZY)
    Child=(Type=Bone,Name="deform_glute_l")
    Parents=((Item=(Type=Bone,Name="thigh_l"),Weight=0.25),(Item=(Type=Bone,Name="pelvis"),Weight=0.75))
    Weight=1
    ExecutePin=@SequenceExecution_1_1.A
   )
   n9 = RigVMVariableNode() {variable=value_struct_var}
   > &RigUnit_ModifyTransforms(
    ItemToModify=((Item=(Type=Bone,Name="deform_glute_l"),Transform=(Rotation=(X=0,Y=0,Z=0,W=1),Translation=(X=0,Y=0,Z=0),Scale3D=(X=1,Y=1,Z=1))))
    ItemToModify.0.Transform=@n9.Value.glute_36_#3
    Weight=@n2.OutputParam
   )
   &RigUnit_ParentConstraint(
    AdvancedSettings=(RotationOrderForFilter=XZY)
    Child=(Type=Bone,Name="deform_glute_r")
    Parents=((Item=(Type=Bone,Name="thigh_r"),Weight=0.25),(Item=(Type=Bone,Name="pelvis"),Weight=0.75))
    Weight=1
    ExecutePin=@SequenceExecution_1_1.B
   )
   > DeformRig_GetMirrorTransform_FNC_1 = RigVMFunctionReferenceNode(transform=@n9.Value.glute_36_#3)
   > &RigUnit_ModifyTransforms(
    ItemToModify=((Item=(Type=Bone,Name="deform_glute_r"),Transform=(Rotation=(X=0,Y=0,Z=0,W=1),Translation=(X=0,Y=0,Z=0),Scale3D=(X=1,Y=1,Z=1))))
    ItemToModify.0.Transform=@DeformRig_GetMirrorTransform_FNC_1.mirrorTransform
    Weight=@n3.OutputParam
   )
   RigVMFunction_ControlFlowBranch = &RigVMFunction_ControlFlowBranch(ExecuteContext=@SequenceExecution.B, Condition=@n1.groin)
   SequenceExecution_1 = RigVMAggregateNode(ExecuteContext=@RigVMFunction_ControlFlowBranch.True)
   &RigUnit_ParentConstraint(
    AdvancedSettings=(RotationOrderForFilter=XZY)
    Child=(Type=Bone,Name="deform_groin_l")
    Parents=((Item=(Type=Bone,Name="thigh_l"),Weight=0.7),(Item=(Type=Bone,Name="pelvis"),Weight=0.3))
    Weight=1
    ExecutePin=@SequenceExecution_1.A
   )
   n10 = RigVMVariableNode() {variable=value_struct_var}
   > &RigUnit_ModifyTransforms(
    ItemToModify=((Item=(Type=Bone,Name="deform_groin_l"),Transform=(Rotation=(X=0,Y=0,Z=0,W=1),Translation=(X=0,Y=0,Z=0),Scale3D=(X=1,Y=1,Z=1))))
    ItemToModify.0.Transform=@n10.Value.groin_35_#4
    Weight=@n2.OutputParam
   )
   &RigUnit_ParentConstraint(
    AdvancedSettings=(RotationOrderForFilter=XZY)
    Child=(Type=Bone,Name="deform_groin_r")
    Parents=((Item=(Type=Bone,Name="thigh_r"),Weight=0.7),(Item=(Type=Bone,Name="pelvis"),Weight=0.3))
    Weight=1
    ExecutePin=@SequenceExecution_1.B
   )
   > DeformRig_GetMirrorTransform_FNC = RigVMFunctionReferenceNode(transform=@n10.Value.groin_35_#4)
   > &RigUnit_ModifyTransforms(
    ItemToModify=((Item=(Type=Bone,Name="deform_groin_r"),Transform=(Rotation=(X=0,Y=0,Z=0,W=1),Translation=(X=0,Y=0,Z=0),Scale3D=(X=1,Y=1,Z=1))))
    ItemToModify.0.Transform=@DeformRig_GetMirrorTransform_FNC.mirrorTransform
    Weight=@n3.OutputParam
   )
  ]}
 ContainedGraph {nodes [
   n1 = RigVMFunctionEntryNode()
   SequenceExecution = &RigVMFunction_Sequence(ExecuteContext=@n1.ExecuteContext)
   SequenceExecution_1 = &RigVMFunction_Sequence(ExecuteContext=@SequenceExecution.B)
   SequenceExecution_2 = &RigVMFunction_Sequence(ExecuteContext=@SequenceExecution_1.B)
   RigVMFunctionReturnNode(A=@SequenceExecution.A, B=@SequenceExecution_1.A, C=@SequenceExecution_2.A, D=@SequenceExecution_2.B)
  ]}
 ContainedGraph {nodes [
   n1 = RigVMFunctionEntryNode()
   SequenceExecution_1 = &RigVMFunction_Sequence(ExecuteContext=@n1.ExecuteContext)
   SequenceExecution_1_1 = &RigVMFunction_Sequence(ExecuteContext=@SequenceExecution_1.B)
   SequenceExecution_1_2 = &RigVMFunction_Sequence(ExecuteContext=@SequenceExecution_1_1.B)
   RigVMFunctionReturnNode(A=@SequenceExecution_1.A, B=@SequenceExecution_1_1.A, C=@SequenceExecution_1_2.A, D=@SequenceExecution_1_2.B)
  ]}
 ContainedGraph {nodes [
   n1 = RigVMFunctionEntryNode()
   SequenceExecution_1_1 = &RigVMFunction_Sequence(ExecuteContext=@n1.ExecuteContext)
   SequenceExecution_1_1_1 = &RigVMFunction_Sequence(ExecuteContext=@SequenceExecution_1_1.B)
   SequenceExecution_1_1_2 = &RigVMFunction_Sequence(ExecuteContext=@SequenceExecution_1_1_1.B)
   RigVMFunctionReturnNode(A=@SequenceExecution_1_1.A, B=@SequenceExecution_1_1_1.A, C=@SequenceExecution_1_1_2.A, D=@SequenceExecution_1_1_2.B)
  ]}
 ContainedGraph {nodes [
   n1 = RigVMFunctionEntryNode()
   > RigVMVariableNode(Value=@n1.value_struct) {variable=value_struct_var}
   > RigVMVariableNode(Value=@n1.curve_struct) {variable=curve_struct_var}
   > SequenceExecution_2 = RigVMAggregateNode()
   DeformRig_GetDeltaRefTransform_FNC = RigVMFunctionReferenceNode(Item=(Type=Bone,Name="lowerarm_l"), ExecuteContext=@SequenceExecution_2.A)
   n2 = RigVMVariableNode() {variable=curve_struct_var}
   n3 = &RigVMFunction_ControlFlowBranch(
    ExecuteContext=@DeformRig_GetDeltaRefTransform_FNC.ExecuteContext
    Condition=@n2.Value.deformRig_l_elbow_alpha_86_#5
   )
   n4 = &RigVMFunction_MathFloatRemap(SourceMaximum=90, SourceMinimum=15, bClamp=true, Value=@DeformRig_GetDeltaRefTransform_FNC.OutTransform.Rotation.Yaw)
   RigVMDispatchNode(BoneName=root, Name=deformRig_l_elbow_alpha, Value=@n4.Result, ExecuteContext=@n3.True) {template="DISPATCH_RigDispatch_SetAnimAttribute(in Name,in BoneName,in Value,out Success)"}
   DeformRig_GetDeltaRefTransform_FNC_1 = RigVMFunctionReferenceNode(Item=(Type=Bone,Name="lowerarm_r"), ExecuteContext=@SequenceExecution_2.B)
   n5 = &RigVMFunction_ControlFlowBranch(
    ExecuteContext=@DeformRig_GetDeltaRefTransform_FNC_1.ExecuteContext
    Condition=@n2.Value.deformRig_r_elbow_alpha_87_#6
   )
   n6 = &RigVMFunction_MathFloatRemap(SourceMaximum=90, SourceMinimum=15, bClamp=true, Value=@DeformRig_GetDeltaRefTransform_FNC_1.OutTransform.Rotation.Yaw)
   RigVMDispatchNode(BoneName=root, Name=deformRig_r_elbow_alpha, Value=@n6.Result, ExecuteContext=@n5.True) {template="DISPATCH_RigDispatch_SetAnimAttribute(in Name,in BoneName,in Value,out Success)"}
   RigVMFunction_ControlFlowBranch = &RigVMFunction_ControlFlowBranch(ExecuteContext=@SequenceExecution_2.C, Condition=@n1.elbow)
   SequenceExecution = RigVMAggregateNode(ExecuteContext=@RigVMFunction_ControlFlowBranch.True)
   SequenceExecution_1 = RigVMAggregateNode(ExecuteContext=@SequenceExecution.A)
   &RigUnit_ParentConstraint(
    AdvancedSettings=(RotationOrderForFilter=XZY)
    Child=(Type=Bone,Name="deform_elbow_in_l")
    Parents=((Item=(Type=Bone,Name="lowerarm_l"),Weight=0.4),(Item=(Type=Bone,Name="upperarm_l"),Weight=0.6))
    Weight=1
    ExecutePin=@SequenceExecution_1.A
   )
   n7 = RigVMVariableNode() {variable=value_struct_var}
   > &RigUnit_ModifyTransforms(
    ItemToModify=((Item=(Type=Bone,Name="deform_elbow_in_l"),Transform=(Rotation=(X=0,Y=0,Z=0,W=1),Translation=(X=0,Y=0,Z=0),Scale3D=(X=1,Y=1,Z=1))))
    ItemToModify.0.Transform=@n7.Value.elbow_in_24_#7
    Weight=@n4.Result
   )
   &RigUnit_ParentConstraint(
    AdvancedSettings=(RotationOrderForFilter=XZY)
    Child=(Type=Bone,Name="deform_elbow_in_r")
    Parents=((Item=(Type=Bone,Name="lowerarm_r"),Weight=0.4),(Item=(Type=Bone,Name="upperarm_r"),Weight=0.6))
    Weight=1
    ExecutePin=@SequenceExecution_1.B
   )
   > DeformRig_GetMirrorTransform_FNC_1 = RigVMFunctionReferenceNode(transform=@n7.Value.elbow_in_24_#7)
   > &RigUnit_ModifyTransforms(
    ItemToModify=((Item=(Type=Bone,Name="deform_elbow_in_r"),Transform=(Rotation=(X=0,Y=0,Z=0,W=1),Translation=(X=0,Y=0,Z=0),Scale3D=(X=1,Y=1,Z=1))))
    ItemToModify.0.Transform=@DeformRig_GetMirrorTransform_FNC_1.mirrorTransform
    Weight=@n6.Result
   )
   SequenceExecution_1_1 = RigVMAggregateNode(ExecuteContext=@SequenceExecution.B)
   &RigUnit_ParentConstraint(
    AdvancedSettings=(RotationOrderForFilter=XZY)
    Child=(Type=Bone,Name="deform_elbow_out_l")
    Parents=((Item=(Type=Bone,Name="lowerarm_l"),Weight=0.3),(Item=(Type=Bone,Name="upperarm_l"),Weight=0.7))
    Weight=1
    ExecutePin=@SequenceExecution_1_1.A
   )
   n8 = RigVMVariableNode() {variable=value_struct_var}
   > &RigUnit_ModifyTransforms(
    ItemToModify=((Item=(Type=Bone,Name="deform_elbow_out_l"),Transform=(Rotation=(X=0,Y=0,Z=0,W=1),Translation=(X=0,Y=0,Z=0),Scale3D=(X=1,Y=1,Z=1))))
    ItemToModify.0.Transform=@n8.Value.elbow_out_26_#8
    Weight=@n4.Result
   )
   &RigUnit_ParentConstraint(
    AdvancedSettings=(RotationOrderForFilter=XZY)
    Child=(Type=Bone,Name="deform_elbow_out_r")
    Parents=((Item=(Type=Bone,Name="lowerarm_r"),Weight=0.3),(Item=(Type=Bone,Name="upperarm_r"),Weight=0.7))
    Weight=1
    ExecutePin=@SequenceExecution_1_1.B
   )
   > DeformRig_GetMirrorTransform_FNC_2 = RigVMFunctionReferenceNode(transform=@n8.Value.elbow_out_26_#8)
   > &RigUnit_ModifyTransforms(
    ItemToModify=((Item=(Type=Bone,Name="deform_elbow_out_r"),Transform=(Rotation=(X=0,Y=0,Z=0,W=1),Translation=(X=0,Y=0,Z=0),Scale3D=(X=1,Y=1,Z=1))))
    ItemToModify.0.Transform=@DeformRig_GetMirrorTransform_FNC_2.mirrorTransform
    Weight=@n6.Result
   )
   RigVMFunctionReturnNode(ExecuteContext=@SequenceExecution_2.E)
   n9 = &RigVMFunction_ControlFlowBranch(ExecuteContext=(), Condition=@n1.bicep)
   SequenceExecution_1_1_1 = RigVMAggregateNode(ExecuteContext=@n9.True)
   n10 = RigVMVariableNode() {variable=value_struct_var}
   &RigUnit_ModifyTransforms(
    ItemToModify=((Item=(Type=Bone,Name="deform_bicep_l"),Transform=(Rotation=(X=0,Y=0,Z=0,W=1),Translation=(X=0,Y=0,Z=0),Scale3D=(X=1,Y=1,Z=1))))
    ItemToModify.0.Transform=@n10.Value.bicep_44_#9
    ExecutePin=@SequenceExecution_1_1_1.A
    Weight=@n4.Result
   )
   DeformRig_GetMirrorTransform_FNC = RigVMFunctionReferenceNode(transform=@n10.Value.bicep_44_#9, ExecuteContext=@SequenceExecution_1_1_1.B)
   > &RigUnit_ModifyTransforms(
    ItemToModify=((Item=(Type=Bone,Name="deform_bicep_r"),Transform=(Rotation=(X=0,Y=0,Z=0,W=1),Translation=(X=0,Y=0,Z=0),Scale3D=(X=1,Y=1,Z=1))))
    ItemToModify.0.Transform=@DeformRig_GetMirrorTransform_FNC.mirrorTransform
    Weight=@n6.Result
   )
  ]}
 ContainedGraph {nodes [
   n1 = RigVMFunctionEntryNode()
   SequenceExecution_2 = &RigVMFunction_Sequence(ExecuteContext=@n1.ExecuteContext)
   SequenceExecution_2_1 = &RigVMFunction_Sequence(ExecuteContext=@SequenceExecution_2.B)
   SequenceExecution_2_2 = &RigVMFunction_Sequence(ExecuteContext=@SequenceExecution_2_1.B)
   SequenceExecution_2_3 = &RigVMFunction_Sequence(ExecuteContext=@SequenceExecution_2_2.B)
   RigVMFunctionReturnNode(A=@SequenceExecution_2.A, B=@SequenceExecution_2_1.A, C=@SequenceExecution_2_2.A, D=@SequenceExecution_2_3.A, E=@SequenceExecution_2_3.B)
  ]}
 ContainedGraph {nodes [
   n1 = RigVMFunctionEntryNode()
   SequenceExecution = &RigVMFunction_Sequence(ExecuteContext=@n1.ExecuteContext)
   SequenceExecution_1 = &RigVMFunction_Sequence(ExecuteContext=@SequenceExecution.B)
   SequenceExecution_2 = &RigVMFunction_Sequence(ExecuteContext=@SequenceExecution_1.B)
   RigVMFunctionReturnNode(A=@SequenceExecution.A, B=@SequenceExecution_1.A, C=@SequenceExecution_2.A, D=@SequenceExecution_2.B)
  ]}
 ContainedGraph {nodes [
   n1 = RigVMFunctionEntryNode()
   SequenceExecution_1 = &RigVMFunction_Sequence(ExecuteContext=@n1.ExecuteContext)
   SequenceExecution_1_1 = &RigVMFunction_Sequence(ExecuteContext=@SequenceExecution_1.B)
   SequenceExecution_1_2 = &RigVMFunction_Sequence(ExecuteContext=@SequenceExecution_1_1.B)
   RigVMFunctionReturnNode(A=@SequenceExecution_1.A, B=@SequenceExecution_1_1.A, C=@SequenceExecution_1_2.A, D=@SequenceExecution_1_2.B)
  ]}
 ContainedGraph {nodes [
   n1 = RigVMFunctionEntryNode()
   SequenceExecution_1_1 = &RigVMFunction_Sequence(ExecuteContext=@n1.ExecuteContext)
   SequenceExecution_1_1_1 = &RigVMFunction_Sequence(ExecuteContext=@SequenceExecution_1_1.B)
   SequenceExecution_1_1_2 = &RigVMFunction_Sequence(ExecuteContext=@SequenceExecution_1_1_1.B)
   RigVMFunctionReturnNode(A=@SequenceExecution_1_1.A, B=@SequenceExecution_1_1_1.A, C=@SequenceExecution_1_1_2.A, D=@SequenceExecution_1_1_2.B)
  ]}
 ContainedGraph {nodes [
   n1 = RigVMFunctionEntryNode()
   SequenceExecution_1_1_1 = &RigVMFunction_Sequence(ExecuteContext=@n1.ExecuteContext)
   SequenceExecution_1_1_1_1 = &RigVMFunction_Sequence(ExecuteContext=@SequenceExecution_1_1_1.B)
   SequenceExecution_1_1_1_2 = &RigVMFunction_Sequence(ExecuteContext=@SequenceExecution_1_1_1_1.B)
   RigVMFunctionReturnNode(A=@SequenceExecution_1_1_1.A, B=@SequenceExecution_1_1_1_1.A, C=@SequenceExecution_1_1_1_2.A, D=@SequenceExecution_1_1_1_2.B)
  ]}
 ContainedGraph {nodes [
   n1 = RigVMFunctionEntryNode()
   > RigVMVariableNode(Value=@n1.value_struct) {variable=value_struct_var}
   > RigVMVariableNode(Value=@n1.curve_struct) {variable=curve_struct_var}
   > SequenceExecution = RigVMAggregateNode()
   n2 = &RigUnit_SphericalPoseReader(
    ActiveRegionScaleFactors=(PositiveWidth=4,NegativeWidth=0.5,PositiveHeight=1.5,NegativeHeight=4)
    ActiveRegionSize=0.05
    Debug=(bDrawDebug=false,DebugScale=25,DebugSegments=20,DebugThickness=0.25)
    DriverAxis=(X=1,Y=0,Z=0)
    DriverItem=(Type=Bone,Name="upperarm_l")
    FalloffRegionScaleFactors=(PositiveWidth=1,NegativeWidth=1,PositiveHeight=1,NegativeHeight=1)
    FalloffSize=0.2
    OptionalParentItem=(Type=Bone,Name="spine_03")
    RotationOffset=(X=0,Y=-150,Z=0)
    Debug.bDrawDebug=@n1.debug
    ExecutePin=@SequenceExecution.A
   )
   n3 = RigVMVariableNode() {variable=value_struct_var}
   > n4 = &RigUnit_ModifyTransforms(
    ItemToModify=((Item=(Type=Bone,Name="deform_lat_l"),Transform=(Rotation=(X=0,Y=0,Z=0,W=1),Translation=(X=0,Y=0,Z=0),Scale3D=(X=1,Y=1,Z=1))))
    Weight=@n2.OutputParam
    ItemToModify.0.Transform=@n3.Value.lat_45_#10
   )
   n5 = RigVMVariableNode() {variable=curve_struct_var}
   RigVMFunction_ControlFlowBranch = &RigVMFunction_ControlFlowBranch(ExecuteContext=@n4.ExecutePin, Condition=@n5.Value.deformRig_l_lat_alpha_91_#11)
   RigVMDispatchNode(BoneName=root, Name=deformRig_l_lat_alpha, ExecuteContext=@RigVMFunction_ControlFlowBranch.True, Value=@n2.OutputParam) {template="DISPATCH_RigDispatch_SetAnimAttribute(in Name,in BoneName,in Value,out Success)"}
   n6 = &RigUnit_SphericalPoseReader(
    ActiveRegionScaleFactors=(PositiveWidth=4,NegativeWidth=0.5,PositiveHeight=1.5,NegativeHeight=4)
    ActiveRegionSize=0.05
    Debug=(bDrawDebug=false,DebugScale=25,DebugSegments=20,DebugThickness=0.25)
    DriverAxis=(X=-1,Y=0,Z=0)
    DriverItem=(Type=Bone,Name="upperarm_r")
    FalloffRegionScaleFactors=(PositiveWidth=1,NegativeWidth=1,PositiveHeight=1,NegativeHeight=1)
    FalloffSize=0.2
    FlipHeightScaling=true
    OptionalParentItem=(Type=Bone,Name="spine_03")
    RotationOffset=(X=0,Y=30,Z=0)
    Debug.bDrawDebug=@n1.debug
    ExecutePin=@SequenceExecution.B
   )
   > DeformRig_GetMirrorTransform_FNC = RigVMFunctionReferenceNode(transform=@n3.Value.lat_45_#10)
   > n7 = &RigUnit_ModifyTransforms(
    ItemToModify=((Item=(Type=Bone,Name="deform_lat_r"),Transform=(Rotation=(X=0,Y=0,Z=0,W=1),Translation=(X=0,Y=0,Z=0),Scale3D=(X=1,Y=1,Z=1))))
    ItemToModify.0.Transform=@DeformRig_GetMirrorTransform_FNC.mirrorTransform
    Weight=@n6.OutputParam
   )
   n8 = &RigVMFunction_ControlFlowBranch(ExecuteContext=@n7.ExecutePin, Condition=@n5.Value.deformRig_r_lat_alpha_93_#12)
   RigVMDispatchNode(BoneName=root, Name=deformRig_r_lat_alpha, ExecuteContext=@n8.True, Value=@n6.OutputParam) {template="DISPATCH_RigDispatch_SetAnimAttribute(in Name,in BoneName,in Value,out Success)"}
   RigVMFunctionReturnNode(ExecuteContext=@SequenceExecution.C)
  ]}
 ContainedGraph {nodes [
   n1 = RigVMFunctionEntryNode()
   SequenceExecution = &RigVMFunction_Sequence(ExecuteContext=@n1.ExecuteContext)
   SequenceExecution_1 = &RigVMFunction_Sequence(ExecuteContext=@SequenceExecution.B)
   SequenceExecution_2 = &RigVMFunction_Sequence(ExecuteContext=@SequenceExecution_1.B)
   RigVMFunctionReturnNode(A=@SequenceExecution.A, B=@SequenceExecution_1.A, C=@SequenceExecution_2.A, D=@SequenceExecution_2.B)
  ]}
 ContainedGraph {nodes [
   n1 = RigVMFunctionEntryNode()
   > RigVMVariableNode(Value=@n1.value_struct) {variable=value_struct_var}
   > RigVMVariableNode(Value=@n1.curve_struct) {variable=curve_struct_var}
   > SequenceExecution = RigVMAggregateNode()
   RigVMFunctionReturnNode(ExecuteContext=@SequenceExecution.D)
   n2 = /Script/ControlRig.RigUnit_AimBoneMath(
    DebugSettings=(Scale=10,WorldOffset=(Rotation=(X=0,Y=0,Z=0,W=1),Translation=(X=0,Y=0,Z=0),Scale3D=(X=1,Y=1,Z=1)))
    Primary=(Weight=1,Axis=(X=-1,Y=0,Z=0),Target=(X=0,Y=0,Z=0),Kind=Location,Space=(Type=Bone,Name="upperarm_r"))
    Secondary=(Weight=1,Axis=(X=0,Y=1,Z=0),Target=(X=0,Y=0,Z=0),Kind=Location,Space=(Type=Bone,Name="spine_04"))
    Weight=1
    InputTransform=&RigUnit_GetTransform(Item=(Type=Bone,Name="deform_pec_aim_r")).Transform
   ) {is=struct RigVMUnitNode}
   &RigUnit_SetTransform(Item=(Type=Bone,Name="deform_pec_aim_r"), Weight=1, bInitial=false, Value=@n2.Result, ExecutePin=@SequenceExecution.B)
   > /Script/ControlRig.RigUnit_PositionConstraintLocalSpaceOffset(
    Child=(Type=Bone,Name="deform_pec_r")
    Filter=(bX=True,bY=false,bZ=false)
    Parents=((Item=(Type=Bone,Name="spine_04"),Weight=0),(Item=(Type=Bone,Name="upperarm_r"),Weight=1))
    Weight=1
    bMaintainOffset=True
   ) {is=struct RigVMUnitNode}
   > DeformRig_GetDeltaRefTransform_FNC_1 = RigVMFunctionReferenceNode(Item=(Type=Bone,Name="clavicle_r"))
   n3 = RigVMVariableNode() {variable=value_struct_var}
   > DeformRig_GetMirrorTransform_FNC = RigVMFunctionReferenceNode(transform=@n3.Value.pec_51_#13)
   n4 = &RigVMFunction_MathFloatRemap(SourceMaximum=30, bClamp=true, Value=@DeformRig_GetDeltaRefTransform_FNC_1.OutTransform.Rotation.Pitch)
   > n5 = &RigUnit_ModifyTransforms(
    ItemToModify=((Item=(Type=Bone,Name="deform_pec_r"),Transform=(Rotation=(X=0,Y=0,Z=0,W=1),Translation=(X=0,Y=0,Z=0),Scale3D=(X=1,Y=1,Z=1))))
    ItemToModify.0.Transform=@DeformRig_GetMirrorTransform_FNC.mirrorTransform
    Weight=@n4.Result
   )
   n6 = RigVMVariableNode() {variable=curve_struct_var}
   n7 = &RigVMFunction_ControlFlowBranch(ExecuteContext=@n5.ExecutePin, Condition=@n6.Value.deformRig_r_pec_alpha_85_#14)
   RigVMDispatchNode(BoneName=root, Name=deformRig_r_pec_alpha, Value=@n4.Result, ExecuteContext=@n7.True) {template="DISPATCH_RigDispatch_SetAnimAttribute(in Name,in BoneName,in Value,out Success)"}
   n8 = &RigUnit_AimBoneMath(
    DebugSettings=(Scale=10,WorldOffset=(Rotation=(X=0,Y=0,Z=0,W=1),Translation=(X=0,Y=0,Z=0),Scale3D=(X=1,Y=1,Z=1)))
    Primary=(Weight=1,Axis=(X=1,Y=0,Z=0),Target=(X=0,Y=0,Z=0),Kind=Location,Space=(Type=Bone,Name="upperarm_l"))
    Secondary=(Weight=1,Axis=(X=0,Y=-1,Z=0),Target=(X=0,Y=0,Z=0),Kind=Location,Space=(Type=Bone,Name="spine_04"))
    Weight=1
    InputTransform=&RigUnit_GetTransform(Item=(Type=Bone,Name="deform_pec_aim_l")).Transform
   )
   &RigUnit_SetTransform(Item=(Type=Bone,Name="deform_pec_aim_l"), Weight=1, bInitial=false, Value=@n8.Result, ExecutePin=@SequenceExecution.A)
   > &RigUnit_PositionConstraintLocalSpaceOffset(
    Child=(Type=Bone,Name="deform_pec_l")
    Filter=(bX=True,bY=false,bZ=false)
    Parents=((Item=(Type=Bone,Name="spine_04"),Weight=0),(Item=(Type=Bone,Name="upperarm_l"),Weight=1))
    Weight=1
    bMaintainOffset=True
   )
   > DeformRig_GetDeltaRefTransform_FNC = RigVMFunctionReferenceNode(Item=(Type=Bone,Name="clavicle_l"))
   n9 = &RigVMFunction_MathFloatRemap(SourceMaximum=30, bClamp=true, Value=@DeformRig_GetDeltaRefTransform_FNC.OutTransform.Rotation.Pitch)
   > n10 = &RigUnit_ModifyTransforms(
    ItemToModify=((Item=(Type=Bone,Name="deform_pec_l"),Transform=(Rotation=(X=0,Y=0,Z=0,W=1),Translation=(X=0,Y=0,Z=0),Scale3D=(X=1,Y=1,Z=1))))
    Weight=@n9.Result
    ItemToModify.0.Transform=@n3.Value.pec_51_#13
   )
   RigVMFunction_ControlFlowBranch = &RigVMFunction_ControlFlowBranch(ExecuteContext=@n10.ExecutePin, Condition=@n6.Value.deformRig_l_pec_alpha_84_#15)
   DISPATCH_RigDispatch_SetAnimAttribute = RigVMDispatchNode(BoneName=root, Name=deformRig_l_pec_alpha, Value=@n9.Result, ExecuteContext=@RigVMFunction_ControlFlowBranch.True) {template="DISPATCH_RigDispatch_SetAnimAttribute(in Name,in BoneName,in Value,out Success)"}
   &RigUnit_ParentConstraint(
    AdvancedSettings=(RotationOrderForFilter=XZY)
    Child=(Type=Bone,Name="deform_pec_aim_base_l")
    ExecutePin=()
    Parents=((Item=(Type=Bone,Name="spine_03"),Weight=0.25),(Item=(Type=Bone,Name="spine_04"),Weight=0.55),(Item=(Type=Bone,Name="spine_05"),Weight=0.2))
    Weight=1
   )
   &RigUnit_ParentConstraint(
    AdvancedSettings=(RotationOrderForFilter=XZY)
    Child=(Type=Bone,Name="deform_pec_aim_base_r")
    ExecutePin=()
    Parents=((Item=(Type=Bone,Name="spine_03"),Weight=0.25),(Item=(Type=Bone,Name="spine_04"),Weight=0.55),(Item=(Type=Bone,Name="spine_05"),Weight=0.2))
    Weight=1
   )
  ]}
 ContainedGraph {nodes [
   n1 = RigVMFunctionEntryNode()
   SequenceExecution = &RigVMFunction_Sequence(ExecuteContext=@n1.ExecuteContext)
   SequenceExecution_1 = &RigVMFunction_Sequence(ExecuteContext=@SequenceExecution.B)
   SequenceExecution_2 = &RigVMFunction_Sequence(ExecuteContext=@SequenceExecution_1.B)
   RigVMFunctionReturnNode(A=@SequenceExecution.A, B=@SequenceExecution_1.A, C=@SequenceExecution_2.A, D=@SequenceExecution_2.B)
  ]}
 ContainedGraph {nodes [
   n1 = RigVMFunctionEntryNode()
   > RigVMVariableNode(Value=@n1.value_struct) {variable=value_struct_var}
   > RigVMVariableNode(Value=@n1.curve_struct) {variable=curve_struct_var}
   > SequenceExecution_2 = RigVMAggregateNode()
   DeformRig_GetDeltaRefTransform_FNC = RigVMFunctionReferenceNode(Item=(Type=Bone,Name="calf_l"), ExecuteContext=@SequenceExecution_2.A)
   n2 = RigVMVariableNode() {variable=curve_struct_var}
   RigVMFunction_ControlFlowBranch = &RigVMFunction_ControlFlowBranch(
    ExecuteContext=@DeformRig_GetDeltaRefTransform_FNC.ExecuteContext
    Condition=@n2.Value.deformRig_l_knee_alpha_82_#16
   )
   n3 = &RigVMFunction_MathFloatRemap(SourceMaximum=100, SourceMinimum=10, bClamp=true, Value=@DeformRig_GetDeltaRefTransform_FNC.OutTransform.Rotation.Yaw)
   RigVMDispatchNode(BoneName=root, Name=deformRig_l_knee_alpha, ExecuteContext=@RigVMFunction_ControlFlowBranch.True, Value=@n3.Result) {template="DISPATCH_RigDispatch_SetAnimAttribute(in Name,in BoneName,in Value,out Success)"}
   DeformRig_GetDeltaRefTransform_FNC_1 = RigVMFunctionReferenceNode(Item=(Type=Bone,Name="calf_r"), ExecuteContext=@SequenceExecution_2.B)
   n4 = &RigVMFunction_ControlFlowBranch(
    ExecuteContext=@DeformRig_GetDeltaRefTransform_FNC_1.ExecuteContext
    Condition=@n2.Value.deformRig_r_knee_alpha_83_#17
   )
   n5 = &RigVMFunction_MathFloatRemap(SourceMaximum=100, SourceMinimum=10, bClamp=true, Value=@DeformRig_GetDeltaRefTransform_FNC_1.OutTransform.Rotation.Yaw)
   RigVMDispatchNode(BoneName=root, Name=deformRig_r_knee_alpha, ExecuteContext=@n4.True, Value=@n5.Result) {template="DISPATCH_RigDispatch_SetAnimAttribute(in Name,in BoneName,in Value,out Success)"}
   SequenceExecution = RigVMAggregateNode(ExecuteContext=@SequenceExecution_2.C)
   SequenceExecution_1 = RigVMAggregateNode(ExecuteContext=@SequenceExecution.A)
   &RigUnit_ParentConstraint(
    AdvancedSettings=(RotationOrderForFilter=XZY)
    Child=(Type=Bone,Name="deform_knee_in_l")
    Parents=((Item=(Type=Bone,Name="calf_l"),Weight=0.5),(Item=(Type=Bone,Name="thigh_l"),Weight=0.5))
    Weight=1
    ExecutePin=@SequenceExecution_1.A
   )
   n6 = RigVMVariableNode() {variable=value_struct_var}
   > &RigUnit_ModifyTransforms(
    ItemToModify=((Item=(Type=Bone,Name="deform_knee_in_l"),Transform=(Rotation=(X=0,Y=0,Z=0,W=1),Translation=(X=0,Y=0,Z=0),Scale3D=(X=1,Y=1,Z=1))))
    ItemToModify.0.Transform=@n6.Value.knee_in_41_#18
    Weight=@n3.Result
   )
   &RigUnit_ParentConstraint(
    AdvancedSettings=(RotationOrderForFilter=XZY)
    Child=(Type=Bone,Name="deform_knee_in_r")
    Parents=((Item=(Type=Bone,Name="calf_r"),Weight=0.5),(Item=(Type=Bone,Name="thigh_r"),Weight=0.5))
    Weight=1
    ExecutePin=@SequenceExecution_1.B
   )
   > DeformRig_GetMirrorTransform_FNC = RigVMFunctionReferenceNode(transform=@n6.Value.knee_in_41_#18)
   > &RigUnit_ModifyTransforms(
    ItemToModify=((Item=(Type=Bone,Name="deform_knee_in_r"),Transform=(Rotation=(X=0,Y=0,Z=0,W=1),Translation=(X=0,Y=0,Z=0),Scale3D=(X=1,Y=1,Z=1))))
    ItemToModify.0.Transform=@DeformRig_GetMirrorTransform_FNC.mirrorTransform
    Weight=@n5.Result
   )
   SequenceExecution_1_1 = RigVMAggregateNode(ExecuteContext=@SequenceExecution.B)
   &RigUnit_ParentConstraint(
    AdvancedSettings=(RotationOrderForFilter=XZY)
    Child=(Type=Bone,Name="deform_knee_out_l")
    Parents=((Item=(Type=Bone,Name="calf_l"),Weight=0.1),(Item=(Type=Bone,Name="thigh_l"),Weight=0.9))
    Weight=1
    ExecutePin=@SequenceExecution_1_1.A
   )
   n7 = RigVMVariableNode() {variable=value_struct_var}
   > &RigUnit_ModifyTransforms(
    ItemToModify=((Item=(Type=Bone,Name="deform_knee_out_l"),Transform=(Rotation=(X=0,Y=0,Z=0,W=1),Translation=(X=0,Y=0,Z=0),Scale3D=(X=1,Y=1,Z=1))))
    ItemToModify.0.Transform=@n7.Value.knee_out_42_#19
    Weight=@n3.Result
   )
   &RigUnit_ParentConstraint(
    AdvancedSettings=(RotationOrderForFilter=XZY)
    Child=(Type=Bone,Name="deform_knee_out_r")
    Parents=((Item=(Type=Bone,Name="calf_r"),Weight=0.1),(Item=(Type=Bone,Name="thigh_r"),Weight=0.9))
    Weight=1
    ExecutePin=@SequenceExecution_1_1.B
   )
   > DeformRig_GetMirrorTransform_FNC_1 = RigVMFunctionReferenceNode(transform=@n7.Value.knee_out_42_#19)
   > &RigUnit_ModifyTransforms(
    ItemToModify=((Item=(Type=Bone,Name="deform_knee_out_r"),Transform=(Rotation=(X=0,Y=0,Z=0,W=1),Translation=(X=0,Y=0,Z=0),Scale3D=(X=1,Y=1,Z=1))))
    ItemToModify.0.Transform=@DeformRig_GetMirrorTransform_FNC_1.mirrorTransform
    Weight=@n5.Result
   )
   RigVMFunctionReturnNode(ExecuteContext=@SequenceExecution_2.D)
  ]}
 ContainedGraph {nodes [
   n1 = RigVMFunctionEntryNode()
   SequenceExecution = &RigVMFunction_Sequence(ExecuteContext=@n1.ExecuteContext)
   SequenceExecution_1 = &RigVMFunction_Sequence(ExecuteContext=@SequenceExecution.B)
   SequenceExecution_2 = &RigVMFunction_Sequence(ExecuteContext=@SequenceExecution_1.B)
   RigVMFunctionReturnNode(A=@SequenceExecution.A, B=@SequenceExecution_1.A, C=@SequenceExecution_2.A, D=@SequenceExecution_2.B)
  ]}
 ContainedGraph {nodes [
   n1 = RigVMFunctionEntryNode()
   SequenceExecution_2 = &RigVMFunction_Sequence(ExecuteContext=@n1.ExecuteContext)
   SequenceExecution_2_1 = &RigVMFunction_Sequence(ExecuteContext=@SequenceExecution_2.B)
   SequenceExecution_2_2 = &RigVMFunction_Sequence(ExecuteContext=@SequenceExecution_2_1.B)
   RigVMFunctionReturnNode(A=@SequenceExecution_2.A, B=@SequenceExecution_2_1.A, C=@SequenceExecution_2_2.A, D=@SequenceExecution_2_2.B)
  ]}
 ContainedGraph {nodes [
   n1 = RigVMFunctionEntryNode()
   SequenceExecution_1_1 = &RigVMFunction_Sequence(ExecuteContext=@n1.ExecuteContext)
   SequenceExecution_1_1_1 = &RigVMFunction_Sequence(ExecuteContext=@SequenceExecution_1_1.B)
   SequenceExecution_1_1_2 = &RigVMFunction_Sequence(ExecuteContext=@SequenceExecution_1_1_1.B)
   RigVMFunctionReturnNode(A=@SequenceExecution_1_1.A, B=@SequenceExecution_1_1_1.A, C=@SequenceExecution_1_1_2.A, D=@SequenceExecution_1_1_2.B)
  ]}
 ContainedGraph {nodes [
   n1 = RigVMFunctionEntryNode()
   SequenceExecution_1 = &RigVMFunction_Sequence(ExecuteContext=@n1.ExecuteContext)
   SequenceExecution_1_1 = &RigVMFunction_Sequence(ExecuteContext=@SequenceExecution_1.B)
   SequenceExecution_1_2 = &RigVMFunction_Sequence(ExecuteContext=@SequenceExecution_1_1.B)
   RigVMFunctionReturnNode(A=@SequenceExecution_1.A, B=@SequenceExecution_1_1.A, C=@SequenceExecution_1_2.A, D=@SequenceExecution_1_2.B)
  ]}
 ContainedGraph {nodes [
   n1 = RigVMFunctionEntryNode()
   > RigVMVariableNode(Value=@n1.value_struct) {variable=value_struct_var}
   > RigVMVariableNode(Value=@n1.curve_struct) {variable=curve_struct_var}
   > SequenceExecution = RigVMAggregateNode()
   RigVMFunctionReturnNode(ExecuteContext=@SequenceExecution.D)
   SequenceExecution_1 = RigVMAggregateNode(ExecuteContext=@SequenceExecution.A)
   &RigUnit_ParentConstraint(
    AdvancedSettings=(RotationOrderForFilter=XZY)
    Child=(Type=Bone,Name="deform_wrist_up_l")
    Parents=((Item=(Type=Bone,Name="hand_l"),Weight=0.4),(Item=(Type=Bone,Name="lowerarm_twist_01_l"),Weight=0.6))
    Weight=1
    ExecutePin=@SequenceExecution_1.A
   )
   > n2 = &RigUnit_SphericalPoseReader(
    ActiveRegionScaleFactors=(PositiveWidth=0.5,NegativeWidth=3.5,PositiveHeight=3,NegativeHeight=3)
    ActiveRegionSize=0.1
    Debug=(bDrawDebug=false,DebugScale=25,DebugSegments=20,DebugThickness=0.25)
    DriverAxis=(X=1,Y=0,Z=0)
    DriverItem=(Type=Bone,Name="hand_l")
    FalloffRegionScaleFactors=(PositiveWidth=1,NegativeWidth=1,PositiveHeight=1,NegativeHeight=1)
    FalloffSize=0.45
    OptionalParentItem=(Type=Bone,Name="lowerarm_twist_01_l")
    RotationOffset=(X=90,Y=0,Z=0)
    Debug.bDrawDebug=@n1.debug
   )
   n3 = RigVMVariableNode() {variable=value_struct_var}
   > n4 = &RigUnit_ModifyTransforms(
    ItemToModify=((Item=(Type=Bone,Name="deform_wrist_up_l"),Transform=(Rotation=(X=0,Y=0,Z=0,W=1),Translation=(X=0,Y=0,Z=0),Scale3D=(X=1,Y=1,Z=1))))
    ItemToModify.0.Transform=@n3.Value.wrist_up_31_#20
    Weight=@n2.OutputParam
   )
   n5 = RigVMVariableNode() {variable=curve_struct_var}
   RigVMFunction_ControlFlowBranch = &RigVMFunction_ControlFlowBranch(ExecuteContext=@n4.ExecutePin, Condition=@n5.Value.deformRig_l_wrist_up_alpha_78_#21)
   RigVMDispatchNode(BoneName=root, Name=deformRig_l_wrist_up_alpha, ExecuteContext=@RigVMFunction_ControlFlowBranch.True, Value=@n2.OutputParam) {template="DISPATCH_RigDispatch_SetAnimAttribute(in Name,in BoneName,in Value,out Success)"}
   &RigUnit_ParentConstraint(
    AdvancedSettings=(RotationOrderForFilter=XZY)
    Child=(Type=Bone,Name="deform_wrist_up_r")
    Parents=((Item=(Type=Bone,Name="hand_r"),Weight=0.4),(Item=(Type=Bone,Name="lowerarm_twist_01_r"),Weight=0.6))
    Weight=1
    ExecutePin=@SequenceExecution_1.B
   )
   > n6 = &RigUnit_SphericalPoseReader(
    ActiveRegionScaleFactors=(PositiveWidth=0.5,NegativeWidth=3.5,PositiveHeight=3,NegativeHeight=3)
    ActiveRegionSize=0.1
    Debug=(bDrawDebug=false,DebugScale=25,DebugSegments=20,DebugThickness=0.25)
    DriverAxis=(X=-1,Y=0,Z=0)
    DriverItem=(Type=Bone,Name="hand_r")
    FalloffRegionScaleFactors=(PositiveWidth=1,NegativeWidth=1,PositiveHeight=1,NegativeHeight=1)
    FalloffSize=0.45
    FlipHeightScaling=true
    OptionalParentItem=(Type=Bone,Name="lowerarm_twist_01_r")
    RotationOffset=(X=90,Y=0,Z=180)
    Debug.bDrawDebug=@n1.debug
   )
   > DeformRig_GetMirrorTransform_FNC = RigVMFunctionReferenceNode(transform=@n3.Value.wrist_up_31_#20)
   > n7 = &RigUnit_ModifyTransforms(
    ItemToModify=((Item=(Type=Bone,Name="deform_wrist_up_r"),Transform=(Rotation=(X=0,Y=0,Z=0,W=1),Translation=(X=0,Y=0,Z=0),Scale3D=(X=1,Y=1,Z=1))))
    ItemToModify.0.Transform=@DeformRig_GetMirrorTransform_FNC.mirrorTransform
    Weight=@n6.OutputParam
   )
   n8 = &RigVMFunction_ControlFlowBranch(ExecuteContext=@n7.ExecutePin, Condition=@n5.Value.deformRig_r_wrist_up_alpha_79_#22)
   RigVMDispatchNode(BoneName=root, Name=deformRig_r_wrist_up_alpha, ExecuteContext=@n8.True, Value=@n6.OutputParam) {template="DISPATCH_RigDispatch_SetAnimAttribute(in Name,in BoneName,in Value,out Success)"}
   SequenceExecution_1_1 = RigVMAggregateNode(ExecuteContext=@SequenceExecution.B)
   &RigUnit_ParentConstraint(
    AdvancedSettings=(RotationOrderForFilter=XZY)
    Child=(Type=Bone,Name="deform_wrist_dn_l")
    Parents=((Item=(Type=Bone,Name="hand_l"),Weight=0.4),(Item=(Type=Bone,Name="lowerarm_twist_01_l"),Weight=0.6))
    Weight=1
    ExecutePin=@SequenceExecution_1_1.A
   )
   > n9 = &RigUnit_SphericalPoseReader(
    ActiveRegionScaleFactors=(PositiveWidth=1,NegativeWidth=4,PositiveHeight=3,NegativeHeight=3)
    ActiveRegionSize=0.1
    Debug=(bDrawDebug=false,DebugScale=25,DebugSegments=20,DebugThickness=0.25)
    DriverAxis=(X=1,Y=0,Z=0)
    DriverItem=(Type=Bone,Name="hand_l")
    FalloffRegionScaleFactors=(PositiveWidth=1,NegativeWidth=1,PositiveHeight=1,NegativeHeight=1)
    FalloffSize=0.4
    OptionalParentItem=(Type=Bone,Name="lowerarm_twist_01_l")
    RotationOffset=(X=-90,Y=0,Z=0)
    Debug.bDrawDebug=@n1.debug
   )
   n10 = RigVMVariableNode() {variable=value_struct_var}
   > n11 = &RigUnit_ModifyTransforms(
    ItemToModify=((Item=(Type=Bone,Name="deform_wrist_dn_l"),Transform=(Rotation=(X=0,Y=0,Z=0,W=1),Translation=(X=0,Y=0,Z=0),Scale3D=(X=1,Y=1,Z=1))))
    ItemToModify.0.Transform=@n10.Value.wrist_dn_32_#23
    Weight=@n9.OutputParam
   )
   n12 = RigVMVariableNode() {variable=curve_struct_var}
   n13 = &RigVMFunction_ControlFlowBranch(ExecuteContext=@n11.ExecutePin, Condition=@n12.Value.deformRig_l_wrist_dn_alpha_80_#24)
   RigVMDispatchNode(BoneName=root, Name=deformRig_l_wrist_dn_alpha, ExecuteContext=@n13.True, Value=@n9.OutputParam) {template="DISPATCH_RigDispatch_SetAnimAttribute(in Name,in BoneName,in Value,out Success)"}
   &RigUnit_ParentConstraint(
    AdvancedSettings=(RotationOrderForFilter=XZY)
    Child=(Type=Bone,Name="deform_wrist_dn_r")
    Parents=((Item=(Type=Bone,Name="hand_r"),Weight=0.4),(Item=(Type=Bone,Name="lowerarm_twist_01_r"),Weight=0.6))
    Weight=1
    ExecutePin=@SequenceExecution_1_1.B
   )
   > n14 = &RigUnit_SphericalPoseReader(
    ActiveRegionScaleFactors=(PositiveWidth=1,NegativeWidth=4,PositiveHeight=3,NegativeHeight=3)
    ActiveRegionSize=0.1
    Debug=(bDrawDebug=false,DebugScale=25,DebugSegments=20,DebugThickness=0.25)
    DriverAxis=(X=-1,Y=0,Z=0)
    DriverItem=(Type=Bone,Name="hand_r")
    FalloffRegionScaleFactors=(PositiveWidth=1,NegativeWidth=1,PositiveHeight=1,NegativeHeight=1)
    FalloffSize=0.4
    FlipHeightScaling=true
    OptionalParentItem=(Type=Bone,Name="lowerarm_twist_01_r")
    RotationOffset=(X=-90,Y=0,Z=190)
    Debug.bDrawDebug=@n1.debug
   )
   > DeformRig_GetMirrorTransform_FNC_1 = RigVMFunctionReferenceNode(transform=@n10.Value.wrist_dn_32_#23)
   > n15 = &RigUnit_ModifyTransforms(
    ItemToModify=((Item=(Type=Bone,Name="deform_wrist_dn_r"),Transform=(Rotation=(X=0,Y=0,Z=0,W=1),Translation=(X=0,Y=0,Z=0),Scale3D=(X=1,Y=1,Z=1))))
    ItemToModify.0.Transform=@DeformRig_GetMirrorTransform_FNC_1.mirrorTransform
    Weight=@n14.OutputParam
   )
   n16 = &RigVMFunction_ControlFlowBranch(ExecuteContext=@n15.ExecutePin, Condition=@n12.Value.deformRig_r_wrist_dn_alpha_81_#25)
   RigVMDispatchNode(BoneName=root, Name=deformRig_r_wrist_dn_alpha, ExecuteContext=@n16.True, Value=@n14.OutputParam) {template="DISPATCH_RigDispatch_SetAnimAttribute(in Name,in BoneName,in Value,out Success)"}
  ]}
 ContainedGraph {nodes [
   n1 = RigVMFunctionEntryNode()
   SequenceExecution = &RigVMFunction_Sequence(ExecuteContext=@n1.ExecuteContext)
   SequenceExecution_1 = &RigVMFunction_Sequence(ExecuteContext=@SequenceExecution.B)
   SequenceExecution_2 = &RigVMFunction_Sequence(ExecuteContext=@SequenceExecution_1.B)
   RigVMFunctionReturnNode(A=@SequenceExecution.A, B=@SequenceExecution_1.A, C=@SequenceExecution_2.A, D=@SequenceExecution_2.B)
  ]}
 ContainedGraph {nodes [
   n1 = RigVMFunctionEntryNode()
   SequenceExecution_1 = &RigVMFunction_Sequence(ExecuteContext=@n1.ExecuteContext)
   SequenceExecution_1_1 = &RigVMFunction_Sequence(ExecuteContext=@SequenceExecution_1.B)
   SequenceExecution_1_2 = &RigVMFunction_Sequence(ExecuteContext=@SequenceExecution_1_1.B)
   RigVMFunctionReturnNode(A=@SequenceExecution_1.A, B=@SequenceExecution_1_1.A, C=@SequenceExecution_1_2.A, D=@SequenceExecution_1_2.B)
  ]}
 ContainedGraph {nodes [
   n1 = RigVMFunctionEntryNode()
   SequenceExecution_1_1 = &RigVMFunction_Sequence(ExecuteContext=@n1.ExecuteContext)
   SequenceExecution_1_1_1 = &RigVMFunction_Sequence(ExecuteContext=@SequenceExecution_1_1.B)
   SequenceExecution_1_1_2 = &RigVMFunction_Sequence(ExecuteContext=@SequenceExecution_1_1_1.B)
   RigVMFunctionReturnNode(A=@SequenceExecution_1_1.A, B=@SequenceExecution_1_1_1.A, C=@SequenceExecution_1_1_2.A, D=@SequenceExecution_1_1_2.B)
  ]}
 ContainedGraph {nodes [
   n1 = RigVMFunctionEntryNode()
   > RigVMVariableNode(Value=@n1.value_struct) {variable=value_struct_var}
   > RigVMVariableNode(Value=@n1.curve_struct) {variable=curve_struct_var}
   > SequenceExecution_2 = RigVMAggregateNode()
   RigVMFunctionReturnNode(ExecuteContext=@SequenceExecution_2.D)
   SequenceExecution_3_1 = RigVMAggregateNode(ExecuteContext=@SequenceExecution_2.B)
   &RigUnit_ParentConstraint(
    AdvancedSettings=(RotationOrderForFilter=XZY)
    Child=(Type=Bone,Name="deform_upperarm_fr_l")
    Parents=((Item=(Type=Bone,Name="upperarm_twist_01_l"),Weight=0.3),(Item=(Type=Bone,Name="clavicle_l"),Weight=0.7))
    Weight=1
    ExecutePin=@SequenceExecution_3_1.A
   )
   > n2 = &RigUnit_SphericalPoseReader(
    ActiveRegionScaleFactors=(PositiveWidth=0.5,NegativeWidth=3,PositiveHeight=1.5,NegativeHeight=1.5)
    ActiveRegionSize=0.1
    Debug=(bDrawDebug=false,DebugScale=25,DebugSegments=20,DebugThickness=0.25)
    DriverAxis=(X=1,Y=0,Z=0)
    DriverItem=(Type=Bone,Name="upperarm_l")
    FalloffRegionScaleFactors=(PositiveWidth=1,NegativeWidth=1,PositiveHeight=1,NegativeHeight=1)
    FalloffSize=0.4
    OptionalParentItem=(Type=Bone)
    RotationOffset=(X=90,Y=0,Z=0)
    Debug.bDrawDebug=@n1.debug
   )
   n3 = RigVMVariableNode() {variable=value_struct_var}
   > &RigUnit_ModifyTransforms(
    ItemToModify=((Item=(Type=Bone,Name="deform_upperarm_fr_l"),Transform=(Rotation=(X=0,Y=0,Z=0,W=1),Translation=(X=0,Y=0,Z=0),Scale3D=(X=1,Y=1,Z=1))))
    ItemToModify.0.Transform=@n3.Value.upperarm_fwd_fr_19_#26
    Weight=@n2.OutputParam
   )
   n4 = RigVMVariableNode() {variable=value_struct_var}
   > n5 = &RigUnit_ModifyTransforms(
    ItemToModify=((Item=(Type=Bone,Name="deform_upperarm_bk_l"),Transform=(Rotation=(X=0,Y=0,Z=0,W=1),Translation=(X=0,Y=0,Z=0),Scale3D=(X=1,Y=1,Z=1))))
    ItemToModify.0.Transform=@n4.Value.upperarm_fwd_bk_20_#27
    Weight=@n2.OutputParam
   )
   n6 = RigVMVariableNode() {variable=curve_struct_var}
   n7 = &RigVMFunction_ControlFlowBranch(ExecuteContext=@n5.ExecutePin, Condition=@n6.Value.deformRig_l_upperarm_fr_alpha_74_#28)
   RigVMDispatchNode(BoneName=root, Name=deformRig_l_upperarm_fr_alpha, ExecuteContext=@n7.True, Value=@n2.OutputParam) {template="DISPATCH_RigDispatch_SetAnimAttribute(in Name,in BoneName,in Value,out Success)"}
   &RigUnit_ParentConstraint(
    AdvancedSettings=(RotationOrderForFilter=XZY)
    Child=(Type=Bone,Name="deform_upperarm_fr_r")
    Parents=((Item=(Type=Bone,Name="upperarm_twist_01_r"),Weight=0.3),(Item=(Type=Bone,Name="clavicle_r"),Weight=0.7))
    Weight=1
    ExecutePin=@SequenceExecution_3_1.B
   )
   > n8 = &RigUnit_SphericalPoseReader(
    ActiveRegionScaleFactors=(PositiveWidth=0.5,NegativeWidth=3,PositiveHeight=1.5,NegativeHeight=1.5)
    ActiveRegionSize=0.1
    Debug=(bDrawDebug=false,DebugScale=25,DebugSegments=20,DebugThickness=0.25)
    DriverAxis=(X=-1,Y=0,Z=0)
    DriverItem=(Type=Bone,Name="upperarm_r")
    FalloffRegionScaleFactors=(PositiveWidth=1,NegativeWidth=1,PositiveHeight=1,NegativeHeight=1)
    FalloffSize=0.4
    FlipWidthScaling=true
    OptionalParentItem=(Type=Bone)
    RotationOffset=(X=90,Y=180,Z=180)
    Debug.bDrawDebug=@n1.debug
   )
   > DeformRig_GetMirrorTransform_FNC_2 = RigVMFunctionReferenceNode(transform=@n3.Value.upperarm_fwd_fr_19_#26)
   > &RigUnit_ModifyTransforms(
    ItemToModify=((Item=(Type=Bone,Name="deform_upperarm_fr_r"),Transform=(Rotation=(X=0,Y=0,Z=0,W=1),Translation=(X=0,Y=0,Z=0),Scale3D=(X=1,Y=1,Z=1))))
    ItemToModify.0.Transform=@DeformRig_GetMirrorTransform_FNC_2.mirrorTransform
    Weight=@n8.OutputParam
   )
   > DeformRig_GetMirrorTransform_FNC_1 = RigVMFunctionReferenceNode(transform=@n4.Value.upperarm_fwd_bk_20_#27)
   > n9 = &RigUnit_ModifyTransforms(
    ItemToModify=((Item=(Type=Bone,Name="deform_upperarm_bk_r"),Transform=(Rotation=(X=0,Y=0,Z=0,W=1),Translation=(X=0,Y=0,Z=0),Scale3D=(X=1,Y=1,Z=1))))
    ItemToModify.0.Transform=@DeformRig_GetMirrorTransform_FNC_1.mirrorTransform
    Weight=@n8.OutputParam
   )
   n10 = &RigVMFunction_ControlFlowBranch(ExecuteContext=@n9.ExecutePin, Condition=@n6.Value.deformRig_r_upperarm_fr_alpha_75_#29)
   RigVMDispatchNode(BoneName=root, Name=deformRig_r_upperarm_fr_alpha, ExecuteContext=@n10.True, Value=@n8.OutputParam) {template="DISPATCH_RigDispatch_SetAnimAttribute(in Name,in BoneName,in Value,out Success)"}
   SequenceExecution_3_1_1 = RigVMAggregateNode(ExecuteContext=@SequenceExecution_2.C)
   &RigUnit_ParentConstraint(
    AdvancedSettings=(RotationOrderForFilter=XZY)
    Child=(Type=Bone,Name="deform_upperarm_bk_l")
    Parents=((Item=(Type=Bone,Name="upperarm_twist_01_l"),Weight=0.3),(Item=(Type=Bone,Name="clavicle_l"),Weight=0.7))
    Weight=1
    ExecutePin=@SequenceExecution_3_1_1.A
   )
   > n11 = &RigUnit_SphericalPoseReader(
    ActiveRegionScaleFactors=(PositiveWidth=3,NegativeWidth=1,PositiveHeight=1.5,NegativeHeight=1.5)
    ActiveRegionSize=0.1
    Debug=(bDrawDebug=false,DebugScale=25,DebugSegments=20,DebugThickness=0.25)
    DriverAxis=(X=1,Y=0,Z=0)
    DriverItem=(Type=Bone,Name="upperarm_l")
    FalloffRegionScaleFactors=(PositiveWidth=1,NegativeWidth=1,PositiveHeight=1,NegativeHeight=1)
    FalloffSize=0.4
    OptionalParentItem=(Type=Bone)
    RotationOffset=(X=90,Y=0,Z=180)
    Debug.bDrawDebug=@n1.debug
   )
   n12 = RigVMVariableNode() {variable=value_struct_var}
   > n13 = &RigUnit_ModifyTransforms(
    ItemToModify=((Item=(Type=Bone,Name="deform_upperarm_bk_l"),Transform=(Rotation=(X=0,Y=0,Z=0,W=1),Translation=(X=0,Y=0,Z=0),Scale3D=(X=1,Y=1,Z=1))))
    ItemToModify.0.Transform=@n12.Value.upperarm_bk_22_#30
    Weight=@n11.OutputParam
   )
   n14 = RigVMVariableNode() {variable=curve_struct_var}
   n15 = &RigVMFunction_ControlFlowBranch(ExecuteContext=@n13.ExecutePin, Condition=@n14.Value.deformRig_l_upperarm_bk_alpha_76_#31)
   RigVMDispatchNode(BoneName=root, Name=deformRig_l_upperarm_bk_alpha, ExecuteContext=@n15.True, Value=@n11.OutputParam) {template="DISPATCH_RigDispatch_SetAnimAttribute(in Name,in BoneName,in Value,out Success)"}
   &RigUnit_ParentConstraint(
    AdvancedSettings=(RotationOrderForFilter=XZY)
    Child=(Type=Bone,Name="deform_upperarm_bk_r")
    Parents=((Item=(Type=Bone,Name="upperarm_twist_01_r"),Weight=0.3),(Item=(Type=Bone,Name="clavicle_r"),Weight=0.7))
    Weight=1
    ExecutePin=@SequenceExecution_3_1_1.B
   )
   > n16 = &RigUnit_SphericalPoseReader(
    ActiveRegionScaleFactors=(PositiveWidth=3,NegativeWidth=1,PositiveHeight=1.5,NegativeHeight=1.5)
    ActiveRegionSize=0.1
    Debug=(bDrawDebug=false,DebugScale=25,DebugSegments=20,DebugThickness=0.25)
    DriverAxis=(X=-1,Y=0,Z=0)
    DriverItem=(Type=Bone,Name="upperarm_r")
    FalloffRegionScaleFactors=(PositiveWidth=1,NegativeWidth=1,PositiveHeight=1,NegativeHeight=1)
    FalloffSize=0.4
    FlipHeightScaling=true
    FlipWidthScaling=false
    OptionalParentItem=(Type=Bone)
    RotationOffset=(X=90,Y=0,Z=0)
    Debug.bDrawDebug=@n1.debug
   )
   > DeformRig_GetMirrorTransform_FNC_3 = RigVMFunctionReferenceNode(transform=@n12.Value.upperarm_bk_22_#30)
   > n17 = &RigUnit_ModifyTransforms(
    ItemToModify=((Item=(Type=Bone,Name="deform_upperarm_bk_r"),Transform=(Rotation=(X=0,Y=0,Z=0,W=1),Translation=(X=0,Y=0,Z=0),Scale3D=(X=1,Y=1,Z=1))))
    ItemToModify.0.Transform=@DeformRig_GetMirrorTransform_FNC_3.mirrorTransform
    Weight=@n16.OutputParam
   )
   n18 = &RigVMFunction_ControlFlowBranch(ExecuteContext=@n17.ExecutePin, Condition=@n14.Value.deformRig_r_upperarm_bk_alpha_77_#32)
   RigVMDispatchNode(BoneName=root, Name=deformRig_r_upperarm_bk_alpha, ExecuteContext=@n18.True, Value=@n16.OutputParam) {template="DISPATCH_RigDispatch_SetAnimAttribute(in Name,in BoneName,in Value,out Success)"}
   SequenceExecution_3 = RigVMAggregateNode(ExecuteContext=@SequenceExecution_2.A)
   &RigUnit_ParentConstraint(
    AdvancedSettings=(RotationOrderForFilter=XZY)
    Child=(Type=Bone,Name="deform_upperarm_up_l")
    Parents=((Item=(Type=Bone,Name="upperarm_twist_01_l"),Weight=0.5),(Item=(Type=Bone,Name="clavicle_l"),Weight=0.5))
    Weight=1
    ExecutePin=@SequenceExecution_3.A
   )
   > n19 = &RigUnit_SphericalPoseReader(
    ActiveRegionScaleFactors=(PositiveWidth=2.5,NegativeWidth=3,PositiveHeight=3,NegativeHeight=3)
    ActiveRegionSize=0.1
    Debug=(bDrawDebug=false,DebugScale=25,DebugSegments=20,DebugThickness=0.25)
    DriverAxis=(X=1,Y=0,Z=0)
    DriverItem=(Type=Bone,Name="upperarm_l")
    FalloffRegionScaleFactors=(PositiveWidth=1,NegativeWidth=1,PositiveHeight=1,NegativeHeight=1)
    FalloffSize=0.5
    OptionalParentItem=(Type=Bone)
    RotationOffset=(X=0,Y=0,Z=0)
    Debug.bDrawDebug=@n1.debug
   )
   n20 = RigVMVariableNode() {variable=value_struct_var}
   > n21 = &RigUnit_ModifyTransforms(
    ItemToModify=((Item=(Type=Bone,Name="deform_upperarm_up_l"),Transform=(Rotation=(X=0,Y=0,Z=0,W=1),Translation=(X=0,Y=0,Z=0),Scale3D=(X=1,Y=1,Z=1))))
    Weight=@n19.OutputParam
    ItemToModify.0.Transform=@n20.Value.upperarm_up_18_#33
   )
   n22 = RigVMVariableNode() {variable=curve_struct_var}
   RigVMFunction_ControlFlowBranch = &RigVMFunction_ControlFlowBranch(ExecuteContext=@n21.ExecutePin, Condition=@n22.Value.deformRig_l_upperarm_up_alpha_72_#34)
   RigVMDispatchNode(BoneName=root, Name=deformRig_l_upperarm_up_alpha, ExecuteContext=@RigVMFunction_ControlFlowBranch.True, Value=@n19.OutputParam) {template="DISPATCH_RigDispatch_SetAnimAttribute(in Name,in BoneName,in Value,out Success)"}
   &RigUnit_ParentConstraint(
    AdvancedSettings=(RotationOrderForFilter=XZY)
    Child=(Type=Bone,Name="deform_upperarm_up_r")
    Parents=((Item=(Type=Bone,Name="upperarm_twist_01_r"),Weight=0.5),(Item=(Type=Bone,Name="clavicle_r"),Weight=0.5))
    Weight=1
    ExecutePin=@SequenceExecution_3.B
   )
   > n23 = &RigUnit_SphericalPoseReader(
    ActiveRegionScaleFactors=(PositiveWidth=2.5,NegativeWidth=3,PositiveHeight=3,NegativeHeight=3)
    ActiveRegionSize=0.1
    Debug=(bDrawDebug=false,DebugScale=25,DebugSegments=20,DebugThickness=0.25)
    DriverAxis=(X=-1,Y=0,Z=0)
    DriverItem=(Type=Bone,Name="upperarm_r")
    FalloffRegionScaleFactors=(PositiveWidth=1,NegativeWidth=1,PositiveHeight=1,NegativeHeight=1)
    FalloffSize=0.5
    FlipHeightScaling=true
    OptionalParentItem=(Type=Bone)
    RotationOffset=(X=180,Y=0,Z=180)
    Debug.bDrawDebug=@n1.debug
   )
   > DeformRig_GetMirrorTransform_FNC = RigVMFunctionReferenceNode(transform=@n20.Value.upperarm_up_18_#33)
   > n24 = &RigUnit_ModifyTransforms(
    ItemToModify=((Item=(Type=Bone,Name="deform_upperarm_up_r"),Transform=(Rotation=(X=0,Y=0,Z=0,W=1),Translation=(X=0,Y=0,Z=0),Scale3D=(X=1,Y=1,Z=1))))
    ItemToModify.0.Transform=@DeformRig_GetMirrorTransform_FNC.mirrorTransform
    Weight=@n23.OutputParam
   )
   n25 = &RigVMFunction_ControlFlowBranch(ExecuteContext=@n24.ExecutePin, Condition=@n22.Value.deformRig_r_upperarm_up_alpha_73_#35)
   RigVMDispatchNode(BoneName=root, Name=deformRig_r_upperarm_up_alpha, ExecuteContext=@n25.True, Value=@n23.OutputParam) {template="DISPATCH_RigDispatch_SetAnimAttribute(in Name,in BoneName,in Value,out Success)"}
  ]}
 ContainedGraph {nodes [
   n1 = RigVMFunctionEntryNode()
   SequenceExecution_2 = &RigVMFunction_Sequence(ExecuteContext=@n1.ExecuteContext)
   SequenceExecution_2_1 = &RigVMFunction_Sequence(ExecuteContext=@SequenceExecution_2.B)
   SequenceExecution_2_2 = &RigVMFunction_Sequence(ExecuteContext=@SequenceExecution_2_1.B)
   RigVMFunctionReturnNode(A=@SequenceExecution_2.A, B=@SequenceExecution_2_1.A, C=@SequenceExecution_2_2.A, D=@SequenceExecution_2_2.B)
  ]}
 ContainedGraph {nodes [
   n1 = RigVMFunctionEntryNode()
   SequenceExecution_3_1 = &RigVMFunction_Sequence(ExecuteContext=@n1.ExecuteContext)
   SequenceExecution_3_1_1 = &RigVMFunction_Sequence(ExecuteContext=@SequenceExecution_3_1.B)
   SequenceExecution_3_1_2 = &RigVMFunction_Sequence(ExecuteContext=@SequenceExecution_3_1_1.B)
   RigVMFunctionReturnNode(A=@SequenceExecution_3_1.A, B=@SequenceExecution_3_1_1.A, C=@SequenceExecution_3_1_2.A, D=@SequenceExecution_3_1_2.B)
  ]}
 ContainedGraph {nodes [
   n1 = RigVMFunctionEntryNode()
   SequenceExecution_3 = &RigVMFunction_Sequence(ExecuteContext=@n1.ExecuteContext)
   SequenceExecution_3_1 = &RigVMFunction_Sequence(ExecuteContext=@SequenceExecution_3.B)
   SequenceExecution_3_2 = &RigVMFunction_Sequence(ExecuteContext=@SequenceExecution_3_1.B)
   RigVMFunctionReturnNode(A=@SequenceExecution_3.A, B=@SequenceExecution_3_1.A, C=@SequenceExecution_3_2.A, D=@SequenceExecution_3_2.B)
  ]}
 ContainedGraph {nodes [
   n1 = RigVMFunctionEntryNode()
   SequenceExecution_3_1_1 = &RigVMFunction_Sequence(ExecuteContext=@n1.ExecuteContext)
   SequenceExecution_3_1_1_1 = &RigVMFunction_Sequence(ExecuteContext=@SequenceExecution_3_1_1.B)
   SequenceExecution_3_1_1_2 = &RigVMFunction_Sequence(ExecuteContext=@SequenceExecution_3_1_1_1.B)
   RigVMFunctionReturnNode(A=@SequenceExecution_3_1_1.A, B=@SequenceExecution_3_1_1_1.A, C=@SequenceExecution_3_1_1_2.A, D=@SequenceExecution_3_1_1_2.B)
  ]}
 New Function_ContainedGraph {nodes [
   RigVMFunctionEntryNode()
   > RigVMFunctionReturnNode()
  ]}
 New Function_ContainedGraph {nodes [
   RigVMFunctionEntryNode()
   > RigVMFunctionReturnNode()
  ]}
 New Function_ContainedGraph {nodes [
   RigVMFunctionEntryNode()
   > RigVMFunctionReturnNode()
  ]}
 New Function_ContainedGraph {nodes [
   RigVMFunctionEntryNode()
   > &RigUnit_ParentConstraint(
    AdvancedSettings=(RotationOrderForFilter=XZY)
    Child=(Type=Bone,Name="def_body_fur_22")
    Parents="((Item=(Type=Bone,Name=\"spine_02\"),Weight=0.0047),(Item=(Type=Bone,Name=\"deform_upperarm_up_l\"),Weight=0.0179),(Item=(Type=Bone,Name=\"deform_upperarm_bk_l\"),Weight=0.069),(Item=(Type=Bone,Name=\"upperarm_twist_01_l\"),Weight=0.0698),(Item=(Type=Bone,Name=\"sp ...(truncated)"
    Weight=1
   )
   > RigVMFunctionReturnNode()
   &RigUnit_ParentConstraint(
    AdvancedSettings=(RotationOrderForFilter=XZY)
    Child=(Type=Bone,Name="def_hand_l_fur_01")
    ExecutePin=()
    Parents="((Item=(Type=Bone,Name=\"middle_01_l\"),Weight=0.0008),(Item=(Type=Bone,Name=\"lowerarm_twist_01_l\"),Weight=0.011),(Item=(Type=Bone,Name=\"ring_metacarpal_l\"),Weight=0.0125),(Item=(Type=Bone,Name=\"index_01_l\"),Weight=0.0356),(Item=(Type=Bone,Name=\"hand_l\"),Wei ...(truncated)"
    Weight=1
   )
   > &RigUnit_ParentConstraint(
    AdvancedSettings=(RotationOrderForFilter=XZY)
    Child=(Type=Bone,Name="def_hand_l_fur_02")
    Parents="((Item=(Type=Bone,Name=\"middle_metacarpal_l\"),Weight=0.0195),(Item=(Type=Bone,Name=\"middle_01_l\"),Weight=0.0483),(Item=(Type=Bone,Name=\"index_metacarpal_l\"),Weight=0.0599),(Item=(Type=Bone,Name=\"index_02_l\"),Weight=0.0663),(Item=(Type=Bone,Name=\"index_01_l ...(truncated)"
    Weight=1
   )
   > &RigUnit_ParentConstraint(
    AdvancedSettings=(RotationOrderForFilter=XZY)
    Child=(Type=Bone,Name="def_hand_l_fur_03")
    Parents="((Item=(Type=Bone,Name=\"ring_metacarpal_l\"),Weight=0.0208),(Item=(Type=Bone,Name=\"middle_metacarpal_l\"),Weight=0.0229),(Item=(Type=Bone,Name=\"index_01_l\"),Weight=0.0375),(Item=(Type=Bone,Name=\"ring_01_l\"),Weight=0.0541),(Item=(Type=Bone,Name=\"middle_02_l\") ...(truncated)"
    Weight=1
   )
   > &RigUnit_ParentConstraint(
    AdvancedSettings=(RotationOrderForFilter=XZY)
    Child=(Type=Bone,Name="def_hand_l_fur_04")
    Parents="((Item=(Type=Bone,Name=\"index_01_l\"),Weight=0.0147),(Item=(Type=Bone,Name=\"pinky_01_l\"),Weight=0.0425),(Item=(Type=Bone,Name=\"middle_metacarpal_l\"),Weight=0.0722),(Item=(Type=Bone,Name=\"pinky_metacarpal_l\"),Weight=0.107),(Item=(Type=Bone,Name=\"middle_01_l\" ...(truncated)"
    Weight=1
   )
   > &RigUnit_ParentConstraint(
    AdvancedSettings=(RotationOrderForFilter=XZY)
    Child=(Type=Bone,Name="def_hand_l_fur_05")
    Parents="((Item=(Type=Bone,Name=\"pinky_02_l\"),Weight=0.0019),(Item=(Type=Bone,Name=\"ring_metacarpal_l\"),Weight=0.0104),(Item=(Type=Bone,Name=\"ring_01_l\"),Weight=0.045),(Item=(Type=Bone,Name=\"pinky_metacarpal_l\"),Weight=0.3684),(Item=(Type=Bone,Name=\"pinky_01_l\"),We ...(truncated)"
    Weight=1
   )
   > &RigUnit_ParentConstraint(
    AdvancedSettings=(RotationOrderForFilter=XZY)
    Child=(Type=Bone,Name="def_hand_l_fur_06")
    Parents=((Item=(Type=Bone,Name="ring_metacarpal_l"),Weight=0.001),(Item=(Type=Bone,Name="pinky_01_l"),Weight=0.1176),(Item=(Type=Bone,Name="pinky_metacarpal_l"),Weight=0.8813))
    Weight=1
   )
   > &RigUnit_ParentConstraint(
    AdvancedSettings=(RotationOrderForFilter=XZY)
    Child=(Type=Bone,Name="def_hand_l_fur_07")
    Parents=((Item=(Type=Bone,Name="lowerarm_twist_01_l"),Weight=0.0026),(Item=(Type=Bone,Name="hand_l"),Weight=0.0104),(Item=(Type=Bone,Name="pinky_metacarpal_l"),Weight=0.987))
    Weight=1
   )
   > &RigUnit_ParentConstraint(
    AdvancedSettings=(RotationOrderForFilter=XZY)
    Child=(Type=Bone,Name="def_hand_l_fur_08")
    Parents="((Item=(Type=Bone,Name=\"ring_metacarpal_l\"),Weight=0.0003),(Item=(Type=Bone,Name=\"pinky_metacarpal_l\"),Weight=0.0008),(Item=(Type=Bone,Name=\"deform_wrist_dn_l\"),Weight=0.0094),(Item=(Type=Bone,Name=\"lowerarm_twist_01_l\"),Weight=0.0182),(Item=(Type=Bone,Nam ...(truncated)"
    Weight=1
   )
   > &RigUnit_ParentConstraint(
    AdvancedSettings=(RotationOrderForFilter=XZY)
    Child=(Type=Bone,Name="def_hand_l_fur_09")
    Parents="((Item=(Type=Bone,Name=\"lowerarm_twist_01_l\"),Weight=0.0005),(Item=(Type=Bone,Name=\"middle_metacarpal_l\"),Weight=0.0047),(Item=(Type=Bone,Name=\"hand_l\"),Weight=0.0169),(Item=(Type=Bone,Name=\"thumb_01_l\"),Weight=0.229),(Item=(Type=Bone,Name=\"thumb_02_l\"),We ...(truncated)"
    Weight=1
   )
   > &RigUnit_ParentConstraint(
    AdvancedSettings=(RotationOrderForFilter=XZY)
    Child=(Type=Bone,Name="def_hand_l_fur_10")
    Parents="((Item=(Type=Bone,Name=\"thumb_02_l\"),Weight=0.0018),(Item=(Type=Bone,Name=\"middle_01_l\"),Weight=0.009),(Item=(Type=Bone,Name=\"middle_metacarpal_l\"),Weight=0.0206),(Item=(Type=Bone,Name=\"index_01_l\"),Weight=0.2741),(Item=(Type=Bone,Name=\"index_metacarpal_l\" ...(truncated)"
    Weight=1
   )
   > &RigUnit_ParentConstraint(
    AdvancedSettings=(RotationOrderForFilter=XZY)
    Child=(Type=Bone,Name="def_hand_l_fur_11")
    Parents="((Item=(Type=Bone,Name=\"lowerarm_l\"),Weight=0.0007),(Item=(Type=Bone,Name=\"lowerarm_twist_02_l\"),Weight=0.0035),(Item=(Type=Bone,Name=\"lowerarm_twist_01_l\"),Weight=0.0429),(Item=(Type=Bone,Name=\"thumb_02_l\"),Weight=0.0451),(Item=(Type=Bone,Name=\"ring_metac ...(truncated)"
    Weight=1
   )
   > &RigUnit_ParentConstraint(
    AdvancedSettings=(RotationOrderForFilter=XZY)
    Child=(Type=Bone,Name="def_hand_l_fur_12")
    Parents="((Item=(Type=Bone,Name=\"lowerarm_twist_01_l\"),Weight=0.0199),(Item=(Type=Bone,Name=\"ring_metacarpal_l\"),Weight=0.0219),(Item=(Type=Bone,Name=\"pinky_metacarpal_l\"),Weight=0.0324),(Item=(Type=Bone,Name=\"hand_l\"),Weight=0.0659),(Item=(Type=Bone,Name=\"thumb_02 ...(truncated)"
    Weight=1
   )
   > &RigUnit_ParentConstraint(
    AdvancedSettings=(RotationOrderForFilter=XZY)
    Child=(Type=Bone,Name="def_hand_l_fur_13")
    Parents="((Item=(Type=Bone,Name=\"thumb_02_l\"),Weight=0.0095),(Item=(Type=Bone,Name=\"lowerarm_twist_01_l\"),Weight=0.0806),(Item=(Type=Bone,Name=\"ring_metacarpal_l\"),Weight=0.0967),(Item=(Type=Bone,Name=\"hand_l\"),Weight=0.1175),(Item=(Type=Bone,Name=\"thumb_01_l\"),Wei ...(truncated)"
    Weight=1
   )
   > &RigUnit_ParentConstraint(
    AdvancedSettings=(RotationOrderForFilter=XZY)
    Child=(Type=Bone,Name="def_lowerarm_l_fur_01")
    Parents="((Item=(Type=Bone,Name=\"lowerarm_twist_01_l\"),Weight=0.0192),(Item=(Type=Bone,Name=\"upperarm_twist_02_l\"),Weight=0.0249),(Item=(Type=Bone,Name=\"deform_elbow_out_l\"),Weight=0.0498),(Item=(Type=Bone,Name=\"deform_elbow_in_l\"),Weight=0.1016),(Item=(Type=Bone,N ...(truncated)"
    Weight=1
   )
   > &RigUnit_ParentConstraint(
    AdvancedSettings=(RotationOrderForFilter=XZY)
    Child=(Type=Bone,Name="def_lowerarm_l_fur_02")
    Parents="((Item=(Type=Bone,Name=\"deform_elbow_in_l\"),Weight=0.0154),(Item=(Type=Bone,Name=\"upperarm_l\"),Weight=0.032),(Item=(Type=Bone,Name=\"lowerarm_twist_01_l\"),Weight=0.1804),(Item=(Type=Bone,Name=\"lowerarm_l\"),Weight=0.2599),(Item=(Type=Bone,Name=\"lowerarm_twis ...(truncated)"
    Weight=1
   )
   > &RigUnit_ParentConstraint(
    AdvancedSettings=(RotationOrderForFilter=XZY)
    Child=(Type=Bone,Name="def_lowerarm_l_fur_03")
    Parents="((Item=(Type=Bone,Name=\"thumb_01_l\"),Weight=0.002),(Item=(Type=Bone,Name=\"deform_wrist_dn_l\"),Weight=0.003),(Item=(Type=Bone,Name=\"hand_l\"),Weight=0.0035),(Item=(Type=Bone,Name=\"deform_wrist_up_l\"),Weight=0.0077),(Item=(Type=Bone,Name=\"lowerarm_l\"),Weight= ...(truncated)"
    Weight=1
   )
   > &RigUnit_ParentConstraint(
    AdvancedSettings=(RotationOrderForFilter=XZY)
    Child=(Type=Bone,Name="def_lowerarm_l_fur_04")
    Parents="((Item=(Type=Bone,Name=\"deform_elbow_out_l\"),Weight=0.0025),(Item=(Type=Bone,Name=\"upperarm_l\"),Weight=0.0083),(Item=(Type=Bone,Name=\"lowerarm_l\"),Weight=0.2027),(Item=(Type=Bone,Name=\"lowerarm_twist_01_l\"),Weight=0.2385),(Item=(Type=Bone,Name=\"lowerarm_tw ...(truncated)"
    Weight=1
   )
   > &RigUnit_ParentConstraint(
    AdvancedSettings=(RotationOrderForFilter=XZY)
    Child=(Type=Bone,Name="def_lowerarm_l_fur_05")
    Parents="((Item=(Type=Bone,Name=\"middle_metacarpal_l\"),Weight=0.0005),(Item=(Type=Bone,Name=\"ring_metacarpal_l\"),Weight=0.0024),(Item=(Type=Bone,Name=\"pinky_metacarpal_l\"),Weight=0.0041),(Item=(Type=Bone,Name=\"lowerarm_l\"),Weight=0.0067),(Item=(Type=Bone,Name=\"thum ...(truncated)"
    Weight=1
   )
   > &RigUnit_ParentConstraint(
    AdvancedSettings=(RotationOrderForFilter=XZY)
    Child=(Type=Bone,Name="def_lowerarm_l_fur_06")
    Parents="((Item=(Type=Bone,Name=\"thumb_02_l\"),Weight=0.0001),(Item=(Type=Bone,Name=\"ring_metacarpal_l\"),Weight=0.0013),(Item=(Type=Bone,Name=\"middle_metacarpal_l\"),Weight=0.0052),(Item=(Type=Bone,Name=\"lowerarm_twist_02_l\"),Weight=0.0229),(Item=(Type=Bone,Name=\"ind ...(truncated)"
    Weight=1
   )
   > &RigUnit_ParentConstraint(
    AdvancedSettings=(RotationOrderForFilter=XZY)
    Child=(Type=Bone,Name="def_lowerarm_l_fur_07")
    Parents="((Item=(Type=Bone,Name=\"lowerarm_l\"),Weight=0.0019),(Item=(Type=Bone,Name=\"ring_metacarpal_l\"),Weight=0.0022),(Item=(Type=Bone,Name=\"thumb_01_l\"),Weight=0.0089),(Item=(Type=Bone,Name=\"deform_wrist_up_l\"),Weight=0.0137),(Item=(Type=Bone,Name=\"lowerarm_twist ...(truncated)"
    Weight=1
   )
   > &RigUnit_ParentConstraint(
    AdvancedSettings=(RotationOrderForFilter=XZY)
    Child=(Type=Bone,Name="def_lowerarm_l_fur_08")
    Parents="((Item=(Type=Bone,Name=\"lowerarm_twist_01_l\"),Weight=0.0062),(Item=(Type=Bone,Name=\"upperarm_twist_02_l\"),Weight=0.0147),(Item=(Type=Bone,Name=\"lowerarm_twist_02_l\"),Weight=0.1213),(Item=(Type=Bone,Name=\"upperarm_l\"),Weight=0.1223),(Item=(Type=Bone,Name=\"l ...(truncated)"
    Weight=1
   )
   > &RigUnit_ParentConstraint(
    AdvancedSettings=(RotationOrderForFilter=XZY)
    Child=(Type=Bone,Name="def_upperarm_l_fur_06")
    Parents="((Item=(Type=Bone,Name=\"upperarm_twist_01_l\"),Weight=0.0015),(Item=(Type=Bone,Name=\"deform_elbow_in_l\"),Weight=0.0073),(Item=(Type=Bone,Name=\"lowerarm_twist_02_l\"),Weight=0.0387),(Item=(Type=Bone,Name=\"upperarm_twist_02_l\"),Weight=0.068),(Item=(Type=Bone,N ...(truncated)"
    Weight=1
   )
   > &RigUnit_ParentConstraint(
    AdvancedSettings=(RotationOrderForFilter=XZY)
    Child=(Type=Bone,Name="def_upperarm_l_fur_05")
    Parents=((Item=(Type=Bone,Name="lowerarm_l"),Weight=0.0118),(Item=(Type=Bone,Name="upperarm_twist_01_l"),Weight=0.1975),(Item=(Type=Bone,Name="upperarm_twist_02_l"),Weight=0.3202),(Item=(Type=Bone,Name="upperarm_l"),Weight=0.4705))
    Weight=1
   )
   > &RigUnit_ParentConstraint(
    AdvancedSettings=(RotationOrderForFilter=XZY)
    Child=(Type=Bone,Name="def_upperarm_l_fur_04")
    Parents="((Item=(Type=Bone,Name=\"lowerarm_twist_02_l\"),Weight=0.0149),(Item=(Type=Bone,Name=\"lowerarm_l\"),Weight=0.0577),(Item=(Type=Bone,Name=\"deform_elbow_in_l\"),Weight=0.0811),(Item=(Type=Bone,Name=\"upperarm_twist_01_l\"),Weight=0.1052),(Item=(Type=Bone,Name=\"upp ...(truncated)"
    Weight=1
   )
   > &RigUnit_ParentConstraint(
    AdvancedSettings=(RotationOrderForFilter=XZY)
    Child=(Type=Bone,Name="def_upperarm_l_fur_03")
    Parents="((Item=(Type=Bone,Name=\"deform_upperarm_fr_l\"),Weight=0.0037),(Item=(Type=Bone,Name=\"deform_elbow_in_l\"),Weight=0.0083),(Item=(Type=Bone,Name=\"lowerarm_l\"),Weight=0.0093),(Item=(Type=Bone,Name=\"upperarm_twist_01_l\"),Weight=0.2307),(Item=(Type=Bone,Name=\"up ...(truncated)"
    Weight=1
   )
   > &RigUnit_ParentConstraint(
    AdvancedSettings=(RotationOrderForFilter=XZY)
    Child=(Type=Bone,Name="def_upperarm_l_fur_02")
    Parents="((Item=(Type=Bone,Name=\"spine_03\"),Weight=0.008),(Item=(Type=Bone,Name=\"spine_05\"),Weight=0.0138),(Item=(Type=Bone,Name=\"upperarm_l\"),Weight=0.0442),(Item=(Type=Bone,Name=\"deform_upperarm_fr_l\"),Weight=0.0674),(Item=(Type=Bone,Name=\"clavicle_l\"),Weight=0.1 ...(truncated)"
    Weight=1
   )
   > &RigUnit_ParentConstraint(
    AdvancedSettings=(RotationOrderForFilter=XZY)
    Child=(Type=Bone,Name="def_upperarm_l_fur_01")
    Parents="((Item=(Type=Bone,Name=\"spine_04\"),Weight=0.0014),(Item=(Type=Bone,Name=\"spine_03\"),Weight=0.0151),(Item=(Type=Bone,Name=\"spine_05\"),Weight=0.0236),(Item=(Type=Bone,Name=\"upperarm_l\"),Weight=0.0239),(Item=(Type=Bone,Name=\"deform_upperarm_fr_l\"),Weight=0.10 ...(truncated)"
    Weight=1
   )
   > &RigUnit_ParentConstraint(
    AdvancedSettings=(RotationOrderForFilter=XZY)
    Child=(Type=Bone,Name="def_hand_r_fur_01")
    Parents="((Item=(Type=Bone,Name=\"middle_01_r\"),Weight=0.0008),(Item=(Type=Bone,Name=\"lowerarm_twist_01_r\"),Weight=0.011),(Item=(Type=Bone,Name=\"ring_metacarpal_r\"),Weight=0.0125),(Item=(Type=Bone,Name=\"index_01_r\"),Weight=0.0356),(Item=(Type=Bone,Name=\"hand_r\"),Wei ...(truncated)"
    Weight=1
   )
   > &RigUnit_ParentConstraint(
    AdvancedSettings=(RotationOrderForFilter=XZY)
    Child=(Type=Bone,Name="def_hand_r_fur_02")
    Parents="((Item=(Type=Bone,Name=\"middle_metacarpal_r\"),Weight=0.0195),(Item=(Type=Bone,Name=\"middle_01_r\"),Weight=0.0483),(Item=(Type=Bone,Name=\"index_metacarpal_r\"),Weight=0.0599),(Item=(Type=Bone,Name=\"index_02_r\"),Weight=0.0663),(Item=(Type=Bone,Name=\"index_01_r ...(truncated)"
    Weight=1
   )
   > &RigUnit_ParentConstraint(
    AdvancedSettings=(RotationOrderForFilter=XZY)
    Child=(Type=Bone,Name="def_hand_r_fur_03")
    Parents="((Item=(Type=Bone,Name=\"ring_metacarpal_r\"),Weight=0.0208),(Item=(Type=Bone,Name=\"middle_metacarpal_r\"),Weight=0.0229),(Item=(Type=Bone,Name=\"index_01_r\"),Weight=0.0375),(Item=(Type=Bone,Name=\"ring_01_r\"),Weight=0.0541),(Item=(Type=Bone,Name=\"middle_02_r\") ...(truncated)"
    Weight=1
   )
   > &RigUnit_ParentConstraint(
    AdvancedSettings=(RotationOrderForFilter=XZY)
    Child=(Type=Bone,Name="def_hand_r_fur_04")
    Parents="((Item=(Type=Bone,Name=\"index_01_r\"),Weight=0.0147),(Item=(Type=Bone,Name=\"pinky_01_r\"),Weight=0.0425),(Item=(Type=Bone,Name=\"middle_metacarpal_r\"),Weight=0.0722),(Item=(Type=Bone,Name=\"pinky_metacarpal_r\"),Weight=0.107),(Item=(Type=Bone,Name=\"middle_01_r\" ...(truncated)"
    Weight=1
   )
   > &RigUnit_ParentConstraint(
    AdvancedSettings=(RotationOrderForFilter=XZY)
    Child=(Type=Bone,Name="def_hand_r_fur_05")
    Parents="((Item=(Type=Bone,Name=\"pinky_02_r\"),Weight=0.0019),(Item=(Type=Bone,Name=\"ring_metacarpal_r\"),Weight=0.0104),(Item=(Type=Bone,Name=\"ring_01_r\"),Weight=0.045),(Item=(Type=Bone,Name=\"pinky_metacarpal_r\"),Weight=0.3684),(Item=(Type=Bone,Name=\"pinky_01_r\"),We ...(truncated)"
    Weight=1
   )
   > &RigUnit_ParentConstraint(
    AdvancedSettings=(RotationOrderForFilter=XZY)
    Child=(Type=Bone,Name="def_hand_r_fur_06")
    Parents=((Item=(Type=Bone,Name="ring_metacarpal_r"),Weight=0.001),(Item=(Type=Bone,Name="pinky_01_r"),Weight=0.1176),(Item=(Type=Bone,Name="pinky_metacarpal_r"),Weight=0.8813))
    Weight=1
   )
   > &RigUnit_ParentConstraint(
    AdvancedSettings=(RotationOrderForFilter=XZY)
    Child=(Type=Bone,Name="def_hand_r_fur_07")
    Parents=((Item=(Type=Bone,Name="lowerarm_twist_01_r"),Weight=0.0026),(Item=(Type=Bone,Name="hand_r"),Weight=0.0104),(Item=(Type=Bone,Name="pinky_metacarpal_r"),Weight=0.987))
    Weight=1
   )
   > &RigUnit_ParentConstraint(
    AdvancedSettings=(RotationOrderForFilter=XZY)
    Child=(Type=Bone,Name="def_hand_r_fur_08")
    Parents="((Item=(Type=Bone,Name=\"lowerarm_twist_01_r\"),Weight=0.0005),(Item=(Type=Bone,Name=\"middle_metacarpal_r\"),Weight=0.0047),(Item=(Type=Bone,Name=\"hand_r\"),Weight=0.0169),(Item=(Type=Bone,Name=\"thumb_01_r\"),Weight=0.229),(Item=(Type=Bone,Name=\"thumb_02_r\"),We ...(truncated)"
    Weight=1
   )
   > &RigUnit_ParentConstraint(
    AdvancedSettings=(RotationOrderForFilter=XZY)
    Child=(Type=Bone,Name="def_hand_r_fur_09")
    Parents="((Item=(Type=Bone,Name=\"thumb_02_r\"),Weight=0.0018),(Item=(Type=Bone,Name=\"middle_01_r\"),Weight=0.009),(Item=(Type=Bone,Name=\"middle_metacarpal_r\"),Weight=0.0206),(Item=(Type=Bone,Name=\"index_01_r\"),Weight=0.2741),(Item=(Type=Bone,Name=\"index_metacarpal_r\" ...(truncated)"
    Weight=1
   )
   > &RigUnit_ParentConstraint(
    AdvancedSettings=(RotationOrderForFilter=XZY)
    Child=(Type=Bone,Name="def_hand_r_fur_10")
    Parents="((Item=(Type=Bone,Name=\"ring_metacarpal_r\"),Weight=0.0003),(Item=(Type=Bone,Name=\"pinky_metacarpal_r\"),Weight=0.0008),(Item=(Type=Bone,Name=\"deform_wrist_dn_r\"),Weight=0.0094),(Item=(Type=Bone,Name=\"lowerarm_twist_01_r\"),Weight=0.0182),(Item=(Type=Bone,Nam ...(truncated)"
    Weight=1
   )
   > &RigUnit_ParentConstraint(
    AdvancedSettings=(RotationOrderForFilter=XZY)
    Child=(Type=Bone,Name="def_hand_r_fur_11")
    Parents="((Item=(Type=Bone,Name=\"thumb_02_r\"),Weight=0.0095),(Item=(Type=Bone,Name=\"lowerarm_twist_01_r\"),Weight=0.0806),(Item=(Type=Bone,Name=\"ring_metacarpal_r\"),Weight=0.0967),(Item=(Type=Bone,Name=\"hand_r\"),Weight=0.1175),(Item=(Type=Bone,Name=\"thumb_01_r\"),Wei ...(truncated)"
    Weight=1
   )
   > &RigUnit_ParentConstraint(
    AdvancedSettings=(RotationOrderForFilter=XZY)
    Child=(Type=Bone,Name="def_hand_r_fur_12")
    Parents="((Item=(Type=Bone,Name=\"lowerarm_r\"),Weight=0.0007),(Item=(Type=Bone,Name=\"lowerarm_twist_02_r\"),Weight=0.0035),(Item=(Type=Bone,Name=\"lowerarm_twist_01_r\"),Weight=0.0429),(Item=(Type=Bone,Name=\"thumb_02_r\"),Weight=0.0451),(Item=(Type=Bone,Name=\"ring_metac ...(truncated)"
    Weight=1
   )
   > &RigUnit_ParentConstraint(
    AdvancedSettings=(RotationOrderForFilter=XZY)
    Child=(Type=Bone,Name="def_hand_r_fur_13")
    Parents="((Item=(Type=Bone,Name=\"lowerarm_twist_01_r\"),Weight=0.0199),(Item=(Type=Bone,Name=\"ring_metacarpal_r\"),Weight=0.0219),(Item=(Type=Bone,Name=\"pinky_metacarpal_r\"),Weight=0.0324),(Item=(Type=Bone,Name=\"hand_r\"),Weight=0.0659),(Item=(Type=Bone,Name=\"thumb_02 ...(truncated)"
    Weight=1
   )
   > &RigUnit_ParentConstraint(
    AdvancedSettings=(RotationOrderForFilter=XZY)
    Child=(Type=Bone,Name="def_lowerarm_r_fur_01")
    Parents="((Item=(Type=Bone,Name=\"lowerarm_twist_01_r\"),Weight=0.0192),(Item=(Type=Bone,Name=\"upperarm_twist_02_r\"),Weight=0.0248),(Item=(Type=Bone,Name=\"deform_elbow_out_r\"),Weight=0.0498),(Item=(Type=Bone,Name=\"deform_elbow_in_r\"),Weight=0.1016),(Item=(Type=Bone,N ...(truncated)"
    Weight=1
   )
   > &RigUnit_ParentConstraint(
    AdvancedSettings=(RotationOrderForFilter=XZY)
    Child=(Type=Bone,Name="def_lowerarm_r_fur_02")
    Parents="((Item=(Type=Bone,Name=\"deform_elbow_in_r\"),Weight=0.0271),(Item=(Type=Bone,Name=\"upperarm_r\"),Weight=0.0449),(Item=(Type=Bone,Name=\"lowerarm_twist_01_r\"),Weight=0.1532),(Item=(Type=Bone,Name=\"lowerarm_r\"),Weight=0.2824),(Item=(Type=Bone,Name=\"lowerarm_twi ...(truncated)"
    Weight=1
   )
   > &RigUnit_ParentConstraint(
    AdvancedSettings=(RotationOrderForFilter=XZY)
    Child=(Type=Bone,Name="def_lowerarm_r_fur_03")
    Parents="((Item=(Type=Bone,Name=\"thumb_01_r\"),Weight=0.002),(Item=(Type=Bone,Name=\"deform_wrist_dn_r\"),Weight=0.003),(Item=(Type=Bone,Name=\"hand_r\"),Weight=0.0035),(Item=(Type=Bone,Name=\"deform_wrist_up_r\"),Weight=0.0077),(Item=(Type=Bone,Name=\"lowerarm_r\"),Weight= ...(truncated)"
    Weight=1
   )
   > &RigUnit_ParentConstraint(
    AdvancedSettings=(RotationOrderForFilter=XZY)
    Child=(Type=Bone,Name="def_lowerarm_r_fur_04")
    Parents="((Item=(Type=Bone,Name=\"deform_elbow_out_r\"),Weight=0.0025),(Item=(Type=Bone,Name=\"upperarm_r\"),Weight=0.0083),(Item=(Type=Bone,Name=\"lowerarm_r\"),Weight=0.2027),(Item=(Type=Bone,Name=\"lowerarm_twist_01_r\"),Weight=0.2385),(Item=(Type=Bone,Name=\"lowerarm_tw ...(truncated)"
    Weight=1
   )
   > &RigUnit_ParentConstraint(
    AdvancedSettings=(RotationOrderForFilter=XZY)
    Child=(Type=Bone,Name="def_lowerarm_r_fur_05")
    Parents="((Item=(Type=Bone,Name=\"middle_metacarpal_r\"),Weight=0.0007),(Item=(Type=Bone,Name=\"ring_metacarpal_r\"),Weight=0.0028),(Item=(Type=Bone,Name=\"pinky_metacarpal_r\"),Weight=0.0045),(Item=(Type=Bone,Name=\"lowerarm_r\"),Weight=0.0069),(Item=(Type=Bone,Name=\"thum ...(truncated)"
    Weight=1
   )
   > &RigUnit_ParentConstraint(
    AdvancedSettings=(RotationOrderForFilter=XZY)
    Child=(Type=Bone,Name="def_lowerarm_r_fur_06")
    Parents="((Item=(Type=Bone,Name=\"thumb_02_r\"),Weight=0.0001),(Item=(Type=Bone,Name=\"ring_metacarpal_r\"),Weight=0.0013),(Item=(Type=Bone,Name=\"middle_metacarpal_r\"),Weight=0.0052),(Item=(Type=Bone,Name=\"lowerarm_twist_02_r\"),Weight=0.0229),(Item=(Type=Bone,Name=\"ind ...(truncated)"
    Weight=1
   )
   > &RigUnit_ParentConstraint(
    AdvancedSettings=(RotationOrderForFilter=XZY)
    Child=(Type=Bone,Name="def_lowerarm_r_fur_07")
    Parents="((Item=(Type=Bone,Name=\"lowerarm_r\"),Weight=0.0019),(Item=(Type=Bone,Name=\"ring_metacarpal_r\"),Weight=0.0022),(Item=(Type=Bone,Name=\"thumb_01_r\"),Weight=0.0089),(Item=(Type=Bone,Name=\"deform_wrist_up_r\"),Weight=0.0137),(Item=(Type=Bone,Name=\"lowerarm_twist ...(truncated)"
    Weight=1
   )
   > &RigUnit_ParentConstraint(
    AdvancedSettings=(RotationOrderForFilter=XZY)
    Child=(Type=Bone,Name="def_lowerarm_r_fur_08")
    Parents="((Item=(Type=Bone,Name=\"lowerarm_twist_01_r\"),Weight=0.0062),(Item=(Type=Bone,Name=\"upperarm_twist_02_r\"),Weight=0.0147),(Item=(Type=Bone,Name=\"lowerarm_twist_02_r\"),Weight=0.1213),(Item=(Type=Bone,Name=\"upperarm_r\"),Weight=0.1223),(Item=(Type=Bone,Name=\"l ...(truncated)"
    Weight=1
   )
   > &RigUnit_ParentConstraint(
    AdvancedSettings=(RotationOrderForFilter=XZY)
    Child=(Type=Bone,Name="def_upperarm_r_fur_06")
    Parents="((Item=(Type=Bone,Name=\"upperarm_twist_01_r\"),Weight=0.0015),(Item=(Type=Bone,Name=\"deform_elbow_in_r\"),Weight=0.0073),(Item=(Type=Bone,Name=\"lowerarm_twist_02_r\"),Weight=0.0387),(Item=(Type=Bone,Name=\"upperarm_twist_02_r\"),Weight=0.068),(Item=(Type=Bone,N ...(truncated)"
    Weight=1
   )
   > &RigUnit_ParentConstraint(
    AdvancedSettings=(RotationOrderForFilter=XZY)
    Child=(Type=Bone,Name="def_upperarm_r_fur_05")
    Parents=((Item=(Type=Bone,Name="lowerarm_r"),Weight=0.0118),(Item=(Type=Bone,Name="upperarm_twist_01_r"),Weight=0.1975),(Item=(Type=Bone,Name="upperarm_twist_02_r"),Weight=0.3202),(Item=(Type=Bone,Name="upperarm_r"),Weight=0.4705))
    Weight=1
   )
   > &RigUnit_ParentConstraint(
    AdvancedSettings=(RotationOrderForFilter=XZY)
    Child=(Type=Bone,Name="def_upperarm_r_fur_04")
    Parents="((Item=(Type=Bone,Name=\"deform_upperarm_fr_r\"),Weight=0.0001),(Item=(Type=Bone,Name=\"lowerarm_twist_02_r\"),Weight=0.014),(Item=(Type=Bone,Name=\"lowerarm_r\"),Weight=0.0565),(Item=(Type=Bone,Name=\"deform_elbow_in_r\"),Weight=0.0793),(Item=(Type=Bone,Name=\"upp ...(truncated)"
    Weight=1
   )
   > &RigUnit_ParentConstraint(
    AdvancedSettings=(RotationOrderForFilter=XZY)
    Child=(Type=Bone,Name="def_upperarm_r_fur_03")
    Parents="((Item=(Type=Bone,Name=\"deform_upperarm_fr_r\"),Weight=0.0035),(Item=(Type=Bone,Name=\"deform_elbow_in_r\"),Weight=0.0082),(Item=(Type=Bone,Name=\"lowerarm_r\"),Weight=0.0093),(Item=(Type=Bone,Name=\"upperarm_twist_01_r\"),Weight=0.2299),(Item=(Type=Bone,Name=\"up ...(truncated)"
    Weight=1
   )
   > &RigUnit_ParentConstraint(
    AdvancedSettings=(RotationOrderForFilter=XZY)
    Child=(Type=Bone,Name="def_body_fur_01")
    Parents="((Item=(Type=Bone,Name=\"spine_02\"),Weight=0.0044),(Item=(Type=Bone,Name=\"deform_upperarm_up_l\"),Weight=0.0171),(Item=(Type=Bone,Name=\"upperarm_twist_01_l\"),Weight=0.0634),(Item=(Type=Bone,Name=\"deform_upperarm_bk_l\"),Weight=0.0664),(Item=(Type=Bone,Name=\"s ...(truncated)"
    Weight=1
   )
   > &RigUnit_ParentConstraint(
    AdvancedSettings=(RotationOrderForFilter=XZY)
    Child=(Type=Bone,Name="def_body_fur_02")
    Parents="((Item=(Type=Bone,Name=\"deform_upperarm_bk_l\"),Weight=0.0396),(Item=(Type=Bone,Name=\"pelvis\"),Weight=0.0508),(Item=(Type=Bone,Name=\"spine_01\"),Weight=0.1075),(Item=(Type=Bone,Name=\"clavicle_l\"),Weight=0.1126),(Item=(Type=Bone,Name=\"spine_04\"),Weight=0.1401 ...(truncated)"
    Weight=1
   )
   > &RigUnit_ParentConstraint(
    AdvancedSettings=(RotationOrderForFilter=XZY)
    Child=(Type=Bone,Name="def_body_fur_03")
    Parents="((Item=(Type=Bone,Name=\"clavicle_r\"),Weight=0.0101),(Item=(Type=Bone,Name=\"pelvis\"),Weight=0.0107),(Item=(Type=Bone,Name=\"spine_01\"),Weight=0.0487),(Item=(Type=Bone,Name=\"clavicle_l\"),Weight=0.0755),(Item=(Type=Bone,Name=\"spine_05\"),Weight=0.0783),(Item=(T ...(truncated)"
    Weight=1
   )
   > &RigUnit_ParentConstraint(
    AdvancedSettings=(RotationOrderForFilter=XZY)
    Child=(Type=Bone,Name="def_body_fur_04")
    Parents=((Item=(Type=Bone,Name="spine_04"),Weight=0.0022),(Item=(Type=Bone,Name="spine_03"),Weight=0.1183),(Item=(Type=Bone,Name="spine_01"),Weight=0.1689),(Item=(Type=Bone,Name="spine_02"),Weight=0.2081),(Item=(Type=Bone,Name="pelvis"),Weight=0.5026))
    Weight=1
   )
   > &RigUnit_ParentConstraint(
    AdvancedSettings=(RotationOrderForFilter=XZY)
    Child=(Type=Bone,Name="def_body_fur_05")
    Parents="((Item=(Type=Bone,Name=\"clavicle_l\"),Weight=0.0117),(Item=(Type=Bone,Name=\"spine_04\"),Weight=0.0674),(Item=(Type=Bone,Name=\"pelvis\"),Weight=0.1408),(Item=(Type=Bone,Name=\"spine_01\"),Weight=0.1694),(Item=(Type=Bone,Name=\"spine_03\"),Weight=0.2986),(Item=(Typ ...(truncated)"
    Weight=1
   )
   > &RigUnit_ParentConstraint(
    AdvancedSettings=(RotationOrderForFilter=XZY)
    Child=(Type=Bone,Name="def_body_fur_06")
    Parents="((Item=(Type=Bone,Name=\"thigh_twist_01_l\"),Weight=0.0034),(Item=(Type=Bone,Name=\"spine_03\"),Weight=0.0147),(Item=(Type=Bone,Name=\"spine_01\"),Weight=0.0217),(Item=(Type=Bone,Name=\"spine_02\"),Weight=0.0229),(Item=(Type=Bone,Name=\"deform_glute_l\"),Weight=0.02 ...(truncated)"
    Weight=1
   )
   > &RigUnit_ParentConstraint(
    AdvancedSettings=(RotationOrderForFilter=XZY)
    Child=(Type=Bone,Name="def_body_fur_07")
    Parents="((Item=(Type=Bone,Name=\"thigh_twist_01_r\"),Weight=0.0003),(Item=(Type=Bone,Name=\"thigh_twist_01_l\"),Weight=0.0025),(Item=(Type=Bone,Name=\"deform_glute_l\"),Weight=0.0091),(Item=(Type=Bone,Name=\"spine_03\"),Weight=0.0102),(Item=(Type=Bone,Name=\"spine_02\"),Wei ...(truncated)"
    Weight=1
   )
   > &RigUnit_ParentConstraint(
    AdvancedSettings=(RotationOrderForFilter=XZY)
    Child=(Type=Bone,Name="def_body_fur_08")
    Parents="((Item=(Type=Bone,Name=\"clavicle_l\"),Weight=0.0005),(Item=(Type=Bone,Name=\"spine_05\"),Weight=0.0033),(Item=(Type=Bone,Name=\"clavicle_r\"),Weight=0.0201),(Item=(Type=Bone,Name=\"pelvis\"),Weight=0.0802),(Item=(Type=Bone,Name=\"spine_04\"),Weight=0.1082),(Item=(T ...(truncated)"
    Weight=1
   )
   > &RigUnit_ParentConstraint(
    AdvancedSettings=(RotationOrderForFilter=XZY)
    Child=(Type=Bone,Name="def_body_fur_09")
    Parents="((Item=(Type=Bone,Name=\"deform_upperarm_bk_r\"),Weight=0.021),(Item=(Type=Bone,Name=\"pelvis\"),Weight=0.0733),(Item=(Type=Bone,Name=\"clavicle_r\"),Weight=0.0809),(Item=(Type=Bone,Name=\"spine_04\"),Weight=0.1118),(Item=(Type=Bone,Name=\"spine_01\"),Weight=0.1304) ...(truncated)"
    Weight=1
   )
   > &RigUnit_ParentConstraint(
    AdvancedSettings=(RotationOrderForFilter=XZY)
    Child=(Type=Bone,Name="def_body_fur_10")
    Parents=((Item=(Type=Bone,Name="spine_03"),Weight=0.0903),(Item=(Type=Bone,Name="spine_01"),Weight=0.1337),(Item=(Type=Bone,Name="spine_02"),Weight=0.1602),(Item=(Type=Bone,Name="pelvis"),Weight=0.6158))
    Weight=1
   )
   > &RigUnit_ParentConstraint(
    AdvancedSettings=(RotationOrderForFilter=XZY)
    Child=(Type=Bone,Name="def_body_fur_11")
    Parents="((Item=(Type=Bone,Name=\"deform_glute_l\"),Weight=0.0003),(Item=(Type=Bone,Name=\"spine_03\"),Weight=0.0011),(Item=(Type=Bone,Name=\"spine_02\"),Weight=0.0019),(Item=(Type=Bone,Name=\"spine_01\"),Weight=0.002),(Item=(Type=Bone,Name=\"thigh_twist_01_l\"),Weight=0.011 ...(truncated)"
    Weight=1
   )
   > &RigUnit_ParentConstraint(
    AdvancedSettings=(RotationOrderForFilter=XZY)
    Child=(Type=Bone,Name="def_body_fur_12")
    Parents="((Item=(Type=Bone,Name=\"deform_groin_l\"),Weight=0.0007),(Item=(Type=Bone,Name=\"thigh_twist_01_l\"),Weight=0.0017),(Item=(Type=Bone,Name=\"thigh_twist_01_r\"),Weight=0.0065),(Item=(Type=Bone,Name=\"deform_groin_r\"),Weight=0.0082),(Item=(Type=Bone,Name=\"spine_03 ...(truncated)"
    Weight=1
   )
   > &RigUnit_ParentConstraint(
    AdvancedSettings=(RotationOrderForFilter=XZY)
    Child=(Type=Bone,Name="def_body_fur_13")
    Parents="((Item=(Type=Bone,Name=\"deform_glute_r\"),Weight=0.0001),(Item=(Type=Bone,Name=\"thigh_r\"),Weight=0.0056),(Item=(Type=Bone,Name=\"deform_groin_l\"),Weight=0.019),(Item=(Type=Bone,Name=\"thigh_twist_01_l\"),Weight=0.0313),(Item=(Type=Bone,Name=\"thigh_twist_01_r\") ...(truncated)"
    Weight=1
   )
   > &RigUnit_ParentConstraint(
    AdvancedSettings=(RotationOrderForFilter=XZY)
    Child=(Type=Bone,Name="def_body_fur_14")
    Parents=((Item=(Type=Bone,Name="spine_04"),Weight=0.0353),(Item=(Type=Bone,Name="pelvis"),Weight=0.1877),(Item=(Type=Bone,Name="spine_01"),Weight=0.2043),(Item=(Type=Bone,Name="spine_03"),Weight=0.2459),(Item=(Type=Bone,Name="spine_02"),Weight=0.3268))
    Weight=1
   )
   > &RigUnit_ParentConstraint(
    AdvancedSettings=(RotationOrderForFilter=XZY)
    Child=(Type=Bone,Name="def_body_fur_15")
    Parents=((Item=(Type=Bone,Name="spine_04"),Weight=0.0121),(Item=(Type=Bone,Name="spine_03"),Weight=0.1342),(Item=(Type=Bone,Name="spine_01"),Weight=0.1741),(Item=(Type=Bone,Name="spine_02"),Weight=0.2177),(Item=(Type=Bone,Name="pelvis"),Weight=0.4619))
    Weight=1
   )
   > &RigUnit_ParentConstraint(
    AdvancedSettings=(RotationOrderForFilter=XZY)
    Child=(Type=Bone,Name="def_body_fur_16")
    Parents="((Item=(Type=Bone,Name=\"thigh_twist_01_r\"),Weight=0.0032),(Item=(Type=Bone,Name=\"spine_04\"),Weight=0.0033),(Item=(Type=Bone,Name=\"deform_groin_r\"),Weight=0.0053),(Item=(Type=Bone,Name=\"spine_03\"),Weight=0.0732),(Item=(Type=Bone,Name=\"spine_01\"),Weight=0.10 ...(truncated)"
    Weight=1
   )
   > &RigUnit_ParentConstraint(
    AdvancedSettings=(RotationOrderForFilter=XZY)
    Child=(Type=Bone,Name="def_body_fur_17")
    Parents="((Item=(Type=Bone,Name=\"upperarm_twist_02_l\"),Weight=0.0289),(Item=(Type=Bone,Name=\"spine_05\"),Weight=0.0389),(Item=(Type=Bone,Name=\"spine_02\"),Weight=0.0434),(Item=(Type=Bone,Name=\"upperarm_twist_01_l\"),Weight=0.0881),(Item=(Type=Bone,Name=\"deform_upperar ...(truncated)"
    Weight=1
   )
   > &RigUnit_ParentConstraint(
    AdvancedSettings=(RotationOrderForFilter=XZY)
    Child=(Type=Bone,Name="def_body_fur_18")
    Parents="((Item=(Type=Bone,Name=\"deform_upperarm_fr_l\"),Weight=0.0015),(Item=(Type=Bone,Name=\"upperarm_twist_01_l\"),Weight=0.0063),(Item=(Type=Bone,Name=\"spine_05\"),Weight=0.0141),(Item=(Type=Bone,Name=\"spine_01\"),Weight=0.0441),(Item=(Type=Bone,Name=\"clavicle_l\"), ...(truncated)"
    Weight=1
   )
   > &RigUnit_ParentConstraint(
    AdvancedSettings=(RotationOrderForFilter=XZY)
    Child=(Type=Bone,Name="def_body_fur_19")
    Parents=((Item=(Type=Bone,Name="spine_04"),Weight=0.0387),(Item=(Type=Bone,Name="pelvis"),Weight=0.1561),(Item=(Type=Bone,Name="spine_01"),Weight=0.1993),(Item=(Type=Bone,Name="spine_03"),Weight=0.2659),(Item=(Type=Bone,Name="spine_02"),Weight=0.34))
    Weight=1
   )
   > &RigUnit_ParentConstraint(
    AdvancedSettings=(RotationOrderForFilter=XZY)
    Child=(Type=Bone,Name="def_body_fur_20")
    Parents="((Item=(Type=Bone,Name=\"deform_upperarm_fr_l\"),Weight=0.0027),(Item=(Type=Bone,Name=\"spine_01\"),Weight=0.0208),(Item=(Type=Bone,Name=\"clavicle_l\"),Weight=0.0799),(Item=(Type=Bone,Name=\"spine_05\"),Weight=0.0915),(Item=(Type=Bone,Name=\"spine_02\"),Weight=0.10 ...(truncated)"
    Weight=1
   )
   > &RigUnit_ParentConstraint(
    AdvancedSettings=(RotationOrderForFilter=XZY)
    Child=(Type=Bone,Name="def_body_fur_21")
    Parents="((Item=(Type=Bone,Name=\"pelvis\"),Weight=0.0038),(Item=(Type=Bone,Name=\"clavicle_l\"),Weight=0.0374),(Item=(Type=Bone,Name=\"spine_05\"),Weight=0.0378),(Item=(Type=Bone,Name=\"deform_pec_l\"),Weight=0.0382),(Item=(Type=Bone,Name=\"spine_01\"),Weight=0.0531),(Item= ...(truncated)"
    Weight=1
   )
   > &RigUnit_ParentConstraint(
    AdvancedSettings=(RotationOrderForFilter=XZY)
    Child=(Type=Bone,Name="def_body_fur_22")
    Parents="((Item=(Type=Bone,Name=\"pelvis\"),Weight=0.0073),(Item=(Type=Bone,Name=\"deform_pec_r\"),Weight=0.0087),(Item=(Type=Bone,Name=\"clavicle_r\"),Weight=0.0146),(Item=(Type=Bone,Name=\"spine_05\"),Weight=0.0359),(Item=(Type=Bone,Name=\"spine_01\"),Weight=0.0555),(Item= ...(truncated)"
    Weight=1
   )
   > &RigUnit_ParentConstraint(
    AdvancedSettings=(RotationOrderForFilter=XZY)
    Child=(Type=Bone,Name="def_body_fur_23")
    Parents="((Item=(Type=Bone,Name=\"spine_01\"),Weight=0.0049),(Item=(Type=Bone,Name=\"clavicle_r\"),Weight=0.042),(Item=(Type=Bone,Name=\"deform_pec_r\"),Weight=0.0642),(Item=(Type=Bone,Name=\"spine_02\"),Weight=0.0985),(Item=(Type=Bone,Name=\"spine_05\"),Weight=0.168),(Item= ...(truncated)"
    Weight=1
   )
   > &RigUnit_ParentConstraint(
    AdvancedSettings=(RotationOrderForFilter=XZY)
    Child=(Type=Bone,Name="def_body_fur_24")
    Parents="((Item=(Type=Bone,Name=\"spine_01\"),Weight=0.0129),(Item=(Type=Bone,Name=\"clavicle_l\"),Weight=0.0288),(Item=(Type=Bone,Name=\"deform_pec_l\"),Weight=0.0471),(Item=(Type=Bone,Name=\"spine_02\"),Weight=0.1158),(Item=(Type=Bone,Name=\"spine_05\"),Weight=0.1399),(Ite ...(truncated)"
    Weight=1
   )
   > &RigUnit_ParentConstraint(
    AdvancedSettings=(RotationOrderForFilter=XZY)
    Child=(Type=Bone,Name="def_body_fur_25")
    Parents="((Item=(Type=Bone,Name=\"upperarm_twist_01_r\"),Weight=0.0057),(Item=(Type=Bone,Name=\"deform_upperarm_fr_r\"),Weight=0.0068),(Item=(Type=Bone,Name=\"spine_02\"),Weight=0.0565),(Item=(Type=Bone,Name=\"clavicle_r\"),Weight=0.1214),(Item=(Type=Bone,Name=\"spine_03\"), ...(truncated)"
    Weight=1
   )
   > &RigUnit_ParentConstraint(
    AdvancedSettings=(RotationOrderForFilter=XZY)
    Child=(Type=Bone,Name="def_body_fur_26")
    Parents="((Item=(Type=Bone,Name=\"deform_upperarm_fr_r\"),Weight=0.0028),(Item=(Type=Bone,Name=\"neck_02\"),Weight=0.0031),(Item=(Type=Bone,Name=\"upperarm_twist_01_r\"),Weight=0.0054),(Item=(Type=Bone,Name=\"spine_02\"),Weight=0.0257),(Item=(Type=Bone,Name=\"head\"),Weight= ...(truncated)"
    Weight=1
   )
   > &RigUnit_ParentConstraint(
    AdvancedSettings=(RotationOrderForFilter=XZY)
    Child=(Type=Bone,Name="def_body_fur_27")
    Parents="((Item=(Type=Bone,Name=\"neck_01\"),Weight=0.0014),(Item=(Type=Bone,Name=\"neck_02\"),Weight=0.0038),(Item=(Type=Bone,Name=\"deform_pec_l\"),Weight=0.0241),(Item=(Type=Bone,Name=\"spine_02\"),Weight=0.0297),(Item=(Type=Bone,Name=\"spine_03\"),Weight=0.0905),(Item=(T ...(truncated)"
    Weight=1
   )
   > &RigUnit_ParentConstraint(
    AdvancedSettings=(RotationOrderForFilter=XZY)
    Child=(Type=Bone,Name="def_body_fur_28")
    Parents="((Item=(Type=Bone,Name=\"deform_upperarm_fr_l\"),Weight=0.0197),(Item=(Type=Bone,Name=\"upperarm_twist_01_l\"),Weight=0.0204),(Item=(Type=Bone,Name=\"spine_02\"),Weight=0.0438),(Item=(Type=Bone,Name=\"spine_03\"),Weight=0.114),(Item=(Type=Bone,Name=\"clavicle_l\"),W ...(truncated)"
    Weight=1
   )
   > &RigUnit_ParentConstraint(
    AdvancedSettings=(RotationOrderForFilter=XZY)
    Child=(Type=Bone,Name="def_body_fur_29")
    Parents="((Item=(Type=Bone,Name=\"spine_02\"),Weight=0.0068),(Item=(Type=Bone,Name=\"deform_upperarm_up_l\"),Weight=0.0174),(Item=(Type=Bone,Name=\"deform_upperarm_fr_l\"),Weight=0.042),(Item=(Type=Bone,Name=\"upperarm_twist_01_l\"),Weight=0.044),(Item=(Type=Bone,Name=\"spi ...(truncated)"
    Weight=1
   )
   > &RigUnit_ParentConstraint(
    AdvancedSettings=(RotationOrderForFilter=XZY)
    Child=(Type=Bone,Name="def_body_fur_30")
    Parents="((Item=(Type=Bone,Name=\"spine_03\"),Weight=0.0002),(Item=(Type=Bone,Name=\"spine_01\"),Weight=0.0003),(Item=(Type=Bone,Name=\"spine_02\"),Weight=0.0016),(Item=(Type=Bone,Name=\"thigh_twist_01_r\"),Weight=0.0186),(Item=(Type=Bone,Name=\"deform_groin_r\"),Weight=0.02 ...(truncated)"
    Weight=1
   )
   > &RigUnit_ParentConstraint(
    AdvancedSettings=(RotationOrderForFilter=XZY)
    Child=(Type=Bone,Name="def_body_fur_31")
    Parents="((Item=(Type=Bone,Name=\"deform_glute_l\"),Weight=0.0021),(Item=(Type=Bone,Name=\"deform_groin_r\"),Weight=0.0028),(Item=(Type=Bone,Name=\"thigh_twist_01_r\"),Weight=0.0175),(Item=(Type=Bone,Name=\"thigh_l\"),Weight=0.0328),(Item=(Type=Bone,Name=\"pelvis\"),Weight=0 ...(truncated)"
    Weight=1
   )
   > &RigUnit_ParentConstraint(
    AdvancedSettings=(RotationOrderForFilter=XZY)
    Child=(Type=Bone,Name="def_calf_l_fur_01")
    Parents="((Item=(Type=Bone,Name=\"deform_glute_r\"),Weight=0.0015),(Item=(Type=Bone,Name=\"deform_groin_l\"),Weight=0.0015),(Item=(Type=Bone,Name=\"thigh_twist_01_l\"),Weight=0.0144),(Item=(Type=Bone,Name=\"thigh_r\"),Weight=0.0402),(Item=(Type=Bone,Name=\"pelvis\"),Weight=0 ...(truncated)"
    Weight=1
   )
   > &RigUnit_ParentConstraint(
    AdvancedSettings=(RotationOrderForFilter=XZY)
    Child=(Type=Bone,Name="def_calf_l_fur_01")
    Parents=((Item=(Type=Bone,Name="calf_twist_01_l"),Weight=0.0374),(Item=(Type=Bone,Name="calf_l"),Weight=0.2694),(Item=(Type=Bone,Name="calf_twist_02_l"),Weight=0.6932))
    Weight=1
   )
   > &RigUnit_ParentConstraint(
    AdvancedSettings=(RotationOrderForFilter=XZY)
    Child=(Type=Bone,Name="def_calf_l_fur_02")
    Parents=((Item=(Type=Bone,Name="calf_twist_01_l"),Weight=0.0805),(Item=(Type=Bone,Name="calf_l"),Weight=0.2047),(Item=(Type=Bone,Name="calf_twist_02_l"),Weight=0.7148))
    Weight=1
   )
   > &RigUnit_ParentConstraint(
    AdvancedSettings=(RotationOrderForFilter=XZY)
    Child=(Type=Bone,Name="def_calf_l_fur_03")
    Parents=((Item=(Type=Bone,Name="calf_l"),Weight=0.0809),(Item=(Type=Bone,Name="calf_twist_01_l"),Weight=0.228),(Item=(Type=Bone,Name="calf_twist_02_l"),Weight=0.6911))
    Weight=1
   )
   > &RigUnit_ParentConstraint(
    AdvancedSettings=(RotationOrderForFilter=XZY)
    Child=(Type=Bone,Name="def_calf_l_fur_04")
    Parents=((Item=(Type=Bone,Name="calf_l"),Weight=0.0356),(Item=(Type=Bone,Name="calf_twist_01_l"),Weight=0.3823),(Item=(Type=Bone,Name="calf_twist_02_l"),Weight=0.5821))
    Weight=1
   )
   > &RigUnit_ParentConstraint(
    AdvancedSettings=(RotationOrderForFilter=XZY)
    Child=(Type=Bone,Name="def_calf_l_fur_05")
    Parents=((Item=(Type=Bone,Name="calf_l"),Weight=0.0004),(Item=(Type=Bone,Name="calf_twist_02_l"),Weight=0.3115),(Item=(Type=Bone,Name="calf_twist_01_l"),Weight=0.6881))
    Weight=1
   )
   > &RigUnit_ParentConstraint(
    AdvancedSettings=(RotationOrderForFilter=XZY)
    Child=(Type=Bone,Name="def_calf_l_fur_06")
    Parents=((Item=(Type=Bone,Name="calf_twist_02_l"),Weight=0.0286),(Item=(Type=Bone,Name="foot_l"),Weight=0.3487),(Item=(Type=Bone,Name="calf_twist_01_l"),Weight=0.6227))
    Weight=1
   )
   > &RigUnit_ParentConstraint(
    AdvancedSettings=(RotationOrderForFilter=XZY)
    Child=(Type=Bone,Name="def_calf_l_fur_07")
    Parents=((Item=(Type=Bone,Name="calf_l"),Weight=0.0342),(Item=(Type=Bone,Name="calf_twist_01_l"),Weight=0.4008),(Item=(Type=Bone,Name="calf_twist_02_l"),Weight=0.565))
    Weight=1
   )
   > &RigUnit_ParentConstraint(
    AdvancedSettings=(RotationOrderForFilter=XZY)
    Child=(Type=Bone,Name="def_calf_l_fur_08")
    Parents=((Item=(Type=Bone,Name="calf_l"),Weight=0.0042),(Item=(Type=Bone,Name="foot_l"),Weight=0.0704),(Item=(Type=Bone,Name="calf_twist_02_l"),Weight=0.2252),(Item=(Type=Bone,Name="calf_twist_01_l"),Weight=0.7002))
    Weight=1
   )
   > &RigUnit_ParentConstraint(
    AdvancedSettings=(RotationOrderForFilter=XZY)
    Child=(Type=Bone,Name="def_calf_l_fur_09")
    Parents=((Item=(Type=Bone,Name="calf_l"),Weight=0.0002),(Item=(Type=Bone,Name="calf_twist_02_l"),Weight=0.0204),(Item=(Type=Bone,Name="calf_twist_01_l"),Weight=0.3309),(Item=(Type=Bone,Name="foot_l"),Weight=0.6485))
    Weight=1
   )
   > &RigUnit_ParentConstraint(
    AdvancedSettings=(RotationOrderForFilter=XZY)
    Child=(Type=Bone,Name="def_calf_l_fur_10")
    Parents=((Item=(Type=Bone,Name="calf_twist_02_l"),Weight=0.0118),(Item=(Type=Bone,Name="calf_twist_01_l"),Weight=0.4401),(Item=(Type=Bone,Name="foot_l"),Weight=0.5481))
    Weight=1
   )
   > &RigUnit_ParentConstraint(
    AdvancedSettings=(RotationOrderForFilter=XZY)
    Child=(Type=Bone,Name="def_calf_l_fur_11")
    Parents=((Item=(Type=Bone,Name="foot_l"),Weight=0.2287),(Item=(Type=Bone,Name="calf_twist_02_l"),Weight=0.2803),(Item=(Type=Bone,Name="calf_twist_01_l"),Weight=0.491))
    Weight=1
   )
   > &RigUnit_ParentConstraint(
    AdvancedSettings=(RotationOrderForFilter=XZY)
    Child=(Type=Bone,Name="def_calf_l_fur_12")
    Parents=((Item=(Type=Bone,Name="calf_twist_01_l"),Weight=0.0301),(Item=(Type=Bone,Name="calf_l"),Weight=0.2759),(Item=(Type=Bone,Name="calf_twist_02_l"),Weight=0.6939))
    Weight=1
   )
   > &RigUnit_ParentConstraint(
    AdvancedSettings=(RotationOrderForFilter=XZY)
    Child=(Type=Bone,Name="def_calf_l_fur_13")
    Parents=((Item=(Type=Bone,Name="calf_l"),Weight=0.0539),(Item=(Type=Bone,Name="calf_twist_01_l"),Weight=0.2411),(Item=(Type=Bone,Name="calf_twist_02_l"),Weight=0.705))
    Weight=1
   )
   > &RigUnit_ParentConstraint(
    AdvancedSettings=(RotationOrderForFilter=XZY)
    Child=(Type=Bone,Name="def_calf_l_fur_14")
    Parents=((Item=(Type=Bone,Name="calf_l"),Weight=0.0242),(Item=(Type=Bone,Name="calf_twist_01_l"),Weight=0.3753),(Item=(Type=Bone,Name="calf_twist_02_l"),Weight=0.6004))
    Weight=1
   )
   > &RigUnit_ParentConstraint(
    AdvancedSettings=(RotationOrderForFilter=XZY)
    Child=(Type=Bone,Name="def_calf_l_fur_15")
    Parents=((Item=(Type=Bone,Name="calf_twist_01_l"),Weight=0.0995),(Item=(Type=Bone,Name="calf_l"),Weight=0.1402),(Item=(Type=Bone,Name="calf_twist_02_l"),Weight=0.7603))
    Weight=1
   )
   > &RigUnit_ParentConstraint(
    AdvancedSettings=(RotationOrderForFilter=XZY)
    Child=(Type=Bone,Name="def_calf_l_fur_16")
    Parents=((Item=(Type=Bone,Name="calf_l"),Weight=0.0277),(Item=(Type=Bone,Name="calf_twist_01_l"),Weight=0.322),(Item=(Type=Bone,Name="calf_twist_02_l"),Weight=0.6504))
    Weight=1
   )
   > &RigUnit_ParentConstraint(
    AdvancedSettings=(RotationOrderForFilter=XZY)
    Child=(Type=Bone,Name="def_thigh_l_fur_01")
    Parents=((Item=(Type=Bone,Name="thigh_l"),Weight=0.0124),(Item=(Type=Bone,Name="deform_glute_l"),Weight=0.1541),(Item=(Type=Bone,Name="pelvis"),Weight=0.1779),(Item=(Type=Bone,Name="thigh_twist_01_l"),Weight=0.6556))
    Weight=1
   )
   > &RigUnit_ParentConstraint(
    AdvancedSettings=(RotationOrderForFilter=XZY)
    Child=(Type=Bone,Name="def_thigh_l_fur_02")
    Parents="((Item=(Type=Bone,Name=\"deform_knee_in_l\"),Weight=0.002),(Item=(Type=Bone,Name=\"deform_knee_out_l\"),Weight=0.0032),(Item=(Type=Bone,Name=\"calf_l\"),Weight=0.0078),(Item=(Type=Bone,Name=\"thigh_twist_01_l\"),Weight=0.4106),(Item=(Type=Bone,Name=\"thigh_l\"),Weig ...(truncated)"
    Weight=1
   )
   > &RigUnit_ParentConstraint(
    AdvancedSettings=(RotationOrderForFilter=XZY)
    Child=(Type=Bone,Name="def_thigh_l_fur_03")
    Parents=((Item=(Type=Bone,Name="calf_l"),Weight=0.0875),(Item=(Type=Bone,Name="thigh_twist_01_l"),Weight=0.1068),(Item=(Type=Bone,Name="deform_knee_out_l"),Weight=0.2387),(Item=(Type=Bone,Name="thigh_l"),Weight=0.567))
    Weight=1
   )
   > &RigUnit_ParentConstraint(
    AdvancedSettings=(RotationOrderForFilter=XZY)
    Child=(Type=Bone,Name="def_thigh_l_fur_04")
    Parents=((Item=(Type=Bone,Name="deform_knee_in_l"),Weight=0.0005),(Item=(Type=Bone,Name="thigh_l"),Weight=0.4562),(Item=(Type=Bone,Name="thigh_twist_01_l"),Weight=0.5433))
    Weight=1
   )
   > &RigUnit_ParentConstraint(
    AdvancedSettings=(RotationOrderForFilter=XZY)
    Child=(Type=Bone,Name="def_thigh_l_fur_05")
    Parents="((Item=(Type=Bone,Name=\"thigh_twist_01_l\"),Weight=0.0058),(Item=(Type=Bone,Name=\"calf_twist_02_l\"),Weight=0.0402),(Item=(Type=Bone,Name=\"calf_l\"),Weight=0.273),(Item=(Type=Bone,Name=\"deform_knee_in_l\"),Weight=0.293),(Item=(Type=Bone,Name=\"thigh_l\"),Weight= ...(truncated)"
    Weight=1
   )
   > &RigUnit_ParentConstraint(
    AdvancedSettings=(RotationOrderForFilter=XZY)
    Child=(Type=Bone,Name="def_calf_r_fur_01")
    Parents=((Item=(Type=Bone,Name="calf_twist_01_r"),Weight=0.0374),(Item=(Type=Bone,Name="calf_r"),Weight=0.2694),(Item=(Type=Bone,Name="calf_twist_02_r"),Weight=0.6932))
    Weight=1
   )
   > &RigUnit_ParentConstraint(
    AdvancedSettings=(RotationOrderForFilter=XZY)
    Child=(Type=Bone,Name="def_calf_r_fur_02")
    Parents=((Item=(Type=Bone,Name="calf_twist_01_r"),Weight=0.0805),(Item=(Type=Bone,Name="calf_r"),Weight=0.2047),(Item=(Type=Bone,Name="calf_twist_02_r"),Weight=0.7148))
    Weight=1
   )
   > &RigUnit_ParentConstraint(
    AdvancedSettings=(RotationOrderForFilter=XZY)
    Child=(Type=Bone,Name="def_calf_r_fur_03")
    Parents=((Item=(Type=Bone,Name="calf_r"),Weight=0.0809),(Item=(Type=Bone,Name="calf_twist_01_r"),Weight=0.228),(Item=(Type=Bone,Name="calf_twist_02_r"),Weight=0.6911))
    Weight=1
   )
   > &RigUnit_ParentConstraint(
    AdvancedSettings=(RotationOrderForFilter=XZY)
    Child=(Type=Bone,Name="def_calf_r_fur_04")
    Parents=((Item=(Type=Bone,Name="calf_r"),Weight=0.0356),(Item=(Type=Bone,Name="calf_twist_01_r"),Weight=0.3823),(Item=(Type=Bone,Name="calf_twist_02_r"),Weight=0.5821))
    Weight=1
   )
   > &RigUnit_ParentConstraint(
    AdvancedSettings=(RotationOrderForFilter=XZY)
    Child=(Type=Bone,Name="def_calf_r_fur_05")
    Parents=((Item=(Type=Bone,Name="calf_r"),Weight=0.0005),(Item=(Type=Bone,Name="calf_twist_02_r"),Weight=0.3123),(Item=(Type=Bone,Name="calf_twist_01_r"),Weight=0.6873))
    Weight=1
   )
   > &RigUnit_ParentConstraint(
    AdvancedSettings=(RotationOrderForFilter=XZY)
    Child=(Type=Bone,Name="def_calf_r_fur_06")
    Parents=((Item=(Type=Bone,Name="calf_twist_02_r"),Weight=0.0286),(Item=(Type=Bone,Name="foot_r"),Weight=0.3487),(Item=(Type=Bone,Name="calf_twist_01_r"),Weight=0.6227))
    Weight=1
   )
   > &RigUnit_ParentConstraint(
    AdvancedSettings=(RotationOrderForFilter=XZY)
    Child=(Type=Bone,Name="def_calf_r_fur_07")
    Parents=((Item=(Type=Bone,Name="calf_r"),Weight=0.0286),(Item=(Type=Bone,Name="calf_twist_01_r"),Weight=0.4336),(Item=(Type=Bone,Name="calf_twist_02_r"),Weight=0.5377))
    Weight=1
   )
   > &RigUnit_ParentConstraint(
    AdvancedSettings=(RotationOrderForFilter=XZY)
    Child=(Type=Bone,Name="def_calf_r_fur_08")
    Parents=((Item=(Type=Bone,Name="calf_r"),Weight=0.0042),(Item=(Type=Bone,Name="foot_r"),Weight=0.0718),(Item=(Type=Bone,Name="calf_twist_02_r"),Weight=0.2232),(Item=(Type=Bone,Name="calf_twist_01_r"),Weight=0.7007))
    Weight=1
   )
   > &RigUnit_ParentConstraint(
    AdvancedSettings=(RotationOrderForFilter=XZY)
    Child=(Type=Bone,Name="def_calf_r_fur_09")
    Parents=((Item=(Type=Bone,Name="calf_r"),Weight=0.0002),(Item=(Type=Bone,Name="calf_twist_02_r"),Weight=0.0204),(Item=(Type=Bone,Name="calf_twist_01_r"),Weight=0.3309),(Item=(Type=Bone,Name="foot_r"),Weight=0.6485))
    Weight=1
   )
   > &RigUnit_ParentConstraint(
    AdvancedSettings=(RotationOrderForFilter=XZY)
    Child=(Type=Bone,Name="def_calf_r_fur_10")
    Parents=((Item=(Type=Bone,Name="calf_twist_02_r"),Weight=0.0118),(Item=(Type=Bone,Name="calf_twist_01_r"),Weight=0.4401),(Item=(Type=Bone,Name="foot_r"),Weight=0.5481))
    Weight=1
   )
   > &RigUnit_ParentConstraint(
    AdvancedSettings=(RotationOrderForFilter=XZY)
    Child=(Type=Bone,Name="def_calf_r_fur_11")
    Parents=((Item=(Type=Bone,Name="foot_r"),Weight=0.2287),(Item=(Type=Bone,Name="calf_twist_02_r"),Weight=0.2803),(Item=(Type=Bone,Name="calf_twist_01_r"),Weight=0.491))
    Weight=1
   )
   > &RigUnit_ParentConstraint(
    AdvancedSettings=(RotationOrderForFilter=XZY)
    Child=(Type=Bone,Name="def_calf_r_fur_12")
    Parents=((Item=(Type=Bone,Name="calf_twist_01_r"),Weight=0.0301),(Item=(Type=Bone,Name="calf_r"),Weight=0.2759),(Item=(Type=Bone,Name="calf_twist_02_r"),Weight=0.6939))
    Weight=1
   )
   > &RigUnit_ParentConstraint(
    AdvancedSettings=(RotationOrderForFilter=XZY)
    Child=(Type=Bone,Name="def_calf_r_fur_13")
    Parents=((Item=(Type=Bone,Name="calf_r"),Weight=0.0488),(Item=(Type=Bone,Name="calf_twist_01_r"),Weight=0.2497),(Item=(Type=Bone,Name="calf_twist_02_r"),Weight=0.7016))
    Weight=1
   )
   > &RigUnit_ParentConstraint(
    AdvancedSettings=(RotationOrderForFilter=XZY)
    Child=(Type=Bone,Name="def_calf_r_fur_14")
    Parents=((Item=(Type=Bone,Name="calf_r"),Weight=0.0205),(Item=(Type=Bone,Name="calf_twist_01_r"),Weight=0.4033),(Item=(Type=Bone,Name="calf_twist_02_r"),Weight=0.5762))
    Weight=1
   )
   > &RigUnit_ParentConstraint(
    AdvancedSettings=(RotationOrderForFilter=XZY)
    Child=(Type=Bone,Name="def_calf_r_fur_15")
    Parents=((Item=(Type=Bone,Name="calf_twist_01_r"),Weight=0.0979),(Item=(Type=Bone,Name="calf_r"),Weight=0.1428),(Item=(Type=Bone,Name="calf_twist_02_r"),Weight=0.7594))
    Weight=1
   )
   > &RigUnit_ParentConstraint(
    AdvancedSettings=(RotationOrderForFilter=XZY)
    Child=(Type=Bone,Name="def_calf_r_fur_16")
    Parents=((Item=(Type=Bone,Name="calf_r"),Weight=0.0277),(Item=(Type=Bone,Name="calf_twist_01_r"),Weight=0.322),(Item=(Type=Bone,Name="calf_twist_02_r"),Weight=0.6504))
    Weight=1
   )
   > &RigUnit_ParentConstraint(
    AdvancedSettings=(RotationOrderForFilter=XZY)
    Child=(Type=Bone,Name="def_thigh_r_fur_01")
    Parents="((Item=(Type=Bone,Name=\"deform_knee_in_r\"),Weight=0.0022),(Item=(Type=Bone,Name=\"deform_knee_out_r\"),Weight=0.0039),(Item=(Type=Bone,Name=\"calf_r\"),Weight=0.0087),(Item=(Type=Bone,Name=\"thigh_twist_01_r\"),Weight=0.4035),(Item=(Type=Bone,Name=\"thigh_r\"),Wei ...(truncated)"
    Weight=1
   )
   > &RigUnit_ParentConstraint(
    AdvancedSettings=(RotationOrderForFilter=XZY)
    Child=(Type=Bone,Name="def_thigh_r_fur_02")
    Parents="((Item=(Type=Bone,Name=\"thigh_twist_01_r\"),Weight=0.0063),(Item=(Type=Bone,Name=\"calf_twist_02_r\"),Weight=0.0432),(Item=(Type=Bone,Name=\"deform_knee_in_r\"),Weight=0.2399),(Item=(Type=Bone,Name=\"calf_r\"),Weight=0.2935),(Item=(Type=Bone,Name=\"thigh_r\"),Weigh ...(truncated)"
    Weight=1
   )
   &RigUnit_ParentConstraint(
    AdvancedSettings=(RotationOrderForFilter=XZY)
    Child=(Type=Bone,Name="def_hand_l_fur_01")
    ExecutePin=()
    Parents="((Item=(Type=Bone,Name=\"lowerarm_twist_01_l\"),Weight=0.011),(Item=(Type=Bone,Name=\"ring_metacarpal_l\"),Weight=0.0125),(Item=(Type=Bone,Name=\"index_01_l\"),Weight=0.0357),(Item=(Type=Bone,Name=\"hand_l\"),Weight=0.0401),(Item=(Type=Bone,Name=\"middle_metacarpal ...(truncated)"
    Weight=1
   )
   > &RigUnit_ParentConstraint(
    AdvancedSettings=(RotationOrderForFilter=XZY)
    Child=(Type=Bone,Name="def_hand_l_fur_09")
    Parents="((Item=(Type=Bone,Name=\"lowerarm_twist_01_l\"),Weight=0.0005),(Item=(Type=Bone,Name=\"middle_metacarpal_l\"),Weight=0.0047),(Item=(Type=Bone,Name=\"hand_l\"),Weight=0.0169),(Item=(Type=Bone,Name=\"thumb_01_l\"),Weight=0.229),(Item=(Type=Bone,Name=\"thumb_02_l\"),We ...(truncated)"
    Weight=1
   )
   > &RigUnit_ParentConstraint(
    AdvancedSettings=(RotationOrderForFilter=XZY)
    Child=(Type=Bone,Name="def_hand_l_fur_03")
    Parents="((Item=(Type=Bone,Name=\"ring_metacarpal_l\"),Weight=0.0208),(Item=(Type=Bone,Name=\"middle_metacarpal_l\"),Weight=0.0229),(Item=(Type=Bone,Name=\"index_01_l\"),Weight=0.0375),(Item=(Type=Bone,Name=\"ring_01_l\"),Weight=0.0541),(Item=(Type=Bone,Name=\"middle_02_l\") ...(truncated)"
    Weight=1
   )
   > &RigUnit_ParentConstraint(
    AdvancedSettings=(RotationOrderForFilter=XZY)
    Child=(Type=Bone,Name="def_hand_l_fur_08")
    Parents="((Item=(Type=Bone,Name=\"ring_metacarpal_l\"),Weight=0.0003),(Item=(Type=Bone,Name=\"pinky_metacarpal_l\"),Weight=0.0008),(Item=(Type=Bone,Name=\"deform_wrist_dn_l\"),Weight=0.0094),(Item=(Type=Bone,Name=\"lowerarm_twist_01_l\"),Weight=0.0182),(Item=(Type=Bone,Nam ...(truncated)"
    Weight=1
   )
   > &RigUnit_ParentConstraint(
    AdvancedSettings=(RotationOrderForFilter=XZY)
    Child=(Type=Bone,Name="def_hand_l_fur_04")
    Parents="((Item=(Type=Bone,Name=\"index_01_l\"),Weight=0.0147),(Item=(Type=Bone,Name=\"pinky_01_l\"),Weight=0.0425),(Item=(Type=Bone,Name=\"middle_metacarpal_l\"),Weight=0.0722),(Item=(Type=Bone,Name=\"pinky_metacarpal_l\"),Weight=0.107),(Item=(Type=Bone,Name=\"middle_01_l\" ...(truncated)"
    Weight=1
   )
   > &RigUnit_ParentConstraint(
    AdvancedSettings=(RotationOrderForFilter=XZY)
    Child=(Type=Bone,Name="def_hand_l_fur_02")
    Parents="((Item=(Type=Bone,Name=\"middle_metacarpal_l\"),Weight=0.0195),(Item=(Type=Bone,Name=\"middle_01_l\"),Weight=0.0483),(Item=(Type=Bone,Name=\"index_metacarpal_l\"),Weight=0.0599),(Item=(Type=Bone,Name=\"index_02_l\"),Weight=0.0663),(Item=(Type=Bone,Name=\"index_01_l ...(truncated)"
    Weight=1
   )
   > &RigUnit_ParentConstraint(
    AdvancedSettings=(RotationOrderForFilter=XZY)
    Child=(Type=Bone,Name="def_hand_l_fur_05")
    Parents="((Item=(Type=Bone,Name=\"pinky_02_l\"),Weight=0.0019),(Item=(Type=Bone,Name=\"ring_metacarpal_l\"),Weight=0.0104),(Item=(Type=Bone,Name=\"ring_01_l\"),Weight=0.045),(Item=(Type=Bone,Name=\"pinky_metacarpal_l\"),Weight=0.3684),(Item=(Type=Bone,Name=\"pinky_01_l\"),We ...(truncated)"
    Weight=1
   )
   > &RigUnit_ParentConstraint(
    AdvancedSettings=(RotationOrderForFilter=XZY)
    Child=(Type=Bone,Name="def_hand_l_fur_13")
    Parents="((Item=(Type=Bone,Name=\"thumb_02_l\"),Weight=0.0095),(Item=(Type=Bone,Name=\"lowerarm_twist_01_l\"),Weight=0.0806),(Item=(Type=Bone,Name=\"ring_metacarpal_l\"),Weight=0.0967),(Item=(Type=Bone,Name=\"hand_l\"),Weight=0.1175),(Item=(Type=Bone,Name=\"thumb_01_l\"),Wei ...(truncated)"
    Weight=1
   )
   > &RigUnit_ParentConstraint(
    AdvancedSettings=(RotationOrderForFilter=XZY)
    Child=(Type=Bone,Name="def_lowerarm_l_fur_06")
    Parents="((Item=(Type=Bone,Name=\"ring_metacarpal_l\"),Weight=0.0013),(Item=(Type=Bone,Name=\"middle_metacarpal_l\"),Weight=0.0052),(Item=(Type=Bone,Name=\"lowerarm_twist_02_l\"),Weight=0.0229),(Item=(Type=Bone,Name=\"index_metacarpal_l\"),Weight=0.0765),(Item=(Type=Bone,N ...(truncated)"
    Weight=1
   )
   > &RigUnit_ParentConstraint(
    AdvancedSettings=(RotationOrderForFilter=XZY)
    Child=(Type=Bone,Name="def_hand_r_fur_03")
    Parents="((Item=(Type=Bone,Name=\"ring_metacarpal_r\"),Weight=0.0208),(Item=(Type=Bone,Name=\"middle_metacarpal_r\"),Weight=0.0229),(Item=(Type=Bone,Name=\"index_01_r\"),Weight=0.0375),(Item=(Type=Bone,Name=\"ring_01_r\"),Weight=0.0541),(Item=(Type=Bone,Name=\"middle_02_r\") ...(truncated)"
    Weight=1
   )
   > &RigUnit_ParentConstraint(
    AdvancedSettings=(RotationOrderForFilter=XZY)
    Child=(Type=Bone,Name="def_hand_r_fur_02")
    Parents="((Item=(Type=Bone,Name=\"middle_metacarpal_r\"),Weight=0.0195),(Item=(Type=Bone,Name=\"middle_01_r\"),Weight=0.0483),(Item=(Type=Bone,Name=\"index_metacarpal_r\"),Weight=0.0599),(Item=(Type=Bone,Name=\"index_02_r\"),Weight=0.0663),(Item=(Type=Bone,Name=\"index_01_r ...(truncated)"
    Weight=1
   )
   > &RigUnit_ParentConstraint(
    AdvancedSettings=(RotationOrderForFilter=XZY)
    Child=(Type=Bone,Name="def_hand_r_fur_10")
    Parents="((Item=(Type=Bone,Name=\"ring_metacarpal_r\"),Weight=0.0003),(Item=(Type=Bone,Name=\"pinky_metacarpal_r\"),Weight=0.0008),(Item=(Type=Bone,Name=\"deform_wrist_dn_r\"),Weight=0.0094),(Item=(Type=Bone,Name=\"lowerarm_twist_01_r\"),Weight=0.0182),(Item=(Type=Bone,Nam ...(truncated)"
    Weight=1
   )
   > &RigUnit_ParentConstraint(
    AdvancedSettings=(RotationOrderForFilter=XZY)
    Child=(Type=Bone,Name="def_hand_r_fur_04")
    Parents="((Item=(Type=Bone,Name=\"index_01_r\"),Weight=0.0147),(Item=(Type=Bone,Name=\"pinky_01_r\"),Weight=0.0425),(Item=(Type=Bone,Name=\"middle_metacarpal_r\"),Weight=0.0722),(Item=(Type=Bone,Name=\"pinky_metacarpal_r\"),Weight=0.107),(Item=(Type=Bone,Name=\"middle_01_r\" ...(truncated)"
    Weight=1
   )
   > &RigUnit_ParentConstraint(
    AdvancedSettings=(RotationOrderForFilter=XZY)
    Child=(Type=Bone,Name="def_hand_r_fur_08")
    Parents="((Item=(Type=Bone,Name=\"lowerarm_twist_01_r\"),Weight=0.0005),(Item=(Type=Bone,Name=\"middle_metacarpal_r\"),Weight=0.0047),(Item=(Type=Bone,Name=\"hand_r\"),Weight=0.0169),(Item=(Type=Bone,Name=\"thumb_01_r\"),Weight=0.229),(Item=(Type=Bone,Name=\"thumb_02_r\"),We ...(truncated)"
    Weight=1
   )
   > &RigUnit_ParentConstraint(
    AdvancedSettings=(RotationOrderForFilter=XZY)
    Child=(Type=Bone,Name="def_hand_r_fur_01")
    Parents="((Item=(Type=Bone,Name=\"lowerarm_twist_01_r\"),Weight=0.011),(Item=(Type=Bone,Name=\"ring_metacarpal_r\"),Weight=0.0125),(Item=(Type=Bone,Name=\"index_01_r\"),Weight=0.0357),(Item=(Type=Bone,Name=\"hand_r\"),Weight=0.0401),(Item=(Type=Bone,Name=\"middle_metacarpal ...(truncated)"
    Weight=1
   )
   > &RigUnit_ParentConstraint(
    AdvancedSettings=(RotationOrderForFilter=XZY)
    Child=(Type=Bone,Name="def_hand_r_fur_05")
    Parents="((Item=(Type=Bone,Name=\"pinky_02_r\"),Weight=0.0019),(Item=(Type=Bone,Name=\"ring_metacarpal_r\"),Weight=0.0104),(Item=(Type=Bone,Name=\"ring_01_r\"),Weight=0.045),(Item=(Type=Bone,Name=\"pinky_metacarpal_r\"),Weight=0.3684),(Item=(Type=Bone,Name=\"pinky_01_r\"),We ...(truncated)"
    Weight=1
   )
   > &RigUnit_ParentConstraint(
    AdvancedSettings=(RotationOrderForFilter=XZY)
    Child=(Type=Bone,Name="def_hand_r_fur_11")
    Parents="((Item=(Type=Bone,Name=\"thumb_02_r\"),Weight=0.0095),(Item=(Type=Bone,Name=\"lowerarm_twist_01_r\"),Weight=0.0806),(Item=(Type=Bone,Name=\"ring_metacarpal_r\"),Weight=0.0967),(Item=(Type=Bone,Name=\"hand_r\"),Weight=0.1175),(Item=(Type=Bone,Name=\"thumb_01_r\"),Wei ...(truncated)"
    Weight=1
   )
   > &RigUnit_ParentConstraint(
    AdvancedSettings=(RotationOrderForFilter=XZY)
    Child=(Type=Bone,Name="def_lowerarm_r_fur_06")
    Parents="((Item=(Type=Bone,Name=\"ring_metacarpal_r\"),Weight=0.0013),(Item=(Type=Bone,Name=\"middle_metacarpal_r\"),Weight=0.0052),(Item=(Type=Bone,Name=\"lowerarm_twist_02_r\"),Weight=0.0229),(Item=(Type=Bone,Name=\"index_metacarpal_r\"),Weight=0.0765),(Item=(Type=Bone,N ...(truncated)"
    Weight=1
   )
   &RigUnit_ParentConstraint(
    AdvancedSettings=(RotationOrderForFilter=XZY)
    Child=(Type=Bone,Name="def_lowerarm_l_fur_04")
    ExecutePin=()
    Parents="((Item=(Type=Bone,Name=\"deform_elbow_out_l\"),Weight=0.0025),(Item=(Type=Bone,Name=\"upperarm_l\"),Weight=0.0083),(Item=(Type=Bone,Name=\"lowerarm_l\"),Weight=0.2027),(Item=(Type=Bone,Name=\"lowerarm_twist_01_l\"),Weight=0.2385),(Item=(Type=Bone,Name=\"lowerarm_tw ...(truncated)"
    Weight=1
   )
   > &RigUnit_ParentConstraint(
    AdvancedSettings=(RotationOrderForFilter=XZY)
    Child=(Type=Bone,Name="def_lowerarm_l_fur_07")
    Parents="((Item=(Type=Bone,Name=\"ring_metacarpal_l\"),Weight=0.0022),(Item=(Type=Bone,Name=\"thumb_01_l\"),Weight=0.0089),(Item=(Type=Bone,Name=\"deform_wrist_up_l\"),Weight=0.0138),(Item=(Type=Bone,Name=\"lowerarm_twist_02_l\"),Weight=0.0268),(Item=(Type=Bone,Name=\"defor ...(truncated)"
    Weight=1
   )
   > &RigUnit_ParentConstraint(
    AdvancedSettings=(RotationOrderForFilter=XZY)
    Child=(Type=Bone,Name="def_lowerarm_l_fur_08")
    Parents="((Item=(Type=Bone,Name=\"lowerarm_twist_01_l\"),Weight=0.0062),(Item=(Type=Bone,Name=\"upperarm_twist_02_l\"),Weight=0.0147),(Item=(Type=Bone,Name=\"lowerarm_twist_02_l\"),Weight=0.1213),(Item=(Type=Bone,Name=\"upperarm_l\"),Weight=0.1223),(Item=(Type=Bone,Name=\"l ...(truncated)"
    Weight=1
   )
   > &RigUnit_ParentConstraint(
    AdvancedSettings=(RotationOrderForFilter=XZY)
    Child=(Type=Bone,Name="def_lowerarm_l_fur_01")
    Parents="((Item=(Type=Bone,Name=\"lowerarm_twist_01_l\"),Weight=0.0192),(Item=(Type=Bone,Name=\"upperarm_twist_02_l\"),Weight=0.0249),(Item=(Type=Bone,Name=\"deform_elbow_out_l\"),Weight=0.0498),(Item=(Type=Bone,Name=\"deform_elbow_in_l\"),Weight=0.1016),(Item=(Type=Bone,N ...(truncated)"
    Weight=1
   )
   > &RigUnit_ParentConstraint(
    AdvancedSettings=(RotationOrderForFilter=XZY)
    Child=(Type=Bone,Name="def_lowerarm_l_fur_02")
    Parents="((Item=(Type=Bone,Name=\"deform_elbow_in_l\"),Weight=0.0154),(Item=(Type=Bone,Name=\"upperarm_l\"),Weight=0.032),(Item=(Type=Bone,Name=\"lowerarm_twist_01_l\"),Weight=0.1804),(Item=(Type=Bone,Name=\"lowerarm_l\"),Weight=0.2599),(Item=(Type=Bone,Name=\"lowerarm_twis ...(truncated)"
    Weight=1
   )
   &RigUnit_ParentConstraint(
    AdvancedSettings=(RotationOrderForFilter=XZY)
    Child=(Type=Bone,Name="def_upperarm_l_fur_06")
    ExecutePin=()
    Parents="((Item=(Type=Bone,Name=\"upperarm_twist_01_l\"),Weight=0.0015),(Item=(Type=Bone,Name=\"deform_elbow_in_l\"),Weight=0.0073),(Item=(Type=Bone,Name=\"lowerarm_twist_02_l\"),Weight=0.0387),(Item=(Type=Bone,Name=\"upperarm_twist_02_l\"),Weight=0.068),(Item=(Type=Bone,N ...(truncated)"
    Weight=1
   )
   > &RigUnit_ParentConstraint(
    AdvancedSettings=(RotationOrderForFilter=XZY)
    Child=(Type=Bone,Name="def_upperarm_l_fur_04")
    Parents="((Item=(Type=Bone,Name=\"lowerarm_twist_02_l\"),Weight=0.0149),(Item=(Type=Bone,Name=\"lowerarm_l\"),Weight=0.0577),(Item=(Type=Bone,Name=\"deform_elbow_in_l\"),Weight=0.0811),(Item=(Type=Bone,Name=\"upperarm_twist_01_l\"),Weight=0.1052),(Item=(Type=Bone,Name=\"upp ...(truncated)"
    Weight=1
   )
   > &RigUnit_ParentConstraint(
    AdvancedSettings=(RotationOrderForFilter=XZY)
    Child=(Type=Bone,Name="def_upperarm_l_fur_03")
    Parents="((Item=(Type=Bone,Name=\"deform_upperarm_fr_l\"),Weight=0.0037),(Item=(Type=Bone,Name=\"deform_elbow_in_l\"),Weight=0.0083),(Item=(Type=Bone,Name=\"lowerarm_l\"),Weight=0.0093),(Item=(Type=Bone,Name=\"upperarm_twist_01_l\"),Weight=0.2307),(Item=(Type=Bone,Name=\"up ...(truncated)"
    Weight=1
   )
   > &RigUnit_ParentConstraint(
    AdvancedSettings=(RotationOrderForFilter=XZY)
    Child=(Type=Bone,Name="def_body_fur_05")
    Parents="((Item=(Type=Bone,Name=\"clavicle_l\"),Weight=0.0117),(Item=(Type=Bone,Name=\"spine_04\"),Weight=0.0674),(Item=(Type=Bone,Name=\"pelvis\"),Weight=0.1408),(Item=(Type=Bone,Name=\"spine_01\"),Weight=0.1694),(Item=(Type=Bone,Name=\"spine_03\"),Weight=0.2986),(Item=(Typ ...(truncated)"
    Weight=1
   )
   > &RigUnit_ParentConstraint(
    AdvancedSettings=(RotationOrderForFilter=XZY)
    Child=(Type=Bone,Name="def_body_fur_03")
    Parents="((Item=(Type=Bone,Name=\"clavicle_r\"),Weight=0.0101),(Item=(Type=Bone,Name=\"pelvis\"),Weight=0.0107),(Item=(Type=Bone,Name=\"spine_01\"),Weight=0.0487),(Item=(Type=Bone,Name=\"clavicle_l\"),Weight=0.0755),(Item=(Type=Bone,Name=\"spine_05\"),Weight=0.0783),(Item=(T ...(truncated)"
    Weight=1
   )
   > &RigUnit_ParentConstraint(
    AdvancedSettings=(RotationOrderForFilter=XZY)
    Child=(Type=Bone,Name="def_body_fur_02")
    Parents="((Item=(Type=Bone,Name=\"deform_upperarm_bk_l\"),Weight=0.0396),(Item=(Type=Bone,Name=\"pelvis\"),Weight=0.0508),(Item=(Type=Bone,Name=\"spine_01\"),Weight=0.1075),(Item=(Type=Bone,Name=\"clavicle_l\"),Weight=0.1126),(Item=(Type=Bone,Name=\"spine_04\"),Weight=0.1401 ...(truncated)"
    Weight=1
   )
   > &RigUnit_ParentConstraint(
    AdvancedSettings=(RotationOrderForFilter=XZY)
    Child=(Type=Bone,Name="def_body_fur_04")
    Parents=((Item=(Type=Bone,Name="spine_04"),Weight=0.0022),(Item=(Type=Bone,Name="spine_03"),Weight=0.1183),(Item=(Type=Bone,Name="spine_01"),Weight=0.1689),(Item=(Type=Bone,Name="spine_02"),Weight=0.2081),(Item=(Type=Bone,Name="pelvis"),Weight=0.5026))
    Weight=1
   )
   > &RigUnit_ParentConstraint(
    AdvancedSettings=(RotationOrderForFilter=XZY)
    Child=(Type=Bone,Name="def_body_fur_08")
    Parents="((Item=(Type=Bone,Name=\"clavicle_l\"),Weight=0.0005),(Item=(Type=Bone,Name=\"spine_05\"),Weight=0.0033),(Item=(Type=Bone,Name=\"clavicle_r\"),Weight=0.0201),(Item=(Type=Bone,Name=\"pelvis\"),Weight=0.0802),(Item=(Type=Bone,Name=\"spine_04\"),Weight=0.1082),(Item=(T ...(truncated)"
    Weight=1
   )
   > &RigUnit_ParentConstraint(
    AdvancedSettings=(RotationOrderForFilter=XZY)
    Child=(Type=Bone,Name="def_body_fur_07")
    Parents="((Item=(Type=Bone,Name=\"thigh_twist_01_r\"),Weight=0.0003),(Item=(Type=Bone,Name=\"thigh_twist_01_l\"),Weight=0.0025),(Item=(Type=Bone,Name=\"deform_glute_l\"),Weight=0.0091),(Item=(Type=Bone,Name=\"spine_03\"),Weight=0.0102),(Item=(Type=Bone,Name=\"spine_02\"),Wei ...(truncated)"
    Weight=1
   )
   > &RigUnit_ParentConstraint(
    AdvancedSettings=(RotationOrderForFilter=XZY)
    Child=(Type=Bone,Name="def_body_fur_19")
    Parents=((Item=(Type=Bone,Name="spine_04"),Weight=0.0387),(Item=(Type=Bone,Name="pelvis"),Weight=0.1561),(Item=(Type=Bone,Name="spine_01"),Weight=0.1993),(Item=(Type=Bone,Name="spine_03"),Weight=0.2659),(Item=(Type=Bone,Name="spine_02"),Weight=0.34))
    Weight=1
   )
   > &RigUnit_ParentConstraint(
    AdvancedSettings=(RotationOrderForFilter=XZY)
    Child=(Type=Bone,Name="def_body_fur_12")
    Parents="((Item=(Type=Bone,Name=\"deform_groin_l\"),Weight=0.0007),(Item=(Type=Bone,Name=\"thigh_twist_01_l\"),Weight=0.0017),(Item=(Type=Bone,Name=\"thigh_twist_01_r\"),Weight=0.0065),(Item=(Type=Bone,Name=\"deform_groin_r\"),Weight=0.0082),(Item=(Type=Bone,Name=\"spine_03 ...(truncated)"
    Weight=1
   )
   > &RigUnit_ParentConstraint(
    AdvancedSettings=(RotationOrderForFilter=XZY)
    Child=(Type=Bone,Name="def_body_fur_16")
    Parents="((Item=(Type=Bone,Name=\"thigh_twist_01_r\"),Weight=0.0032),(Item=(Type=Bone,Name=\"spine_04\"),Weight=0.0033),(Item=(Type=Bone,Name=\"deform_groin_r\"),Weight=0.0053),(Item=(Type=Bone,Name=\"spine_03\"),Weight=0.0732),(Item=(Type=Bone,Name=\"spine_01\"),Weight=0.10 ...(truncated)"
    Weight=1
   )
   > &RigUnit_ParentConstraint(
    AdvancedSettings=(RotationOrderForFilter=XZY)
    Child=(Type=Bone,Name="def_body_fur_11")
    Parents="((Item=(Type=Bone,Name=\"deform_glute_l\"),Weight=0.0003),(Item=(Type=Bone,Name=\"spine_03\"),Weight=0.0011),(Item=(Type=Bone,Name=\"spine_02\"),Weight=0.0019),(Item=(Type=Bone,Name=\"spine_01\"),Weight=0.002),(Item=(Type=Bone,Name=\"thigh_twist_01_l\"),Weight=0.011 ...(truncated)"
    Weight=1
   )
   > &RigUnit_ParentConstraint(
    AdvancedSettings=(RotationOrderForFilter=XZY)
    Child=(Type=Bone,Name="def_body_fur_14")
    Parents=((Item=(Type=Bone,Name="spine_04"),Weight=0.0353),(Item=(Type=Bone,Name="pelvis"),Weight=0.1877),(Item=(Type=Bone,Name="spine_01"),Weight=0.2043),(Item=(Type=Bone,Name="spine_03"),Weight=0.2459),(Item=(Type=Bone,Name="spine_02"),Weight=0.3268))
    Weight=1
   )
   > &RigUnit_ParentConstraint(
    AdvancedSettings=(RotationOrderForFilter=XZY)
    Child=(Type=Bone,Name="def_body_fur_09")
    Parents="((Item=(Type=Bone,Name=\"deform_upperarm_bk_r\"),Weight=0.021),(Item=(Type=Bone,Name=\"pelvis\"),Weight=0.0733),(Item=(Type=Bone,Name=\"clavicle_r\"),Weight=0.0809),(Item=(Type=Bone,Name=\"spine_04\"),Weight=0.1118),(Item=(Type=Bone,Name=\"spine_01\"),Weight=0.1304) ...(truncated)"
    Weight=1
   )
   > &RigUnit_ParentConstraint(
    AdvancedSettings=(RotationOrderForFilter=XZY)
    Child=(Type=Bone,Name="def_body_fur_10")
    Parents=((Item=(Type=Bone,Name="spine_03"),Weight=0.0903),(Item=(Type=Bone,Name="spine_01"),Weight=0.1337),(Item=(Type=Bone,Name="spine_02"),Weight=0.1602),(Item=(Type=Bone,Name="pelvis"),Weight=0.6158))
    Weight=1
   )
   > &RigUnit_ParentConstraint(
    AdvancedSettings=(RotationOrderForFilter=XZY)
    Child=(Type=Bone,Name="def_body_fur_15")
    Parents=((Item=(Type=Bone,Name="spine_04"),Weight=0.0121),(Item=(Type=Bone,Name="spine_03"),Weight=0.1342),(Item=(Type=Bone,Name="spine_01"),Weight=0.1741),(Item=(Type=Bone,Name="spine_02"),Weight=0.2177),(Item=(Type=Bone,Name="pelvis"),Weight=0.4619))
    Weight=1
   )
   > &RigUnit_ParentConstraint(
    AdvancedSettings=(RotationOrderForFilter=XZY)
    Child=(Type=Bone,Name="def_body_fur_20")
    Parents="((Item=(Type=Bone,Name=\"spine_01\"),Weight=0.0172),(Item=(Type=Bone,Name=\"clavicle_l\"),Weight=0.0662),(Item=(Type=Bone,Name=\"spine_05\"),Weight=0.0758),(Item=(Type=Bone,Name=\"spine_02\"),Weight=0.0908),(Item=(Type=Bone,Name=\"deform_pec_l\"),Weight=0.1405),(Ite ...(truncated)"
    Weight=1
   )
   > &RigUnit_ParentConstraint(
    AdvancedSettings=(RotationOrderForFilter=XZY)
    Child=(Type=Bone,Name="def_body_fur_21")
    Parents="((Item=(Type=Bone,Name=\"clavicle_l\"),Weight=0.0343),(Item=(Type=Bone,Name=\"spine_05\"),Weight=0.0347),(Item=(Type=Bone,Name=\"deform_pec_l\"),Weight=0.035),(Item=(Type=Bone,Name=\"spine_01\"),Weight=0.0487),(Item=(Type=Bone,Name=\"def_neck_fur_front\"),Weight=0.0 ...(truncated)"
    Weight=1
   )
   > &RigUnit_ParentConstraint(
    AdvancedSettings=(RotationOrderForFilter=XZY)
    Child=(Type=Bone,Name="def_calf_l_fur_05")
    Parents=((Item=(Type=Bone,Name="calf_l"),Weight=0.0004),(Item=(Type=Bone,Name="calf_twist_02_l"),Weight=0.3115),(Item=(Type=Bone,Name="calf_twist_01_l"),Weight=0.6881))
    Weight=1
   )
   > &RigUnit_ParentConstraint(
    AdvancedSettings=(RotationOrderForFilter=XZY)
    Child=(Type=Bone,Name="def_calf_l_fur_04")
    Parents=((Item=(Type=Bone,Name="calf_l"),Weight=0.0356),(Item=(Type=Bone,Name="calf_twist_01_l"),Weight=0.3823),(Item=(Type=Bone,Name="calf_twist_02_l"),Weight=0.5821))
    Weight=1
   )
   > &RigUnit_ParentConstraint(
    AdvancedSettings=(RotationOrderForFilter=XZY)
    Child=(Type=Bone,Name="def_calf_l_fur_07")
    Parents=((Item=(Type=Bone,Name="calf_l"),Weight=0.0342),(Item=(Type=Bone,Name="calf_twist_01_l"),Weight=0.4008),(Item=(Type=Bone,Name="calf_twist_02_l"),Weight=0.565))
    Weight=1
   )
   &RigUnit_ParentConstraint(
    AdvancedSettings=(RotationOrderForFilter=XZY)
    Child=(Type=Bone,Name="def_lowerarm_r_fur_04")
    ExecutePin=()
    Parents="((Item=(Type=Bone,Name=\"deform_elbow_out_r\"),Weight=0.0025),(Item=(Type=Bone,Name=\"upperarm_r\"),Weight=0.0083),(Item=(Type=Bone,Name=\"lowerarm_r\"),Weight=0.2027),(Item=(Type=Bone,Name=\"lowerarm_twist_01_r\"),Weight=0.2385),(Item=(Type=Bone,Name=\"lowerarm_tw ...(truncated)"
    Weight=1
   )
   > &RigUnit_ParentConstraint(
    AdvancedSettings=(RotationOrderForFilter=XZY)
    Child=(Type=Bone,Name="def_lowerarm_r_fur_02")
    Parents="((Item=(Type=Bone,Name=\"deform_elbow_in_r\"),Weight=0.0271),(Item=(Type=Bone,Name=\"upperarm_r\"),Weight=0.0449),(Item=(Type=Bone,Name=\"lowerarm_twist_01_r\"),Weight=0.1532),(Item=(Type=Bone,Name=\"lowerarm_r\"),Weight=0.2824),(Item=(Type=Bone,Name=\"lowerarm_twi ...(truncated)"
    Weight=1
   )
   > &RigUnit_ParentConstraint(
    AdvancedSettings=(RotationOrderForFilter=XZY)
    Child=(Type=Bone,Name="def_lowerarm_r_fur_01")
    Parents="((Item=(Type=Bone,Name=\"lowerarm_twist_01_r\"),Weight=0.0192),(Item=(Type=Bone,Name=\"upperarm_twist_02_r\"),Weight=0.0248),(Item=(Type=Bone,Name=\"deform_elbow_out_r\"),Weight=0.0498),(Item=(Type=Bone,Name=\"deform_elbow_in_r\"),Weight=0.1016),(Item=(Type=Bone,N ...(truncated)"
    Weight=1
   )
   > &RigUnit_ParentConstraint(
    AdvancedSettings=(RotationOrderForFilter=XZY)
    Child=(Type=Bone,Name="def_lowerarm_r_fur_07")
    Parents="((Item=(Type=Bone,Name=\"ring_metacarpal_r\"),Weight=0.0022),(Item=(Type=Bone,Name=\"thumb_01_r\"),Weight=0.0089),(Item=(Type=Bone,Name=\"deform_wrist_up_r\"),Weight=0.0138),(Item=(Type=Bone,Name=\"lowerarm_twist_02_r\"),Weight=0.0268),(Item=(Type=Bone,Name=\"defor ...(truncated)"
    Weight=1
   )
   &RigUnit_ParentConstraint(
    AdvancedSettings=(RotationOrderForFilter=XZY)
    Child=(Type=Bone,Name="def_lowerarm_r_fur_08")
    ExecutePin=()
    Parents="((Item=(Type=Bone,Name=\"lowerarm_twist_01_r\"),Weight=0.0062),(Item=(Type=Bone,Name=\"upperarm_twist_02_r\"),Weight=0.0147),(Item=(Type=Bone,Name=\"lowerarm_twist_02_r\"),Weight=0.1213),(Item=(Type=Bone,Name=\"upperarm_r\"),Weight=0.1223),(Item=(Type=Bone,Name=\"l ...(truncated)"
    Weight=1
   )
   > &RigUnit_ParentConstraint(
    AdvancedSettings=(RotationOrderForFilter=XZY)
    Child=(Type=Bone,Name="def_upperarm_r_fur_06")
    Parents="((Item=(Type=Bone,Name=\"upperarm_twist_01_r\"),Weight=0.0015),(Item=(Type=Bone,Name=\"deform_elbow_in_r\"),Weight=0.0073),(Item=(Type=Bone,Name=\"lowerarm_twist_02_r\"),Weight=0.0387),(Item=(Type=Bone,Name=\"upperarm_twist_02_r\"),Weight=0.068),(Item=(Type=Bone,N ...(truncated)"
    Weight=1
   )
   > &RigUnit_ParentConstraint(
    AdvancedSettings=(RotationOrderForFilter=XZY)
    Child=(Type=Bone,Name="def_upperarm_r_fur_04")
    Parents="((Item=(Type=Bone,Name=\"deform_upperarm_fr_r\"),Weight=0.0001),(Item=(Type=Bone,Name=\"lowerarm_twist_02_r\"),Weight=0.014),(Item=(Type=Bone,Name=\"lowerarm_r\"),Weight=0.0565),(Item=(Type=Bone,Name=\"deform_elbow_in_r\"),Weight=0.0793),(Item=(Type=Bone,Name=\"upp ...(truncated)"
    Weight=1
   )
   > &RigUnit_ParentConstraint(
    AdvancedSettings=(RotationOrderForFilter=XZY)
    Child=(Type=Bone,Name="def_upperarm_r_fur_03")
    Parents="((Item=(Type=Bone,Name=\"deform_upperarm_fr_r\"),Weight=0.0035),(Item=(Type=Bone,Name=\"deform_elbow_in_r\"),Weight=0.0082),(Item=(Type=Bone,Name=\"lowerarm_r\"),Weight=0.0093),(Item=(Type=Bone,Name=\"upperarm_twist_01_r\"),Weight=0.2299),(Item=(Type=Bone,Name=\"up ...(truncated)"
    Weight=1
   )
   > &RigUnit_ParentConstraint(
    AdvancedSettings=(RotationOrderForFilter=XZY)
    Child=(Type=Bone,Name="def_calf_l_fur_14")
    Parents=((Item=(Type=Bone,Name="calf_l"),Weight=0.0242),(Item=(Type=Bone,Name="calf_twist_01_l"),Weight=0.3753),(Item=(Type=Bone,Name="calf_twist_02_l"),Weight=0.6004))
    Weight=1
   )
   > &RigUnit_ParentConstraint(
    AdvancedSettings=(RotationOrderForFilter=XZY)
    Child=(Type=Bone,Name="def_calf_l_fur_13")
    Parents=((Item=(Type=Bone,Name="calf_l"),Weight=0.0539),(Item=(Type=Bone,Name="calf_twist_01_l"),Weight=0.2411),(Item=(Type=Bone,Name="calf_twist_02_l"),Weight=0.705))
    Weight=1
   )
   > &RigUnit_ParentConstraint(
    AdvancedSettings=(RotationOrderForFilter=XZY)
    Child=(Type=Bone,Name="def_calf_l_fur_15")
    Parents=((Item=(Type=Bone,Name="calf_twist_01_l"),Weight=0.0995),(Item=(Type=Bone,Name="calf_l"),Weight=0.1402),(Item=(Type=Bone,Name="calf_twist_02_l"),Weight=0.7603))
    Weight=1
   )
   > &RigUnit_ParentConstraint(
    AdvancedSettings=(RotationOrderForFilter=XZY)
    Child=(Type=Bone,Name="def_thigh_l_fur_02")
    Parents="((Item=(Type=Bone,Name=\"deform_knee_in_l\"),Weight=0.002),(Item=(Type=Bone,Name=\"deform_knee_out_l\"),Weight=0.0032),(Item=(Type=Bone,Name=\"calf_l\"),Weight=0.0078),(Item=(Type=Bone,Name=\"thigh_twist_01_l\"),Weight=0.4106),(Item=(Type=Bone,Name=\"thigh_l\"),Weig ...(truncated)"
    Weight=1
   )
   > &RigUnit_ParentConstraint(
    AdvancedSettings=(RotationOrderForFilter=XZY)
    Child=(Type=Bone,Name="def_thigh_l_fur_04")
    Parents=((Item=(Type=Bone,Name="deform_knee_in_l"),Weight=0.0005),(Item=(Type=Bone,Name="thigh_l"),Weight=0.4562),(Item=(Type=Bone,Name="thigh_twist_01_l"),Weight=0.5433))
    Weight=1
   )
   > &RigUnit_ParentConstraint(
    AdvancedSettings=(RotationOrderForFilter=XZY)
    Child=(Type=Bone,Name="def_calf_r_fur_04")
    Parents=((Item=(Type=Bone,Name="calf_r"),Weight=0.0356),(Item=(Type=Bone,Name="calf_twist_01_r"),Weight=0.3823),(Item=(Type=Bone,Name="calf_twist_02_r"),Weight=0.5821))
    Weight=1
   )
   > &RigUnit_ParentConstraint(
    AdvancedSettings=(RotationOrderForFilter=XZY)
    Child=(Type=Bone,Name="def_calf_r_fur_13")
    Parents=((Item=(Type=Bone,Name="calf_r"),Weight=0.0488),(Item=(Type=Bone,Name="calf_twist_01_r"),Weight=0.2497),(Item=(Type=Bone,Name="calf_twist_02_r"),Weight=0.7016))
    Weight=1
   )
   > &RigUnit_ParentConstraint(
    AdvancedSettings=(RotationOrderForFilter=XZY)
    Child=(Type=Bone,Name="def_calf_r_fur_15")
    Parents=((Item=(Type=Bone,Name="calf_twist_01_r"),Weight=0.0979),(Item=(Type=Bone,Name="calf_r"),Weight=0.1428),(Item=(Type=Bone,Name="calf_twist_02_r"),Weight=0.7594))
    Weight=1
   )
   > &RigUnit_ParentConstraint(
    AdvancedSettings=(RotationOrderForFilter=XZY)
    Child=(Type=Bone,Name="def_calf_r_fur_07")
    Parents=((Item=(Type=Bone,Name="calf_r"),Weight=0.0286),(Item=(Type=Bone,Name="calf_twist_01_r"),Weight=0.4336),(Item=(Type=Bone,Name="calf_twist_02_r"),Weight=0.5377))
    Weight=1
   )
   > &RigUnit_ParentConstraint(
    AdvancedSettings=(RotationOrderForFilter=XZY)
    Child=(Type=Bone,Name="def_calf_r_fur_05")
    Parents=((Item=(Type=Bone,Name="calf_r"),Weight=0.0005),(Item=(Type=Bone,Name="calf_twist_02_r"),Weight=0.3123),(Item=(Type=Bone,Name="calf_twist_01_r"),Weight=0.6873))
    Weight=1
   )
   > &RigUnit_ParentConstraint(
    AdvancedSettings=(RotationOrderForFilter=XZY)
    Child=(Type=Bone,Name="def_calf_r_fur_14")
    Parents=((Item=(Type=Bone,Name="calf_r"),Weight=0.0205),(Item=(Type=Bone,Name="calf_twist_01_r"),Weight=0.4033),(Item=(Type=Bone,Name="calf_twist_02_r"),Weight=0.5762))
    Weight=1
   )
   &RigUnit_ParentConstraint(
    AdvancedSettings=(RotationOrderForFilter=XZY)
    Child=(Type=Bone,Name="def_hand_l_fur_06")
    ExecutePin=()
    Parents=((Item=(Type=Bone,Name="ring_metacarpal_l"),Weight=0.001),(Item=(Type=Bone,Name="pinky_01_l"),Weight=0.1176),(Item=(Type=Bone,Name="pinky_metacarpal_l"),Weight=0.8813))
    Weight=1
   )
   &RigUnit_ParentConstraint(
    AdvancedSettings=(RotationOrderForFilter=XZY)
    Child=(Type=Bone,Name="def_hand_l_fur_07")
    ExecutePin=()
    Parents=((Item=(Type=Bone,Name="lowerarm_twist_01_l"),Weight=0.0026),(Item=(Type=Bone,Name="hand_l"),Weight=0.0104),(Item=(Type=Bone,Name="pinky_metacarpal_l"),Weight=0.987))
    Weight=1
   )
   &RigUnit_ParentConstraint(
    AdvancedSettings=(RotationOrderForFilter=XZY)
    Child=(Type=Bone,Name="def_hand_r_fur_07")
    ExecutePin=()
    Parents=((Item=(Type=Bone,Name="lowerarm_twist_01_r"),Weight=0.0026),(Item=(Type=Bone,Name="hand_r"),Weight=0.0104),(Item=(Type=Bone,Name="pinky_metacarpal_r"),Weight=0.987))
    Weight=1
   )
   &RigUnit_ParentConstraint(
    AdvancedSettings=(RotationOrderForFilter=XZY)
    Child=(Type=Bone,Name="def_hand_r_fur_06")
    ExecutePin=()
    Parents=((Item=(Type=Bone,Name="ring_metacarpal_r"),Weight=0.001),(Item=(Type=Bone,Name="pinky_01_r"),Weight=0.1176),(Item=(Type=Bone,Name="pinky_metacarpal_r"),Weight=0.8813))
    Weight=1
   )
  ]}
 New Function_ContainedGraph {nodes [
   RigVMFunctionEntryNode()
   > n1 = &RigUnit_SphericalPoseReader(
    ActiveRegionScaleFactors=(PositiveWidth=1,NegativeWidth=4,PositiveHeight=4,NegativeHeight=4)
    ActiveRegionSize=0.1
    Debug=(bDrawDebug=true,bDraw2D=False,bDrawLocalAxes=False,DebugScale=25,DebugSegments=20,DebugThickness=0.25)
    DriverAxis=(X=1,Y=0,Z=0)
    DriverItem=(Type=Bone,Name="head")
    FalloffRegionScaleFactors=(PositiveWidth=1,NegativeWidth=1,PositiveHeight=1,NegativeHeight=1)
    FalloffSize=0.45
    FlipHeightScaling=False
    FlipWidthScaling=False
    OptionalParentItem=(Type=Bone,Name="spine_05")
    RotationOffset=(X=-90,Y=0,Z=20)
   )
   > &RigUnit_ParentConstraint(
    AdvancedSettings=(InterpolationType=Average,RotationOrderForFilter=XZY)
    Child=(Type=Bone,Name="def_neck_fur_front")
    Filter=(TranslationFilter=(bX=True,bY=True,bZ=True),RotationFilter=(bX=false,bY=false,bZ=false),ScaleFilter=(bX=false,bY=false,bZ=false))
    Parents=((Item=(Type=Bone,Name="head"),Weight=1),(Item=(Type=Bone,Name="spine_05"),Weight=0.4))
    Weight=@n1.OutputParam
   )
   > &RigUnit_ModifyTransforms(
    ItemToModify=((Item=(Type=Bone,Name="def_neck_fur_front"),Transform=(Rotation=(X=0,Y=0,Z=0,W=1),Translation=(X=3,Y=-2,Z=0),Scale3D=(X=0.2,Y=1,Z=1))))
    WeightMinimum=0
    Weight=@n1.OutputParam
   )
   > RigVMFunctionReturnNode()
  ]}
 New Function_ContainedGraph {nodes [
   n1 = RigVMFunctionEntryNode()
   > RigVMFunction_Sequence = RigVMAggregateNode()
   DeformRig_GroinGlue_FNC = RigVMFunctionReferenceNode(
    debug=False
    glute=true
    groin=true
    ExecuteContext=@RigVMFunction_Sequence.A
    curve_struct=@n1.DeformRig_ Curve
    value_struct=@n1.DeformRig_Value
   )
   DeformRig_Elbow_FNC = RigVMFunctionReferenceNode(bicep=true, elbow=true, ExecuteContext=@RigVMFunction_Sequence.B, curve_struct=@n1.DeformRig_ Curve, value_struct=@n1.DeformRig_Value)
   DeformRig_Pec_FNC = RigVMFunctionReferenceNode(ExecuteContext=@RigVMFunction_Sequence.C, curve_struct=@n1.DeformRig_ Curve, value_struct=@n1.DeformRig_Value)
   DeformRig_Knee_FNC = RigVMFunctionReferenceNode(
    curve_struct="(deformRig_l_lat_alpha_91_#11=False,deformRig_r_lat_alpha_93_#12=False,deformRig_l_upperarm_up_alpha_72_#34=False,deformRig_r_upperarm_up_alpha_73_7435B411475DE10CCB23A0 ...(truncated)"
    ExecuteContext=@RigVMFunction_Sequence.D
    value_struct=@n1.DeformRig_Value
   )
   DeformRig_Wrist_FNC = RigVMFunctionReferenceNode(debug=false, ExecuteContext=@RigVMFunction_Sequence.E, curve_struct=@n1.DeformRig_ Curve, value_struct=@n1.DeformRig_Value)
   DeformRig_UpperArm_FNC = RigVMFunctionReferenceNode(debug=False, ExecuteContext=@RigVMFunction_Sequence.F, curve_struct=@n1.DeformRig_ Curve, value_struct=@n1.DeformRig_Value)
   RigVMFunctionReturnNode(ExecuteContext=@RigVMFunction_Sequence.G)
  ]}
 New Function_ContainedGraph {nodes [
   RigVMFunctionEntryNode()
   > &RigUnit_ParentConstraint(
    AdvancedSettings=(RotationOrderForFilter=XZY)
    Child=(Type=Bone,Name="dyn_head_fur_08")
    Parents=((Item=(Type=Bone,Name="neck_01"),Weight=0.0001),(Item=(Type=Bone,Name="spine_05"),Weight=0.0004),(Item=(Type=Bone,Name="neck_02"),Weight=0.0004),(Item=(Type=Bone,Name="head_twist"),Weight=0.0101),(Item=(Type=Bone,Name="head"),Weight=0.9889))
    Weight=1
   )
   > &RigUnit_ParentConstraint(
    AdvancedSettings=(RotationOrderForFilter=XZY)
    Child=(Type=Bone,Name="dyn_head_fur_10")
    Parents="((Item=(Type=Bone,Name=\"spine_03\"),Weight=0.0002),(Item=(Type=Bone,Name=\"clavicle_l\"),Weight=0.0005),(Item=(Type=Bone,Name=\"clavicle_r\"),Weight=0.0007),(Item=(Type=Bone,Name=\"neck_01\"),Weight=0.0259),(Item=(Type=Bone,Name=\"spine_05\"),Weight=0.0694),(Item=( ...(truncated)"
    Weight=1
   )
   > &RigUnit_ParentConstraint(
    AdvancedSettings=(RotationOrderForFilter=XZY)
    Child=(Type=Bone,Name="dyn_head_fur_11")
    Parents="((Item=(Type=Bone,Name=\"clavicle_l\"),Weight=0.0001),(Item=(Type=Bone,Name=\"neck_01\"),Weight=0.0008),(Item=(Type=Bone,Name=\"spine_05\"),Weight=0.0019),(Item=(Type=Bone,Name=\"neck_02\"),Weight=0.0029),(Item=(Type=Bone,Name=\"head_twist\"),Weight=0.0192),(Item=(T ...(truncated)"
    Weight=1
   )
   > &RigUnit_ParentConstraint(
    AdvancedSettings=(RotationOrderForFilter=XZY)
    Child=(Type=Bone,Name="dyn_head_fur_12")
    Parents=((Item=(Type=Bone,Name="neck_01"),Weight=0.0023),(Item=(Type=Bone,Name="head_twist"),Weight=0.0029),(Item=(Type=Bone,Name="spine_05"),Weight=0.0047),(Item=(Type=Bone,Name="neck_02"),Weight=0.0087),(Item=(Type=Bone,Name="head"),Weight=0.9815))
    Weight=1
   )
   > &RigUnit_ParentConstraint(
    AdvancedSettings=(RotationOrderForFilter=XZY)
    Child=(Type=Bone,Name="dyn_head_fur_13")
    Parents="((Item=(Type=Bone,Name=\"clavicle_l\"),Weight=0.0035),(Item=(Type=Bone,Name=\"clavicle_r\"),Weight=0.0041),(Item=(Type=Bone,Name=\"deform_neck_hair_rear_01\"),Weight=0.0191),(Item=(Type=Bone,Name=\"neck_01\"),Weight=0.063),(Item=(Type=Bone,Name=\"head\"),Weight=0.12 ...(truncated)"
    Weight=1
   )
   > &RigUnit_ParentConstraint(
    AdvancedSettings=(RotationOrderForFilter=XZY)
    Child=(Type=Bone,Name="dyn_head_fur_14")
    Parents="((Item=(Type=Bone,Name=\"deform_neck_hair_d_l_02\"),Weight=0.0067),(Item=(Type=Bone,Name=\"clavicle_l\"),Weight=0.0174),(Item=(Type=Bone,Name=\"head\"),Weight=0.0386),(Item=(Type=Bone,Name=\"deform_neck_hair_rear_01\"),Weight=0.0417),(Item=(Type=Bone,Name=\"neck_01 ...(truncated)"
    Weight=1
   )
   > &RigUnit_ParentConstraint(
    AdvancedSettings=(RotationOrderForFilter=XZY)
    Child=(Type=Bone,Name="dyn_head_fur_15")
    Parents="((Item=(Type=Bone,Name=\"deform_neck_hair_c_l_01\"),Weight=0.0039),(Item=(Type=Bone,Name=\"clavicle_l\"),Weight=0.0277),(Item=(Type=Bone,Name=\"neck_01\"),Weight=0.0578),(Item=(Type=Bone,Name=\"head\"),Weight=0.0805),(Item=(Type=Bone,Name=\"deform_neck_hair_d_l_01\" ...(truncated)"
    Weight=1
   )
   > &RigUnit_ParentConstraint(
    AdvancedSettings=(RotationOrderForFilter=XZY)
    Child=(Type=Bone,Name="dyn_head_fur_16")
    Parents="((Item=(Type=Bone,Name=\"spine_03\"),Weight=0.0003),(Item=(Type=Bone,Name=\"deform_neck_hair_c_l_01\"),Weight=0.0004),(Item=(Type=Bone,Name=\"clavicle_l\"),Weight=0.0035),(Item=(Type=Bone,Name=\"neck_01\"),Weight=0.0107),(Item=(Type=Bone,Name=\"spine_05\"),Weight=0. ...(truncated)"
    Weight=1
   )
   > &RigUnit_ParentConstraint(
    AdvancedSettings=(RotationOrderForFilter=XZY)
    Child=(Type=Bone,Name="dyn_head_fur_17")
    Parents="((Item=(Type=Bone,Name=\"head\"),Weight=0.0025),(Item=(Type=Bone,Name=\"head_twist\"),Weight=0.0045),(Item=(Type=Bone,Name=\"neck_01\"),Weight=0.006),(Item=(Type=Bone,Name=\"neck_02\"),Weight=0.0143),(Item=(Type=Bone,Name=\"clavicle_l\"),Weight=0.029),(Item=(Type=Bo ...(truncated)"
    Weight=1
   )
   > &RigUnit_ParentConstraint(
    AdvancedSettings=(RotationOrderForFilter=XZY)
    Child=(Type=Bone,Name="dyn_head_fur_18")
    Parents="((Item=(Type=Bone,Name=\"upperarm_twist_01_l\"),Weight=0.0004),(Item=(Type=Bone,Name=\"teeth_lower\"),Weight=0.002),(Item=(Type=Bone,Name=\"neck_01\"),Weight=0.004),(Item=(Type=Bone,Name=\"clavicle_l\"),Weight=0.0049),(Item=(Type=Bone,Name=\"neck_02\"),Weight=0.0108 ...(truncated)"
    Weight=1
   )
   > &RigUnit_ParentConstraint(
    AdvancedSettings=(RotationOrderForFilter=XZY)
    Child=(Type=Bone,Name="dyn_head_fur_19")
    Parents="((Item=(Type=Bone,Name=\"neck_01\"),Weight=0.0122),(Item=(Type=Bone,Name=\"neck_02\"),Weight=0.0327),(Item=(Type=Bone,Name=\"clavicle_l\"),Weight=0.0473),(Item=(Type=Bone,Name=\"deform_neck_hair_a_l_01\"),Weight=0.0661),(Item=(Type=Bone,Name=\"spine_05\"),Weight=0.1 ...(truncated)"
    Weight=1
   )
   > &RigUnit_ParentConstraint(
    AdvancedSettings=(RotationOrderForFilter=XZY)
    Child=(Type=Bone,Name="dyn_head_fur_20")
    Parents="((Item=(Type=Bone,Name=\"neck_01\"),Weight=0.0042),(Item=(Type=Bone,Name=\"head_twist\"),Weight=0.0043),(Item=(Type=Bone,Name=\"teeth_lower\"),Weight=0.0048),(Item=(Type=Bone,Name=\"neck_02\"),Weight=0.0119),(Item=(Type=Bone,Name=\"clavicle_l\"),Weight=0.0135),(Item ...(truncated)"
    Weight=1
   )
   > &RigUnit_ParentConstraint(
    AdvancedSettings=(RotationOrderForFilter=XZY)
    Child=(Type=Bone,Name="dyn_head_fur_21")
    Parents="((Item=(Type=Bone,Name=\"C_lip_lower_mid\"),Weight=0.0002),(Item=(Type=Bone,Name=\"R_lip_upper_outer\"),Weight=0.0002),(Item=(Type=Bone,Name=\"L_lip_lower_outer\"),Weight=0.001),(Item=(Type=Bone,Name=\"C_lip_upper_mid\"),Weight=0.0021),(Item=(Type=Bone,Name=\"C_jaw ...(truncated)"
    Weight=1
   )
   > &RigUnit_ParentConstraint(
    AdvancedSettings=(RotationOrderForFilter=XZY)
    Child=(Type=Bone,Name="dyn_head_fur_22")
    Parents="((Item=(Type=Bone,Name=\"neck_02\"),Weight=0.0034),(Item=(Type=Bone,Name=\"def_neck_fur_front\"),Weight=0.0035),(Item=(Type=Bone,Name=\"clavicle_l\"),Weight=0.0056),(Item=(Type=Bone,Name=\"L_lip_corner\"),Weight=0.0125),(Item=(Type=Bone,Name=\"spine_05\"),Weight=0.0 ...(truncated)"
    Weight=1
   )
   > &RigUnit_ParentConstraint(
    AdvancedSettings=(RotationOrderForFilter=XZY)
    Child=(Type=Bone,Name="dyn_head_fur_23")
    Parents="((Item=(Type=Bone,Name=\"spine_03\"),Weight=0.0001),(Item=(Type=Bone,Name=\"deform_neck_hair_d_r_01\"),Weight=0.0002),(Item=(Type=Bone,Name=\"clavicle_r\"),Weight=0.0045),(Item=(Type=Bone,Name=\"neck_01\"),Weight=0.0611),(Item=(Type=Bone,Name=\"neck_02\"),Weight=0.1 ...(truncated)"
    Weight=1
   )
   > &RigUnit_ParentConstraint(
    AdvancedSettings=(RotationOrderForFilter=XZY)
    Child=(Type=Bone,Name="dyn_head_fur_29")
    Parents="((Item=(Type=Bone,Name=\"clavicle_r\"),Weight=0.0006),(Item=(Type=Bone,Name=\"neck_01\"),Weight=0.004),(Item=(Type=Bone,Name=\"spine_05\"),Weight=0.0098),(Item=(Type=Bone,Name=\"neck_02\"),Weight=0.0157),(Item=(Type=Bone,Name=\"head_twist\"),Weight=0.0192),(Item=(Ty ...(truncated)"
    Weight=1
   )
   > &RigUnit_ParentConstraint(
    AdvancedSettings=(RotationOrderForFilter=XZY)
    Child=(Type=Bone,Name="dyn_head_fur_31")
    Parents="((Item=(Type=Bone,Name=\"upperarm_twist_01_r\"),Weight=0.0006),(Item=(Type=Bone,Name=\"deform_upperarm_up_r\"),Weight=0.001),(Item=(Type=Bone,Name=\"neck_01\"),Weight=0.0184),(Item=(Type=Bone,Name=\"clavicle_r\"),Weight=0.0402),(Item=(Type=Bone,Name=\"neck_02\"),Wei ...(truncated)"
    Weight=1
   )
   > &RigUnit_ParentConstraint(
    AdvancedSettings=(RotationOrderForFilter=XZY)
    Child=(Type=Bone,Name="dyn_head_fur_32")
    Parents=((Item=(Type=Bone,Name="head_twist"),Weight=0.0001),(Item=(Type=Bone,Name="neck_02"),Weight=0.0002),(Item=(Type=Bone,Name="spine_05"),Weight=0.0002),(Item=(Type=Bone,Name="head"),Weight=0.9995))
    Weight=1
   )
   > &RigUnit_ParentConstraint(
    AdvancedSettings=(RotationOrderForFilter=XZY)
    Child=(Type=Bone,Name="dyn_head_fur_33")
    Parents="((Item=(Type=Bone,Name=\"upperarm_twist_01_r\"),Weight=0.0018),(Item=(Type=Bone,Name=\"deform_upperarm_up_r\"),Weight=0.0027),(Item=(Type=Bone,Name=\"neck_01\"),Weight=0.01),(Item=(Type=Bone,Name=\"clavicle_r\"),Weight=0.0259),(Item=(Type=Bone,Name=\"neck_02\"),Weig ...(truncated)"
    Weight=1
   )
   > &RigUnit_ParentConstraint(
    AdvancedSettings=(RotationOrderForFilter=XZY)
    Child=(Type=Bone,Name="dyn_head_fur_34")
    Parents="((Item=(Type=Bone,Name=\"deform_upperarm_up_r\"),Weight=0.0081),(Item=(Type=Bone,Name=\"neck_01\"),Weight=0.0116),(Item=(Type=Bone,Name=\"deform_neck_hair_b_r_01\"),Weight=0.0303),(Item=(Type=Bone,Name=\"neck_02\"),Weight=0.0359),(Item=(Type=Bone,Name=\"clavicle_r\" ...(truncated)"
    Weight=1
   )
   > &RigUnit_ParentConstraint(
    AdvancedSettings=(RotationOrderForFilter=XZY)
    Child=(Type=Bone,Name="dyn_head_fur_35")
    Parents=((Item=(Type=Bone,Name="neck_02"),Weight=0.0002),(Item=(Type=Bone,Name="spine_05"),Weight=0.0002),(Item=(Type=Bone,Name="R_cheek_inner"),Weight=0.0211),(Item=(Type=Bone,Name="head"),Weight=0.9785))
    Weight=1
   )
   > &RigUnit_ParentConstraint(
    AdvancedSettings=(RotationOrderForFilter=XZY)
    Child=(Type=Bone,Name="dyn_head_fur_37")
    Parents=((Item=(Type=Bone,Name="R_brow_mid"),Weight=0.0006),(Item=(Type=Bone,Name="R_brow_outer"),Weight=0.2879),(Item=(Type=Bone,Name="head"),Weight=0.7115))
    Weight=1
   )
   > &RigUnit_ParentConstraint(
    AdvancedSettings=(RotationOrderForFilter=XZY)
    Child=(Type=Bone,Name="dyn_head_fur_38")
    Parents="((Item=(Type=Bone,Name=\"teeth_lower\"),Weight=0.0009),(Item=(Type=Bone,Name=\"clavicle_r\"),Weight=0.0018),(Item=(Type=Bone,Name=\"neck_01\"),Weight=0.0023),(Item=(Type=Bone,Name=\"C_jaw\"),Weight=0.0038),(Item=(Type=Bone,Name=\"neck_02\"),Weight=0.0074),(Item=(Typ ...(truncated)"
    Weight=1
   )
   > &RigUnit_ParentConstraint(
    AdvancedSettings=(RotationOrderForFilter=XZY)
    Child=(Type=Bone,Name="dyn_head_fur_39")
    Parents="((Item=(Type=Bone,Name=\"L_lip_upper_outer\"),Weight=0.0002),(Item=(Type=Bone,Name=\"R_lip_lower_outer\"),Weight=0.0006),(Item=(Type=Bone,Name=\"C_lip_upper_mid\"),Weight=0.002),(Item=(Type=Bone,Name=\"C_jaw\"),Weight=0.0041),(Item=(Type=Bone,Name=\"R_lip_upper_out ...(truncated)"
    Weight=1
   )
   > &RigUnit_ParentConstraint(
    AdvancedSettings=(RotationOrderForFilter=XZY)
    Child=(Type=Bone,Name="dyn_head_fur_40")
    Parents="((Item=(Type=Bone,Name=\"neck_02\"),Weight=0.0036),(Item=(Type=Bone,Name=\"deform_neck_hair_a_r_01\"),Weight=0.006),(Item=(Type=Bone,Name=\"spine_05\"),Weight=0.0119),(Item=(Type=Bone,Name=\"R_lip_lower_outer\"),Weight=0.0163),(Item=(Type=Bone,Name=\"R_lip_upper_ou ...(truncated)"
    Weight=1
   )
   > &RigUnit_ParentConstraint(
    AdvancedSettings=(RotationOrderForFilter=XZY)
    Child=(Type=Bone,Name="dyn_head_fur_41")
    Parents="((Item=(Type=Bone,Name=\"L_lip_corner\"),Weight=0.002),(Item=(Type=Bone,Name=\"L_lip_upper_outer\"),Weight=0.0359),(Item=(Type=Bone,Name=\"R_lip_corner\"),Weight=0.037),(Item=(Type=Bone,Name=\"R_lip_upper_outer\"),Weight=0.1445),(Item=(Type=Bone,Name=\"C_lip_upper_ ...(truncated)"
    Weight=1
   )
   > &RigUnit_ParentConstraint(
    AdvancedSettings=(RotationOrderForFilter=XZY)
    Child=(Type=Bone,Name="dyn_head_fur_42")
    Parents="((Item=(Type=Bone,Name=\"L_lip_lower_outer\"),Weight=0.0001),(Item=(Type=Bone,Name=\"R_lip_corner\"),Weight=0.0005),(Item=(Type=Bone,Name=\"R_lip_upper_outer\"),Weight=0.0102),(Item=(Type=Bone,Name=\"C_lip_upper_mid\"),Weight=0.0635),(Item=(Type=Bone,Name=\"L_lip_u ...(truncated)"
    Weight=1
   )
   > &RigUnit_ParentConstraint(
    AdvancedSettings=(RotationOrderForFilter=XZY)
    Child=(Type=Bone,Name="dyn_head_fur_46")
    Parents=((Item=(Type=Bone,Name="R_brow_outer"),Weight=0.0097),(Item=(Type=Bone,Name="R_brow_mid"),Weight=0.0314),(Item=(Type=Bone,Name="head"),Weight=0.9589))
    Weight=1
   )
   > &RigUnit_ParentConstraint(
    AdvancedSettings=(RotationOrderForFilter=XZY)
    Child=(Type=Bone,Name="dyn_head_fur_48")
    Parents=((Item=(Type=Bone,Name="L_brow_mid"),Weight=0.0003),(Item=(Type=Bone,Name="C_brow_mid"),Weight=0.0008),(Item=(Type=Bone,Name="R_brow_outer"),Weight=0.0921),(Item=(Type=Bone,Name="R_brow_mid"),Weight=0.2897),(Item=(Type=Bone,Name="head"),Weight=0.6171))
    Weight=1
   )
   > &RigUnit_ParentConstraint(
    AdvancedSettings=(RotationOrderForFilter=XZY)
    Child=(Type=Bone,Name="dyn_head_fur_49")
    Parents=((Item=(Type=Bone,Name="R_brow_mid"),Weight=0.0003),(Item=(Type=Bone,Name="C_brow_mid"),Weight=0.0008),(Item=(Type=Bone,Name="L_brow_outer"),Weight=0.0921),(Item=(Type=Bone,Name="L_brow_mid"),Weight=0.2897),(Item=(Type=Bone,Name="head"),Weight=0.6171))
    Weight=1
   )
   > &RigUnit_ParentConstraint(
    AdvancedSettings=(InterpolationType=Average,RotationOrderForFilter=XZY)
    Child=(Type=Bone,Name="dyn_head_fur_50")
    Parents=((Item=(Type=Bone,Name="deform_neck_hair_rear_02"),Weight=1))
    Weight=1
   )
   > RigVMFunctionReturnNode()
  ]}
 RigVMFunctionLibrary {nodes [
   Build Bandolier = RigVMCollapseNode(ExecuteContext=())
   Build Bandolier Poses = RigVMCollapseNode(ExecuteContext=())
   DeformRig_Elbow_FNC = RigVMCollapseNode(
    ExecuteContext=()
    bicep=true
    curve_struct=(deformRig_l_lat_alpha_91_#11=False,deformRig_r_lat_alpha_93_#12=False)
    elbow=true
    value_struct="(upperarm_up_18_#33=(Rotation=(X=0,Y=0,Z=0,W=1),Translation=(X=0,Y=0,Z=8),Scale3D=(X=0.7,Y=1,Z=1)),upperarm_fwd_fr_19_#26=(Rotation=(X=0,Y=0,Z=0,W=1),Translation=(X=0,Y=10,Z=0),Scale3D=(X=0.5,Y=1,Z= ...(truncated)"
   )
   DeformRig_Fast = RigVMCollapseNode(
    DeformRig_ Curve="(deformRig_l_upperarm_up_alpha_72_#34=False,deformRig_r_upperarm_up_alpha_73_#35=False,deformRig_l_upperarm_fr_alpha_74_#28=False,deformRig_r_upperarm_fr_alpha_75_CB51B8 ...(truncated)"
    DeformRig_Bool="(upperarm_1_#36=True,elbow_3_#37=True,wrist_7_#38=True,groin_8_#39=True,glute_17_#40=True,knee_9_40C408324CA261 ...(truncated)"
    DeformRig_Debug=(upperarm_debug_18_#36=False,wrist_debug_20_#38=False,groinGlute_debug_27_#39=False,lat_debug_26_#41=False)
    DeformRig_Value="(upperarm_up_18_#33=(Rotation=(X=0,Y=0,Z=0,W=1),Translation=(X=0,Y=0,Z=8),Scale3D=(X=0.7,Y=1,Z=1)),upperarm_fwd_fr_19_#26=(Rotation=(X=0,Y=0,Z=0,W=1),Translation=(X=0,Y=10,Z=0),Scale3D=(X=0.5,Y=1,Z= ...(truncated)"
    ExecuteContext=()
   )
   DeformRig_GetDeltaRefTransform_FNC = RigVMCollapseNode(ExecuteContext=(), Item=(Type=Bone))
   DeformRig_GetMirrorTransform_FNC = RigVMCollapseNode(ExecuteContext=(), transform=(Rotation=(X=0,Y=0,Z=0,W=1),Translation=(X=0,Y=0,Z=0),Scale3D=(X=1,Y=1,Z=1)))
   DeformRig_GroinGlue_FNC = RigVMCollapseNode(
    ExecuteContext=()
    curve_struct=(deformRig_l_lat_alpha_91_#11=False,deformRig_r_lat_alpha_93_#12=False)
    debug=False
    glute=true
    groin=true
    value_struct="(upperarm_up_18_#33=(Rotation=(X=0,Y=0,Z=0,W=1),Translation=(X=0,Y=0,Z=8),Scale3D=(X=0.7,Y=1,Z=1)),upperarm_fwd_fr_19_#26=(Rotation=(X=0,Y=0,Z=0,W=1),Translation=(X=0,Y=10,Z=0),Scale3D=(X=0.5,Y=1,Z= ...(truncated)"
   )
   DeformRig_Knee_FNC = RigVMCollapseNode(
    ExecuteContext=()
    curve_struct=(deformRig_l_lat_alpha_91_#11=False,deformRig_r_lat_alpha_93_#12=False)
    value_struct="(upperarm_up_18_#33=(Rotation=(X=0,Y=0,Z=0,W=1),Translation=(X=0,Y=0,Z=8),Scale3D=(X=0.7,Y=1,Z=1)),upperarm_fwd_fr_19_#26=(Rotation=(X=0,Y=0,Z=0,W=1),Translation=(X=0,Y=10,Z=0),Scale3D=(X=0.5,Y=1,Z= ...(truncated)"
   )
   DeformRig_Lat_FNC = RigVMCollapseNode(
    ExecuteContext=()
    curve_struct=(deformRig_l_lat_alpha_91_#11=False,deformRig_r_lat_alpha_93_#12=False)
    debug=False
    value_struct="(upperarm_up_18_#33=(Rotation=(X=0,Y=0,Z=0,W=1),Translation=(X=0,Y=0,Z=8),Scale3D=(X=0.7,Y=1,Z=1)),upperarm_fwd_fr_19_#26=(Rotation=(X=0,Y=0,Z=0,W=1),Translation=(X=0,Y=10,Z=0),Scale3D=(X=0.5,Y=1,Z= ...(truncated)"
   )
   DeformRig_Pec_FNC = RigVMCollapseNode(
    ExecuteContext=()
    curve_struct=()
    value_struct="(upperarm_up_18_#33=(Rotation=(X=0,Y=0,Z=0,W=1),Translation=(X=0,Y=0,Z=8),Scale3D=(X=0.7,Y=1,Z=1)),upperarm_fwd_fr_19_#26=(Rotation=(X=0,Y=0,Z=0,W=1),Translation=(X=0,Y=10,Z=0),Scale3D=(X=0.5,Y=1,Z= ...(truncated)"
   )
   DeformRig_UpperArm_FNC = RigVMCollapseNode(
    ExecuteContext=()
    curve_struct=(deformRig_l_lat_alpha_91_#11=False,deformRig_r_lat_alpha_93_#12=False)
    debug=False
    value_struct="(upperarm_up_18_#33=(Rotation=(X=0,Y=0,Z=0,W=1),Translation=(X=0,Y=0,Z=8),Scale3D=(X=0.7,Y=1,Z=1)),upperarm_fwd_fr_19_#26=(Rotation=(X=0,Y=0,Z=0,W=1),Translation=(X=0,Y=10,Z=0),Scale3D=(X=0.5,Y=1,Z= ...(truncated)"
   )
   DeformRig_Wrist_FNC = RigVMCollapseNode(
    ExecuteContext=()
    curve_struct=(deformRig_l_lat_alpha_91_#11=False,deformRig_r_lat_alpha_93_#12=False)
    debug=false
    value_struct="(upperarm_up_18_#33=(Rotation=(X=0,Y=0,Z=0,W=1),Translation=(X=0,Y=0,Z=8),Scale3D=(X=0.7,Y=1,Z=1)),upperarm_fwd_fr_19_#26=(Rotation=(X=0,Y=0,Z=0,W=1),Translation=(X=0,Y=10,Z=0),Scale3D=(X=0.5,Y=1,Z= ...(truncated)"
   )
   HeadCardPins = RigVMCollapseNode(ExecuteContext=())
   Pins = RigVMCollapseNode(ExecuteContext=())
   Solve Bandolier = RigVMCollapseNode(ExecuteContext=())
   Solve Neck Front Fur = RigVMCollapseNode(ExecuteContext=())
  ]}
 RigVMFunction_Sequence_1_ContainedGraph {nodes [
   n1 = RigVMFunctionEntryNode()
   RigVMFunction_Sequence = &RigVMFunction_Sequence(ExecuteContext=@n1.ExecuteContext)
   n2 = &RigVMFunction_Sequence(ExecuteContext=@RigVMFunction_Sequence.B)
   n3 = &RigVMFunction_Sequence(ExecuteContext=@n2.B)
   n4 = &RigVMFunction_Sequence(ExecuteContext=@n3.B)
   n5 = &RigVMFunction_Sequence(ExecuteContext=@n4.B)
   n6 = &RigVMFunction_Sequence(ExecuteContext=@n5.B)
   RigVMFunctionReturnNode(A=@RigVMFunction_Sequence.A, B=@n2.A, C=@n3.A, D=@n4.A, E=@n5.A, F=@n6.A, G=@n6.B)
  ]}
 RigVMModel {nodes [
   RigUnit_BeginExecution = &RigUnit_BeginExecution()
   DeformRig_Fast = RigVMFunctionReferenceNode(
    DeformRig_ Curve="(deformRig_l_upperarm_up_alpha_72_#34=False,deformRig_r_upperarm_up_alpha_73_#35=False,deformRig_l_upperarm_fr_alpha_74_#28=False,deformRig_r_upperarm_fr_alpha_75_CB51B8 ...(truncated)"
    DeformRig_Bool="(upperarm_1_#36=True,elbow_3_#37=True,wrist_7_#38=True,groin_8_#39=True,glute_17_#40=True,knee_9_40C408324CA261 ...(truncated)"
    DeformRig_Debug=(upperarm_debug_18_#36=False,wrist_debug_20_#38=False,groinGlute_debug_27_#39=False,lat_debug_26_#41=False)
    DeformRig_Value="(upperarm_up_18_#33=(Translation=(X=0,Y=0,Z=8),Scale3D=(X=0.7,Y=1,Z=1)),upperarm_fwd_fr_19_#26=(Translation=(X=0,Y=10,Z=0),Scale3D=(X=0.5,Y=1,Z=1)),upperarm_fwd_bk_20_9973AAD84EE0A1E349BC9E85067D4A3 ...(truncated)"
    ExecuteContext=@RigUnit_BeginExecution.ExecutePin
   )
   > Solve Neck Front Fur = RigVMFunctionReferenceNode()
   > Pins = RigVMFunctionReferenceNode()
   > &RigUnit_ModifyTransforms(
    ItemToModify=((Item=(Type=Bone,Name="def_body_fur_22"),Transform=(Rotation=(X=0.1326,Y=0.4368,Z=0.1305,W=0.8801),Translation=(X=0,Y=0,Z=0),Scale3D=(X=1,Y=1,Z=1))))
    Weight=1
    WeightMinimum=0
   )
   > &RigUnit_ModifyTransforms(
    ItemToModify=((Item=(Type=Bone,Name="foot_l"),Transform=(Rotation=(X=0,Y=0,Z=0,W=1),Translation=(X=0,Y=0,Z=0),Scale3D=(X=1.05,Y=1,Z=1))),(Item=(Type=Bone,Name="foot_r"),Transform=(Rotation=(X=0,Y=0,Z=0,W=1),Translation=(X=0,Y=0,Z=0),Scale3D=(X=1.05,Y=1,Z=1))))
    Weight=1
    WeightMinimum=0
   )
   > HeadCardPins = RigVMFunctionReferenceNode()
  ]}
]
hierarchy [
 root {loc=<0 0 0>, rot=<0 0 0 1>}
 pelvis {parent=root, loc=<0 -0.7071 93.1191>, rot=<0.0127 0.707 -0.0127 -0.707>}
 spine_01 {parent=pelvis, loc=<6.7494 -0.6388 0>, rot=<0 0 0.034 -0.9994>}
 spine_02 {parent=spine_01, loc=<7.252 -0.035 0>, rot=<0 0 0.0238 -0.9997>}
 spine_03 {parent=spine_02, loc=<6.0106 0 0>, rot=<0 0 0.0149 -0.9999>}
 spine_04 {parent=spine_03, loc=<8.1023 0 0>, rot=<0 0 -0.0132 -0.9999>}
 spine_05 {parent=spine_04, loc=<10.0142 0 0>, rot=<0 0 -0.0015 -1>}
 neck_01 {parent=spine_05, loc=<11.3138 0.1529 0>, rot=<0 0 -0.2003 -0.9797>}
 neck_02 {parent=neck_01, loc=<4.5627 0 0>, rot=<0 0 0.0504 -0.9987>}
 head {parent=neck_02, loc=<5.3914 0 0>, rot=<0 0 0.115 -0.9934>}
 faceAttach {parent=head, loc=<0 0 0>, rot=<0.0029 -0.7071 -0.0029 -0.7071>}
 C_jaw {parent=faceAttach, loc=<-0.0134 8.0146 0.3>, rot=<0 0 0 1>}
 dyn_beard_MSTR {parent=C_jaw, loc=<0 5.962 -4.4682>, rot=<0 0 0 1>}
 dyn_beard_r_MSTR {parent=dyn_beard_MSTR, loc=<-2.101 -1.2607 -2.1819>, rot=<-0.352 -0.5178 0.4071 -0.665>}
 dyn_beard_r_01 {parent=dyn_beard_r_MSTR, loc=<0 0 0>, rot=<0 0 0 1>}
 dyn_beard_r_02 {parent=dyn_beard_r_01, loc=<1.9995 0 0>, rot=<0.5979 0.1645 0.0432 -0.7833>}
 dyn_beard_c_MSTR {parent=dyn_beard_MSTR, loc=<0.3766 -0.3039 -2.6214>, rot=<-0.4058 -0.5341 0.6014 -0.4341>}
 dyn_beard_c_01 {parent=dyn_beard_c_MSTR, loc=<0 0 0>, rot=<0 0 0 1>}
 dyn_beard_c_02 {parent=dyn_beard_c_01, loc=<1.3962 0 0>, rot=<-0.0329 0.0288 -0.0856 -0.9954>}
 dyn_beard_l_MSTR {parent=dyn_beard_MSTR, loc=<2.4043 -1.4568 -2.8033>, rot=<-0.5012 -0.3467 0.7319 -0.3048>}
 dyn_beard_l_01 {parent=dyn_beard_l_MSTR, loc=<0 0 0>, rot=<0 0 0 1>}
 dyn_beard_l_02 {parent=dyn_beard_l_01, loc=<1.1882 0 0>, rot=<0.7801 -0.0032 -0.0109 -0.6256>}
 C_lip_lower_mid {parent=C_jaw, loc=<0.0223 6.9593 -1.4987>, rot=<0 0 0 -1>}
 L_lip_lower_outer {parent=C_jaw, loc=<1.7592 6.3307 -1.4414>, rot=<-0.004 0.0068 0.1992 -0.9799>}
 R_lip_lower_outer {parent=C_jaw, loc=<-1.7324 6.3307 -1.4414>, rot=<-0.004 -0.0068 -0.1992 -0.9799>}
 C_brow_mid {parent=faceAttach, loc=<0 13.0448 7.6773>, rot=<0 0 0 -1>}
 C_lip_upper_mid {parent=faceAttach, loc=<-0.0233 14.8373 0.056>, rot=<-0.1219 0.0011 -0.0009 -0.9925>}
 L_brow_mid {parent=faceAttach, loc=<2.6832 12.2852 8.1254>, rot=<-0.1117 0.0088 0.116 -0.9869>}
 L_brow_outer {parent=faceAttach, loc=<5.2456 11.2606 8.0909>, rot=<-0.0515 -0.0263 0.2236 -0.973>}
 L_cheek_inner {parent=faceAttach, loc=<4.8472 11.2445 2.5996>, rot=<-0.2037 -0.1078 0.2662 -0.936>}
 L_eye {parent=faceAttach, loc=<3.106 8.1912 5.9129>, rot=<0 0 0.9998 0.0175>}
 L_eye_lid_lower_mid {parent=L_eye, loc=<0 0 0>, rot=<-0.0707 0 0 -0.9975>}
 L_eye_lid_upper_mid {parent=L_eye, loc=<0 0 0>, rot=<0.0727 0 0 -0.9974>}
 L_lip_corner {parent=faceAttach, loc=<3.2348 13.0542 -0.7297>, rot=<0.003 0.008 -0.0146 -0.9999>}
 L_lip_upper_outer {parent=faceAttach, loc=<1.7557 14.3449 -0.0633>, rot=<-0.0732 -0.0115 0.2302 -0.9703>}
 R_brow_mid {parent=faceAttach, loc=<-2.6832 12.2852 8.1254>, rot=<-0.1117 -0.0088 -0.116 -0.9869>}
 R_brow_outer {parent=faceAttach, loc=<-5.2456 11.2606 8.0909>, rot=<-0.0515 0.0263 -0.2236 -0.973>}
 R_cheek_inner {parent=faceAttach, loc=<-4.8472 11.2445 2.5996>, rot=<-0.2037 0.1078 -0.2662 -0.936>}
 R_eye {parent=faceAttach, loc=<-3.106 8.1912 5.9128>, rot=<0 0 0.9998 -0.0175>}
 R_eye_lid_lower_mid {parent=R_eye, loc=<0 0 0.0001>, rot=<-0.0707 0 0 -0.9975>}
 R_eye_lid_upper_mid {parent=R_eye, loc=<0 0 0.0001>, rot=<0.0727 0 0 -0.9974>}
 R_lip_corner {parent=faceAttach, loc=<-3.2348 13.0542 -0.7297>, rot=<0.003 -0.008 0.0146 -0.9999>}
 R_lip_upper_outer {parent=faceAttach, loc=<-1.7557 14.3449 -0.0633>, rot=<-0.0732 0.0115 -0.2302 -0.9703>}
 C_nose_bridge {parent=faceAttach, loc=<0 13.5309 4.6814>, rot=<-0.4193 0 0 -0.9079>}
 teeth_upper {parent=faceAttach, loc=<0 9.7287 0.7952>, rot=<0 0 0 1>}
 deform_neck_hair_a_l_MSTR {parent=head, loc=<-4.0087 5.6409 -6.2811>, rot=<-0.0067 -0.9661 -0.1741 -0.1906>}
 deform_neck_hair_a_l_01 {parent=deform_neck_hair_a_l_MSTR, loc=<0 0 0>, rot=<0 0 0 1>}
 deform_neck_hair_a_l_02 {parent=deform_neck_hair_a_l_01, loc=<3.9204 0 0>, rot=<0 0 0 1>}
 deform_neck_hair_b_l_MSTR {parent=head, loc=<-2.4611 2.1755 -8.4027>, rot=<0.1928 -0.9565 -0.0696 -0.2074>}
 deform_neck_hair_b_l_01 {parent=deform_neck_hair_b_l_MSTR, loc=<0 0 0>, rot=<0 0 0 1>}
 deform_neck_hair_b_l_02 {parent=deform_neck_hair_b_l_01, loc=<5.973 0 0>, rot=<0 0.2462 0 -0.9692>}
 deform_neck_hair_c_l_MSTR {parent=head, loc=<-5.6305 -5.3382 -10.1057>, rot=<0.2354 -0.9056 0.2318 -0.2658>}
 deform_neck_hair_c_l_01 {parent=deform_neck_hair_c_l_MSTR, loc=<0 0 0>, rot=<0 0 0 1>}
 deform_neck_hair_c_l_02 {parent=deform_neck_hair_c_l_01, loc=<2.309 0 0>, rot=<0 0 0 1>}
 deform_neck_hair_a_r_MSTR {parent=head, loc=<-4.009 5.6409 6.2811>, rot=<0.9661 -0.0067 -0.1906 0.1741>}
 deform_neck_hair_a_r_01 {parent=deform_neck_hair_a_r_MSTR, loc=<0 0 0>, rot=<0 0 0 -1>}
 deform_neck_hair_a_r_02 {parent=deform_neck_hair_a_r_01, loc=<-3.92 0 -0.0002>, rot=<0 0 0 -1>}
 deform_neck_hair_b_r_MSTR {parent=head, loc=<-2.4613 2.1755 8.4027>, rot=<0.9565 0.1928 -0.2074 0.0696>}
 deform_neck_hair_b_r_01 {parent=deform_neck_hair_b_r_MSTR, loc=<0 0 0>, rot=<0 0 0 1>}
 deform_neck_hair_b_r_02 {parent=deform_neck_hair_b_r_01, loc=<-5.9728 0.0001 -0.0001>, rot=<0 0.2462 0 -0.9692>}
 deform_neck_hair_c_r_MSTR {parent=head, loc=<-5.6303 -5.3382 10.1057>, rot=<0.9056 0.2354 -0.2658 -0.2318>}
 deform_neck_hair_c_r_01 {parent=deform_neck_hair_c_r_MSTR, loc=<0 0 0>, rot=<0 0 0 1>}
 deform_neck_hair_c_r_02 {parent=deform_neck_hair_c_r_01, loc=<-2.3088 0.0001 -0.0001>, rot=<0 0 0 1>}
 deform_neck_hair_d_l_MSTR {parent=head, loc=<-4.8587 -9.2115 -5.2715>, rot=<0.229 -0.681 0.6596 -0.2209>}
 deform_neck_hair_d_l_01 {parent=deform_neck_hair_d_l_MSTR, loc=<0 0 0>, rot=<0 0 0 1>}
 deform_neck_hair_d_l_02 {parent=deform_neck_hair_d_l_01, loc=<3.9434 0 0>, rot=<0 0 0 1>}
 deform_neck_hair_rear_MSTR {parent=head, loc=<-5.2893 -10.0691 -0.1196>, rot=<0.229 -0.681 0.6596 -0.2209>}
 deform_neck_hair_rear_01 {parent=deform_neck_hair_rear_MSTR, loc=<0 0 0>, rot=<0 -0.0343 0 -0.9994>}
 deform_neck_hair_rear_02 {parent=deform_neck_hair_rear_01, loc=<5.0389 0 0>, rot=<0 0 0 1>}
 deform_neck_hair_d_r_MSTR {parent=head, loc=<-4.8583 -9.2115 5.2715>, rot=<0.681 0.229 -0.2209 -0.6596>}
 deform_neck_hair_d_r_01 {parent=deform_neck_hair_d_r_MSTR, loc=<0 0 0>, rot=<0 0 0 1>}
 deform_neck_hair_d_r_02 {parent=deform_neck_hair_d_r_01, loc=<-3.944 0 0.0004>, rot=<0 0 0 1>}
 dyn_fur_cards_root {parent=head, loc=<0 0 0>, rot=<0 0 0 1>}
 dyn_head_fur_01 {parent=dyn_fur_cards_root, loc=<17.1479 -2.9671 0.8802>, rot=<0.3373 -0.5692 0.6767 -0.3231>}
 dyn_head_fur_02 {parent=dyn_fur_cards_root, loc=<18.0222 1.7818 -3.3223>, rot=<0.4842 -0.6256 0.4804 -0.3786>}
 dyn_head_fur_03 {parent=dyn_fur_cards_root, loc=<14.0856 -2.7121 -6.8614>, rot=<0.5672 -0.7306 0.2973 -0.2367>}
 dyn_head_fur_04 {parent=dyn_fur_cards_root, loc=<14.8191 3.9625 -7.3159>, rot=<0.6155 -0.7076 0.189 -0.2912>}
 dyn_head_fur_05 {parent=dyn_fur_cards_root, loc=<11.5387 7.9318 -7.5649>, rot=<0.6593 -0.711 0.009 -0.2446>}
 dyn_head_fur_06 {parent=dyn_fur_cards_root, loc=<9.328 8.8246 -7.2983>, rot=<0.6848 -0.7065 -0.0805 -0.1597>}
 dyn_head_fur_07 {parent=dyn_fur_cards_root, loc=<5.6022 8.0409 -7.6614>, rot=<0.6985 -0.6928 -0.16 -0.081>}
 dyn_head_fur_08 {parent=dyn_fur_cards_root, loc=<7.6033 0.2239 -10.1589>, rot=<0.4743 -0.8702 0.0705 -0.1134>}
 dyn_head_fur_09 {parent=dyn_fur_cards_root, loc=<13.6096 -6.4922 -2.05>, rot=<0.1109 -0.7266 0.6522 -0.1855>}
 dyn_head_fur_10 {parent=dyn_fur_cards_root, loc=<4.7368 -10.1818 -1.54>, rot=<0.1257 -0.7774 0.5906 -0.1758>}
 dyn_head_fur_11 {parent=dyn_fur_cards_root, loc=<7.8505 -7.2219 -6.7079>, rot=<0.1741 -0.9225 0.3199 -0.1281>}
 dyn_head_fur_12 {parent=dyn_fur_cards_root, loc=<9.8251 -8.3952 1.2815>, rot=<0.2251 -0.7023 0.643 -0.2065>}
 dyn_head_fur_13 {parent=dyn_fur_cards_root, loc=<-2.4495 -10.869 1.0251>, rot=<0.0058 -0.6611 0.7426 -0.107>}
 dyn_head_fur_14 {parent=dyn_fur_cards_root, loc=<-4.4224 -10.7711 -2.9255>, rot=<0.1236 -0.7881 0.58 -0.1652>}
 dyn_head_fur_15 {parent=dyn_fur_cards_root, loc=<-2.8606 -9.1635 -4.9476>, rot=<0.2587 -0.7546 0.5998 -0.0619>}
 dyn_head_fur_16 {parent=dyn_fur_cards_root, loc=<2.7164 -6.7474 -8.4192>, rot=<0.2965 -0.8907 0.3343 -0.0839>}
 dyn_head_fur_17 {parent=dyn_fur_cards_root, loc=<-4.4585 -6.4318 -10.8898>, rot=<0.3434 -0.8679 0.2858 -0.2169>}
 dyn_head_fur_18 {parent=dyn_fur_cards_root, loc=<2.7426 2.3425 -10.2995>, rot=<0.397 -0.8966 -0.0408 -0.1916>}
 dyn_head_fur_19 {parent=dyn_fur_cards_root, loc=<-2.3744 3.8079 -10.5027>, rot=<0.2777 -0.9317 -0.2198 -0.0814>}
 dyn_head_fur_20 {parent=dyn_fur_cards_root, loc=<0.0753 8.094 -7.9192>, rot=<0.2654 -0.9259 -0.2344 -0.1316>}
 dyn_head_fur_21 {parent=dyn_fur_cards_root, loc=<0.2798 11.4391 -5.1774>, rot=<0.1928 -0.8998 -0.3847 -0.0715>}
 dyn_head_fur_22 {parent=dyn_fur_cards_root, loc=<-3.8166 9.3996 -6.5033>, rot=<0.1658 -0.9052 -0.3867 -0.0599>}
 dyn_head_fur_23 {parent=dyn_fur_cards_root, loc=<-0.7117 -8.9635 5.1449>, rot=<0.0279 -0.5886 0.8055 -0.0633>}
 dyn_head_fur_24 {parent=dyn_fur_cards_root, loc=<17.8269 5.554 1.0101>, rot=<0.5408 -0.3573 0.4775 -0.5931>}
 dyn_head_fur_25 {parent=dyn_head_fur_24, loc=<3.5811 0 0>, rot=<0 -0.1105 0 -0.9939>}
 dyn_head_fur_26 {parent=dyn_fur_cards_root, loc=<16.938 0.0247 4.7226>, rot=<0.3103 -0.4432 0.7248 -0.4266>}
 dyn_head_fur_27 {parent=dyn_head_fur_26, loc=<3.0715 0 0>, rot=<0 -0.1105 0 -0.9939>}
 dyn_head_fur_28 {parent=dyn_fur_cards_root, loc=<13.1765 -2.451 7.526>, rot=<0.2612 -0.2998 0.8922 -0.2141>}
 dyn_head_fur_29 {parent=dyn_fur_cards_root, loc=<9.1446 -6.8892 6.3179>, rot=<0.2178 -0.4616 0.8032 -0.3073>}
 dyn_head_fur_30 {parent=dyn_fur_cards_root, loc=<11.2462 0.2014 9.0192>, rot=<0.1112 -0.0944 0.8793 -0.4534>}
 dyn_head_fur_31 {parent=dyn_fur_cards_root, loc=<3.0816 -4.9033 9.6132>, rot=<0.0485 -0.1726 0.9563 -0.2311>}
 dyn_head_fur_32 {parent=dyn_fur_cards_root, loc=<8.5282 0.7298 9.5921>, rot=<0.17 -0.064 0.9461 -0.268>}
 dyn_head_fur_33 {parent=dyn_fur_cards_root, loc=<1.1727 1.8723 10.5743>, rot=<0.2997 0.0457 0.8701 -0.3886>}
 dyn_head_fur_34 {parent=dyn_fur_cards_root, loc=<-1.6754 3.4681 10.4592>, rot=<0.1254 0.1746 0.9447 -0.2475>}
 dyn_head_fur_35 {parent=dyn_fur_cards_root, loc=<4.5267 7.9829 7.9715>, rot=<0.1123 0.1833 0.9603 -0.1776>}
 dyn_head_fur_36 {parent=dyn_fur_cards_root, loc=<6.7443 7.2222 8.1263>, rot=<0.1152 0.2078 0.811 -0.5346>}
 dyn_head_fur_37 {parent=dyn_fur_cards_root, loc=<8.616 8.4206 7.8311>, rot=<0.1997 0.1402 0.6455 -0.7237>}
 dyn_head_fur_38 {parent=dyn_fur_cards_root, loc=<-0.0596 7.8632 8.2343>, rot=<0.1123 0.1833 0.9603 -0.1776>}
 dyn_head_fur_40 {parent=dyn_fur_cards_root, loc=<-3.629 10.4142 5.3868>, rot=<0.118 0.233 0.9323 -0.25>}
 dyn_head_fur_41 {parent=dyn_fur_cards_root, loc=<0.8493 14.6097 1.5553>, rot=<0.4556 0.3464 0.6477 -0.5028>}
 dyn_head_fur_42 {parent=dyn_fur_cards_root, loc=<0.9618 14.5573 -1.5754>, rot=<0.413 -0.7298 -0.3528 -0.4151>}
 dyn_head_fur_43 {parent=dyn_fur_cards_root, loc=<16.1422 7.3174 -4.0457>, rot=<0.6703 -0.4864 0.2156 -0.5174>}
 dyn_head_fur_44 {parent=dyn_fur_cards_root, loc=<14.5676 9.771 -2.6501>, rot=<0.7636 -0.3429 0.0435 -0.5454>}
 dyn_head_fur_45 {parent=dyn_fur_cards_root, loc=<14.8178 9.6753 1.0515>, rot=<0.6823 -0.2149 0.3299 -0.6161>}
 dyn_head_fur_46 {parent=dyn_fur_cards_root, loc=<12.2487 11.0713 2.9309>, rot=<0.6007 -0.0123 0.311 -0.7364>}
 dyn_head_fur_47 {parent=dyn_fur_cards_root, loc=<12.3349 10.6219 -4.1329>, rot=<0.737 -0.3884 0.0199 -0.5528>}
 dyn_head_fur_48 {parent=dyn_fur_cards_root, loc=<8.7968 12.1275 2.7843>, rot=<0.6494 0.1304 0.2111 -0.7188>}
 dyn_head_fur_49 {parent=dyn_fur_cards_root, loc=<8.676 12.3697 -2.7433>, rot=<0.738 -0.2189 -0.0943 -0.6313>}
 dyn_head_fur_50 {parent=dyn_fur_cards_root, loc=<-8.8918 -13.7423 -0.2526>, rot=<0.2061 -0.672 0.6667 -0.2479>}
 head_twist {parent=neck_01, loc=<4.563 0 0>, rot=<0 0 0.0504 -0.9987>}
 clavicle_l {parent=spine_05, loc=<4.3575 -0.1466 -2.3223>, rot=<0.0465 -0.7289 -0.009 -0.6829>}
 upperarm_l {parent=clavicle_l, loc=<15.5111 0 0>, rot=<0.0226 -0.3234 -0.0088 -0.9459>}
 lowerarm_l {parent=upperarm_l, loc=<26.6514 0 0>, rot=<0.0391 -0.0099 -0.2453 -0.9686>}
 hand_l {parent=lowerarm_l, loc=<22.3491 0 0>, rot=<-0.7081 -0.0507 0.0975 -0.6975>}
 thumb_01_l {parent=hand_l, loc=<1.3118 -0.3191 -2.459>, rot=<0.6546 -0.4042 0.0664 -0.6354>}
 thumb_02_l {parent=thumb_01_l, loc=<5.9336 -0.0448 0.0401>, rot=<-0.0037 0.0213 0.298 -0.9543>}
 thumb_03_l {parent=thumb_02_l, loc=<2.2223 0.3682 -0.0907>, rot=<0.0118 -0.013 0.0066 -0.9998>}
 index_metacarpal_l {parent=hand_l, loc=<1.6535 0.0183 -2.96>, rot=<0.0158 -0.0124 -0.0173 -0.9996>}
 index_01_l {parent=index_metacarpal_l, loc=<8.3008 0.0357 -0.0524>, rot=<-0.0044 0.0485 0.0924 -0.9945>}
 index_02_l {parent=index_01_l, loc=<3.3587 -0.3004 0.0452>, rot=<-0.002 0.0027 0.3225 -0.9466>}
 index_03_l {parent=index_02_l, loc=<2.5578 -0.123 -0.0252>, rot=<-0.002 0.0067 0.2558 -0.9667>}
 middle_metacarpal_l {parent=hand_l, loc=<2.1856 -0.1725 -0.5984>, rot=<0.006 0.0168 -0.0329 -0.9993>}
 middle_01_l {parent=middle_metacarpal_l, loc=<7.3275 -0.2447 0.0129>, rot=<-0.0208 0.0042 0.0858 -0.9961>}
 middle_02_l {parent=middle_01_l, loc=<4.0123 0.1401 0.1175>, rot=<-0.0218 0.0258 0.3512 -0.9357>}
 middle_03_l {parent=middle_02_l, loc=<2.6077 0.0534 0.1373>, rot=<0.0082 0.0188 0.1708 -0.9851>}
 ring_metacarpal_l {parent=hand_l, loc=<2.2074 -0.3625 1.3638>, rot=<-0.0284 0.0722 -0.0458 -0.9959>}
 ring_01_l {parent=ring_metacarpal_l, loc=<7.0295 -0.6197 -0.1446>, rot=<-0.0489 -0.0252 0.1892 -0.9804>}
 ring_02_l {parent=ring_01_l, loc=<3.6174 0.4647 -0.1027>, rot=<0 0 0.2806 -0.9598>}
 ring_03_l {parent=ring_02_l, loc=<2.4152 0.114 0.0199>, rot=<0 0 0.1189 -0.9929>}
 pinky_metacarpal_l {parent=hand_l, loc=<1.7616 -0.9753 3.3814>, rot=<-0.0397 0.0814 -0.069 -0.9935>}
 pinky_01_l {parent=pinky_metacarpal_l, loc=<7.1709 -0.4831 -0.1094>, rot=<-0.0654 -0.0035 0.2242 -0.9723>}
 pinky_02_l {parent=pinky_01_l, loc=<2.8274 0.0893 -0.1043>, rot=<0.0037 0.0011 0.2854 -0.9584>}
 pinky_03_l {parent=pinky_02_l, loc=<2.2046 0.0307 -0.0078>, rot=<-0.0002 0.0098 0.1445 -0.9895>}
 deform_wrist_dn_l {parent=hand_l, loc=<0 -2.2342 0>, rot=<0 0 0 1>}
 deform_wrist_up_l {parent=hand_l, loc=<0 2.2682 0>, rot=<0 0 0 1>}
 lowerarm_twist_02_l {parent=lowerarm_l, loc=<9.6199 0 0>, rot=<0 0 0 -1>}
 lowerarm_twist_01_l {parent=lowerarm_l, loc=<17.8549 0 0>, rot=<0 0 0 -1>}
 deform_elbow_in_l {parent=lowerarm_l, loc=<0.2112 5.259 0.8274>, rot=<0 0 0 -1>}
 deform_elbow_out_l {parent=lowerarm_l, loc=<0 -4 0>, rot=<0 0 0 -1>}
 upperarm_twist_01_l {parent=upperarm_l, loc=<0.0041 0 0>, rot=<0 0 0 1>}
 deform_upperarm_bk_l {parent=upperarm_twist_01_l, loc=<1.0813 -7.2193 -1.9453>, rot=<0 0 0 1>}
 deform_upperarm_fr_l {parent=upperarm_twist_01_l, loc=<-0.3691 7.0506 -2.765>, rot=<0 0 0 1>}
 deform_upperarm_up_l {parent=upperarm_twist_01_l, loc=<-3.2216 -1.1391 3.3311>, rot=<0 0 0 1>}
 upperarm_twist_02_l {parent=upperarm_l, loc=<13.7941 0 0>, rot=<0 0 0 1>}
 clavicle_r {parent=spine_05, loc=<4.3575 -0.1466 2.3223>, rot=<0.7289 0.0465 -0.6829 0.009>}
 upperarm_r {parent=clavicle_r, loc=<-15.5111 0 0>, rot=<0.0226 -0.3234 -0.0088 -0.9459>}
 lowerarm_r {parent=upperarm_r, loc=<-26.6514 0 0>, rot=<0.0391 -0.0099 -0.2453 -0.9686>}
 hand_r {parent=lowerarm_r, loc=<-22.3491 0 0>, rot=<-0.7081 -0.0507 0.0975 -0.6975>}
 thumb_01_r {parent=hand_r, loc=<-1.3118 0.3191 2.459>, rot=<0.6546 -0.4042 0.0664 -0.6354>}
 thumb_02_r {parent=thumb_01_r, loc=<-5.9336 0.0448 -0.0401>, rot=<-0.0037 0.0213 0.298 -0.9543>}
 thumb_03_r {parent=thumb_02_r, loc=<-2.2223 -0.3682 0.0907>, rot=<0.0118 -0.013 0.0066 -0.9998>}
 index_metacarpal_r {parent=hand_r, loc=<-1.6535 -0.0183 2.96>, rot=<0.0158 -0.0124 -0.0173 -0.9996>}
 index_01_r {parent=index_metacarpal_r, loc=<-8.3008 -0.0357 0.0524>, rot=<-0.0044 0.0485 0.0924 -0.9945>}
 index_02_r {parent=index_01_r, loc=<-3.3587 0.3004 -0.0452>, rot=<-0.002 0.0027 0.3225 -0.9466>}
 index_03_r {parent=index_02_r, loc=<-2.5578 0.123 0.0252>, rot=<-0.002 0.0067 0.2558 -0.9667>}
 middle_metacarpal_r {parent=hand_r, loc=<-2.1856 0.1725 0.5984>, rot=<0.006 0.0168 -0.0329 -0.9993>}
 middle_01_r {parent=middle_metacarpal_r, loc=<-7.3275 0.2447 -0.0129>, rot=<-0.0208 0.0042 0.0858 -0.9961>}
 middle_02_r {parent=middle_01_r, loc=<-4.0123 -0.1401 -0.1175>, rot=<-0.0218 0.0258 0.3512 -0.9357>}
 middle_03_r {parent=middle_02_r, loc=<-2.6077 -0.0534 -0.1373>, rot=<0.0082 0.0188 0.1708 -0.9851>}
 ring_metacarpal_r {parent=hand_r, loc=<-2.2074 0.3625 -1.3638>, rot=<-0.0284 0.0722 -0.0458 -0.9959>}
 ring_01_r {parent=ring_metacarpal_r, loc=<-7.0295 0.6197 0.1446>, rot=<-0.0489 -0.0252 0.1892 -0.9804>}
 ring_02_r {parent=ring_01_r, loc=<-3.6174 -0.4647 0.1027>, rot=<0 0 0.2806 -0.9598>}
 ring_03_r {parent=ring_02_r, loc=<-2.4152 -0.114 -0.0199>, rot=<0 0 0.1189 -0.9929>}
 pinky_metacarpal_r {parent=hand_r, loc=<-1.7616 0.9753 -3.3814>, rot=<-0.0397 0.0814 -0.069 -0.9935>}
 pinky_01_r {parent=pinky_metacarpal_r, loc=<-7.1709 0.4831 0.1094>, rot=<-0.0654 -0.0035 0.2242 -0.9723>}
 pinky_02_r {parent=pinky_01_r, loc=<-2.8274 -0.0893 0.1043>, rot=<0.0037 0.0011 0.2854 -0.9584>}
 pinky_03_r {parent=pinky_02_r, loc=<-2.2046 -0.0307 0.0078>, rot=<-0.0002 0.0098 0.1445 -0.9895>}
 deform_wrist_dn_r {parent=hand_r, loc=<-0.0004 2.2345 -0.0001>, rot=<0 0 0 1>}
 deform_wrist_up_r {parent=hand_r, loc=<0.0003 -2.2684 0.0001>, rot=<0 0 0 1>}
 lowerarm_twist_02_r {parent=lowerarm_r, loc=<-9.6199 0 0>, rot=<0 0 0 -1>}
 lowerarm_twist_01_r {parent=lowerarm_r, loc=<-17.8549 0 0>, rot=<0 0 0 -1>}
 deform_elbow_in_r {parent=lowerarm_r, loc=<-0.2112 -5.2591 -0.8273>, rot=<0 0 0 -1>}
 deform_elbow_out_r {parent=lowerarm_r, loc=<0 4 -0.0001>, rot=<0 0 0 -1>}
 upperarm_twist_01_r {parent=upperarm_r, loc=<-0.0041 0 0>, rot=<0 0 0 -1>}
 deform_upperarm_bk_r {parent=upperarm_twist_01_r, loc=<-1.0814 7.2192 1.9456>, rot=<0 0 0 1>}
 deform_upperarm_fr_r {parent=upperarm_twist_01_r, loc=<0.3689 -7.0506 2.7652>, rot=<0 0 0 1>}
 deform_upperarm_up_r {parent=upperarm_twist_01_r, loc=<3.2214 1.1391 -3.3308>, rot=<0 0 0 -1>}
 upperarm_twist_02_r {parent=upperarm_r, loc=<-13.7941 0 0>, rot=<0 0 0 -1>}
 attach_cape {parent=spine_05, loc=<9.017 -11.714 0>, rot=<0.0616 -0.7044 -0.0616 -0.7044>}
 attach_backpack {parent=spine_05, loc=<-0.123 -13.473 0>, rot=<-0.0093 -0.707 0.0093 -0.707>}
 deform_pec_aim_l {parent=spine_04, loc=<7.8728 12.6388 -2.4123>, rot=<0.0102 -0.4602 0.3979 -0.7936>}
 deform_pec_l {parent=deform_pec_aim_l, loc=<10.8209 0 0>, rot=<0 0 0 1>}
 deform_pec_aim_r {parent=spine_04, loc=<7.8723 12.6388 2.4123>, rot=<-0.4602 -0.0102 0.7936 0.3979>}
 deform_pec_r {parent=deform_pec_aim_r, loc=<-10.7662 -0.0006 -0.0007>, rot=<0 0 0 1>}
 def_neck_fur_front {parent=spine_04, loc=<14.0101 14.4709 0>, rot=<0 0 0.1217 -0.9926>}
 def_body_fur_root {parent=spine_04, loc=<0 0 0>, rot=<0 0 0 1>}
 def_body_fur_22 {parent=def_body_fur_root, loc=<14.2394 -11.5723 -8.369>, rot=<0.381 -0.9089 0.001 -0.1698>}
 thigh_l {parent=pelvis, loc=<-2.5693 -0.0109 -7.7501>, rot=<0.0012 0.0572 0.0155 -0.9982>}
 calf_l {parent=thigh_l, loc=<-39.0484 0 0>, rot=<0.0006 -0.0161 -0.0391 -0.9991>}
 foot_l {parent=calf_l, loc=<-42.561 0.1077 0.4292>, rot=<0.0418 -0.0455 0.0343 -0.9975>}
 ball_l {parent=foot_l, loc=<-6.1599 13.8264 0.6806>, rot=<0.0001 -0.0001 -0.7063 -0.7079>}
 calf_twist_02_l {parent=calf_l, loc=<-18.8064 -0.0967 0.0397>, rot=<0.0237 0.0095 -0.0035 -0.9997>}
 calf_twist_01_l {parent=calf_l, loc=<-35.813 -0.0967 0.0397>, rot=<0.0237 0.0095 -0.0035 -0.9997>}
 deform_knee_in_l {parent=calf_l, loc=<1.298 -5.2957 -0.0496>, rot=<0 0 0 -1>}
 deform_knee_out_l {parent=calf_l, loc=<0 4 0>, rot=<0 0 0 -1>}
 thigh_twist_01_l {parent=thigh_l, loc=<-21.6303 0 0>, rot=<0.0013 0 0 -1>}
 deform_groin_l {parent=thigh_l, loc=<-2.1908 8.1142 -1.7442>, rot=<0 0 0 -1>}
 thigh_r {parent=pelvis, loc=<-2.5693 -0.0109 7.7501>, rot=<0.0572 -0.0012 0.9982 0.0155>}
 calf_r {parent=thigh_r, loc=<39.0484 0 0>, rot=<0.0006 -0.0161 -0.0391 -0.9991>}
 foot_r {parent=calf_r, loc=<42.561 -0.1077 -0.4292>, rot=<0.0418 -0.0455 0.0343 -0.9975>}
 ball_r {parent=foot_r, loc=<6.1599 -13.8264 -0.6806>, rot=<0.0001 -0.0001 -0.7063 -0.7079>}
 calf_twist_02_r {parent=calf_r, loc=<18.8064 0.0967 -0.0397>, rot=<0.0237 0.0095 -0.0035 -0.9997>}
 calf_twist_01_r {parent=calf_r, loc=<35.813 0.0967 -0.0397>, rot=<0.0237 0.0095 -0.0035 -0.9997>}
 deform_knee_in_r {parent=calf_r, loc=<-1.298 5.2957 0.0496>, rot=<0 0 0 -1>}
 deform_knee_out_r {parent=calf_r, loc=<0 -4 0>, rot=<0 0 0 -1>}
 thigh_twist_01_r {parent=thigh_r, loc=<21.6303 0 0>, rot=<0.0013 0 0 -1>}
 deform_groin_r {parent=thigh_r, loc=<2.1908 -8.1142 1.7442>, rot=<0 0 0 -1>}
 deform_glute_l {parent=pelvis, loc=<-8.8064 -8.8222 -8.4287>, rot=<0.0275 0.0501 -0.4472 -0.8926>}
 deform_glute_r {parent=pelvis, loc=<-8.8062 -8.8222 8.4287>, rot=<0.0501 -0.0275 0.8926 -0.4472>}
]
