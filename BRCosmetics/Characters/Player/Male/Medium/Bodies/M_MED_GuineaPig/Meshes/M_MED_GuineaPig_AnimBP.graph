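# Animation Blueprint for the male medium GuineaPig player body, adding dynamic head fur motion to poses copied from other meshes.
# It blends head and jaw bones from the Head component onto the attached parent's pose, then runs a Control Rig and caches the result.
# A long Anim Dynamics chain simulates individual fur bones with angular springs and gravity.
# The output layers the cached HeadFurCards pose over the rigged pose from dyn_fur_cards_root.
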AnimBlueprint /BRCosmetics/Characters/Player/Male/Medium/Bodies/M_MED_GuineaPig/Meshes/M_MED_GuineaPig_AnimBP
target_skeleton=/Game/Characters/Player/Male/Medium/Bodies/M_MED_GuineaPig/Meshes/M_MED_GuineaPig_Skeleton
graphs [
 AnimGraph {nodes [
   Anim Dynamics - Bone: dyn_head_fur_39!(ComponentPose, Alpha=1, GravityOverride:Vector="0,0,-3000", ExternalForce="0,0,0") -> (Pose) {
    angularDampingOverride=0.95
    angularSpringConstant=15000
    bAngularSpring=true
    bOverrideAngularDamping=true
    bUseGravityOverride=true
    bUsePlanarLimit=false
    gravityScale=0.35
    numSolverIterationsPreUpdate=1
    alphaBoolBlend {blendOption=Linear}
    alphaScaleBiasClamp {scale=0.7}
    boundBone {boneName=dyn_head_fur_39}
    componentAppliedLinearAccClamp=<1300 1300 2400>
    componentLinearAccScale=<0.5 0.5 0.5>
    componentLinearVelScale=<0.5 0.5 0.5>
    gravityOverride {z=-3000}
    physicsBodyDefinitions [{
      boundBone {boneName=dyn_head_fur_39}
      boxExtents=<25 25 25>
      constraintSetup {angularLimitsMax=<20 5 20>, angularLimitsMin=<-20 -40 -20>, angularTarget=<1 0 0>}
      localJointOffset=<-3 0 0>
     }]
   }
   Output Pose(Result=Layered blend per bone(BasePose=Use cached pose 'PostDeformationRig'(), BlendPoses_0=Use cached pose 'HeadFurCards'()) {layerSetup [{branchFilters [{boneName=dyn_fur_cards_root}]}]}) {layerGroup=None}
   n1 = Layered blend per bone(
    BasePose=Copy Pose From Mesh() {bUseAttachedParent=true}
    BlendPoses_0=Copy Pose From Mesh(SourceMeshComponent=Get Head() -> (Head:Skeletal Mesh Component Budgeted Object Reference)) {bUseAttachedParent=true}
   ) {layerSetup [{branchFilters [{boneName=head}, {boneName=head_twist}, {boneName=jaw}, {boneName=None}]}]}
   n2 = Control Rig(Source=@n1) {
    bTransferInputCurves=false
    alphaBoolBlend {blendOption=Linear}
    controlRigAssetReference {blueprintRigClass=&/M_MED_GuineaPig_CtrlRig.M_MED_GuineaPig_CtrlRig_C}
   }
   Save cached pose 'PostDeformationRig'(Pose=@n2) {cacheName=PostDeformationRig}
   n3 = "Constraint - dyn_head_fur_50 (P)"(ComponentPose=Local To Component(LocalPose=Use cached pose 'PostDeformationRig'()), Alpha=1, ConstraintWeights_0=1) {
    alphaBoolBlend {blendOption=Linear}
    boneToModify {boneName=dyn_head_fur_50}
    constraintSetup [{targetBone {boneName=deform_neck_hair_rear_02}}]
    constraintWeights [1]
   }
   n4 = "Transform (Modify) Bone - Bone: dyn_head_fur_24"(ComponentPose=@n3, Alpha=1, Translation="0,0,0", Rotation="-5.434,-0.022,0.265", Scale="1,1,1") {
    rotationMode=BMM_Additive
    rotationSpace=BCS_BoneSpace
    alphaBoolBlend {blendOption=Linear}
    boneToModify {boneName=dyn_head_fur_24}
    rotation=<rot -5.434 -0.022 0.265>
   }
   n5 = Anim Dynamics - Bone: dyn_head_fur_01!(ComponentPose=@n4, Alpha=1, GravityOverride:Vector="0,0,-3000", ExternalForce="0,0,0") -> (Pose) {
    angularDampingOverride=0.95
    angularSpringConstant=15000
    bAngularSpring=true
    bOverrideAngularDamping=true
    bUseGravityOverride=true
    bUsePlanarLimit=false
    gravityScale=0.35
    numSolverIterationsPreUpdate=1
    alphaBoolBlend {blendOption=Linear}
    alphaScaleBiasClamp {scale=0.7}
    boundBone {boneName=dyn_head_fur_01}
    componentAppliedLinearAccClamp=<1300 1300 1300>
    componentLinearAccScale=<0.5 0.5 0.5>
    componentLinearVelScale=<0.5 0.5 0.5>
    gravityOverride {z=-3000}
    physicsBodyDefinitions [{
      boundBone {boneName=dyn_head_fur_01}
      boxExtents=<30 30 30>
      constraintSetup {angularLimitsMax=<20 5 20>, angularLimitsMin=<-20 -40 -20>, angularTarget=<1 0 0>}
      localJointOffset=<-2 0.2 0.1>
     }]
   }
   n6 = Anim Dynamics - Bone: dyn_head_fur_02!(ComponentPose=@n5, Alpha=1, GravityOverride:Vector="0,0,-3000", ExternalForce="0,0,0") -> (Pose) {
    angularDampingOverride=0.95
    angularSpringConstant=15000
    bAngularSpring=true
    bOverrideAngularDamping=true
    bUseGravityOverride=true
    bUsePlanarLimit=false
    gravityScale=0.35
    numSolverIterationsPreUpdate=1
    alphaBoolBlend {blendOption=Linear}
    alphaScaleBiasClamp {scale=0.7}
    boundBone {boneName=dyn_head_fur_02}
    componentAppliedLinearAccClamp=<1300 1300 2400>
    componentLinearAccScale=<0.5 0.5 0.5>
    componentLinearVelScale=<0.5 0.5 0.5>
    gravityOverride {z=-3000}
    physicsBodyDefinitions [{
      boundBone {boneName=dyn_head_fur_02}
      boxExtents=<25 35 25>
      constraintSetup {angularLimitsMax=<20 5 20>, angularLimitsMin=<-20 -40 -20>, angularTarget=<1 0 0>}
      localJointOffset=<-2 -0.5 0>
     }]
   }
   n7 = Anim Dynamics - Bone: dyn_head_fur_03!(ComponentPose=@n6, Alpha=1, GravityOverride:Vector="0,0,-3000", ExternalForce="0,0,0") -> (Pose) {
    angularDampingOverride=0.95
    angularSpringConstant=15000
    bAngularSpring=true
    bOverrideAngularDamping=true
    bUseGravityOverride=true
    bUsePlanarLimit=false
    gravityScale=0.35
    numSolverIterationsPreUpdate=1
    alphaBoolBlend {blendOption=Linear}
    alphaScaleBiasClamp {scale=0.7}
    boundBone {boneName=dyn_head_fur_03}
    componentAppliedLinearAccClamp=<1300 1300 2000>
    componentLinearAccScale=<0.5 0.5 0.5>
    componentLinearVelScale=<0.5 0.5 0.5>
    gravityOverride {z=-3000}
    physicsBodyDefinitions [{
      boundBone {boneName=dyn_head_fur_03}
      boxExtents=<35 25 20>
      constraintSetup {angularLimitsMax=<20 5 20>, angularLimitsMin=<-20 -40 -20>, angularTarget=<1 0 0>}
      localJointOffset=<-3 0 0.2>
     }]
   }
   n8 = Anim Dynamics - Bone: dyn_head_fur_04!(ComponentPose=@n7, Alpha=1, GravityOverride:Vector="0,0,-3000", ExternalForce="0,0,0") -> (Pose) {
    angularDampingOverride=0.95
    angularSpringConstant=15000
    bAngularSpring=true
    bOverrideAngularDamping=true
    bUseGravityOverride=true
    bUsePlanarLimit=false
    gravityScale=0.35
    numSolverIterationsPreUpdate=1
    alphaBoolBlend {blendOption=Linear}
    alphaScaleBiasClamp {scale=0.7}
    boundBone {boneName=dyn_head_fur_04}
    componentAppliedLinearAccClamp=<1300 1300 1800>
    componentLinearAccScale=<0.5 0.5 0.5>
    componentLinearVelScale=<0.5 0.5 0.5>
    gravityOverride {z=-3000}
    physicsBodyDefinitions [{
      boundBone {boneName=dyn_head_fur_04}
      boxExtents=<20 40 20>
      constraintSetup {angularLimitsMax=<20 5 20>, angularLimitsMin=<-20 -40 -20>, angularTarget=<1 0 0>}
      localJointOffset=<-2 0.6 -0.6>
     }]
   }
   n9 = Anim Dynamics - Bone: dyn_head_fur_05!(ComponentPose=@n8, Alpha=1, GravityOverride:Vector="0,0,-3000", ExternalForce="0,0,0") -> (Pose) {
    angularDampingOverride=0.95
    angularSpringConstant=15000
    bAngularSpring=true
    bOverrideAngularDamping=true
    bUseGravityOverride=true
    bUsePlanarLimit=false
    gravityScale=0.35
    numSolverIterationsPreUpdate=1
    alphaBoolBlend {blendOption=Linear}
    alphaScaleBiasClamp {scale=0.7}
    boundBone {boneName=dyn_head_fur_05}
    componentAppliedLinearAccClamp=<1300 1300 2400>
    componentLinearAccScale=<0.5 0.5 0.5>
    componentLinearVelScale=<0.5 0.5 0.5>
    gravityOverride {z=-3000}
    physicsBodyDefinitions [{
      boundBone {boneName=dyn_head_fur_05}
      boxExtents=<25 25 25>
      constraintSetup {angularLimitsMax=<20 5 20>, angularLimitsMin=<-20 -40 -20>, angularTarget=<1 0 0>}
      localJointOffset=<-2 0 0>
     }]
   }
   n10 = Anim Dynamics - Bone: dyn_head_fur_06!(ComponentPose=@n9, Alpha=1, GravityOverride:Vector="0,0,-3000", ExternalForce="0,0,0") -> (Pose) {
    angularDampingOverride=0.95
    angularSpringConstant=15000
    bAngularSpring=true
    bOverrideAngularDamping=true
    bUseGravityOverride=true
    bUsePlanarLimit=false
    gravityScale=0.35
    numSolverIterationsPreUpdate=1
    alphaBoolBlend {blendOption=Linear}
    alphaScaleBiasClamp {scale=0.7}
    boundBone {boneName=dyn_head_fur_06}
    componentAppliedLinearAccClamp=<1300 1300 2400>
    componentLinearAccScale=<0.5 0.5 0.5>
    componentLinearVelScale=<0.5 0.5 0.5>
    gravityOverride {z=-3000}
    physicsBodyDefinitions [{
      boundBone {boneName=dyn_head_fur_06}
      boxExtents=<25 25 25>
      constraintSetup {angularLimitsMax=<20 5 20>, angularLimitsMin=<-20 -40 -20>, angularTarget=<1 0 0>}
      localJointOffset=<-3 0 0>
     }]
   }
   n11 = Anim Dynamics - Bone: dyn_head_fur_07!(ComponentPose=@n10, Alpha=1, GravityOverride:Vector="0,0,-3000", ExternalForce="0,0,0") -> (Pose) {
    angularDampingOverride=0.95
    angularSpringConstant=15000
    bAngularSpring=true
    bOverrideAngularDamping=true
    bUseGravityOverride=true
    bUsePlanarLimit=false
    gravityScale=0.35
    numSolverIterationsPreUpdate=1
    alphaBoolBlend {blendOption=Linear}
    alphaScaleBiasClamp {scale=0.7}
    boundBone {boneName=dyn_head_fur_07}
    componentAppliedLinearAccClamp=<1300 1300 2000>
    componentLinearAccScale=<0.5 0.5 0.5>
    componentLinearVelScale=<0.5 0.5 0.5>
    gravityOverride {z=-3000}
    physicsBodyDefinitions [{
      boundBone {boneName=dyn_head_fur_07}
      boxExtents=<30 30 30>
      constraintSetup {angularLimitsMax=<20 5 20>, angularLimitsMin=<-20 -40 -20>, angularTarget=<1 0 0>}
      localJointOffset=<-3 0 0>
     }]
   }
   n12 = Anim Dynamics - Bone: dyn_head_fur_08!(ComponentPose=@n11, Alpha=1, GravityOverride:Vector="0,0,-3000", ExternalForce="0,0,0") -> (Pose) {
    angularDampingOverride=0.95
    angularSpringConstant=15000
    bAngularSpring=true
    bOverrideAngularDamping=true
    bUseGravityOverride=true
    bUsePlanarLimit=false
    gravityScale=0.35
    numSolverIterationsPreUpdate=1
    alphaBoolBlend {blendOption=Linear}
    alphaScaleBiasClamp {scale=0.7}
    boundBone {boneName=dyn_head_fur_08}
    componentAppliedLinearAccClamp=<1300 1300 2400>
    componentLinearAccScale=<0.5 0.5 0.5>
    componentLinearVelScale=<0.5 0.5 0.5>
    gravityOverride {z=-3000}
    physicsBodyDefinitions [{
      boundBone {boneName=dyn_head_fur_08}
      boxExtents=<25 30 25>
      constraintSetup {angularLimitsMax=<20 5 20>, angularLimitsMin=<-20 -40 -20>, angularTarget=<1 0 0>}
      localJointOffset=<-2 1 0>
     }]
   }
   n13 = Anim Dynamics - Bone: dyn_head_fur_09!(ComponentPose=@n12, Alpha=1, GravityOverride:Vector="0,0,-3000", ExternalForce="0,0,0") -> (Pose) {
    angularDampingOverride=0.95
    angularSpringConstant=15000
    bAngularSpring=true
    bOverrideAngularDamping=true
    bUseGravityOverride=true
    bUsePlanarLimit=false
    gravityScale=0.35
    numSolverIterationsPreUpdate=1
    alphaBoolBlend {blendOption=Linear}
    alphaScaleBiasClamp {scale=0.7}
    boundBone {boneName=dyn_head_fur_09}
    componentAppliedLinearAccClamp=<1300 1300 2400>
    componentLinearAccScale=<0.5 0.5 0.5>
    componentLinearVelScale=<0.5 0.5 0.5>
    gravityOverride {z=-3000}
    physicsBodyDefinitions [{
      boundBone {boneName=dyn_head_fur_09}
      boxExtents=<25 25 25>
      constraintSetup {angularLimitsMax=<20 5 20>, angularLimitsMin=<-20 -40 -20>, angularTarget=<1 0 0>}
      localJointOffset=<-2 0 -1>
     }]
   }
   n14 = Anim Dynamics - Bone: dyn_head_fur_10!(ComponentPose=@n13, Alpha=1, GravityOverride:Vector="0,0,-3000", ExternalForce="0,0,0") -> (Pose) {
    angularDampingOverride=0.95
    angularSpringConstant=15000
    bAngularSpring=true
    bOverrideAngularDamping=true
    bUseGravityOverride=true
    bUsePlanarLimit=false
    gravityScale=0.35
    numSolverIterationsPreUpdate=1
    alphaBoolBlend {blendOption=Linear}
    alphaScaleBiasClamp {scale=0.7}
    boundBone {boneName=dyn_head_fur_10}
    componentAppliedLinearAccClamp=<1300 1300 2400>
    componentLinearAccScale=<0.5 0.5 0.5>
    componentLinearVelScale=<0.5 0.5 0.5>
    gravityOverride {z=-3000}
    physicsBodyDefinitions [{
      boundBone {boneName=dyn_head_fur_10}
      boxExtents=<25 25 25>
      constraintSetup {angularLimitsMax=<20 5 20>, angularLimitsMin=<-20 -40 -20>, angularTarget=<1 0 0>}
      localJointOffset=<-2 0 0>
     }]
   }
   n15 = Anim Dynamics - Bone: dyn_head_fur_11!(ComponentPose=@n14, Alpha=1, GravityOverride:Vector="0,0,-3000", ExternalForce="0,0,0") -> (Pose) {
    angularDampingOverride=0.95
    angularSpringConstant=15000
    bAngularSpring=true
    bOverrideAngularDamping=true
    bUseGravityOverride=true
    bUsePlanarLimit=false
    gravityScale=0.35
    numSolverIterationsPreUpdate=1
    alphaBoolBlend {blendOption=Linear}
    alphaScaleBiasClamp {scale=0.7}
    boundBone {boneName=dyn_head_fur_11}
    componentAppliedLinearAccClamp=<1300 1300 1300>
    componentLinearAccScale=<0.5 0.5 0.5>
    componentLinearVelScale=<0.5 0.5 0.5>
    gravityOverride {z=-3000}
    physicsBodyDefinitions [{
      boundBone {boneName=dyn_head_fur_11}
      boxExtents=<30 30 30>
      constraintSetup {angularLimitsMax=<20 5 20>, angularLimitsMin=<-20 -40 -20>, angularTarget=<1 0 0>}
      localJointOffset=<-2 0.2 0.1>
     }]
   }
   n16 = Anim Dynamics - Bone: dyn_head_fur_12!(ComponentPose=@n15, Alpha=1, GravityOverride:Vector="0,0,-3000", ExternalForce="0,0,0") -> (Pose) {
    angularDampingOverride=0.95
    angularSpringConstant=15000
    bAngularSpring=true
    bOverrideAngularDamping=true
    bUseGravityOverride=true
    bUsePlanarLimit=false
    gravityScale=0.35
    numSolverIterationsPreUpdate=1
    alphaBoolBlend {blendOption=Linear}
    alphaScaleBiasClamp {scale=0.7}
    boundBone {boneName=dyn_head_fur_12}
    componentAppliedLinearAccClamp=<1300 1300 2400>
    componentLinearAccScale=<0.5 0.5 0.5>
    componentLinearVelScale=<0.5 0.5 0.5>
    gravityOverride {z=-3000}
    physicsBodyDefinitions [{
      boundBone {boneName=dyn_head_fur_12}
      boxExtents=<25 35 25>
      constraintSetup {angularLimitsMax=<20 5 20>, angularLimitsMin=<-20 -40 -20>, angularTarget=<1 0 0>}
      localJointOffset=<-2 -0.5 0>
     }]
   }
   n17 = Anim Dynamics - Bone: dyn_head_fur_13!(ComponentPose=@n16, Alpha=1, GravityOverride:Vector="0,0,-3000", ExternalForce="0,0,0") -> (Pose) {
    angularDampingOverride=0.95
    angularSpringConstant=15000
    bAngularSpring=true
    bOverrideAngularDamping=true
    bUseGravityOverride=true
    bUsePlanarLimit=false
    gravityScale=0.35
    numSolverIterationsPreUpdate=1
    alphaBoolBlend {blendOption=Linear}
    alphaScaleBiasClamp {scale=0.7}
    boundBone {boneName=dyn_head_fur_13}
    componentAppliedLinearAccClamp=<1300 1300 2000>
    componentLinearAccScale=<0.5 0.5 0.5>
    componentLinearVelScale=<0.5 0.5 0.5>
    gravityOverride {z=-3000}
    physicsBodyDefinitions [{
      boundBone {boneName=dyn_head_fur_13}
      boxExtents=<35 25 20>
      constraintSetup {angularLimitsMax=<20 5 20>, angularLimitsMin=<-20 -40 -20>, angularTarget=<1 0 0>}
      localJointOffset=<-3 0 0.2>
     }]
   }
   n18 = Anim Dynamics - Bone: dyn_head_fur_14!(ComponentPose=@n17, Alpha=1, GravityOverride:Vector="0,0,-3000", ExternalForce="0,0,0") -> (Pose) {
    angularDampingOverride=0.95
    angularSpringConstant=15000
    bAngularSpring=true
    bOverrideAngularDamping=true
    bUseGravityOverride=true
    bUsePlanarLimit=false
    gravityScale=0.35
    numSolverIterationsPreUpdate=1
    alphaBoolBlend {blendOption=Linear}
    alphaScaleBiasClamp {scale=0.7}
    boundBone {boneName=dyn_head_fur_14}
    componentAppliedLinearAccClamp=<1300 1300 1800>
    componentLinearAccScale=<0.5 0.5 0.5>
    componentLinearVelScale=<0.5 0.5 0.5>
    gravityOverride {z=-3000}
    physicsBodyDefinitions [{
      boundBone {boneName=dyn_head_fur_14}
      boxExtents=<20 40 20>
      constraintSetup {angularLimitsMax=<20 5 20>, angularLimitsMin=<-20 -40 -20>, angularTarget=<1 0 0>}
      localJointOffset=<-2 0.6 -0.6>
     }]
   }
   n19 = Anim Dynamics - Bone: dyn_head_fur_15!(ComponentPose=@n18, Alpha=1, GravityOverride:Vector="0,0,-3000", ExternalForce="0,0,0") -> (Pose) {
    angularDampingOverride=0.95
    angularSpringConstant=15000
    bAngularSpring=true
    bOverrideAngularDamping=true
    bUseGravityOverride=true
    bUsePlanarLimit=false
    gravityScale=0.35
    numSolverIterationsPreUpdate=1
    alphaBoolBlend {blendOption=Linear}
    alphaScaleBiasClamp {scale=0.7}
    boundBone {boneName=dyn_head_fur_15}
    componentAppliedLinearAccClamp=<1300 1300 2400>
    componentLinearAccScale=<0.5 0.5 0.5>
    componentLinearVelScale=<0.5 0.5 0.5>
    gravityOverride {z=-3000}
    physicsBodyDefinitions [{
      boundBone {boneName=dyn_head_fur_15}
      boxExtents=<25 25 25>
      constraintSetup {angularLimitsMax=<20 5 20>, angularLimitsMin=<-20 -40 -20>, angularTarget=<1 0 0>}
      localJointOffset=<-2 0 0>
     }]
   }
   n20 = Anim Dynamics - Bone: dyn_head_fur_16!(ComponentPose=@n19, Alpha=1, GravityOverride:Vector="0,0,-3000", ExternalForce="0,0,0") -> (Pose) {
    angularDampingOverride=0.95
    angularSpringConstant=15000
    bAngularSpring=true
    bOverrideAngularDamping=true
    bUseGravityOverride=true
    bUsePlanarLimit=false
    gravityScale=0.35
    numSolverIterationsPreUpdate=1
    alphaBoolBlend {blendOption=Linear}
    alphaScaleBiasClamp {scale=0.7}
    boundBone {boneName=dyn_head_fur_16}
    componentAppliedLinearAccClamp=<1300 1300 2400>
    componentLinearAccScale=<0.5 0.5 0.5>
    componentLinearVelScale=<0.5 0.5 0.5>
    gravityOverride {z=-3000}
    physicsBodyDefinitions [{
      boundBone {boneName=dyn_head_fur_16}
      boxExtents=<25 25 25>
      constraintSetup {angularLimitsMax=<20 5 20>, angularLimitsMin=<-20 -40 -20>, angularTarget=<1 0 0>}
      localJointOffset=<-3 0 0>
     }]
   }
   n21 = Anim Dynamics - Bone: dyn_head_fur_17!(ComponentPose=@n20, Alpha=1, GravityOverride:Vector="0,0,-3000", ExternalForce="0,0,0") -> (Pose) {
    angularDampingOverride=0.95
    angularSpringConstant=15000
    bAngularSpring=true
    bOverrideAngularDamping=true
    bUseGravityOverride=true
    bUsePlanarLimit=false
    gravityScale=0.35
    numSolverIterationsPreUpdate=1
    alphaBoolBlend {blendOption=Linear}
    alphaScaleBiasClamp {scale=0.7}
    boundBone {boneName=dyn_head_fur_17}
    componentAppliedLinearAccClamp=<1300 1300 2000>
    componentLinearAccScale=<0.5 0.5 0.5>
    componentLinearVelScale=<0.5 0.5 0.5>
    gravityOverride {z=-3000}
    physicsBodyDefinitions [{
      boundBone {boneName=dyn_head_fur_17}
      boxExtents=<30 30 30>
      constraintSetup {angularLimitsMax=<20 5 20>, angularLimitsMin=<-20 -40 -20>, angularTarget=<1 0 0>}
      localJointOffset=<-3 0 0>
     }]
   }
   n22 = Anim Dynamics - Bone: dyn_head_fur_18!(ComponentPose=@n21, Alpha=1, GravityOverride:Vector="0,0,-3000", ExternalForce="0,0,0") -> (Pose) {
    angularDampingOverride=0.95
    angularSpringConstant=15000
    bAngularSpring=true
    bOverrideAngularDamping=true
    bUseGravityOverride=true
    bUsePlanarLimit=false
    gravityScale=0.35
    numSolverIterationsPreUpdate=1
    alphaBoolBlend {blendOption=Linear}
    alphaScaleBiasClamp {scale=0.7}
    boundBone {boneName=dyn_head_fur_18}
    componentAppliedLinearAccClamp=<1300 1300 2400>
    componentLinearAccScale=<0.5 0.5 0.5>
    componentLinearVelScale=<0.5 0.5 0.5>
    gravityOverride {z=-3000}
    physicsBodyDefinitions [{
      boundBone {boneName=dyn_head_fur_18}
      boxExtents=<25 30 25>
      constraintSetup {angularLimitsMax=<20 5 20>, angularLimitsMin=<-20 -40 -20>, angularTarget=<1 0 0>}
      localJointOffset=<-2 1 0>
     }]
   }
   n23 = Anim Dynamics - Bone: dyn_head_fur_19!(ComponentPose=@n22, Alpha=1, GravityOverride:Vector="0,0,-3000", ExternalForce="0,0,0") -> (Pose) {
    angularDampingOverride=0.95
    angularSpringConstant=15000
    bAngularSpring=true
    bOverrideAngularDamping=true
    bUseGravityOverride=true
    bUsePlanarLimit=false
    gravityScale=0.35
    numSolverIterationsPreUpdate=1
    alphaBoolBlend {blendOption=Linear}
    alphaScaleBiasClamp {scale=0.7}
    boundBone {boneName=dyn_head_fur_19}
    componentAppliedLinearAccClamp=<1300 1300 2400>
    componentLinearAccScale=<0.5 0.5 0.5>
    componentLinearVelScale=<0.5 0.5 0.5>
    gravityOverride {z=-3000}
    physicsBodyDefinitions [{
      boundBone {boneName=dyn_head_fur_19}
      boxExtents=<25 25 25>
      constraintSetup {angularLimitsMax=<20 5 20>, angularLimitsMin=<-20 -40 -20>, angularTarget=<1 0 0>}
      localJointOffset=<-2 0 -1>
     }]
   }
   n24 = Anim Dynamics - Bone: dyn_head_fur_20!(ComponentPose=@n23, Alpha=1, GravityOverride:Vector="0,0,-3000", ExternalForce="0,0,0") -> (Pose) {
    angularDampingOverride=0.95
    angularSpringConstant=15000
    bAngularSpring=true
    bOverrideAngularDamping=true
    bUseGravityOverride=true
    bUsePlanarLimit=false
    gravityScale=0.35
    numSolverIterationsPreUpdate=1
    alphaBoolBlend {blendOption=Linear}
    alphaScaleBiasClamp {scale=0.7}
    boundBone {boneName=dyn_head_fur_20}
    componentAppliedLinearAccClamp=<1300 1300 2400>
    componentLinearAccScale=<0.5 0.5 0.5>
    componentLinearVelScale=<0.5 0.5 0.5>
    gravityOverride {z=-3000}
    physicsBodyDefinitions [{
      boundBone {boneName=dyn_head_fur_20}
      boxExtents=<25 25 25>
      constraintSetup {angularLimitsMax=<20 5 20>, angularLimitsMin=<-20 -40 -20>, angularTarget=<1 0 0>}
      localJointOffset=<-2 0 0>
     }]
   }
   n25 = Anim Dynamics - Bone: dyn_head_fur_21!(ComponentPose=@n24, Alpha=1, GravityOverride:Vector="0,0,-3000", ExternalForce="0,0,0") -> (Pose) {
    angularDampingOverride=0.95
    angularSpringConstant=15000
    bAngularSpring=true
    bOverrideAngularDamping=true
    bUseGravityOverride=true
    bUsePlanarLimit=false
    gravityScale=0.35
    numSolverIterationsPreUpdate=1
    alphaBoolBlend {blendOption=Linear}
    alphaScaleBiasClamp {scale=0.7}
    boundBone {boneName=dyn_head_fur_21}
    componentAppliedLinearAccClamp=<1300 1300 1300>
    componentLinearAccScale=<0.5 0.5 0.5>
    componentLinearVelScale=<0.5 0.5 0.5>
    gravityOverride {z=-3000}
    physicsBodyDefinitions [{
      boundBone {boneName=dyn_head_fur_21}
      boxExtents=<30 30 30>
      constraintSetup {angularLimitsMax=<20 5 20>, angularLimitsMin=<-20 -40 -20>, angularTarget=<1 0 0>}
      localJointOffset=<-2 0.2 0.1>
     }]
   }
   n26 = Anim Dynamics - Bone: dyn_head_fur_22!(ComponentPose=@n25, Alpha=1, GravityOverride:Vector="0,0,-3000", ExternalForce="0,0,0") -> (Pose) {
    angularDampingOverride=0.95
    angularSpringConstant=15000
    bAngularSpring=true
    bOverrideAngularDamping=true
    bUseGravityOverride=true
    bUsePlanarLimit=false
    gravityScale=0.35
    numSolverIterationsPreUpdate=1
    alphaBoolBlend {blendOption=Linear}
    alphaScaleBiasClamp {scale=0.7}
    boundBone {boneName=dyn_head_fur_22}
    componentAppliedLinearAccClamp=<1300 1300 2400>
    componentLinearAccScale=<0.5 0.5 0.5>
    componentLinearVelScale=<0.5 0.5 0.5>
    gravityOverride {z=-3000}
    physicsBodyDefinitions [{
      boundBone {boneName=dyn_head_fur_22}
      boxExtents=<25 35 25>
      constraintSetup {angularLimitsMax=<20 5 20>, angularLimitsMin=<-20 -40 -20>, angularTarget=<1 0 0>}
      localJointOffset=<-2 -0.5 0>
     }]
   }
   n27 = Anim Dynamics - Bone: dyn_head_fur_23!(ComponentPose=@n26, Alpha=1, GravityOverride:Vector="0,0,-3000", ExternalForce="0,0,0") -> (Pose) {
    angularDampingOverride=0.95
    angularSpringConstant=15000
    bAngularSpring=true
    bOverrideAngularDamping=true
    bUseGravityOverride=true
    bUsePlanarLimit=false
    gravityScale=0.35
    numSolverIterationsPreUpdate=1
    alphaBoolBlend {blendOption=Linear}
    alphaScaleBiasClamp {scale=0.7}
    boundBone {boneName=dyn_head_fur_23}
    componentAppliedLinearAccClamp=<1300 1300 2000>
    componentLinearAccScale=<0.5 0.5 0.5>
    componentLinearVelScale=<0.5 0.5 0.5>
    gravityOverride {z=-3000}
    physicsBodyDefinitions [{
      boundBone {boneName=dyn_head_fur_23}
      boxExtents=<35 25 20>
      constraintSetup {angularLimitsMax=<20 5 20>, angularLimitsMin=<-20 -40 -20>, angularTarget=<1 0 0>}
      localJointOffset=<-3 0 0.2>
     }]
   }
   n28 = Anim Dynamics - Bone: dyn_head_fur_24!(ComponentPose=@n27, Alpha=1, GravityOverride:Vector="0,0,-3000", ExternalForce="0,0,0") -> (Pose) {
    angularDampingOverride=0.95
    angularSpringConstant=15000
    bAngularSpring=true
    bOverrideAngularDamping=true
    bUseGravityOverride=true
    bUsePlanarLimit=false
    gravityScale=0.35
    numSolverIterationsPreUpdate=1
    alphaBoolBlend {blendOption=Linear}
    alphaScaleBiasClamp {scale=0.7}
    boundBone {boneName=dyn_head_fur_24}
    componentAppliedLinearAccClamp=<1300 1300 1800>
    componentLinearAccScale=<0.5 0.5 0.5>
    componentLinearVelScale=<0.5 0.5 0.5>
    gravityOverride {z=-3000}
    physicsBodyDefinitions [{
      boundBone {boneName=dyn_head_fur_24}
      boxExtents=<20 40 20>
      constraintSetup {angularLimitsMax=<20 15 20>, angularLimitsMin=<-20 -5 -20>, angularTarget=<1 0 0>}
      localJointOffset=<-2 0.6 -0.6>
     }]
   }
   n29 = Anim Dynamics - Bone: dyn_head_fur_28!(ComponentPose=@n28, Alpha=1, GravityOverride:Vector="0,0,-3000", ExternalForce="0,0,0") -> (Pose) {
    angularDampingOverride=0.95
    angularSpringConstant=15000
    bAngularSpring=true
    bOverrideAngularDamping=true
    bUseGravityOverride=true
    bUsePlanarLimit=false
    gravityScale=0.35
    numSolverIterationsPreUpdate=1
    alphaBoolBlend {blendOption=Linear}
    alphaScaleBiasClamp {scale=0.7}
    boundBone {boneName=dyn_head_fur_28}
    componentAppliedLinearAccClamp=<1300 1300 2400>
    componentLinearAccScale=<0.5 0.5 0.5>
    componentLinearVelScale=<0.5 0.5 0.5>
    gravityOverride {z=-3000}
    physicsBodyDefinitions [{
      boundBone {boneName=dyn_head_fur_28}
      boxExtents=<25 25 25>
      constraintSetup {angularLimitsMax=<20 5 20>, angularLimitsMin=<-20 -40 -20>, angularTarget=<1 0 0>}
      localJointOffset=<-2 0 0>
     }]
   }
   n30 = Anim Dynamics - Bone: dyn_head_fur_29!(ComponentPose=@n29, Alpha=1, GravityOverride:Vector="0,0,-3000", ExternalForce="0,0,0") -> (Pose) {
    angularDampingOverride=0.95
    angularSpringConstant=15000
    bAngularSpring=true
    bOverrideAngularDamping=true
    bUseGravityOverride=true
    bUsePlanarLimit=false
    gravityScale=0.35
    numSolverIterationsPreUpdate=1
    alphaBoolBlend {blendOption=Linear}
    alphaScaleBiasClamp {scale=0.7}
    boundBone {boneName=dyn_head_fur_29}
    componentAppliedLinearAccClamp=<1300 1300 2400>
    componentLinearAccScale=<0.5 0.5 0.5>
    componentLinearVelScale=<0.5 0.5 0.5>
    gravityOverride {z=-3000}
    physicsBodyDefinitions [{
      boundBone {boneName=dyn_head_fur_29}
      boxExtents=<25 25 25>
      constraintSetup {angularLimitsMax=<20 5 20>, angularLimitsMin=<-20 -40 -20>, angularTarget=<1 0 0>}
      localJointOffset=<-3 0 0>
     }]
   }
   n31 = Anim Dynamics - Bone: dyn_head_fur_30!(ComponentPose=@n30, Alpha=1, GravityOverride:Vector="0,0,-3000", ExternalForce="0,0,0") -> (Pose) {
    angularDampingOverride=0.95
    angularSpringConstant=15000
    bAngularSpring=true
    bOverrideAngularDamping=true
    bUseGravityOverride=true
    bUsePlanarLimit=false
    gravityScale=0.35
    numSolverIterationsPreUpdate=1
    alphaBoolBlend {blendOption=Linear}
    alphaScaleBiasClamp {scale=0.7}
    boundBone {boneName=dyn_head_fur_30}
    componentAppliedLinearAccClamp=<1300 1300 2000>
    componentLinearAccScale=<0.5 0.5 0.5>
    componentLinearVelScale=<0.5 0.5 0.5>
    gravityOverride {z=-3000}
    physicsBodyDefinitions [{
      boundBone {boneName=dyn_head_fur_30}
      boxExtents=<30 30 30>
      constraintSetup {angularLimitsMax=<20 5 20>, angularLimitsMin=<-20 -40 -20>, angularTarget=<1 0 0>}
      localJointOffset=<-3 0 0>
     }]
   }
   n32 = Anim Dynamics - Bone: dyn_head_fur_31!(ComponentPose=@n31, Alpha=1, GravityOverride:Vector="0,0,-3000", ExternalForce="0,0,0") -> (Pose) {
    angularDampingOverride=0.95
    angularSpringConstant=15000
    bAngularSpring=true
    bOverrideAngularDamping=true
    bUseGravityOverride=true
    bUsePlanarLimit=false
    gravityScale=0.35
    numSolverIterationsPreUpdate=1
    alphaBoolBlend {blendOption=Linear}
    alphaScaleBiasClamp {scale=0.7}
    boundBone {boneName=dyn_head_fur_31}
    componentAppliedLinearAccClamp=<1300 1300 2400>
    componentLinearAccScale=<0.5 0.5 0.5>
    componentLinearVelScale=<0.5 0.5 0.5>
    gravityOverride {z=-3000}
    physicsBodyDefinitions [{
      boundBone {boneName=dyn_head_fur_31}
      boxExtents=<25 30 25>
      constraintSetup {angularLimitsMax=<20 5 20>, angularLimitsMin=<-20 -40 -20>, angularTarget=<1 0 0>}
      localJointOffset=<-2 1 0>
     }]
   }
   n33 = Anim Dynamics - Bone: dyn_head_fur_32!(ComponentPose=@n32, Alpha=1, GravityOverride:Vector="0,0,-3000", ExternalForce="0,0,0") -> (Pose) {
    angularDampingOverride=0.95
    angularSpringConstant=15000
    bAngularSpring=true
    bOverrideAngularDamping=true
    bUseGravityOverride=true
    bUsePlanarLimit=false
    gravityScale=0.35
    numSolverIterationsPreUpdate=1
    alphaBoolBlend {blendOption=Linear}
    alphaScaleBiasClamp {scale=0.7}
    boundBone {boneName=dyn_head_fur_32}
    componentAppliedLinearAccClamp=<1300 1300 2400>
    componentLinearAccScale=<0.5 0.5 0.5>
    componentLinearVelScale=<0.5 0.5 0.5>
    gravityOverride {z=-3000}
    physicsBodyDefinitions [{
      boundBone {boneName=dyn_head_fur_32}
      boxExtents=<25 25 25>
      constraintSetup {angularLimitsMax=<20 5 20>, angularLimitsMin=<-20 -40 -20>, angularTarget=<1 0 0>}
      localJointOffset=<-2 0 -1>
     }]
   }
   n34 = Anim Dynamics - Bone: dyn_head_fur_33!(ComponentPose=@n33, Alpha=1, GravityOverride:Vector="0,0,-3000", ExternalForce="0,0,0") -> (Pose) {
    angularDampingOverride=0.95
    angularSpringConstant=15000
    bAngularSpring=true
    bOverrideAngularDamping=true
    bUseGravityOverride=true
    bUsePlanarLimit=false
    gravityScale=0.35
    numSolverIterationsPreUpdate=1
    alphaBoolBlend {blendOption=Linear}
    alphaScaleBiasClamp {scale=0.7}
    boundBone {boneName=dyn_head_fur_33}
    componentAppliedLinearAccClamp=<1300 1300 2400>
    componentLinearAccScale=<0.5 0.5 0.5>
    componentLinearVelScale=<0.5 0.5 0.5>
    gravityOverride {z=-3000}
    physicsBodyDefinitions [{
      boundBone {boneName=dyn_head_fur_33}
      boxExtents=<25 25 25>
      constraintSetup {angularLimitsMax=<20 5 20>, angularLimitsMin=<-20 -40 -20>, angularTarget=<1 0 0>}
      localJointOffset=<-2 0 0>
     }]
   }
   n35 = Anim Dynamics - Bone: dyn_head_fur_34!(ComponentPose=@n34, Alpha=1, GravityOverride:Vector="0,0,-3000", ExternalForce="0,0,0") -> (Pose) {
    angularDampingOverride=0.95
    angularSpringConstant=15000
    bAngularSpring=true
    bOverrideAngularDamping=true
    bUseGravityOverride=true
    bUsePlanarLimit=false
    gravityScale=0.35
    numSolverIterationsPreUpdate=1
    alphaBoolBlend {blendOption=Linear}
    alphaScaleBiasClamp {scale=0.7}
    boundBone {boneName=dyn_head_fur_34}
    componentAppliedLinearAccClamp=<1300 1300 1300>
    componentLinearAccScale=<0.5 0.5 0.5>
    componentLinearVelScale=<0.5 0.5 0.5>
    gravityOverride {z=-3000}
    physicsBodyDefinitions [{
      boundBone {boneName=dyn_head_fur_34}
      boxExtents=<30 30 30>
      constraintSetup {angularLimitsMax=<20 5 20>, angularLimitsMin=<-20 -40 -20>, angularTarget=<1 0 0>}
      localJointOffset=<-2 0.2 0.1>
     }]
   }
   n36 = Anim Dynamics - Bone: dyn_head_fur_35!(ComponentPose=@n35, Alpha=1, GravityOverride:Vector="0,0,-3000", ExternalForce="0,0,0") -> (Pose) {
    angularDampingOverride=0.95
    angularSpringConstant=15000
    bAngularSpring=true
    bOverrideAngularDamping=true
    bUseGravityOverride=true
    bUsePlanarLimit=false
    gravityScale=0.35
    numSolverIterationsPreUpdate=1
    alphaBoolBlend {blendOption=Linear}
    alphaScaleBiasClamp {scale=0.7}
    boundBone {boneName=dyn_head_fur_35}
    componentAppliedLinearAccClamp=<1300 1300 2400>
    componentLinearAccScale=<0.5 0.5 0.5>
    componentLinearVelScale=<0.5 0.5 0.5>
    gravityOverride {z=-3000}
    physicsBodyDefinitions [{
      boundBone {boneName=dyn_head_fur_35}
      boxExtents=<25 35 25>
      constraintSetup {angularLimitsMax=<20 5 20>, angularLimitsMin=<-20 -40 -20>, angularTarget=<1 0 0>}
      localJointOffset=<-2 -0.5 0>
     }]
   }
   n37 = Anim Dynamics - Bone: dyn_head_fur_36!(ComponentPose=@n36, Alpha=1, GravityOverride:Vector="0,0,-3000", ExternalForce="0,0,0") -> (Pose) {
    angularDampingOverride=0.95
    angularSpringConstant=15000
    bAngularSpring=true
    bOverrideAngularDamping=true
    bUseGravityOverride=true
    bUsePlanarLimit=false
    gravityScale=0.35
    numSolverIterationsPreUpdate=1
    alphaBoolBlend {blendOption=Linear}
    alphaScaleBiasClamp {scale=0.7}
    boundBone {boneName=dyn_head_fur_36}
    componentAppliedLinearAccClamp=<1300 1300 2000>
    componentLinearAccScale=<0.5 0.5 0.5>
    componentLinearVelScale=<0.5 0.5 0.5>
    gravityOverride {z=-3000}
    physicsBodyDefinitions [{
      boundBone {boneName=dyn_head_fur_36}
      boxExtents=<35 25 20>
      constraintSetup {angularLimitsMax=<20 5 20>, angularLimitsMin=<-20 -40 -20>, angularTarget=<1 0 0>}
      localJointOffset=<-3 0 0.2>
     }]
   }
   n38 = Anim Dynamics - Bone: dyn_head_fur_37!(ComponentPose=@n37, Alpha=1, GravityOverride:Vector="0,0,-3000", ExternalForce="0,0,0") -> (Pose) {
    angularDampingOverride=0.95
    angularSpringConstant=15000
    bAngularSpring=true
    bOverrideAngularDamping=true
    bUseGravityOverride=true
    bUsePlanarLimit=false
    gravityScale=0.35
    numSolverIterationsPreUpdate=1
    alphaBoolBlend {blendOption=Linear}
    alphaScaleBiasClamp {scale=0.7}
    boundBone {boneName=dyn_head_fur_37}
    componentAppliedLinearAccClamp=<1300 1300 1800>
    componentLinearAccScale=<0.5 0.5 0.5>
    componentLinearVelScale=<0.5 0.5 0.5>
    gravityOverride {z=-3000}
    physicsBodyDefinitions [{
      boundBone {boneName=dyn_head_fur_37}
      boxExtents=<20 40 20>
      constraintSetup {angularLimitsMax=<20 5 20>, angularLimitsMin=<-20 -40 -20>, angularTarget=<1 0 0>}
      localJointOffset=<-2 0.6 -0.6>
     }]
   }
   n39 = Anim Dynamics - Bone: dyn_head_fur_38!(ComponentPose=@n38, Alpha=1, GravityOverride:Vector="0,0,-3000", ExternalForce="0,0,0") -> (Pose) {
    angularDampingOverride=0.95
    angularSpringConstant=15000
    bAngularSpring=true
    bOverrideAngularDamping=true
    bUseGravityOverride=true
    bUsePlanarLimit=false
    gravityScale=0.35
    numSolverIterationsPreUpdate=1
    alphaBoolBlend {blendOption=Linear}
    alphaScaleBiasClamp {scale=0.7}
    boundBone {boneName=dyn_head_fur_38}
    componentAppliedLinearAccClamp=<1300 1300 2400>
    componentLinearAccScale=<0.5 0.5 0.5>
    componentLinearVelScale=<0.5 0.5 0.5>
    gravityOverride {z=-3000}
    physicsBodyDefinitions [{
      boundBone {boneName=dyn_head_fur_38}
      boxExtents=<25 25 25>
      constraintSetup {angularLimitsMax=<20 5 20>, angularLimitsMin=<-20 -40 -20>, angularTarget=<1 0 0>}
      localJointOffset=<-2 0 0>
     }]
   }
   n40 = Anim Dynamics - Bone: dyn_head_fur_40!(ComponentPose=@n39, Alpha=1, GravityOverride:Vector="0,0,-3000", ExternalForce="0,0,0") -> (Pose) {
    angularDampingOverride=0.95
    angularSpringConstant=15000
    bAngularSpring=true
    bOverrideAngularDamping=true
    bUseGravityOverride=true
    bUsePlanarLimit=false
    gravityScale=0.35
    numSolverIterationsPreUpdate=1
    alphaBoolBlend {blendOption=Linear}
    alphaScaleBiasClamp {scale=0.7}
    boundBone {boneName=dyn_head_fur_40}
    componentAppliedLinearAccClamp=<1300 1300 2000>
    componentLinearAccScale=<0.5 0.5 0.5>
    componentLinearVelScale=<0.5 0.5 0.5>
    gravityOverride {z=-3000}
    physicsBodyDefinitions [{
      boundBone {boneName=dyn_head_fur_40}
      boxExtents=<30 30 30>
      constraintSetup {angularLimitsMax=<20 5 20>, angularLimitsMin=<-20 -40 -20>, angularTarget=<1 0 0>}
      localJointOffset=<-3 0 0>
     }]
   }
   n41 = Anim Dynamics - Bone: dyn_head_fur_41!(ComponentPose=@n40, Alpha=0.5, GravityOverride:Vector="0,0,-3000", ExternalForce="0,0,0") -> (Pose) {
    angularDampingOverride=0.95
    angularSpringConstant=15000
    bAngularSpring=true
    bOverrideAngularDamping=true
    bUseGravityOverride=true
    bUsePlanarLimit=false
    gravityScale=0.35
    numSolverIterationsPreUpdate=1
    alphaBoolBlend {blendOption=Linear}
    alphaScaleBiasClamp {scale=0.7}
    boundBone {boneName=dyn_head_fur_41}
    componentAppliedLinearAccClamp=<1300 1300 2400>
    componentLinearAccScale=<0.5 0.5 0.5>
    componentLinearVelScale=<0.5 0.5 0.5>
    gravityOverride {z=-3000}
    physicsBodyDefinitions [{
      boundBone {boneName=dyn_head_fur_41}
      boxExtents=<25 30 25>
      constraintSetup {angularLimitsMax=<20 5 20>, angularLimitsMin=<-20 -40 -20>, angularTarget=<1 0 0>}
      localJointOffset=<-2 1 0>
     }]
   }
   n42 = Anim Dynamics - Bone: dyn_head_fur_42!(ComponentPose=@n41, Alpha=0.5, GravityOverride:Vector="0,0,-3000", ExternalForce="0,0,0") -> (Pose) {
    angularDampingOverride=0.95
    angularSpringConstant=15000
    bAngularSpring=true
    bOverrideAngularDamping=true
    bUseGravityOverride=true
    bUsePlanarLimit=false
    gravityScale=0.35
    numSolverIterationsPreUpdate=1
    alphaBoolBlend {blendOption=Linear}
    alphaScaleBiasClamp {scale=0.7}
    boundBone {boneName=dyn_head_fur_42}
    componentAppliedLinearAccClamp=<1300 1300 2400>
    componentLinearAccScale=<0.5 0.5 0.5>
    componentLinearVelScale=<0.5 0.5 0.5>
    gravityOverride {z=-3000}
    physicsBodyDefinitions [{
      boundBone {boneName=dyn_head_fur_42}
      boxExtents=<25 30 25>
      constraintSetup {angularLimitsMax=<20 5 20>, angularLimitsMin=<-20 -40 -20>, angularTarget=<1 0 0>}
      localJointOffset=<-2 -1 0>
     }]
   }
   n43 = Anim Dynamics - Bone: dyn_head_fur_43!(ComponentPose=@n42, Alpha=1, GravityOverride:Vector="0,0,-3000", ExternalForce="0,0,0") -> (Pose) {
    angularDampingOverride=0.95
    angularSpringConstant=15000
    bAngularSpring=true
    bOverrideAngularDamping=true
    bUseGravityOverride=true
    bUsePlanarLimit=false
    gravityScale=0.35
    numSolverIterationsPreUpdate=1
    alphaBoolBlend {blendOption=Linear}
    alphaScaleBiasClamp {scale=0.7}
    boundBone {boneName=dyn_head_fur_43}
    componentAppliedLinearAccClamp=<1300 1300 2400>
    componentLinearAccScale=<0.5 0.5 0.5>
    componentLinearVelScale=<0.5 0.5 0.5>
    gravityOverride {z=-3000}
    physicsBodyDefinitions [{
      boundBone {boneName=dyn_head_fur_43}
      boxExtents=<25 25 25>
      constraintSetup {angularLimitsMax=<20 5 20>, angularLimitsMin=<-20 -40 -20>, angularTarget=<1 0 0>}
      localJointOffset=<-2 0 0>
     }]
   }
   n44 = Anim Dynamics - Bone: dyn_head_fur_44!(ComponentPose=@n43, Alpha=1, GravityOverride:Vector="0,0,-3000", ExternalForce="0,0,0") -> (Pose) {
    angularDampingOverride=0.95
    angularSpringConstant=15000
    bAngularSpring=true
    bOverrideAngularDamping=true
    bUseGravityOverride=true
    bUsePlanarLimit=false
    gravityScale=0.35
    numSolverIterationsPreUpdate=1
    alphaBoolBlend {blendOption=Linear}
    alphaScaleBiasClamp {scale=0.7}
    boundBone {boneName=dyn_head_fur_44}
    componentAppliedLinearAccClamp=<1300 1300 1300>
    componentLinearAccScale=<0.5 0.5 0.5>
    componentLinearVelScale=<0.5 0.5 0.5>
    gravityOverride {z=-3000}
    physicsBodyDefinitions [{
      boundBone {boneName=dyn_head_fur_44}
      boxExtents=<30 30 30>
      constraintSetup {angularLimitsMax=<20 5 20>, angularLimitsMin=<-20 -40 -20>, angularTarget=<1 0 0>}
      localJointOffset=<-2 0.2 0.1>
     }]
   }
   n45 = Anim Dynamics - Bone: dyn_head_fur_45!(ComponentPose=@n44, Alpha=1, GravityOverride:Vector="0,0,-3000", ExternalForce="0,0,0") -> (Pose) {
    angularDampingOverride=0.95
    angularSpringConstant=15000
    bAngularSpring=true
    bOverrideAngularDamping=true
    bUseGravityOverride=true
    bUsePlanarLimit=false
    gravityScale=0.35
    numSolverIterationsPreUpdate=1
    alphaBoolBlend {blendOption=Linear}
    alphaScaleBiasClamp {scale=0.7}
    boundBone {boneName=dyn_head_fur_45}
    componentAppliedLinearAccClamp=<1300 1300 2400>
    componentLinearAccScale=<0.5 0.5 0.5>
    componentLinearVelScale=<0.5 0.5 0.5>
    gravityOverride {z=-3000}
    physicsBodyDefinitions [{
      boundBone {boneName=dyn_head_fur_45}
      boxExtents=<25 35 25>
      constraintSetup {angularLimitsMax=<20 5 20>, angularLimitsMin=<-20 -40 -20>, angularTarget=<1 0 0>}
      localJointOffset=<-2 -0.5 0>
     }]
   }
   n46 = Anim Dynamics - Bone: dyn_head_fur_46!(ComponentPose=@n45, Alpha=1, GravityOverride:Vector="0,0,-3000", ExternalForce="0,0,0") -> (Pose) {
    angularDampingOverride=0.95
    angularSpringConstant=15000
    bAngularSpring=true
    bOverrideAngularDamping=true
    bUseGravityOverride=true
    bUsePlanarLimit=false
    gravityScale=0.35
    numSolverIterationsPreUpdate=1
    alphaBoolBlend {blendOption=Linear}
    alphaScaleBiasClamp {scale=0.7}
    boundBone {boneName=dyn_head_fur_46}
    componentAppliedLinearAccClamp=<1300 1300 2000>
    componentLinearAccScale=<0.5 0.5 0.5>
    componentLinearVelScale=<0.5 0.5 0.5>
    gravityOverride {z=-3000}
    physicsBodyDefinitions [{
      boundBone {boneName=dyn_head_fur_46}
      boxExtents=<35 25 20>
      constraintSetup {angularLimitsMax=<20 5 20>, angularLimitsMin=<-20 -40 -20>, angularTarget=<1 0 0>}
      localJointOffset=<-3 0 0.2>
     }]
   }
   n47 = Anim Dynamics - Bone: dyn_head_fur_47!(ComponentPose=@n46, Alpha=1, GravityOverride:Vector="0,0,-3000", ExternalForce="0,0,0") -> (Pose) {
    angularDampingOverride=0.95
    angularSpringConstant=15000
    bAngularSpring=true
    bOverrideAngularDamping=true
    bUseGravityOverride=true
    bUsePlanarLimit=false
    gravityScale=0.35
    numSolverIterationsPreUpdate=1
    alphaBoolBlend {blendOption=Linear}
    alphaScaleBiasClamp {scale=0.7}
    boundBone {boneName=dyn_head_fur_47}
    componentAppliedLinearAccClamp=<1300 1300 1800>
    componentLinearAccScale=<0.5 0.5 0.5>
    componentLinearVelScale=<0.5 0.5 0.5>
    gravityOverride {z=-3000}
    physicsBodyDefinitions [{
      boundBone {boneName=dyn_head_fur_47}
      boxExtents=<20 40 20>
      constraintSetup {angularLimitsMax=<20 5 20>, angularLimitsMin=<-20 -40 -20>, angularTarget=<1 0 0>}
      localJointOffset=<-2 0.6 -0.6>
     }]
   }
   n48 = Anim Dynamics - Bone: dyn_head_fur_48!(ComponentPose=@n47, Alpha=1, GravityOverride:Vector="0,0,-3000", ExternalForce="0,0,0") -> (Pose) {
    angularDampingOverride=0.95
    angularSpringConstant=15000
    bAngularSpring=true
    bOverrideAngularDamping=true
    bUseGravityOverride=true
    bUsePlanarLimit=false
    gravityScale=0.35
    numSolverIterationsPreUpdate=1
    alphaBoolBlend {blendOption=Linear}
    alphaScaleBiasClamp {scale=0.7}
    boundBone {boneName=dyn_head_fur_48}
    componentAppliedLinearAccClamp=<1300 1300 2400>
    componentLinearAccScale=<0.5 0.5 0.5>
    componentLinearVelScale=<0.5 0.5 0.5>
    gravityOverride {z=-3000}
    physicsBodyDefinitions [{
      boundBone {boneName=dyn_head_fur_48}
      boxExtents=<25 25 25>
      constraintSetup {angularLimitsMax=<20 5 20>, angularLimitsMin=<-20 -40 -20>, angularTarget=<1 0 0>}
      localJointOffset=<-2 0 0>
     }]
   }
   n49 = Anim Dynamics - Bone: dyn_head_fur_49!(ComponentPose=@n48, Alpha=1, GravityOverride:Vector="0,0,-3000", ExternalForce="0,0,0") -> (Pose) {
    angularDampingOverride=0.95
    angularSpringConstant=15000
    bAngularSpring=true
    bOverrideAngularDamping=true
    bUseGravityOverride=true
    bUsePlanarLimit=false
    gravityScale=0.35
    numSolverIterationsPreUpdate=1
    alphaBoolBlend {blendOption=Linear}
    alphaScaleBiasClamp {scale=0.7}
    boundBone {boneName=dyn_head_fur_49}
    componentAppliedLinearAccClamp=<1300 1300 2400>
    componentLinearAccScale=<0.5 0.5 0.5>
    componentLinearVelScale=<0.5 0.5 0.5>
    gravityOverride {z=-3000}
    physicsBodyDefinitions [{
      boundBone {boneName=dyn_head_fur_49}
      boxExtents=<25 25 25>
      constraintSetup {angularLimitsMax=<20 5 20>, angularLimitsMin=<-20 -40 -20>, angularTarget=<1 0 0>}
      localJointOffset=<-3 0 0>
     }]
   }
   n50 = Anim Dynamics - Bone: dyn_head_fur_25!(ComponentPose=@n49, Alpha=1, GravityOverride:Vector="0,0,-3000", ExternalForce="0,0,0") -> (Pose) {
    angularDampingOverride=0.95
    angularSpringConstant=15000
    bAngularSpring=true
    bOverrideAngularDamping=true
    bUseGravityOverride=true
    bUsePlanarLimit=false
    gravityScale=0.35
    numSolverIterationsPreUpdate=1
    alphaBoolBlend {blendOption=Linear}
    alphaScaleBiasClamp {scale=0.7}
    boundBone {boneName=dyn_head_fur_25}
    componentAppliedLinearAccClamp=<1300 1300 2000>
    componentLinearAccScale=<0.5 0.5 0.5>
    componentLinearVelScale=<0.5 0.5 0.5>
    gravityOverride {z=-3000}
    physicsBodyDefinitions [{
      boundBone {boneName=dyn_head_fur_25}
      boxExtents=<30 30 30>
      constraintSetup {angularLimitsMax=<20 40 20>, angularLimitsMin=<-20 -5 -20>, angularTarget=<1 0 0>}
      localJointOffset=<-3 0 0>
     }]
   }
   n51 = Anim Dynamics - Bone: dyn_head_fur_26!(ComponentPose=@n50, Alpha=1, GravityOverride:Vector="0,0,-3000", ExternalForce="0,0,0") -> (Pose) {
    angularDampingOverride=0.95
    angularSpringConstant=15000
    bAngularSpring=true
    bOverrideAngularDamping=true
    bUseGravityOverride=true
    bUsePlanarLimit=false
    gravityScale=0.35
    numSolverIterationsPreUpdate=1
    alphaBoolBlend {blendOption=Linear}
    alphaScaleBiasClamp {scale=0.7}
    boundBone {boneName=dyn_head_fur_26}
    componentAppliedLinearAccClamp=<1300 1300 2400>
    componentLinearAccScale=<0.5 0.5 0.5>
    componentLinearVelScale=<0.5 0.5 0.5>
    gravityOverride {z=-3000}
    physicsBodyDefinitions [{
      boundBone {boneName=dyn_head_fur_26}
      boxExtents=<25 30 25>
      constraintSetup {angularLimitsMax=<20 5 20>, angularLimitsMin=<-20 -40 -20>, angularTarget=<1 0 0>}
      localJointOffset=<-2 1 0>
     }]
   }
   n52 = Anim Dynamics - Bone: dyn_head_fur_27!(ComponentPose=@n51, Alpha=1, GravityOverride:Vector="0,0,-3000", ExternalForce="0,0,0") -> (Pose) {
    angularDampingOverride=0.95
    angularSpringConstant=15000
    bAngularSpring=true
    bOverrideAngularDamping=true
    bUseGravityOverride=true
    bUsePlanarLimit=false
    gravityScale=0.35
    numSolverIterationsPreUpdate=1
    alphaBoolBlend {blendOption=Linear}
    alphaScaleBiasClamp {scale=0.7}
    boundBone {boneName=dyn_head_fur_27}
    componentAppliedLinearAccClamp=<1300 1300 2400>
    componentLinearAccScale=<0.5 0.5 0.5>
    componentLinearVelScale=<0.5 0.5 0.5>
    gravityOverride {z=-3000}
    physicsBodyDefinitions [{
      boundBone {boneName=dyn_head_fur_27}
      boxExtents=<25 25 25>
      constraintSetup {angularLimitsMax=<20 5 20>, angularLimitsMin=<-20 -40 -20>, angularTarget=<1 0 0>}
      localJointOffset=<-2 0 -1>
     }]
   }
   n53 = Anim Dynamics - Bone: dyn_head_fur_50!(ComponentPose=@n52, Alpha=1, GravityOverride:Vector="0,0,-3000", ExternalForce="0,0,0") -> (Pose) {
    angularDampingOverride=0.95
    angularSpringConstant=15000
    bAngularSpring=true
    bOverrideAngularDamping=true
    bUseGravityOverride=true
    bUsePlanarLimit=false
    gravityScale=0.35
    numSolverIterationsPreUpdate=1
    alphaBoolBlend {blendOption=Linear}
    alphaScaleBiasClamp {scale=0.7}
    boundBone {boneName=dyn_head_fur_50}
    componentAppliedLinearAccClamp=<1300 1300 2400>
    componentLinearAccScale=<0.5 0.5 0.5>
    componentLinearVelScale=<0.5 0.5 0.5>
    gravityOverride {z=-3000}
    physicsBodyDefinitions [{
      boundBone {boneName=dyn_head_fur_50}
      boxExtents=<25 25 25>
      constraintSetup {angularLimitsMax=<20 5 20>, angularLimitsMin=<-20 -40 -20>, angularTarget=<1 0 0>}
      localJointOffset=<-2 0 -1>
     }]
   }
   Save cached pose 'HeadFurCards'(Pose=Component To Local(ComponentPose=@n53)) {cacheName=HeadFurCards}
  ]}
 EventGraph {nodes [
   n1 = Event BlueprintUpdateAnimation()
   > n2 = Sequence!(execute) -> (then_0, then_1, then_2:Exec)
   > Cloth Controller Function DEPRECATED!(
    execute
    CharacterPartAnimInstance:Custom Character Part Anim Instance Object Reference=Self-Reference()
    ClothingAssetName:String=M_MED_GuineaPig_Clothing_Plane_0
    Cloth Paramaters:Cloth Params Struct Structure=Get Cloth Paramaters() -> (Cloth Paramaters:Cloth Params Struct Structure)
    Cloth Core Settings:Cloth Core Setting Struct Structure=Get Cloth Core Settings() -> (Cloth Core Settings:Cloth Core Setting Struct Structure)
    ClothUpAxis:Vector="1,0,0"
    OnGround_Moving_SpeedLimit:double=1.5
    OnGround_Moving_AngularVel_Max:double=0.025
    OnGround_Moving_FictitiousVel_Max:double=0.6
    LinVelScale_FwdBwd_Clamp:double=0
    LinVelScale_UpDwn_Clamp:double=0.9
    DeltaTime:double=@n1.DeltaTimeX
    GameTimeInSeconds:double=GetGameTimeInSeconds()
    WindGustsParams:Wind Gust Params Struct Structure=Get Wind Gusts Params() -> (Wind Gusts Params:Wind Gust Params Struct Structure)
    WindGustLocalSettings:Wind Gust Local Setting Struct Structure=Get Wind Gust Local Settings() -> (Wind Gust Local Settings:Wind Gust Local Setting Struct Structure)
   ) -> (
    then
    CurrentStateOut:String
    Cloth Settings Out:Cloth Setting Struct Structure
    Cloth Core Settings Out:Cloth Core Setting Struct Structure
    WindGustSettings Out:Wind Gust Setting Struct Structure
    WindGustLocalSettings Out:Wind Gust Local Setting Struct Structure
    GustStrength - WindVelocityOut:Vector
    LinearVelVector-DELETE LATER:Vector
   )
   > n3 = Gravity Override State Function DEPRECATED(CharacterPartAnimInstance=Self-Reference(), GameTimeInSeconds=GetGameTimeInSeconds(), GravityOverrideParams=Get Gravity Override Params())
   > Set Gravity Override(Gravity Override=@n3)
   n4 = GetOwningActor()
   Cast To FortPlayerMannequin!(execute=@n2.then_1, Object=@n4) -> (then, CastFailed, AsFort Player Mannequin:Fort Player Mannequin Object Reference) {is=title K2Node_DynamicCast}
   > n5 = For Each Loop(Array:Array of Skeletal Mesh Component Object References=Get Components By Class!(self:Actor Object Reference=@n4, ComponentClass:Actor Component Class Reference=/Script/Engine.SkeletalMeshComponent) -> (ReturnValue:Array of Skeletal Mesh Component Object References)) -> (Array Element:Skeletal Mesh Component Object Reference)
   > Branch(Condition="Equal (Object)"(
     A=Get SkeletalMesh!(self:Skinned Mesh Component Object Reference=@n5) -> (SkeletalMesh:Skeletal Mesh Object Reference)
     B=/BRCosmetics/Characters/Player/Male/Medium/Heads/M_MED_GuineaPig_Head/Meshes/M_MED_GuineaPig_Head
    ))
   > n6 = Cast To SkeletalMeshComponentBudgeted!(execute, Object=@n5) -> (then, CastFailed, AsSkeletal Mesh Component Budgeted:Skeletal Mesh Component Budgeted Object Reference) {is=title K2Node_DynamicCast}
   > Set Head(Head:Skeletal Mesh Component Budgeted Object Reference=@n6) -> (Output_Get:Skeletal Mesh Component Budgeted Object Reference)
   n7 = GetOwnerFortPlayerAnimInstance!(self:Custom Character Part Anim Instance Object Reference) -> (ReturnValue:Fort Player Anim Instance Object Reference)
   n8 = Is Valid(exec=@n2.then_2, InputObject=@n7)
   n9 = Branch(
    execute=[@n13.then, @n8.Is Valid]
    Condition=Get bIsInFirstPerson!(self:Fort Player Anim Instance Object Reference=@n7) -> (bIsInFirstPerson:bool)
   )
   n10 = GetPartSkeletalMeshComponent()
   > n11 = Get Dynamic Material Instance(SkeletalMeshComponent=@n10)
   > SetScalarParameterValue(self=@n11, ParameterName=OpacityMaskMultiplier, Value=0)
   n12 = Get Dynamic Material Instance(execute=[@n16.then, @n9.else], SkeletalMeshComponent=@n10)
   > SetScalarParameterValue(self=@n12, ParameterName=OpacityMaskMultiplier, Value=1)
   n13 = Event OnBeginTargeting() {is=title K2Node_Event}
   n14 = Event OnCharacterCustomizationLoaded()
   > Set Head(Head=Get Array Item with EFort Custom Part Type(Array=@n14.CustomMeshParts, PartType=Head))
   > Set Body(Body:Skeletal Mesh Component Object Reference=@n14.CustomBaseMesh) -> (Output_Get:Skeletal Mesh Component Object Reference)
   > n15 = Get Dynamic Material Instance(SkeletalMeshComponent=@n14.CustomBaseMesh)
   > Set FurMaterial(FurMaterial:Material Instance Dynamic Object Reference=@n15) -> (Output_Get:Material Instance Dynamic Object Reference)
   n16 = Event OnEndTargeting() {is=title K2Node_Event}
   Branch(Condition="Equal (Integer)"(A=Get CurrentLOD()))
  ]}
]
variables [
 AnimDynamicsClothSwitch {type=boolean}
 As M MED Guinea Pig Head Anim BP PP {type=M_MED_GuineaPig_Head_AnimBP_PP_C, default=null}
 Body {type=SkeletalMeshComponent}
 Cloth Core Settings {
  type=ClothCoreSettingStruct
  default {
   collisionThickness=1
   dampingCoefficient=0.22
   frictionCoefficient=0.2
   localDampingCoefficient=0
   numberIterations=1
   numberSubsteps=1
   selfCollisionThickness=0
   animDriveDamping=<0 0>
   animDriveStiffness=<0 0>
   areaStiffness=<0 0>
   bendingStiffness=<0 0>
   edgeStiffness=<0.65 0.8>
   tetherStiffness=<0 0>
  }
 }
 Cloth Paramaters {
  type=ClothParamsStruct
  default {
   dBNO {
    angularVelocityScale=0.3
    fictitiousAngularScale=0.1
    jointName=root
    dragCoefficient=<0.1 0.5>
    gravityOverride=<0 -100 -980>
    liftCoefficient=<0.1 0.5>
    linearVelocityScale=<0.2 0.2 0.4>
   }
   dBNO_Carried {
    angularVelocityScale=0.3
    fictitiousAngularScale=0.1
    jointName=root
    dragCoefficient=<0.1 0.5>
    gravityOverride=<0 -100 -980>
    liftCoefficient=<0.1 0.5>
    linearVelocityScale=<0.2 0.2 0.4>
   }
   dBNO_Impulse {
    angularVelocityScale=0
    fictitiousAngularScale=0
    jointName=pelvis
    dragCoefficient=<0 0.2>
    gravityOverride=<-3000 -2000 0>
    liftCoefficient=<0 0.1>
    linearVelocityScale=<0.3 0.1 0.3>
   }
   dBNO_RollBwd {
    angularVelocityScale=0
    fictitiousAngularScale=0
    jointName=pelvis
    dragCoefficient=<0 0.2>
    gravityOverride=<-2000 0 0>
    liftCoefficient=<0 0.1>
    linearVelocityScale=<0.3 0.1 0.3>
   }
   dBNO_RollFwd {
    angularVelocityScale=0
    fictitiousAngularScale=0
    jointName=pelvis
    dragCoefficient=<0 0.2>
    gravityOverride=<-980 0 0>
    liftCoefficient=<0 0.1>
    linearVelocityScale=<0.3 0.1 0.3>
   }
   dBNO_Tumbling {
    angularVelocityScale=0
    fictitiousAngularScale=0
    jointName=pelvis
    dragCoefficient=<0 0.2>
    gravityOverride=<-2500 -300 0>
    liftCoefficient=<0 0.1>
    linearVelocityScale=<0.3 0.1 0.3>
   }
   emoteOrMelee {
    angularVelocityScale=0.3
    fictitiousAngularScale=0.1
    jointName=root
    dragCoefficient=<0.1 0.5>
    gravityOverride=<0 -100 -980>
    liftCoefficient=<0.1 0.5>
    linearVelocityScale=<0.2 0.2 0.4>
   }
   front_End {
    angularVelocityScale=0.15
    fictitiousAngularScale=0.4
    jointName=root
    dragCoefficient=<0.2 0.2>
    gravityOverride=<0 -200 -980>
    liftCoefficient=<0.1 0.1>
    linearVelocityScale=<0.1 0.1 0.1>
   }
   onGround_Crouch_Moving {
    angularVelocityScale=0.3
    fictitiousAngularScale=0.1
    jointName=root
    dragCoefficient=<1 1>
    gravityOverride=<0 0 -980>
    liftCoefficient=<0.2 0.2>
    linearVelocityScale=<0.2 0.1 0.1>
   }
   onGround_Falling {
    angularVelocityScale=0.3
    fictitiousAngularScale=0.1
    jointName=root
    dragCoefficient=<0.1 0.5>
    gravityOverride=<0 -100 -980>
    liftCoefficient=<0.1 0.5>
    linearVelocityScale=<0.2 0.2 0.4>
   }
   onGround_Moving {
    angularVelocityScale=0.15
    fictitiousAngularScale=0.1
    jointName=root
    dragCoefficient=<0.05 0.15>
    gravityOverride=<0 -100 -980>
    liftCoefficient=<0.05 0.15>
    linearVelocityScale=<0.2 0.2 0.2>
   }
   onGround_Standing {
    angularVelocityScale=0.3
    fictitiousAngularScale=0.1
    jointName=root
    dragCoefficient=<0.1 0.5>
    gravityOverride=<0 -100 -980>
    liftCoefficient=<0.1 0.5>
    linearVelocityScale=<0.2 0.2 0.4>
   }
   rail_Grinding {
    angularVelocityScale=0.3
    fictitiousAngularScale=0.3
    jointName=root
    dragCoefficient=<0.08 0.8>
    gravityOverride=<0 -100 -400>
    liftCoefficient=<0.08 0.8>
    linearVelocityScale=<0.05 0.05 0.3>
   }
   riding_Moving {
    angularVelocityScale=0.3
    fictitiousAngularScale=0.1
    jointName=root
    dragCoefficient=<0.1 0.5>
    gravityOverride=<0 -100 -980>
    liftCoefficient=<0.1 0.5>
    linearVelocityScale=<0.2 0.2 0.4>
   }
   riding_Moving_Targeting {
    angularVelocityScale=0.3
    fictitiousAngularScale=0.1
    jointName=root
    dragCoefficient=<0.1 0.5>
    gravityOverride=<0 -100 -980>
    liftCoefficient=<0.1 0.5>
    linearVelocityScale=<0.2 0.2 0.4>
   }
   riding_Standing {
    angularVelocityScale=0.3
    fictitiousAngularScale=0.1
    jointName=root
    dragCoefficient=<0.1 0.5>
    gravityOverride=<0 -100 -980>
    liftCoefficient=<0.1 0.5>
    linearVelocityScale=<0.2 0.2 0.4>
   }
   skydive_DivingDown {
    angularVelocityScale=0.3
    fictitiousAngularScale=0.3
    jointName=root
    dragCoefficient=<0.08 0.5>
    gravityOverride=<0 -100 -400>
    liftCoefficient=<0.08 0.5>
    linearVelocityScale=<0.05 0.05 0.3>
   }
   skydive_FloatingDown {
    angularVelocityScale=0.1
    fictitiousAngularScale=0.1
    jointName=root
    dragCoefficient=<0.08 0.1>
    gravityOverride=<0 -200 300>
    liftCoefficient=<0.08 0.1>
    linearVelocityScale=<0.05 0.05 0.1>
   }
   skydive_GoingUp {
    angularVelocityScale=0.1
    fictitiousAngularScale=0.1
    jointName=root
    dragCoefficient=<0.1 0.3>
    gravityOverride=<0 -100 0>
    liftCoefficient=<0.1 0.3>
    linearVelocityScale=<0.2 0.2 0.2>
   }
   skydive_Parachuting {
    angularVelocityScale=0.3
    fictitiousAngularScale=0.3
    jointName=root
    dragCoefficient=<0.08 0.8>
    gravityOverride=<0 -100 -400>
    liftCoefficient=<0.08 0.8>
    linearVelocityScale=<0.05 0.05 0.3>
   }
   skydive_Surfing {
    angularVelocityScale=0.3
    fictitiousAngularScale=0.1
    jointName=root
    dragCoefficient=<0.1 0.5>
    gravityOverride=<0 -100 -980>
    liftCoefficient=<0.1 0.5>
    linearVelocityScale=<0.2 0.2 0.4>
   }
   swimming {
    angularVelocityScale=0.3
    fictitiousAngularScale=0.1
    jointName=root
    dragCoefficient=<0.1 0.5>
    gravityOverride=<0 -100 -980>
    liftCoefficient=<0.1 0.5>
    linearVelocityScale=<0.2 0.2 0.4>
   }
   tactical_Sprint {
    angularVelocityScale=0.3
    fictitiousAngularScale=0.1
    jointName=root
    dragCoefficient=<0.1 0.5>
    gravityOverride=<0 -100 -980>
    liftCoefficient=<0.1 0.5>
    linearVelocityScale=<0.2 0.2 0.4>
   }
   ziplining {
    angularVelocityScale=0.4
    fictitiousAngularScale=0
    jointName=root
    dragCoefficient=<0 1.5>
    gravityOverride=<0 0 -980>
    liftCoefficient=<0 0.3>
    linearVelocityScale=<0.2 0.2 0.25>
   }
  }
 }
 ClothMat {type=MaterialInstanceDynamic}
 FurMaterial {type=MaterialInstanceDynamic}
 Gravity Override {type=Vector}
 Gravity Override Params {
  type=GravityOverrideParamsStruct
  default {
   emoteOrMelee {gravityOverride=<0 0 -3000>, windAmplitude=<0 200 150>, windFrequency=<3 10 8>, windOffset=<0 20 0>}
   front_End {gravityOverride=<0 0 -3000>, windAmplitude=<0 200 150>, windFrequency=<3 10 8>, windOffset=<0 20 0>}
   onGround_Falling {gravityOverride=<0 0 -3000>, windAmplitude=<0 200 150>, windFrequency=<3 10 8>, windOffset=<0 20 0>}
   onGround_Moving {gravityOverride=<0 0 -3000>, windAmplitude=<0 200 150>, windFrequency=<3 10 8>, windOffset=<0 20 0>}
   onGround_Standing {gravityOverride=<0 0 -4000>, windAmplitude=<0 150 100>, windFrequency=<1.5 8 6>, windOffset=<0 20 0>}
   rail_Grinding {gravityOverride=<0 0 -3000>, windAmplitude=<0 200 150>, windFrequency=<3 10 8>, windOffset=<0 20 0>}
   riding_Moving {gravityOverride=<0 0 -3000>, windAmplitude=<0 200 150>, windFrequency=<3 10 8>, windOffset=<0 20 0>}
   riding_Moving_Targeting {gravityOverride=<0 0 -3000>, windAmplitude=<0 200 150>, windFrequency=<3 10 8>, windOffset=<0 20 0>}
   riding_Standing {gravityOverride=<0 0 -3000>, windAmplitude=<0 200 150>, windFrequency=<3 10 8>, windOffset=<0 20 0>}
   skydive_DivingDown {jointName=root, gravityOverride=<0 700 -1000>, windAmplitude=<50 -5000 100>, windFrequency=<5 30 5>, windOffset=<-200 -500 500>}
   skydive_FloatingDown {gravityOverride=<0 700 -1000>, windAmplitude=<50 -2500 100>, windFrequency=<5 15 5>, windOffset=<-200 -500 500>}
   skydive_GoingUp {gravityOverride=<0 700 -1000>, windAmplitude=<50 -2500 100>, windFrequency=<5 15 5>, windOffset=<-200 -500 500>}
   skydive_Parachuting {gravityOverride=<0 700 -1000>, windAmplitude=<50 -2500 100>, windFrequency=<5 15 5>, windOffset=<-200 -500 500>}
   skydive_Surfing {gravityOverride=<0 0 -3000>, windAmplitude=<0 200 150>, windFrequency=<3 10 8>, windOffset=<0 20 0>}
   swimming {jointName=root, gravityOverride=<0 0 -3000>, windAmplitude=<0 200 150>, windFrequency=<3 10 8>, windOffset=<0 20 0>}
   tactical_Sprint {gravityOverride=<0 0 -3000>, windAmplitude=<0 200 150>, windFrequency=<3 10 8>, windOffset=<0 20 0>}
   ziplining {gravityOverride=<0 0 -3000>, windAmplitude=<0 200 150>, windFrequency=<3 10 8>, windOffset=<0 20 0>}
  }
 }
 Head {type=SkeletalMeshComponentBudgeted}
 HeadMannequin {type=SkeletalMesh, default=null}
 IsMannequin {type=boolean}
 NewVar {type=AnimInstance, default=null}
 Output {type=SkeletalMeshComponent}
 Skeletal Mesh {type=SkeletalMesh, default=null}
 Skeletal Mesh Component {type=SkeletalMeshComponent}
 Wind Gust Local Settings {type=WindGustLocalSettingStruct, default {elapsedTime=0, finalGustStrengthLocal=0, gustDuration=0, gustStrength=0, timeBetweenGusts=0}}
 Wind Gust Settings Out {
  type=WindGustSettingStruct
  default {
   maxDownInterpSpeed=5
   maxGustDuration=2
   maxGustStrength=7
   maxTimeBetweenGusts=7
   maxUpInterpSpeed=5
   minDownInterpSpeed=2
   minGustDuration=0.5
   minGustStrength=3
   minTimeBetweenGusts=0
   minUpInterpSpeed=2
   windTurbulenceScalarA=0.1
   windTurbulenceScalarB=1
   windAmplitude=<0 0 0>
   windFrequency=<0 0 0>
   windOffset=<0 0 0>
  }
 }
 Wind Gusts Params {
  type=WindGustParamsStruct
  default {
   dBNO {
    maxDownInterpSpeed=5
    maxGustDuration=2
    maxGustStrength=5
    maxTimeBetweenGusts=12
    maxUpInterpSpeed=8
    minDownInterpSpeed=2
    minGustDuration=0.5
    minGustStrength=3
    minTimeBetweenGusts=3
    minUpInterpSpeed=1
    windTurbulenceScalarA=0.6
    windTurbulenceScalarB=0.7
    windAmplitude=<0 100 100>
    windFrequency=<50 100 100>
    windOffset=<0 -150 100>
   }
   dBNO_Carried {
    maxDownInterpSpeed=5
    maxGustDuration=2
    maxGustStrength=5
    maxTimeBetweenGusts=12
    maxUpInterpSpeed=8
    minDownInterpSpeed=2
    minGustDuration=0.5
    minGustStrength=3
    minTimeBetweenGusts=3
    minUpInterpSpeed=1
    windTurbulenceScalarA=0.6
    windTurbulenceScalarB=0.7
    windAmplitude=<0 100 100>
    windFrequency=<50 100 100>
    windOffset=<0 -150 100>
   }
   dBNO_Impulse {
    maxDownInterpSpeed=5
    maxGustDuration=2
    maxGustStrength=7
    maxTimeBetweenGusts=7
    maxUpInterpSpeed=5
    minDownInterpSpeed=2
    minGustDuration=0.5
    minGustStrength=3
    minTimeBetweenGusts=0
    minUpInterpSpeed=2
    windTurbulenceScalarA=0.1
    windTurbulenceScalarB=1
    windAmplitude=<0 0 0>
    windFrequency=<0 0 0>
    windOffset=<0 0 0>
   }
   dBNO_RollBwd {
    maxDownInterpSpeed=5
    maxGustDuration=2
    maxGustStrength=7
    maxTimeBetweenGusts=7
    maxUpInterpSpeed=5
    minDownInterpSpeed=2
    minGustDuration=0.5
    minGustStrength=3
    minTimeBetweenGusts=0
    minUpInterpSpeed=2
    windTurbulenceScalarA=0.1
    windTurbulenceScalarB=1
    windAmplitude=<0 0 0>
    windFrequency=<0 0 0>
    windOffset=<0 0 0>
   }
   dBNO_RollFwd {
    maxDownInterpSpeed=5
    maxGustDuration=2
    maxGustStrength=7
    maxTimeBetweenGusts=7
    maxUpInterpSpeed=5
    minDownInterpSpeed=2
    minGustDuration=0.5
    minGustStrength=3
    minTimeBetweenGusts=0
    minUpInterpSpeed=2
    windTurbulenceScalarA=0.1
    windTurbulenceScalarB=1
    windAmplitude=<0 0 0>
    windFrequency=<0 0 0>
    windOffset=<0 0 0>
   }
   dBNO_Tumbling {
    maxDownInterpSpeed=5
    maxGustDuration=2
    maxGustStrength=7
    maxTimeBetweenGusts=7
    maxUpInterpSpeed=5
    minDownInterpSpeed=2
    minGustDuration=0.5
    minGustStrength=3
    minTimeBetweenGusts=0
    minUpInterpSpeed=2
    windTurbulenceScalarA=0.1
    windTurbulenceScalarB=1
    windAmplitude=<0 0 0>
    windFrequency=<0 0 0>
    windOffset=<0 0 0>
   }
   emoteOrMelee {
    maxDownInterpSpeed=5
    maxGustDuration=2
    maxGustStrength=5
    maxTimeBetweenGusts=12
    maxUpInterpSpeed=8
    minDownInterpSpeed=2
    minGustDuration=0.5
    minGustStrength=3
    minTimeBetweenGusts=3
    minUpInterpSpeed=1
    windTurbulenceScalarA=0.6
    windTurbulenceScalarB=0.7
    windAmplitude=<0 100 100>
    windFrequency=<50 100 100>
    windOffset=<0 -150 100>
   }
   front_End {
    maxDownInterpSpeed=5
    maxGustDuration=2
    maxGustStrength=7
    maxTimeBetweenGusts=7
    maxUpInterpSpeed=5
    minDownInterpSpeed=2
    minGustDuration=0.5
    minGustStrength=3
    minTimeBetweenGusts=0
    minUpInterpSpeed=2
    windTurbulenceScalarA=0.1
    windTurbulenceScalarB=1
    windAmplitude=<0 0 0>
    windFrequency=<0 0 0>
    windOffset=<0 0 0>
   }
   onGround_Crouch_Moving {
    maxDownInterpSpeed=5
    maxGustDuration=2
    maxGustStrength=7
    maxTimeBetweenGusts=7
    maxUpInterpSpeed=5
    minDownInterpSpeed=2
    minGustDuration=0.5
    minGustStrength=3
    minTimeBetweenGusts=0
    minUpInterpSpeed=2
    windTurbulenceScalarA=0.1
    windTurbulenceScalarB=1
    windAmplitude=<0 0 0>
    windFrequency=<0 0 0>
    windOffset=<0 0 0>
   }
   onGround_Falling {
    maxDownInterpSpeed=5
    maxGustDuration=2
    maxGustStrength=5
    maxTimeBetweenGusts=12
    maxUpInterpSpeed=8
    minDownInterpSpeed=2
    minGustDuration=0.5
    minGustStrength=3
    minTimeBetweenGusts=3
    minUpInterpSpeed=1
    windTurbulenceScalarA=0.6
    windTurbulenceScalarB=0.7
    windAmplitude=<0 100 100>
    windFrequency=<50 100 100>
    windOffset=<0 -150 100>
   }
   onGround_Moving {
    maxDownInterpSpeed=5
    maxGustDuration=2
    maxGustStrength=5
    maxTimeBetweenGusts=12
    maxUpInterpSpeed=8
    minDownInterpSpeed=2
    minGustDuration=0.5
    minGustStrength=3
    minTimeBetweenGusts=3
    minUpInterpSpeed=1
    windTurbulenceScalarA=0.6
    windTurbulenceScalarB=0.7
    windAmplitude=<0 100 100>
    windFrequency=<50 100 100>
    windOffset=<0 -150 100>
   }
   onGround_Standing {
    maxDownInterpSpeed=5
    maxGustDuration=2
    maxGustStrength=5
    maxTimeBetweenGusts=12
    maxUpInterpSpeed=8
    minDownInterpSpeed=2
    minGustDuration=0.5
    minGustStrength=3
    minTimeBetweenGusts=3
    minUpInterpSpeed=1
    windTurbulenceScalarA=0.6
    windTurbulenceScalarB=0.7
    windAmplitude=<0 100 100>
    windFrequency=<50 100 100>
    windOffset=<0 -150 100>
   }
   rail_Grinding {
    maxDownInterpSpeed=2
    maxGustDuration=1
    maxGustStrength=2
    maxTimeBetweenGusts=5
    maxUpInterpSpeed=2
    minDownInterpSpeed=1
    minGustDuration=0.5
    minGustStrength=1
    minTimeBetweenGusts=1
    minUpInterpSpeed=1
    windTurbulenceScalarA=0.1
    windTurbulenceScalarB=1
    windAmplitude=<100 100 100>
    windFrequency=<100 100 100>
    windOffset=<-100 -300 300>
   }
   riding_Moving {
    maxDownInterpSpeed=5
    maxGustDuration=2
    maxGustStrength=5
    maxTimeBetweenGusts=12
    maxUpInterpSpeed=8
    minDownInterpSpeed=2
    minGustDuration=0.5
    minGustStrength=3
    minTimeBetweenGusts=3
    minUpInterpSpeed=1
    windTurbulenceScalarA=0.6
    windTurbulenceScalarB=0.7
    windAmplitude=<0 100 100>
    windFrequency=<50 100 100>
    windOffset=<0 -150 100>
   }
   riding_Moving_Targeting {
    maxDownInterpSpeed=5
    maxGustDuration=2
    maxGustStrength=5
    maxTimeBetweenGusts=12
    maxUpInterpSpeed=8
    minDownInterpSpeed=2
    minGustDuration=0.5
    minGustStrength=3
    minTimeBetweenGusts=3
    minUpInterpSpeed=1
    windTurbulenceScalarA=0.6
    windTurbulenceScalarB=0.7
    windAmplitude=<0 100 100>
    windFrequency=<50 100 100>
    windOffset=<0 -150 100>
   }
   riding_Standing {
    maxDownInterpSpeed=5
    maxGustDuration=2
    maxGustStrength=5
    maxTimeBetweenGusts=12
    maxUpInterpSpeed=8
    minDownInterpSpeed=2
    minGustDuration=0.5
    minGustStrength=3
    minTimeBetweenGusts=3
    minUpInterpSpeed=1
    windTurbulenceScalarA=0.6
    windTurbulenceScalarB=0.7
    windAmplitude=<0 100 100>
    windFrequency=<50 100 100>
    windOffset=<0 -150 100>
   }
   skydive_DivingDown {
    maxDownInterpSpeed=2
    maxGustDuration=1
    maxGustStrength=2
    maxTimeBetweenGusts=5
    maxUpInterpSpeed=2
    minDownInterpSpeed=1
    minGustDuration=0.5
    minGustStrength=1
    minTimeBetweenGusts=1
    minUpInterpSpeed=1
    windTurbulenceScalarA=0.1
    windTurbulenceScalarB=1
    windAmplitude=<100 100 100>
    windFrequency=<100 100 100>
    windOffset=<-100 -300 300>
   }
   skydive_FloatingDown {
    maxDownInterpSpeed=2
    maxGustDuration=1
    maxGustStrength=2
    maxTimeBetweenGusts=5
    maxUpInterpSpeed=2
    minDownInterpSpeed=1
    minGustDuration=0.5
    minGustStrength=1
    minTimeBetweenGusts=1
    minUpInterpSpeed=1
    windTurbulenceScalarA=0.1
    windTurbulenceScalarB=1
    windAmplitude=<0 5 5>
    windFrequency=<50 50 50>
    windOffset=<0 -300 100>
   }
   skydive_GoingUp {
    maxDownInterpSpeed=5
    maxGustDuration=2
    maxGustStrength=5
    maxTimeBetweenGusts=12
    maxUpInterpSpeed=8
    minDownInterpSpeed=2
    minGustDuration=0.5
    minGustStrength=3
    minTimeBetweenGusts=3
    minUpInterpSpeed=1
    windTurbulenceScalarA=0.6
    windTurbulenceScalarB=0.7
    windAmplitude=<0 100 100>
    windFrequency=<50 100 100>
    windOffset=<0 -150 100>
   }
   skydive_Parachuting {
    maxDownInterpSpeed=2
    maxGustDuration=1
    maxGustStrength=2
    maxTimeBetweenGusts=5
    maxUpInterpSpeed=2
    minDownInterpSpeed=1
    minGustDuration=0.5
    minGustStrength=1
    minTimeBetweenGusts=1
    minUpInterpSpeed=1
    windTurbulenceScalarA=0.1
    windTurbulenceScalarB=1
    windAmplitude=<100 100 100>
    windFrequency=<100 100 100>
    windOffset=<-100 -300 300>
   }
   skydive_Surfing {
    maxDownInterpSpeed=5
    maxGustDuration=2
    maxGustStrength=5
    maxTimeBetweenGusts=12
    maxUpInterpSpeed=8
    minDownInterpSpeed=2
    minGustDuration=0.5
    minGustStrength=3
    minTimeBetweenGusts=3
    minUpInterpSpeed=1
    windTurbulenceScalarA=0.6
    windTurbulenceScalarB=0.7
    windAmplitude=<0 100 100>
    windFrequency=<50 100 100>
    windOffset=<0 -150 100>
   }
   swimming {
    maxDownInterpSpeed=5
    maxGustDuration=2
    maxGustStrength=5
    maxTimeBetweenGusts=12
    maxUpInterpSpeed=8
    minDownInterpSpeed=2
    minGustDuration=0.5
    minGustStrength=3
    minTimeBetweenGusts=3
    minUpInterpSpeed=1
    windTurbulenceScalarA=0.6
    windTurbulenceScalarB=0.7
    windAmplitude=<0 100 100>
    windFrequency=<50 100 100>
    windOffset=<0 -150 100>
   }
   tactical_Sprint {
    maxDownInterpSpeed=5
    maxGustDuration=2
    maxGustStrength=5
    maxTimeBetweenGusts=12
    maxUpInterpSpeed=8
    minDownInterpSpeed=2
    minGustDuration=0.5
    minGustStrength=3
    minTimeBetweenGusts=3
    minUpInterpSpeed=1
    windTurbulenceScalarA=0.6
    windTurbulenceScalarB=0.7
    windAmplitude=<0 100 100>
    windFrequency=<50 100 100>
    windOffset=<0 -150 100>
   }
   ziplining {
    maxDownInterpSpeed=5
    maxGustDuration=2
    maxGustStrength=7
    maxTimeBetweenGusts=7
    maxUpInterpSpeed=5
    minDownInterpSpeed=2
    minGustDuration=0.5
    minGustStrength=3
    minTimeBetweenGusts=0
    minUpInterpSpeed=2
    windTurbulenceScalarA=0.1
    windTurbulenceScalarB=1
    windAmplitude=<0 0 0>
    windFrequency=<0 0 0>
    windOffset=<0 0 0>
   }
  }
 }
]
