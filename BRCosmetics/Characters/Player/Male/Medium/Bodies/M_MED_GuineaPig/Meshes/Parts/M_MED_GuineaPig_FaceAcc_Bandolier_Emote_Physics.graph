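# Physics asset for the GuineaPig character's bandolier accessory, named for emote use, covering two bandolier strands and their hanging bags.
# Front and rear capsule chains run from spine_02 through straps into bag bones, tying into a shared sim-chain body and cross-linking the bags.
# Bandolier and bag joints have limited swing and twist, while body joints and spine anchors rotate freely.
# Pelvis, spine, right-leg capsules and a large root sphere below the origin supply body shapes.
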PhysicsAsset /BRCosmetics/Characters/Player/Male/Medium/Bodies/M_MED_GuineaPig/Meshes/Parts/M_MED_GuineaPig_FaceAcc_Bandolier_Emote_Physics
bodies {
 calf_r [{shapeType=Capsule, length=39.9543, radius=7.5, shapeName=calf_r_capsule, center=<21.3239 1.0844 0>, rotation=<rot 0.0009 -90.0112 88.6527>}]
 def_bandolier_front_04 [{
   shapeType=Capsule
   length=4.4402
   radius=4.1364
   shapeName=def_bandolier_front_04_capsule
   center=<3.0567 1.7926 -0.0527>
   rotation=<rot 0.0001 -84.3296 -89.3425>
  }]
 def_bandolier_front_05 [{
   shapeType=Capsule
   length=4.5719
   radius=3.7901
   shapeName=def_bandolier_front_05_capsule
   center=<4.0595 1.7352 0.0637>
   rotation=<rot 0 -88.3456 -86.5566>
  }]
 def_bandolier_front_06 [{
   shapeType=Capsule
   length=4.1548
   radius=4.3273
   shapeName=def_bandolier_front_06_capsule
   center=<3.2946 0.8371 -0.1115>
   rotation=<rot 0 -93.7944 -86.3925>
  }]
 def_bandolier_front_07 [{
   shapeType=Capsule
   length=3.9571
   radius=4.048
   shapeName=def_bandolier_front_07_capsule
   center=<3.7359 0.7192 -0.2147>
   rotation=<rot 0 -89.6738 -84.4103>
  }]
 def_bandolier_front_08 [{
   shapeType=Capsule
   length=5.5519
   radius=3.2076
   shapeName=def_bandolier_front_08_capsule
   center=<2.767 0.6694 -0.2054>
   rotation=<rot 19.7145 -90 87.7479>
  }]
 def_bandolier_front_09 [{
   shapeType=Capsule
   length=4.0092
   radius=3.6298
   shapeName=def_bandolier_front_09_capsule
   center=<3.9551 0.5651 0.2726>
   rotation=<rot -2.4706 97.0119 -79.0861>
  }]
 def_bandolier_front_10 [{shapeType=Capsule, length=5.5987, radius=2.4953, rotation=<rot 14.3779 87.5035 -87.985>}]
 def_bandolier_front_strap [{
   shapeType=Capsule
   length=6.1092
   radius=3.0711
   shapeName=def_bandolier_front_strap_capsule
   center=<2.4309 -0.1543 0.7575>
   rotation=<rot 34.3854 -90 -104.9026>
  }]
 def_bandolier_rear_04 [{
   shapeType=Capsule
   length=4.7532
   radius=3.8293
   shapeName=def_bandolier_rear_04_capsule
   center=<3.4317 -1.7471 0.0738>
   rotation=<rot 0 -93.6107 -86.2494>
  }]
 def_bandolier_rear_05 [{
   shapeType=Capsule
   length=4.7994
   radius=4.0191
   shapeName=def_bandolier_rear_05_capsule
   center=<4.2771 -1.7674 0.0769>
   rotation=<rot 0 -91.1873 -90.0686>
  }]
 def_bandolier_rear_06 [{
   shapeType=Capsule
   length=4.7473
   radius=3.73
   shapeName=def_bandolier_rear_06_capsule
   center=<3.3516 -1.4843 0.0553>
   rotation=<rot 0.0252 -92.3758 -87.563>
  }]
 def_bandolier_rear_07 [{
   shapeType=Capsule
   length=4.5278
   radius=3.8723
   shapeName=def_bandolier_rear_07_capsule
   center=<4.1289 -1.4662 -0.0477>
   rotation=<rot 0 -88.6999 -83.3916>
  }]
 def_bandolier_rear_08 [{
   shapeType=Capsule
   length=3.892
   radius=4.0183
   shapeName=def_bandolier_rear_08_capsule
   center=<3.2484 -1.3744 -0.2144>
   rotation=<rot -10.2416 -89.9999 -90.0262>
  }]
 def_bandolier_rear_09 [{
   shapeType=Capsule
   length=3.2519
   radius=2.3354
   shapeName=def_bandolier_rear_09_capsule
   center=<3.6022 -2.3932 -1.3526>
   rotation=<rot 0 -82.205 -82.639>
  }]
 def_bandolier_rear_10 [{shapeType=Capsule, length=4.3217, radius=2.0803, center=<0.8274 -2.3188 0>, rotation=<rot -2.3049 104.7509 -115.1461>}]
 def_bandolier_rear_strap [{
   shapeType=Capsule
   length=3.5011
   radius=3.3596
   shapeName=def_bandolier_rear_strap_capsule
   center=<2.584 0.7817 0.2454>
   rotation=<rot 20.9947 -70.5393 -77.7413>
  }]
 dyn_bag_front_01 [{
   shapeType=Capsule
   length=19.0426
   radius=5.0988
   shapeName=dyn_bag_front_01_capsule
   center=<6.742 -1.0568 2.2611>
   rotation=<rot -6.1591 -91.3391 -81.0083>
  }]
 dyn_bag_front_02 [{
   shapeType=Capsule
   length=2.7514
   radius=1.7605
   shapeName=dyn_bag_front_02_capsule
   center=<1.7129 0.1121 -0.3676>
   rotation=<rot 9.1124 -90 10.3745>
  }]
 dyn_bag_front_03 [{
   shapeType=Capsule
   length=2.1658
   radius=1.058
   shapeName=dyn_bag_front_03_capsule
   center=<1.4342 -0.1171 -0.0576>
   rotation=<rot 0 -91.6745 -88.244>
  }]
 dyn_bag_front_04 [{
   shapeType=Capsule
   length=2.0814
   radius=1.3594
   shapeName=dyn_bag_front_04_capsule
   center=<1.2965 -0.0547 0.0624>
   rotation=<rot 0 89.652 -85.8641>
  }]
 dyn_bag_rear_01 [{
   shapeType=Capsule
   length=19.266
   radius=5.7473
   shapeName=dyn_bag_rear_01_capsule
   center=<6.7813 -1.057 1.837>
   rotation=<rot -3.0152 -88.3132 -82.7078>
  }]
 dyn_bag_rear_02 [{
   shapeType=Capsule
   length=2.5767
   radius=1.6428
   shapeName=dyn_bag_rear_02_capsule
   center=<1.5072 0.4609 -0.9451>
   rotation=<rot -12.333 -90 17.9062>
  }]
 dyn_bag_rear_03 [{
   shapeType=Capsule
   length=2.2388
   radius=1.0866
   shapeName=dyn_bag_rear_03_capsule
   center=<1.8532 -0.1811 -0.3449>
   rotation=<rot 0 -88.9607 -79.988>
  }]
 dyn_bag_rear_04 [{
   shapeType=Capsule
   length=2.0814
   radius=1.3594
   shapeName=dyn_bag_rear_04_capsule
   center=<1.3269 0.0562 0.0395>
   rotation=<rot 0 -86.8702 -89.8371>
  }]
 dyn_bag_sim_chain_03 [{shapeType=Capsule, length=0.505, radius=0.505, shapeName=dyn_bag_sim_chain_03_capsule, rotation=<rot 0 -90 0>}]
 pelvis [
  {shapeType=Capsule, length=16.5317, radius=8.3737, center=<-2.3011 -1.6625 0>}
  {shapeType=Capsule, length=11.0925, radius=10.7372, shapeName=pelvis_capsule, center=<-6.612 -3.6876 0>}
 ]
 root [{shapeType=Sphere, radius=110, shapeName=root_sphere, center=<0 0 -95>}]
 spine_01 [
  {shapeType=Capsule, length=10.7601, radius=11.9282, shapeName=spine_01_capsule, center=<0 -0.6598 0>, rotation=<rot 0 -90 0>}
  {shapeType=Capsule, length=6.5515, radius=4.6245, shapeName=spine_01_capsule, center=<-3.0276 -16.3985 -3.3224>, rotation=<rot 52.7808 0 0>}
  {
   shapeType=Capsule
   length=5.0891
   radius=5.3033
   shapeName=spine_01_capsule
   center=<-4.2313 15.8862 -3.126>
   rotation=<rot 57.9523 -1.925 -2.5162>
  }
 ]
 spine_02 [
  {shapeType=Capsule, length=11.5873, radius=12.1278, shapeName=spine_02_capsule, center=<0 -0.8477 0>, rotation=<rot 0 -90 0>}
  {
   shapeType=Capsule
   length=10.996
   radius=4.3502
   shapeName=spine_02_capsule
   center=<-0.6732 -13.3326 -9.9473>
   rotation=<rot 58.9645 0.6365 -5.3463>
  }
  {shapeType=Capsule, length=6.2946, radius=3.5248, shapeName=spine_02_capsule, center=<-0.6169 16.2188 -8.7179>, rotation=<rot 57.6485 0 0>}
 ]
 thigh_r [{shapeType=Capsule, length=37.865, radius=7.6625, shapeName=thigh_r_capsule, center=<19.709 0 0>, rotation=<rot 0 90 90>}]
}
constraints [
 {bone1Name=def_bandolier_front_10, bone2Name=def_bandolier_front_09, swing1LimitDegrees=20, swing2LimitDegrees=20, twistLimitDegrees=4}
 {bone1Name=def_bandolier_rear_10, bone2Name=def_bandolier_rear_09, swing1LimitDegrees=20, swing2LimitDegrees=20, twistLimitDegrees=4}
 {
  bone1Name=thigh_r
  bone2Name=pelvis
  swing1LimitDegrees=45
  swing1Motion=Free
  swing2LimitDegrees=45
  swing2Motion=Free
  twistLimitDegrees=45
  twistMotion=Free
 }
 {bone1Name=dyn_bag_front_01, bone2Name=def_bandolier_front_strap, swing1LimitDegrees=4, swing2LimitDegrees=20, twistLimitDegrees=4}
 {bone1Name=dyn_bag_front_02, bone2Name=dyn_bag_front_01, swing1LimitDegrees=12, swing2LimitDegrees=12, twistLimitDegrees=12}
 {bone1Name=dyn_bag_front_03, bone2Name=dyn_bag_front_02, swing1LimitDegrees=12, swing2LimitDegrees=12, twistLimitDegrees=12}
 {bone1Name=dyn_bag_rear_01, bone2Name=def_bandolier_rear_strap, swing1LimitDegrees=4, swing2LimitDegrees=20, twistLimitDegrees=4}
 {bone1Name=dyn_bag_rear_02, bone2Name=dyn_bag_rear_01, swing1LimitDegrees=12, swing2LimitDegrees=12, twistLimitDegrees=12}
 {bone1Name=dyn_bag_rear_03, bone2Name=dyn_bag_rear_02, swing1LimitDegrees=12, swing2LimitDegrees=12, twistLimitDegrees=12}
 {bone1Name=def_bandolier_rear_09, bone2Name=def_bandolier_rear_08, swing1LimitDegrees=20, swing2LimitDegrees=20, twistLimitDegrees=4}
 {bone1Name=def_bandolier_rear_08, bone2Name=def_bandolier_rear_07, swing1LimitDegrees=24, swing2LimitDegrees=24, twistLimitDegrees=24}
 {bone1Name=def_bandolier_rear_07, bone2Name=def_bandolier_rear_06, swing1LimitDegrees=24, swing2LimitDegrees=24, twistLimitDegrees=24}
 {bone1Name=def_bandolier_front_09, bone2Name=def_bandolier_front_08, swing1LimitDegrees=20, swing2LimitDegrees=20, twistLimitDegrees=4}
 {bone1Name=def_bandolier_front_08, bone2Name=def_bandolier_front_07, swing1LimitDegrees=24, swing2LimitDegrees=24, twistLimitDegrees=24}
 {bone1Name=def_bandolier_front_07, bone2Name=def_bandolier_front_06, swing1LimitDegrees=24, swing2LimitDegrees=24, twistLimitDegrees=24}
 {
  bone1Name=spine_01
  bone2Name=pelvis
  swing1LimitDegrees=45
  swing1Motion=Free
  swing2LimitDegrees=45
  swing2Motion=Free
  twistLimitDegrees=45
  twistMotion=Free
 }
 {
  bone1Name=spine_02
  bone2Name=spine_01
  swing1LimitDegrees=45
  swing1Motion=Free
  swing2LimitDegrees=45
  swing2Motion=Free
  twistLimitDegrees=45
  twistMotion=Free
 }
 {
  bone1Name=dyn_bag_sim_chain_03
  bone2Name=spine_02
  swing1LimitDegrees=45
  swing1Motion=Free
  swing2LimitDegrees=45
  swing2Motion=Free
  twistLimitDegrees=45
  twistMotion=Free
 }
 {bone1Name=def_bandolier_rear_10, bone2Name=dyn_bag_sim_chain_03, swing1LimitDegrees=45, swing2LimitDegrees=45, twistLimitDegrees=45}
 {bone1Name=def_bandolier_front_10, bone2Name=dyn_bag_sim_chain_03, swing1LimitDegrees=45, swing2LimitDegrees=45, twistLimitDegrees=45}
 {bone1Name=def_bandolier_front_06, bone2Name=def_bandolier_front_05, swing1LimitDegrees=24, swing2LimitDegrees=24, twistLimitDegrees=24}
 {bone1Name=def_bandolier_front_05, bone2Name=def_bandolier_front_04, swing1LimitDegrees=24, swing2LimitDegrees=24, twistLimitDegrees=24}
 {
  bone1Name=def_bandolier_front_04
  bone2Name=spine_02
  swing1LimitDegrees=45
  swing1Motion=Free
  swing2LimitDegrees=45
  swing2Motion=Free
  twistLimitDegrees=45
  twistMotion=Free
 }
 {bone1Name=def_bandolier_rear_06, bone2Name=def_bandolier_rear_05, swing1LimitDegrees=24, swing2LimitDegrees=24, twistLimitDegrees=24}
 {bone1Name=def_bandolier_rear_05, bone2Name=def_bandolier_rear_04, swing1LimitDegrees=24, swing2LimitDegrees=24, twistLimitDegrees=24}
 {
  bone1Name=def_bandolier_rear_04
  bone2Name=spine_02
  swing1LimitDegrees=45
  swing1Motion=Free
  swing2LimitDegrees=45
  swing2Motion=Free
  twistLimitDegrees=45
  twistMotion=Free
 }
 {
  bone1Name=calf_r
  bone2Name=thigh_r
  swing1LimitDegrees=45
  swing1Motion=Free
  swing2LimitDegrees=45
  swing2Motion=Free
  twistLimitDegrees=45
  twistMotion=Free
 }
 {bone1Name=dyn_bag_front_01, bone2Name=dyn_bag_rear_01, swing1LimitDegrees=45, swing2LimitDegrees=45, twistLimitDegrees=45}
 {
  bone1Name=pelvis
  bone2Name=root
  swing1LimitDegrees=45
  swing1Motion=Free
  swing2LimitDegrees=45
  swing2Motion=Free
  twistLimitDegrees=45
  twistMotion=Free
 }
 {bone1Name=def_bandolier_rear_strap, bone2Name=def_bandolier_rear_10, swing1LimitDegrees=4, swing2LimitDegrees=4, twistLimitDegrees=4}
 {bone1Name=def_bandolier_front_strap, bone2Name=def_bandolier_front_10, swing1LimitDegrees=4, swing2LimitDegrees=4, twistLimitDegrees=4}
 {bone1Name=dyn_bag_front_04, bone2Name=dyn_bag_front_03, swing1LimitDegrees=10, swing2LimitDegrees=20, twistLimitDegrees=10}
 {bone1Name=dyn_bag_rear_04, bone2Name=dyn_bag_rear_03, swing1LimitDegrees=10, swing2LimitDegrees=20, twistLimitDegrees=10}
]
