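# Control Rig for the GuineaPig character's FaceAcc part that deforms its bandolier, hip bag and neck hair to follow body motion.
# Weighted parent constraints attach bandolier control nulls and neck-hair bones to body bones.
# Spherical pose readers on upper arm, thigh, spine and head weight corrective transform offsets.
# B-spline IK fits front and rear strap chains to those controls, with optional stretch, including an IsSkydiving-driven case.
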ControlRigBlueprint /BRCosmetics/Characters/Player/Male/Medium/Bodies/M_MED_GuineaPig/Meshes/Parts/M_MED_GuineaPig_FaceAcc_CtrlRig
graphs [
 New Function_ContainedGraph {nodes [
   RigVMFunctionEntryNode()
   > &RigUnit_ParentConstraint(
    AdvancedSettings=(RotationOrderForFilter=XZY)
    Child=(Type=Bone,Name="def_bandolier_root")
    Parents="((Item=(Type=Bone,Name=\"spine_04\"),Weight=0.0035),(Item=(Type=Bone,Name=\"spine_03\"),Weight=0.0224),(Item=(Type=Bone,Name=\"spine_05\"),Weight=0.0367),(Item=(Type=Bone,Name=\"deform_upperarm_fr_l\"),Weight=0.0684),(Item=(Type=Bone,Name=\"upperarm_twist_02_l\"),We ...(truncated)"
    Weight=1
   )
   > &RigUnit_ParentConstraint(
    AdvancedSettings=(RotationOrderForFilter=XZY)
    Child=(Type=Null,Name="def_bandolier_rear_01_ctrl_null")
    Parents="((Item=(Type=Bone,Name=\"neck_02\"),Weight=0.0002),(Item=(Type=Bone,Name=\"spine_04\"),Weight=0.0005),(Item=(Type=Bone,Name=\"upperarm_l\"),Weight=0.0076),(Item=(Type=Bone,Name=\"spine_03\"),Weight=0.0409),(Item=(Type=Bone,Name=\"spine_05\"),Weight=0.0563),(Item=(Ty ...(truncated)"
    Weight=1
   )
   > &RigUnit_ParentConstraint(
    AdvancedSettings=(RotationOrderForFilter=XZY)
    Child=(Type=Null,Name="def_bandolier_rear_02_ctrl_null")
    Parents="((Item=(Type=Bone,Name=\"spine_02\"),Weight=0.0001),(Item=(Type=Bone,Name=\"upperarm_l\"),Weight=0.0053),(Item=(Type=Bone,Name=\"spine_04\"),Weight=0.0184),(Item=(Type=Bone,Name=\"upperarm_twist_02_l\"),Weight=0.0603),(Item=(Type=Bone,Name=\"spine_05\"),Weight=0.065 ...(truncated)"
    Weight=1
   )
   > &RigUnit_ParentConstraint(
    AdvancedSettings=(RotationOrderForFilter=XZY)
    Child=(Type=Null,Name="def_bandolier_rear_03_ctrl_null")
    Parents="((Item=(Type=Bone,Name=\"upperarm_twist_02_l\"),Weight=0.0025),(Item=(Type=Bone,Name=\"spine_02\"),Weight=0.0068),(Item=(Type=Bone,Name=\"upperarm_twist_01_l\"),Weight=0.0811),(Item=(Type=Bone,Name=\"spine_04\"),Weight=0.101),(Item=(Type=Bone,Name=\"spine_05\"),Weig ...(truncated)"
    Weight=1
   )
   > &RigUnit_ParentConstraint(
    AdvancedSettings=(RotationOrderForFilter=XZY)
    Child=(Type=Null,Name="def_bandolier_rear_04_ctrl_null")
    Parents="((Item=(Type=Bone,Name=\"upperarm_twist_01_l\"),Weight=0.0002),(Item=(Type=Bone,Name=\"spine_01\"),Weight=0.0098),(Item=(Type=Bone,Name=\"spine_02\"),Weight=0.0484),(Item=(Type=Bone,Name=\"spine_05\"),Weight=0.0948),(Item=(Type=Bone,Name=\"spine_04\"),Weight=0.2435) ...(truncated)"
    Weight=1
   )
   > &RigUnit_ParentConstraint(
    AdvancedSettings=(RotationOrderForFilter=XZY)
    Child=(Type=Null,Name="def_bandolier_front_01_ctrl_null")
    Parents="((Item=(Type=Bone,Name=\"upperarm_l\"),Weight=0.0043),(Item=(Type=Bone,Name=\"spine_04\"),Weight=0.0188),(Item=(Type=Bone,Name=\"spine_03\"),Weight=0.0319),(Item=(Type=Bone,Name=\"spine_05\"),Weight=0.0618),(Item=(Type=Bone,Name=\"upperarm_twist_02_l\"),Weight=0.155 ...(truncated)"
    Weight=1
   )
   > &RigUnit_ParentConstraint(
    AdvancedSettings=(RotationOrderForFilter=XZY)
    Child=(Type=Null,Name="def_bandolier_front_02_ctrl_null")
    Parents="((Item=(Type=Bone,Name=\"spine_02\"),Weight=0.0035),(Item=(Type=Bone,Name=\"upperarm_twist_02_l\"),Weight=0.0202),(Item=(Type=Bone,Name=\"spine_03\"),Weight=0.0516),(Item=(Type=Bone,Name=\"upperarm_twist_01_l\"),Weight=0.1068),(Item=(Type=Bone,Name=\"spine_04\"),Wei ...(truncated)"
    Weight=1
   )
   > &RigUnit_ParentConstraint(
    AdvancedSettings=(RotationOrderForFilter=XZY)
    Child=(Type=Null,Name="def_bandolier_front_03_ctrl_null")
    Parents="((Item=(Type=Bone,Name=\"upperarm_twist_02_l\"),Weight=0.0012),(Item=(Type=Bone,Name=\"upperarm_twist_01_l\"),Weight=0.0513),(Item=(Type=Bone,Name=\"spine_02\"),Weight=0.0584),(Item=(Type=Bone,Name=\"spine_03\"),Weight=0.159),(Item=(Type=Bone,Name=\"spine_05\"),Weig ...(truncated)"
    Weight=1
   )
   > &RigUnit_ParentConstraint(
    AdvancedSettings=(RotationOrderForFilter=XZY)
    Child=(Type=Null,Name="def_bandolier_front_04_ctrl_null")
    Parents="((Item=(Type=Bone,Name=\"spine_01\"),Weight=0.0293),(Item=(Type=Bone,Name=\"clavicle_l\"),Weight=0.0937),(Item=(Type=Bone,Name=\"spine_05\"),Weight=0.0967),(Item=(Type=Bone,Name=\"spine_02\"),Weight=0.1404),(Item=(Type=Bone,Name=\"spine_03\"),Weight=0.2929),(Item=(T ...(truncated)"
    Weight=1
   )
   > &RigUnit_ParentConstraint(
    AdvancedSettings=(RotationOrderForFilter=XZY)
    Child=(Type=Null,Name="def_bandolier_front_05_ctrl_null")
    Parents="((Item=(Type=Bone,Name=\"pelvis\"),Weight=0.0101),(Item=(Type=Bone,Name=\"clavicle_l\"),Weight=0.0234),(Item=(Type=Bone,Name=\"spine_05\"),Weight=0.0305),(Item=(Type=Bone,Name=\"spine_01\"),Weight=0.0609),(Item=(Type=Bone,Name=\"spine_02\"),Weight=0.2061),(Item=(Typ ...(truncated)"
    Weight=1
   )
   > &RigUnit_ParentConstraint(
    AdvancedSettings=(RotationOrderForFilter=XZY)
    Child=(Type=Null,Name="def_bandolier_rear_05_ctrl_null")
    Parents="((Item=(Type=Bone,Name=\"spine_05\"),Weight=0.0093),(Item=(Type=Bone,Name=\"pelvis\"),Weight=0.0437),(Item=(Type=Bone,Name=\"spine_01\"),Weight=0.0881),(Item=(Type=Bone,Name=\"clavicle_l\"),Weight=0.1054),(Item=(Type=Bone,Name=\"spine_04\"),Weight=0.1725),(Item=(Typ ...(truncated)"
    Weight=1
   )
   > RigVMFunctionReturnNode()
   &RigUnit_ParentConstraint(
    AdvancedSettings=(RotationOrderForFilter=XZY)
    Child=(Type=Null,Name="def_bandolier_rear_05_ctrl_null")
    ExecutePin=()
    Parents="((Item=(Type=Bone,Name=\"spine_05\"),Weight=0.0093),(Item=(Type=Bone,Name=\"pelvis\"),Weight=0.0437),(Item=(Type=Bone,Name=\"spine_01\"),Weight=0.0881),(Item=(Type=Bone,Name=\"clavicle_l\"),Weight=0.1054),(Item=(Type=Bone,Name=\"spine_04\"),Weight=0.1725),(Item=(Typ ...(truncated)"
    Weight=1
   )
   > &RigUnit_ParentConstraint(
    AdvancedSettings=(RotationOrderForFilter=XZY)
    Child=(Type=Null,Name="def_bandolier_rear_06_ctrl_null")
    Parents="((Item=(Type=Bone,Name=\"clavicle_l\"),Weight=0.0007),(Item=(Type=Bone,Name=\"spine_05\"),Weight=0.0027),(Item=(Type=Bone,Name=\"clavicle_r\"),Weight=0.0125),(Item=(Type=Bone,Name=\"spine_04\"),Weight=0.0943),(Item=(Type=Bone,Name=\"pelvis\"),Weight=0.1051),(Item=(T ...(truncated)"
    Weight=1
   )
   > &RigUnit_ParentConstraint(
    AdvancedSettings=(RotationOrderForFilter=XZY)
    Child=(Type=Null,Name="def_bandolier_rear_07_ctrl_null")
    Parents="((Item=(Type=Bone,Name=\"clavicle_r\"),Weight=0.0004),(Item=(Type=Bone,Name=\"spine_04\"),Weight=0.0225),(Item=(Type=Bone,Name=\"spine_01\"),Weight=0.1517),(Item=(Type=Bone,Name=\"spine_03\"),Weight=0.1541),(Item=(Type=Bone,Name=\"spine_02\"),Weight=0.1966),(Item=(T ...(truncated)"
    Weight=1
   )
   > &RigUnit_ParentConstraint(
    AdvancedSettings=(RotationOrderForFilter=XZY)
    Child=(Type=Null,Name="def_bandolier_rear_08_ctrl_null")
    Parents="((Item=(Type=Bone,Name=\"thigh_twist_01_l\"),Weight=0.0011),(Item=(Type=Bone,Name=\"spine_03\"),Weight=0.0073),(Item=(Type=Bone,Name=\"spine_01\"),Weight=0.0109),(Item=(Type=Bone,Name=\"spine_02\"),Weight=0.0114),(Item=(Type=Bone,Name=\"thigh_twist_01_r\"),Weight=0. ...(truncated)"
    Weight=1
   )
   > &RigUnit_ParentConstraint(
    AdvancedSettings=(RotationOrderForFilter=XZY)
    Child=(Type=Null,Name="def_bandolier_rear_09_ctrl_null")
    Parents=((Item=(Type=Bone,Name="spine_03"),Weight=0.0008),(Item=(Type=Bone,Name="spine_01"),Weight=0.0012),(Item=(Type=Bone,Name="spine_02"),Weight=0.0015),(Item=(Type=Bone,Name="thigh_twist_01_r"),Weight=0.0566),(Item=(Type=Bone,Name="pelvis"),Weight=0.9399))
    Weight=1
   )
   > &RigUnit_ParentConstraint(
    AdvancedSettings=(RotationOrderForFilter=XZY)
    Child=(Type=Null,Name="def_bandolier_rear_10_ctrl_null")
    Parents=((Item=(Type=Bone,Name="thigh_twist_01_l"),Weight=0.0005),(Item=(Type=Bone,Name="thigh_r"),Weight=0.0018),(Item=(Type=Bone,Name="pelvis"),Weight=0.3346),(Item=(Type=Bone,Name="thigh_twist_01_r"),Weight=0.6632))
    Weight=1
   )
   > &RigUnit_ParentConstraint(
    AdvancedSettings=(RotationOrderForFilter=XZY)
    Child=(Type=Null,Name="def_bandolier_front_06_ctrl_null")
    Parents=((Item=(Type=Bone,Name="spine_04"),Weight=0.0425),(Item=(Type=Bone,Name="pelvis"),Weight=0.1514),(Item=(Type=Bone,Name="spine_01"),Weight=0.1939),(Item=(Type=Bone,Name="spine_03"),Weight=0.2742),(Item=(Type=Bone,Name="spine_02"),Weight=0.3379))
    Weight=1
   )
   > &RigUnit_ParentConstraint(
    AdvancedSettings=(RotationOrderForFilter=XZY)
    Child=(Type=Null,Name="def_bandolier_front_07_ctrl_null")
    Parents=((Item=(Type=Bone,Name="spine_04"),Weight=0.0187),(Item=(Type=Bone,Name="spine_03"),Weight=0.1406),(Item=(Type=Bone,Name="spine_01"),Weight=0.1799),(Item=(Type=Bone,Name="spine_02"),Weight=0.2253),(Item=(Type=Bone,Name="pelvis"),Weight=0.4355))
    Weight=1
   )
   > &RigUnit_ParentConstraint(
    AdvancedSettings=(RotationOrderForFilter=XZY)
    Child=(Type=Null,Name="def_bandolier_front_08_ctrl_null")
    Parents="((Item=(Type=Bone,Name=\"spine_04\"),Weight=0.0023),(Item=(Type=Bone,Name=\"thigh_twist_01_r\"),Weight=0.0051),(Item=(Type=Bone,Name=\"spine_03\"),Weight=0.0572),(Item=(Type=Bone,Name=\"spine_01\"),Weight=0.0828),(Item=(Type=Bone,Name=\"spine_02\"),Weight=0.0972),(I ...(truncated)"
    Weight=1
   )
   > &RigUnit_ParentConstraint(
    AdvancedSettings=(RotationOrderForFilter=XZY)
    Child=(Type=Null,Name="def_bandolier_front_09_ctrl_null")
    Parents="((Item=(Type=Bone,Name=\"thigh_twist_01_l\"),Weight=0.0009),(Item=(Type=Bone,Name=\"spine_03\"),Weight=0.008),(Item=(Type=Bone,Name=\"spine_01\"),Weight=0.0125),(Item=(Type=Bone,Name=\"spine_02\"),Weight=0.0158),(Item=(Type=Bone,Name=\"thigh_twist_01_r\"),Weight=0.1 ...(truncated)"
    Weight=1
   )
   > &RigUnit_ParentConstraint(
    AdvancedSettings=(RotationOrderForFilter=XZY)
    Child=(Type=Null,Name="def_bandolier_front_10_ctrl_null")
    Parents="((Item=(Type=Bone,Name=\"spine_01\"),Weight=0.0001),(Item=(Type=Bone,Name=\"spine_02\"),Weight=0.0002),(Item=(Type=Bone,Name=\"thigh_twist_01_l\"),Weight=0.0006),(Item=(Type=Bone,Name=\"thigh_r\"),Weight=0.0022),(Item=(Type=Bone,Name=\"pelvis\"),Weight=0.4785),(Item ...(truncated)"
    Weight=1
   )
  ]}
 New Function_ContainedGraph {nodes [
   RigVMFunctionEntryNode()
   > RigVMFunction_Sequence = RigVMAggregateNode()
   n1 = &RigUnit_SphericalPoseReader(
    ActiveRegionScaleFactors=(PositiveWidth=1,NegativeWidth=1,PositiveHeight=1,NegativeHeight=1)
    ActiveRegionSize=0.1
    Debug=(bDrawDebug=false,DebugScale=25,DebugSegments=20,DebugThickness=0.25)
    DriverAxis=(X=1,Y=0,Z=0)
    DriverItem=(Type=Bone,Name="upperarm_l")
    FalloffRegionScaleFactors=(PositiveWidth=1,NegativeWidth=1,PositiveHeight=1,NegativeHeight=1)
    FalloffSize=0.3
    OptionalParentItem=(Type=Bone,Name="spine_05")
    RotationOffset=(X=7.0096,Y=-141.722,Z=16.681)
    ExecutePin=@RigVMFunction_Sequence.A
   )
   > &RigUnit_ModifyTransforms(
    ItemToModify="((Item=(Type=Null,Name=\"def_bandolier_rear_01_ctrl_null\"),Transform=(Rotation=(X=0.1305,Y=0,Z=0,W=0.9914),Translation=(X=2,Y=1,Z=-4),Scale3D=(X=1,Y=1,Z=1))),(Item=(Type=Null,Name=\"def_bandolier_rear_02_ctrl_null\"),Transform=(Rotation=(X=0,Y=0,Z=0,W=1),Tran ...(truncated)"
    WeightMinimum=0
    Weight=@n1.OutputParam
   )
   > n2 = &RigUnit_SphericalPoseReader(
    ActiveRegionScaleFactors=(PositiveWidth=2.5,NegativeWidth=3,PositiveHeight=3,NegativeHeight=3)
    ActiveRegionSize=0.1
    Debug=(bDrawDebug=false,DebugScale=25,DebugSegments=20,DebugThickness=0.25)
    DriverAxis=(X=1,Y=0,Z=0)
    DriverItem=(Type=Bone,Name="upperarm_l")
    FalloffRegionScaleFactors=(PositiveWidth=1,NegativeWidth=1,PositiveHeight=1,NegativeHeight=1)
    FalloffSize=0.5
    OptionalParentItem=(Type=Bone,Name="clavicle_l")
    RotationOffset=(X=0,Y=0,Z=0)
   )
   > &RigUnit_ModifyTransforms(
    ItemToModify="((Item=(Type=Null,Name=\"def_bandolier_front_01_ctrl_null\"),Transform=(Rotation=(X=0.2588,Y=0,Z=0,W=0.9659),Translation=(X=-1,Y=4,Z=3),Scale3D=(X=1,Y=1,Z=1))),(Item=(Type=Null,Name=\"def_bandolier_rear_01_ctrl_null\"),Transform=(Rotation=(X=0,Y=0.0436,Z=0,W=0 ...(truncated)"
    WeightMinimum=0
    Weight=@n2.OutputParam
   )
   > n3 = &RigUnit_SphericalPoseReader(
    ActiveRegionScaleFactors=(PositiveWidth=0.5,NegativeWidth=3,PositiveHeight=1.5,NegativeHeight=1.5)
    ActiveRegionSize=0.1
    Debug=(bDrawDebug=false,DebugScale=25,DebugSegments=20,DebugThickness=0.25)
    DriverAxis=(X=1,Y=0,Z=0)
    DriverItem=(Type=Bone,Name="upperarm_l")
    FalloffRegionScaleFactors=(PositiveWidth=1,NegativeWidth=1,PositiveHeight=1,NegativeHeight=1)
    FalloffSize=0.4
    OptionalParentItem=(Type=Bone,Name="clavicle_l")
    RotationOffset=(X=90,Y=0,Z=0)
   )
   RigVMFunction_AnimEvalRichCurve = &RigVMFunction_AnimEvalRichCurve(Curve=(EditorCurveData=(Keys=((),(Time=0.6954,Value=0.7921),(Time=1,Value=1)))), SourceMinimum=0, TargetMinimum=0, Value=@n3.OutputParam)
   > &RigUnit_ModifyTransforms(
    ItemToModify="((Item=(Type=Null,Name=\"def_bandolier_rear_01_ctrl_null\"),Transform=(Rotation=(X=0.2719,Y=0.2511,Z=0.1141,W=0.922),Translation=(X=0.0501,Y=1.3975,Z=-3.7365),Scale3D=(X=1,Y=1,Z=1))),(Item=(Type=Null,Name=\"def_bandolier_rear_02_ctrl_null\"),Transform=(Rotatio ...(truncated)"
    WeightMinimum=0
    Weight=@RigVMFunction_AnimEvalRichCurve.Result
   )
   > n4 = &RigUnit_SphericalPoseReader(
    ActiveRegionScaleFactors=(PositiveWidth=1,NegativeWidth=3.5,PositiveHeight=1.5,NegativeHeight=1.5)
    ActiveRegionSize=0.1
    Debug=(bDrawDebug=true,DebugScale=25,DebugSegments=20,DebugThickness=0.25)
    DriverAxis=(X=1,Y=0,Z=0)
    DriverItem=(Type=Bone,Name="thigh_r")
    FalloffRegionScaleFactors=(PositiveWidth=1,NegativeWidth=1,PositiveHeight=1,NegativeHeight=1)
    FalloffSize=0.35
    OptionalParentItem=(Type=Bone,Name="pelvis")
    RotationOffset=(X=0,Y=-27,Z=0)
   )
   > &RigUnit_ModifyTransforms(
    ItemToModify=((Item=(Type=Bone,Name="dyn_bag_sim_chain_03"),Transform=(Rotation=(X=0,Y=-0.3827,Z=0,W=0.9239),Translation=(X=0,Y=0,Z=4),Scale3D=(X=1,Y=1,Z=1))))
    WeightMinimum=0
    Weight=@n4.OutputParam
   )
   > n5 = &RigUnit_SphericalPoseReader(
    ActiveRegionScaleFactors=(PositiveWidth=1,NegativeWidth=1,PositiveHeight=1,NegativeHeight=1)
    ActiveRegionSize=0.1
    Debug=(bDrawDebug=true,DebugScale=25,DebugSegments=20,DebugThickness=0.25)
    DriverAxis=(X=1,Y=0,Z=0)
    DriverItem=(Type=Bone,Name="spine_05")
    FalloffRegionScaleFactors=(PositiveWidth=1,NegativeWidth=1,PositiveHeight=1,NegativeHeight=1)
    FalloffSize=0.24
    OptionalParentItem=(Type=Bone,Name="pelvis")
    RotationOffset=(X=90,Y=0,Z=-20)
   )
   > &RigUnit_ModifyTransforms(
    ItemToModify="((Item=(Type=Bone,Name=\"dyn_bag_sim_chain_03\"),Transform=(Rotation=(X=0,Y=0,Z=0.0436,W=0.999),Translation=(X=0,Y=3,Z=0),Scale3D=(X=1,Y=1,Z=1))),(Item=(Type=Null,Name=\"def_bandolier_rear_04_ctrl_null\"),Transform=(Rotation=(X=0,Y=0,Z=-0.0436,W=0.999),Transla ...(truncated)"
    WeightMinimum=0
    Weight=@n5.OutputParam
   )
   > &RigUnit_ModifyTransforms(
    ItemToModify=((Item=(Type=Bone,Name="dyn_bag_sim_chain_03"),Transform=(Rotation=(X=0,Y=0,Z=0,W=1),Translation=(X=0,Y=0,Z=3),Scale3D=(X=1,Y=1,Z=1))))
    Weight=1
    WeightMinimum=0
   )
   > &RigUnit_ParentConstraint(
    AdvancedSettings=(InterpolationType=Shortest,RotationOrderForFilter=XYZ)
    Child=(Type=Null,Name="dyn_bag_sim_chain_03_ctrl_null")
    Filter=(TranslationFilter=(bX=True,bY=True,bZ=True),RotationFilter=(bX=true,bY=true,bZ=true),ScaleFilter=(bX=True,bY=True,bZ=True))
    Parents=((Item=(Type=Bone,Name="thigh_r"),Weight=1),(Item=(Type=Bone,Name="pelvis"),Weight=1))
    Weight=1
   )
   > SplineIK_2 = RigVMFunctionReferenceNode(
    Primary Axis=(X=1,Y=0,Z=0)
    Samples Per Segment=6
    Secondary Spline Direction=(X=0,Y=-1,Z=0)
    Spline Mode=BSpline
    Up Axis=(X=0,Y=-1,Z=0)
    debug=false
    Stretch=RigVMVariableNode() {variable=IsSkydiving}.Value
    Bones=RigVMVariableNode() {variable=bandolier_rear_joints_RBAN}.Value
    Controls=RigVMVariableNode() {variable=bandolier_rear_ctrls_RBAN}.Value
   )
   > SplineIK_1_1 = RigVMFunctionReferenceNode(
    Primary Axis=(X=1,Y=0,Z=0)
    Samples Per Segment=6
    Secondary Spline Direction=(X=0,Y=1,Z=0)
    Spline Mode=BSpline
    Stretch=false
    Up Axis=(X=0,Y=1,Z=0)
    debug=false
    Bones=RigVMVariableNode() {variable=bandolier_front_joints_RBAN}.Value
    Controls=RigVMVariableNode() {variable=bandolier_front_ctrls_RBAN}.Value
   )
   RigVMFunctionReturnNode(ExecuteContext=@RigVMFunction_Sequence.C)
   RigVMFunction_MathBoolToFloat = &RigVMFunction_MathBoolToFloat(Value=RigVMVariableNode() {variable=crouch_pistol}.Value)
   AlphaInterp = &RigVMFunction_AlphaInterp(InterpSpeedDecreasing=10, InterpSpeedIncreasing=10, Scale=1, bClampResult=False, Value=@RigVMFunction_MathBoolToFloat.Result)
   &RigUnit_ModifyTransforms(
    ExecutePin=()
    ItemToModify=((Item=(Type=Bone,Name="def_bandolier_root"),Transform=(Rotation=(X=0,Y=0.1079,Z=0,W=0.9942),Translation=(X=0.7275,Y=1.4243,Z=1.141),Scale3D=(X=1,Y=1,Z=1))))
    WeightMaximum=0
    WeightMinimum=0
    Weight=@AlphaInterp.Result
   )
   /Script/ControlRigSpline.RigUnit_PositionFromControlRigSpline(U=0, Spline=@SplineIK_2.Output Spline) {is=struct RigVMUnitNode}
   /Script/ControlRigSpline.RigUnit_TangentFromControlRigSpline(U=0, Spline=@SplineIK_2.Output Spline) {is=struct RigVMUnitNode}
  ]}
 New Function_ContainedGraph {nodes [
   RigVMFunctionEntryNode()
   > RigVMVariableNode(Value=(bVisible=True,Name="Default",Color=(R=1,G=1,B=0,A=1),Transform=(Rotation=(X=0,Y=0,Z=0,W=1),Translation=(X=0,Y=0,Z=0),Scale3D=(X=0.5,Y=0.5,Z=0.5)))) {variable=ControlSettings}
   > RigVMFunction_Sequence = RigVMAggregateNode()
   DISPATCH_RigVMDispatch_ArrayIterator = RigVMDispatchNode(ExecuteContext=@RigVMFunction_Sequence.A, Array=RigVMVariableNode() {variable=bandolier_front_ctrls}.Value) {template="DISPATCH_RigVMDispatch_ArrayIterator(in Array,out Element,out Index,out Count,out Ratio)"}
   n1 = RigVMVariableNode() {variable=ControlSettings}
   > /Script/ControlRig.RigUnit_HierarchySetShapeSettings(Item=@DISPATCH_RigVMDispatch_ArrayIterator.Element, Settings=@n1.Value) {is=struct RigVMUnitNode}
   n2 = RigVMDispatchNode(ExecuteContext=@RigVMFunction_Sequence.B, Array=RigVMVariableNode() {variable=bandolier_rear_ctrls}.Value) {template="DISPATCH_RigVMDispatch_ArrayIterator(in Array,out Element,out Index,out Count,out Ratio)"}
   > &RigUnit_HierarchySetShapeSettings(Item=@n2.Element, Settings=@n1.Value)
   RigVMFunctionReturnNode(ExecuteContext=@RigVMFunction_Sequence.D)
  ]}
 New Function_ContainedGraph {nodes [
   n1 = RigVMDispatchNode(ExecuteContext=(), Array=RigVMVariableNode() {variable=BagCtrls}.Value) {template="DISPATCH_RigVMDispatch_ArrayIterator(in Array,out Element,out Index,out Count,out Ratio)"}
   > &RigUnit_ModifyTransforms(
    ItemToModify=((Item=(Type=Bone,Name="None"),Transform=(Rotation=(X=0,Y=0,Z=0,W=1),Translation=(X=0,Y=0,Z=0),Scale3D=(X=1,Y=1,Z=1))))
    Mode=OverrideLocal
    Weight=1
    WeightMinimum=0
    ItemToModify.0.Item=@n1.Element
   )
   RigVMFunctionEntryNode()
   > RigVMFunction_Sequence = RigVMAggregateNode()
   &RigUnit_ModifyTransforms(
    ItemToModify=((Item=(Type=Control,Name="dyn_bag_sim_chain_03_ctrl"),Transform=(Rotation=(X=0,Y=-0.232,Z=0,W=0.9727),Translation=(X=3.3723,Y=0,Z=7.6903),Scale3D=(X=1,Y=1,Z=1))))
    Weight=1
    WeightMinimum=0
    ExecutePin=@RigVMFunction_Sequence.C
   )
   > RigVMVariableNode(Value=/Script/ControlRig.RigUnit_HierarchyGetPoseItemArray(ElementType=Control, Initial=false, ItemsToGet=RigVMVariableNode() {variable=BagCtrls}.Value) {is=struct RigVMUnitNode}.Pose) {variable=BagThighSideRPose}
   &RigUnit_ModifyTransforms(
    ItemToModify=((Item=(Type=Control,Name="dyn_bag_sim_chain_03_ctrl"),Transform=(Rotation=(X=0,Y=-0.2784,Z=0,W=0.9605),Translation=(X=0,Y=0,Z=4.8207),Scale3D=(X=1,Y=1,Z=1))))
    Weight=1
    WeightMinimum=0
    ExecutePin=@RigVMFunction_Sequence.F
   )
   > RigVMVariableNode(Value=&RigUnit_HierarchyGetPoseItemArray(ElementType=Control, Initial=False, ItemsToGet=RigVMVariableNode() {variable=BagCtrls}.Value).Pose) {variable=HierarchyGetPoseItemArray_3_1_Pose}
   RigVMFunctionReturnNode(ExecuteContext=@RigVMFunction_Sequence.G)
   RigVMVariableNode(
    Value=&RigUnit_HierarchyGetPoseItemArray(ElementType=Control, Initial=true, ItemsToGet=RigVMVariableNode() {variable=BagCtrls}.Value).Pose
    ExecuteContext=@RigVMFunction_Sequence.A
   ) {variable=BagBindPose}
   &RigUnit_ModifyTransforms(
    ExecutePin=()
    ItemToModify=((Item=(Type=Control,Name="dyn_bag_ctrl"),Transform=(Rotation=(X=0,Y=0,Z=0,W=1),Translation=(X=0,Y=0,Z=0),Scale3D=(X=1,Y=1,Z=1))))
    Weight=1
    WeightMinimum=0
   )
   > RigVMVariableNode(Value=&RigUnit_HierarchyGetPoseItemArray(ElementType=Control, Initial=False, ItemsToGet=RigVMVariableNode() {variable=BagCtrls}.Value).Pose) {variable=BagThighBackRPose}
   > n2 = RigVMDispatchNode(Array=RigVMVariableNode() {variable=BagCtrls}.Value) {template="DISPATCH_RigVMDispatch_ArrayIterator(in Array,out Element,out Index,out Count,out Ratio)"}
   > &RigUnit_ModifyTransforms(
    ItemToModify=((Item=(Type=Bone,Name="None"),Transform=(Rotation=(X=0,Y=0,Z=0,W=1),Translation=(X=0,Y=0,Z=0),Scale3D=(X=1,Y=1,Z=1))))
    Weight=1
    WeightMinimum=0
    ItemToModify.0.Item=@n2.Element
   )
   n3 = RigVMRerouteNode(Value=())
   &RigUnit_ModifyTransforms(
    ItemToModify=((Item=(Type=Control,Name="dyn_bag_ctrl"),Transform=(Rotation=(X=0,Y=0,Z=0,W=1),Translation=(X=0,Y=-1.114,Z=0),Scale3D=(X=1,Y=1,Z=1))))
    Weight=1
    WeightMinimum=0
    ExecutePin=@n3.Value
   )
   > RigVMVariableNode(Value=&RigUnit_HierarchyGetPoseItemArray(ElementType=Control, Initial=False, ItemsToGet=RigVMVariableNode() {variable=BagCtrls}.Value).Pose) {variable=BagThighUpRPose}
   > DISPATCH_RigVMDispatch_ArrayIterator = RigVMDispatchNode(Array=RigVMVariableNode() {variable=BagCtrls}.Value) {template="DISPATCH_RigVMDispatch_ArrayIterator(in Array,out Element,out Index,out Count,out Ratio)"}
   > &RigUnit_ModifyTransforms(
    ItemToModify=((Item=(Type=Bone,Name="None"),Transform=(Rotation=(X=0,Y=0,Z=0,W=1),Translation=(X=0,Y=0,Z=0),Scale3D=(X=1,Y=1,Z=1))))
    Weight=1
    WeightMinimum=0
    ItemToModify.0.Item=@DISPATCH_RigVMDispatch_ArrayIterator.Element
   )
  ]}
 New Function_ContainedGraph {nodes [
   n1 = RigVMFunctionEntryNode()
   > RigVMVariableNode(Value=@n1.controls) {variable=local_controls}
   > RigVMVariableNode(Value=@n1.bones) {variable=local_bones}
   > RigVMVariableNode(Value=@n1.primary axis) {variable=local_primary_axis}
   > RigVMVariableNode(Value=@n1.samples) {variable=local_samples}
   > RigVMVariableNode(Value=@n1.stretch) {variable=local_stretch}
   > RigVMVariableNode(Value=@n1.debug draw spline) {variable=local_debug_draw}
   > RigVMVariableNode(Value=@n1.max stretch) {variable=local_max stretch}
   > RigVMVariableNode(Value=@n1.min stretch) {variable=local_min stretch}
   > DISPATCH_RigVMDispatch_ArrayReset = RigVMDispatchNode(Array=RigVMVariableNode() {variable=points}.Value) {template="DISPATCH_RigVMDispatch_ArrayReset(io Array)"}
   > DISPATCH_RigVMDispatch_ArrayIterator = RigVMDispatchNode(Array=RigVMVariableNode() {variable=local_controls}.Value) {template="DISPATCH_RigVMDispatch_ArrayIterator(in Array,out Element,out Index,out Count,out Ratio)"}
   > DISPATCH_RigVMDispatch_ArrayAdd = RigVMDispatchNode(
    Element=&RigUnit_GetTransform(Item=@DISPATCH_RigVMDispatch_ArrayIterator.Element).Transform.Translation
    Array=RigVMVariableNode() {variable=points}.Value
   ) {template="DISPATCH_RigVMDispatch_ArrayAdd(io Array,in Element,out Index)"}
   RigVMVariableNode(
    Value=/Script/ControlRigSpline.RigUnit_ControlRigSplineFromPoints(SplineMode=BSpline, Points=RigVMVariableNode() {variable=points}.Value, SamplesPerSegment=RigVMVariableNode() {variable=local_samples}.Value) {is=struct RigVMUnitNode}.Spline
    ExecuteContext=@DISPATCH_RigVMDispatch_ArrayIterator.Completed
   ) {variable=spline}
   > SequenceAggregate = RigVMAggregateNode()
   RigVMFunction_ControlFlowBranch = &RigVMFunction_ControlFlowBranch(ExecuteContext=@SequenceAggregate.A, Condition=RigVMVariableNode() {variable=local_stretch}.Value)
   /Script/ControlRigSpline.RigUnit_FitChainToSplineCurveItemArray(
    Alignment=Front
    DebugSettings=(Scale=1,CurveColor=(R=1,G=1,B=0,A=1),SegmentsColor=(R=1,G=0,B=0,A=1),WorldOffset=(Rotation=(X=0,Y=0,Z=0,W=1),Translation=(X=0,Y=0,Z=0),Scale3D=(X=1,Y=1,Z=1)))
    Maximum=1
    PoleVectorPosition=(X=0,Y=0,Z=0)
    RotationEaseType=CubicEaseInOut
    Rotations=()
    SamplingPrecision=12
    SecondaryAxis=(X=0,Y=0,Z=0)
    Weight=1
    bPropagateToChildren=True
    ExecutePin=@RigVMFunction_ControlFlowBranch.False
    Spline=RigVMVariableNode() {variable=spline}.Value
    PrimaryAxis=RigVMVariableNode() {variable=local_primary_axis}.Value
    Items=RigVMVariableNode() {variable=local_bones}.Value
   ) {is=struct RigVMUnitNode}
   &RigUnit_FitChainToSplineCurveItemArray(
    Alignment=Stretched
    DebugSettings=(Scale=1,CurveColor=(R=1,G=1,B=0,A=1),SegmentsColor=(R=1,G=0,B=0,A=1),WorldOffset=(Rotation=(X=0,Y=0,Z=0,W=1),Translation=(X=0,Y=0,Z=0),Scale3D=(X=1,Y=1,Z=1)))
    Maximum=1.5
    PoleVectorPosition=(X=0,Y=0,Z=0)
    Rotations=()
    SamplingPrecision=12
    SecondaryAxis=(X=0,Y=0,Z=0)
    Weight=1
    bPropagateToChildren=True
    ExecutePin=@RigVMFunction_ControlFlowBranch.True
    Spline=RigVMVariableNode() {variable=spline}.Value
    PrimaryAxis=RigVMVariableNode() {variable=local_primary_axis}.Value
    Items=RigVMVariableNode() {variable=local_bones}.Value
   )
   n2 = &RigVMFunction_ControlFlowBranch(ExecuteContext=@SequenceAggregate.B, Condition=RigVMVariableNode() {variable=local_debug_draw}.Value)
   /Script/ControlRigSpline.RigUnit_DrawControlRigSpline(Color=(R=1,G=0,B=0,A=1), Detail=16, Thickness=1, ExecutePin=@n2.True, Spline=RigVMVariableNode() {variable=spline}.Value) {is=struct RigVMUnitNode}
   RigVMFunctionReturnNode(ExecuteContext=@SequenceAggregate.C)
   RigVMVariableNode() {variable=local_max stretch}
   RigVMVariableNode() {variable=local_min stretch}
  ]}
 New Function_ContainedGraph {nodes [
   RigVMFunctionEntryNode()
   > RigVMFunction_Sequence = RigVMAggregateNode()
   &RigUnit_ParentConstraint(
    AdvancedSettings=(RotationOrderForFilter=XZY)
    Child=(Type=Bone,Name="deform_neck_hair_a_l_MSTR")
    Parents="((Item=(Type=Bone,Name=\"def_neck_fur_front\"),Weight=0.0044),(Item=(Type=Bone,Name=\"clavicle_l\"),Weight=0.0144),(Item=(Type=Bone,Name=\"teeth_lower\"),Weight=0.0175),(Item=(Type=Bone,Name=\"neck_01\"),Weight=0.0177),(Item=(Type=Bone,Name=\"neck_02\"),Weight=0.049 ...(truncated)"
    Weight=1
    ExecutePin=@RigVMFunction_Sequence.A
   )
   > &RigUnit_ParentConstraint(
    AdvancedSettings=(RotationOrderForFilter=XZY)
    Child=(Type=Bone,Name="deform_neck_hair_b_l_MSTR")
    Parents="((Item=(Type=Bone,Name=\"teeth_lower\"),Weight=0.0041),(Item=(Type=Bone,Name=\"spine_04\"),Weight=0.0049),(Item=(Type=Bone,Name=\"upperarm_twist_01_l\"),Weight=0.0062),(Item=(Type=Bone,Name=\"neck_01\"),Weight=0.0423),(Item=(Type=Bone,Name=\"clavicle_l\"),Weight=0.0 ...(truncated)"
    Weight=1
   )
   > &RigUnit_ParentConstraint(
    AdvancedSettings=(RotationOrderForFilter=XZY)
    Child=(Type=Bone,Name="deform_neck_hair_c_l_MSTR")
    Parents="((Item=(Type=Bone,Name=\"spine_03\"),Weight=0.0011),(Item=(Type=Bone,Name=\"spine_04\"),Weight=0.0033),(Item=(Type=Bone,Name=\"upperarm_twist_01_l\"),Weight=0.0058),(Item=(Type=Bone,Name=\"neck_01\"),Weight=0.0319),(Item=(Type=Bone,Name=\"head\"),Weight=0.0329),(Ite ...(truncated)"
    Weight=1
   )
   > &RigUnit_ParentConstraint(
    AdvancedSettings=(RotationOrderForFilter=XZY)
    Child=(Type=Bone,Name="deform_neck_hair_d_l_MSTR")
    Parents="((Item=(Type=Bone,Name=\"spine_03\"),Weight=0.0011),(Item=(Type=Bone,Name=\"head\"),Weight=0.0231),(Item=(Type=Bone,Name=\"clavicle_l\"),Weight=0.0722),(Item=(Type=Bone,Name=\"neck_01\"),Weight=0.0743),(Item=(Type=Bone,Name=\"neck_02\"),Weight=0.1696),(Item=(Type=Bo ...(truncated)"
    Weight=1
   )
   > &RigUnit_ParentConstraint(
    AdvancedSettings=(RotationOrderForFilter=XZY)
    Child=(Type=Bone,Name="deform_neck_hair_rear_MSTR")
    Parents="((Item=(Type=Bone,Name=\"spine_03\"),Weight=0.0004),(Item=(Type=Bone,Name=\"head\"),Weight=0.0179),(Item=(Type=Bone,Name=\"clavicle_r\"),Weight=0.0421),(Item=(Type=Bone,Name=\"neck_01\"),Weight=0.0568),(Item=(Type=Bone,Name=\"neck_02\"),Weight=0.127),(Item=(Type=Bon ...(truncated)"
    Weight=1
   )
   > &RigUnit_ParentConstraint(
    AdvancedSettings=(RotationOrderForFilter=XZY)
    Child=(Type=Bone,Name="deform_neck_hair_a_r_MSTR")
    Parents="((Item=(Type=Bone,Name=\"clavicle_r\"),Weight=0.0001),(Item=(Type=Bone,Name=\"def_neck_fur_front\"),Weight=0.0001),(Item=(Type=Bone,Name=\"teeth_lower\"),Weight=0.0027),(Item=(Type=Bone,Name=\"neck_01\"),Weight=0.0144),(Item=(Type=Bone,Name=\"neck_02\"),Weight=0.052 ...(truncated)"
    Weight=1
   )
   > &RigUnit_ParentConstraint(
    AdvancedSettings=(RotationOrderForFilter=XZY)
    Child=(Type=Bone,Name="deform_neck_hair_b_r_MSTR")
    Parents="((Item=(Type=Bone,Name=\"spine_03\"),Weight=0.0005),(Item=(Type=Bone,Name=\"spine_04\"),Weight=0.0017),(Item=(Type=Bone,Name=\"teeth_lower\"),Weight=0.0026),(Item=(Type=Bone,Name=\"neck_01\"),Weight=0.0681),(Item=(Type=Bone,Name=\"clavicle_r\"),Weight=0.0721),(Item= ...(truncated)"
    Weight=1
   )
   > &RigUnit_ParentConstraint(
    AdvancedSettings=(RotationOrderForFilter=XZY)
    Child=(Type=Bone,Name="deform_neck_hair_c_r_MSTR")
    Parents="((Item=(Type=Bone,Name=\"deform_pec_r\"),Weight=0.0002),(Item=(Type=Bone,Name=\"spine_03\"),Weight=0.0006),(Item=(Type=Bone,Name=\"spine_04\"),Weight=0.0011),(Item=(Type=Bone,Name=\"head\"),Weight=0.0148),(Item=(Type=Bone,Name=\"neck_01\"),Weight=0.028),(Item=(Type= ...(truncated)"
    Weight=1
   )
   > &RigUnit_ParentConstraint(
    AdvancedSettings=(RotationOrderForFilter=XZY)
    Child=(Type=Bone,Name="deform_neck_hair_d_r_MSTR")
    Parents="((Item=(Type=Bone,Name=\"spine_03\"),Weight=0.0008),(Item=(Type=Bone,Name=\"head\"),Weight=0.023),(Item=(Type=Bone,Name=\"clavicle_r\"),Weight=0.0723),(Item=(Type=Bone,Name=\"neck_01\"),Weight=0.0742),(Item=(Type=Bone,Name=\"neck_02\"),Weight=0.1696),(Item=(Type=Bon ...(truncated)"
    Weight=1
   )
   n1 = &RigUnit_SphericalPoseReader(
    ActiveRegionScaleFactors=(PositiveWidth=2,NegativeWidth=2,PositiveHeight=1,NegativeHeight=3.5)
    ActiveRegionSize=0.1
    Debug=(bDrawDebug=false,bDraw2D=False,bDrawLocalAxes=False,DebugScale=25,DebugSegments=20,DebugThickness=0.25)
    DriverAxis=(X=0,Y=1,Z=0)
    DriverItem=(Type=Bone,Name="head")
    FalloffRegionScaleFactors=(PositiveWidth=1,NegativeWidth=1,PositiveHeight=1,NegativeHeight=1)
    FalloffSize=0.35
    FlipHeightScaling=False
    FlipWidthScaling=False
    OptionalParentItem=(Type=Bone,Name="spine_05")
    RotationOffset=(X=20,Y=180,Z=0)
    ExecutePin=@RigVMFunction_Sequence.B
   )
   > n2 = &RigUnit_SphericalPoseReader(
    ActiveRegionScaleFactors=(PositiveWidth=2,NegativeWidth=2,PositiveHeight=1,NegativeHeight=3.5)
    ActiveRegionSize=0.1
    Debug=(bDrawDebug=false,bDraw2D=False,bDrawLocalAxes=False,DebugScale=25,DebugSegments=20,DebugThickness=0.25)
    DriverAxis=(X=0,Y=1,Z=0)
    DriverItem=(Type=Bone,Name="head")
    FalloffRegionScaleFactors=(PositiveWidth=1,NegativeWidth=1,PositiveHeight=1,NegativeHeight=1)
    FalloffSize=0.35
    FlipHeightScaling=False
    FlipWidthScaling=False
    OptionalParentItem=(Type=Bone,Name="None")
    RotationOffset=(X=20,Y=0,Z=0)
   )
   > &RigUnit_ModifyTransforms(
    ItemToModify="((Item=(Type=Bone,Name=\"deform_neck_hair_a_l_MSTR\"),Transform=(Rotation=(X=0,Y=0,Z=0,W=1),Translation=(X=0,Y=0,Z=0),Scale3D=(X=1,Y=1,Z=1))),(Item=(Type=Bone,Name=\"deform_neck_hair_b_l_MSTR\"),Transform=(Rotation=(X=0,Y=0,Z=0,W=1),Translation=(X=0,Y=0,Z=0),S ...(truncated)"
    WeightMinimum=0
    Weight=@n1.OutputParam
   )
   > &RigUnit_ModifyTransforms(
    ItemToModify="((Item=(Type=Bone,Name=\"deform_neck_hair_a_l_MSTR\"),Transform=(Rotation=(X=0,Y=0,Z=0,W=1),Translation=(X=0,Y=0,Z=0),Scale3D=(X=1,Y=1,Z=1))),(Item=(Type=Bone,Name=\"deform_neck_hair_b_l_MSTR\"),Transform=(Rotation=(X=0,Y=0,Z=0,W=1),Translation=(X=0,Y=0,Z=0),S ...(truncated)"
    WeightMinimum=0
    Weight=@n2.OutputParam
   )
   RigVMFunctionReturnNode(ExecuteContext=@RigVMFunction_Sequence.C)
  ]}
 New Function_ContainedGraph {nodes [
   n1 = RigVMFunctionEntryNode()
   RigVMFunction_Sequence = &RigVMFunction_Sequence(ExecuteContext=@n1.ExecuteContext)
   n2 = &RigUnit_SphericalPoseReader(
    ActiveRegionScaleFactors=(PositiveWidth=1,NegativeWidth=1,PositiveHeight=1,NegativeHeight=1)
    ActiveRegionSize=0.1
    Debug=(bDrawDebug=false,DebugScale=25,DebugSegments=20,DebugThickness=0.25)
    DriverAxis=(X=1,Y=0,Z=0)
    DriverItem=(Type=Bone,Name="upperarm_l")
    FalloffRegionScaleFactors=(PositiveWidth=1,NegativeWidth=1,PositiveHeight=1,NegativeHeight=1)
    FalloffSize=0.35
    OptionalParentItem=(Type=Bone,Name="spine_05")
    RotationOffset=(X=7.0096,Y=-141.722,Z=16.681)
    ExecutePin=@RigVMFunction_Sequence.A
   )
   > &RigUnit_ModifyTransforms(
    ItemToModify="((Item=(Type=Null,Name=\"def_bandolier_rear_01_ctrl_null\"),Transform=(Rotation=(X=0.1332,Y=0,Z=0,W=0.9911),Translation=(X=1.7248,Y=0.9874,Z=-4),Scale3D=(X=1,Y=1,Z=1))),(Item=(Type=Null,Name=\"def_bandolier_rear_02_ctrl_null\"),Transform=(Rotation=(X=0,Y=0,Z=0 ...(truncated)"
    WeightMinimum=0
    Weight=@n2.OutputParam
   )
   > n3 = &RigUnit_SphericalPoseReader(
    ActiveRegionScaleFactors=(PositiveWidth=2.5,NegativeWidth=3,PositiveHeight=3,NegativeHeight=3)
    ActiveRegionSize=0.1
    Debug=(bDrawDebug=false,DebugScale=25,DebugSegments=20,DebugThickness=0.25)
    DriverAxis=(X=1,Y=0,Z=0)
    DriverItem=(Type=Bone,Name="upperarm_l")
    FalloffRegionScaleFactors=(PositiveWidth=1,NegativeWidth=1,PositiveHeight=1,NegativeHeight=1)
    FalloffSize=0.5
    OptionalParentItem=(Type=Bone,Name="clavicle_l")
    RotationOffset=(X=0,Y=0,Z=0)
   )
   > &RigUnit_ModifyTransforms(
    ItemToModify="((Item=(Type=Null,Name=\"def_bandolier_front_01_ctrl_null\"),Transform=(Rotation=(X=0.4876,Y=0,Z=0,W=0.8731),Translation=(X=-0.9842,Y=3.0732,Z=5.3101),Scale3D=(X=1,Y=1,Z=1))),(Item=(Type=Null,Name=\"def_bandolier_rear_01_ctrl_null\"),Transform=(Rotation=(X=-0. ...(truncated)"
    WeightMinimum=0
    Weight=@n3.OutputParam
   )
   > n4 = &RigUnit_SphericalPoseReader(
    ActiveRegionScaleFactors=(PositiveWidth=0.5,NegativeWidth=3,PositiveHeight=1.5,NegativeHeight=1.5)
    ActiveRegionSize=0.1
    Debug=(bDrawDebug=false,DebugScale=25,DebugSegments=20,DebugThickness=0.25)
    DriverAxis=(X=1,Y=0,Z=0)
    DriverItem=(Type=Bone,Name="upperarm_l")
    FalloffRegionScaleFactors=(PositiveWidth=1,NegativeWidth=1,PositiveHeight=1,NegativeHeight=1)
    FalloffSize=0.4
    OptionalParentItem=(Type=Bone,Name="clavicle_l")
    RotationOffset=(X=90,Y=0,Z=0)
   )
   > &RigUnit_ModifyTransforms(
    ItemToModify="((Item=(Type=Null,Name=\"def_bandolier_rear_01_ctrl_null\"),Transform=(Rotation=(X=0.2865,Y=0.1023,Z=0.0697,W=0.9501),Translation=(X=0.0501,Y=1.3975,Z=-3.7365),Scale3D=(X=1,Y=1,Z=1))),(Item=(Type=Bone,Name=\"def_bandolier_root\"),Transform=(Rotation=(X=-0.1505 ...(truncated)"
    WeightMinimum=0
    Weight=@n4.OutputParam
   )
   > &RigUnit_ParentConstraint(
    AdvancedSettings=(InterpolationType=Shortest,RotationOrderForFilter=XYZ)
    Child=(Type=Null,Name="dyn_bag_sim_chain_03_ctrl_null")
    Filter=(TranslationFilter=(bX=True,bY=True,bZ=True),RotationFilter=(bX=true,bY=true,bZ=true),ScaleFilter=(bX=True,bY=True,bZ=True))
    Parents=((Item=(Type=Bone,Name="thigh_r"),Weight=1),(Item=(Type=Bone,Name="pelvis"),Weight=1))
    Weight=1
   )
   > &RigUnit_ParentConstraint(
    AdvancedSettings=(InterpolationType=Average,RotationOrderForFilter=XZY)
    Child=(Type=Null,Name="def_bandolier_front_10_ctrl_null")
    Parents=((Item=(Type=Control,Name="dyn_bag_sim_chain_03_ctrl"),Weight=1))
    Weight=1
   )
   > &RigUnit_ParentConstraint(
    AdvancedSettings=(InterpolationType=Average,RotationOrderForFilter=XZY)
    Child=(Type=Null,Name="def_bandolier_rear_10_ctrl_null")
    Parents=((Item=(Type=Control,Name="dyn_bag_sim_chain_03_ctrl"),Weight=1))
    Weight=1
   )
   > n5 = &RigUnit_SphericalPoseReader(
    ActiveRegionScaleFactors=(PositiveWidth=3.5,NegativeWidth=0.5,PositiveHeight=3,NegativeHeight=3)
    ActiveRegionSize=0.1
    Debug=(bDrawDebug=false,DebugScale=25,DebugSegments=20,DebugThickness=0.25)
    DriverAxis=(X=1,Y=0,Z=0)
    DriverItem=(Type=Bone,Name="thigh_r")
    FalloffRegionScaleFactors=(PositiveWidth=1,NegativeWidth=1,PositiveHeight=1,NegativeHeight=1)
    FalloffSize=0.45
    FlipHeightScaling=false
    FlipWidthScaling=true
    OptionalParentItem=(Type=Bone)
    RotationOffset=(X=-90,Y=0,Z=0)
   )
   > &RigUnit_ModifyTransforms(
    ItemToModify="((Item=(Type=Null,Name=\"dyn_bag_sim_chain_03_ctrl_null\"),Transform=(Rotation=(X=0,Y=0,Z=-0.0058,W=1),Translation=(X=-7.2087,Y=0,Z=0.9173),Scale3D=(X=1,Y=1,Z=1))),(Item=(Type=Null,Name=\"def_bandolier_front_10_ctrl_null\"),Transform=(Rotation=(X=0,Y=0,Z=-0.12 ...(truncated)"
    WeightMinimum=0
    Weight=@n5.OutputParam
   )
   > SplineIK = RigVMFunctionReferenceNode(
    Primary Axis=(X=1,Y=0,Z=0)
    Samples Per Segment=3
    Secondary Spline Direction=(X=0,Y=-5,Z=0)
    Spline Mode=BSpline
    Stretch=true
    Up Axis=(X=0,Y=-1,Z=0)
    debug=false
    Bones=RigVMVariableNode() {variable=bandolier_rear_joints}.Value
    Controls=RigVMVariableNode() {variable=bandolier_rear_ctrls}.Value
   )
   > SplineIK_1 = RigVMFunctionReferenceNode(
    Primary Axis=(X=1,Y=0,Z=0)
    Samples Per Segment=3
    Secondary Spline Direction=(X=0,Y=1,Z=0)
    Spline Mode=BSpline
    Stretch=true
    Up Axis=(X=0,Y=1,Z=0)
    debug=false
    Controls=RigVMVariableNode() {variable=bandolier_front_ctrls}.Value
    Bones=RigVMVariableNode() {variable=bandolier_front_joints}.Value
   )
   > &RigUnit_ParentConstraint(
    AdvancedSettings=(InterpolationType=Average,RotationOrderForFilter=XZY)
    Child=(Type=Bone,Name="dyn_bag_front_01")
    Parents=((Item=(Type=Control,Name="dyn_bag_sim_chain_03_ctrl"),Weight=1))
    Weight=1
   )
   > &RigUnit_ParentConstraint(
    AdvancedSettings=(InterpolationType=Average,RotationOrderForFilter=XZY)
    Child=(Type=Bone,Name="dyn_bag_rear_01")
    Parents=((Item=(Type=Control,Name="dyn_bag_sim_chain_03_ctrl"),Weight=1))
    Weight=1
   )
   RigVMFunctionReturnNode(ExecuteContext=@RigVMFunction_Sequence.B)
   n6 = &RigUnit_SphericalPoseReader(
    ActiveRegionScaleFactors=(PositiveWidth=1,NegativeWidth=3.5,PositiveHeight=1.5,NegativeHeight=1.5)
    ActiveRegionSize=0.1
    Debug=(bDrawDebug=true,DebugScale=25,DebugSegments=20,DebugThickness=0.25)
    DriverAxis=(X=1,Y=0,Z=0)
    DriverItem=(Type=Bone,Name="thigh_r")
    ExecutePin=()
    FalloffRegionScaleFactors=(PositiveWidth=1,NegativeWidth=1,PositiveHeight=1,NegativeHeight=1)
    FalloffSize=0.35
    OptionalParentItem=(Type=Bone,Name="pelvis")
    RotationOffset=(X=0,Y=-27,Z=0)
   )
   > &RigUnit_ModifyTransforms(
    ItemToModify=((Item=(Type=Bone,Name="dyn_bag_sim_chain_03"),Transform=(Rotation=(X=0,Y=-0.3827,Z=0,W=0.9239),Translation=(X=0,Y=0,Z=4),Scale3D=(X=1,Y=1,Z=1))))
    WeightMinimum=0
    Weight=@n6.OutputParam
   )
  ]}
 New Function_ContainedGraph {nodes [
   RigVMFunctionEntryNode()
   > &RigUnit_ParentConstraint(
    AdvancedSettings=(RotationOrderForFilter=XZY)
    Child=(Type=Null,Name="def_bandolier_rear_06_ctrl_null")
    Parents="((Item=(Type=Bone,Name=\"clavicle_l\"),Weight=0.0007),(Item=(Type=Bone,Name=\"spine_05\"),Weight=0.0027),(Item=(Type=Bone,Name=\"clavicle_r\"),Weight=0.0125),(Item=(Type=Bone,Name=\"spine_04\"),Weight=0.0943),(Item=(Type=Bone,Name=\"pelvis\"),Weight=0.1051),(Item=(T ...(truncated)"
    Weight=1
   )
   > &RigUnit_ParentConstraint(
    AdvancedSettings=(RotationOrderForFilter=XZY)
    Child=(Type=Null,Name="def_bandolier_rear_07_ctrl_null")
    Parents="((Item=(Type=Bone,Name=\"clavicle_r\"),Weight=0.0004),(Item=(Type=Bone,Name=\"spine_04\"),Weight=0.0225),(Item=(Type=Bone,Name=\"spine_01\"),Weight=0.1517),(Item=(Type=Bone,Name=\"spine_03\"),Weight=0.1541),(Item=(Type=Bone,Name=\"spine_02\"),Weight=0.1966),(Item=(T ...(truncated)"
    Weight=1
   )
   > &RigUnit_ParentConstraint(
    AdvancedSettings=(RotationOrderForFilter=XZY)
    Child=(Type=Null,Name="def_bandolier_rear_08_ctrl_null")
    Parents="((Item=(Type=Bone,Name=\"thigh_twist_01_l\"),Weight=0.0011),(Item=(Type=Bone,Name=\"spine_03\"),Weight=0.0073),(Item=(Type=Bone,Name=\"spine_01\"),Weight=0.0109),(Item=(Type=Bone,Name=\"spine_02\"),Weight=0.0114),(Item=(Type=Bone,Name=\"thigh_twist_01_r\"),Weight=0. ...(truncated)"
    Weight=1
   )
   > &RigUnit_ParentConstraint(
    AdvancedSettings=(RotationOrderForFilter=XZY)
    Child=(Type=Null,Name="def_bandolier_rear_09_ctrl_null")
    Parents=((Item=(Type=Bone,Name="spine_03"),Weight=0.0008),(Item=(Type=Bone,Name="spine_01"),Weight=0.0012),(Item=(Type=Bone,Name="spine_02"),Weight=0.0015),(Item=(Type=Bone,Name="thigh_twist_01_r"),Weight=0.0566),(Item=(Type=Bone,Name="pelvis"),Weight=0.9399))
    Weight=1
   )
   > &RigUnit_ParentConstraint(
    AdvancedSettings=(RotationOrderForFilter=XZY)
    Child=(Type=Null,Name="def_bandolier_rear_10_ctrl_null")
    Parents=((Item=(Type=Bone,Name="thigh_twist_01_l"),Weight=0.0005),(Item=(Type=Bone,Name="thigh_r"),Weight=0.0018),(Item=(Type=Bone,Name="pelvis"),Weight=0.3346),(Item=(Type=Bone,Name="thigh_twist_01_r"),Weight=0.6632))
    Weight=1
   )
   > &RigUnit_ParentConstraint(
    AdvancedSettings=(RotationOrderForFilter=XZY)
    Child=(Type=Null,Name="def_bandolier_front_06_ctrl_null")
    Parents=((Item=(Type=Bone,Name="spine_04"),Weight=0.0425),(Item=(Type=Bone,Name="pelvis"),Weight=0.1514),(Item=(Type=Bone,Name="spine_01"),Weight=0.1939),(Item=(Type=Bone,Name="spine_03"),Weight=0.2742),(Item=(Type=Bone,Name="spine_02"),Weight=0.3379))
    Weight=1
   )
   > &RigUnit_ParentConstraint(
    AdvancedSettings=(RotationOrderForFilter=XZY)
    Child=(Type=Null,Name="def_bandolier_front_07_ctrl_null")
    Parents=((Item=(Type=Bone,Name="spine_04"),Weight=0.0187),(Item=(Type=Bone,Name="spine_03"),Weight=0.1406),(Item=(Type=Bone,Name="spine_01"),Weight=0.1799),(Item=(Type=Bone,Name="spine_02"),Weight=0.2253),(Item=(Type=Bone,Name="pelvis"),Weight=0.4355))
    Weight=1
   )
   > &RigUnit_ParentConstraint(
    AdvancedSettings=(RotationOrderForFilter=XZY)
    Child=(Type=Null,Name="def_bandolier_front_08_ctrl_null")
    Parents="((Item=(Type=Bone,Name=\"spine_04\"),Weight=0.0023),(Item=(Type=Bone,Name=\"thigh_twist_01_r\"),Weight=0.0051),(Item=(Type=Bone,Name=\"spine_03\"),Weight=0.0572),(Item=(Type=Bone,Name=\"spine_01\"),Weight=0.0828),(Item=(Type=Bone,Name=\"spine_02\"),Weight=0.0972),(I ...(truncated)"
    Weight=1
   )
   > &RigUnit_ParentConstraint(
    AdvancedSettings=(RotationOrderForFilter=XZY)
    Child=(Type=Null,Name="def_bandolier_front_09_ctrl_null")
    Parents="((Item=(Type=Bone,Name=\"thigh_twist_01_l\"),Weight=0.0009),(Item=(Type=Bone,Name=\"spine_03\"),Weight=0.008),(Item=(Type=Bone,Name=\"spine_01\"),Weight=0.0125),(Item=(Type=Bone,Name=\"spine_02\"),Weight=0.0158),(Item=(Type=Bone,Name=\"thigh_twist_01_r\"),Weight=0.1 ...(truncated)"
    Weight=1
   )
   > &RigUnit_ParentConstraint(
    AdvancedSettings=(RotationOrderForFilter=XZY)
    Child=(Type=Null,Name="def_bandolier_front_10_ctrl_null")
    Parents="((Item=(Type=Bone,Name=\"spine_01\"),Weight=0.0001),(Item=(Type=Bone,Name=\"spine_02\"),Weight=0.0002),(Item=(Type=Bone,Name=\"thigh_twist_01_l\"),Weight=0.0006),(Item=(Type=Bone,Name=\"thigh_r\"),Weight=0.0022),(Item=(Type=Bone,Name=\"pelvis\"),Weight=0.4785),(Item ...(truncated)"
    Weight=1
   )
   > RigVMFunctionReturnNode()
  ]}
 RigVMFunctionLibrary {nodes [
   Build Bandolier = RigVMCollapseNode(ExecuteContext=())
   Build Bandolier Pose Cache = RigVMCollapseNode(ExecuteContext=())
   Pin Default = RigVMCollapseNode(ExecuteContext=())
   Pin No RBAN = RigVMCollapseNode(ExecuteContext=())
   Solve Bandolier No RBAN = RigVMCollapseNode(ExecuteContext=())
   Solve Bandolier RBAN = RigVMCollapseNode(ExecuteContext=())
   Solve Neckhair = RigVMCollapseNode(ExecuteContext=())
   set_pose_func = RigVMCollapseNode(ExecuteContext=(), bind_pose=(HierarchyTopologyVersion=-1,PoseHash=-1), target_pose=(HierarchyTopologyVersion=-1,PoseHash=-1))
   simple_spline = RigVMCollapseNode(
    ExecuteContext=()
    bones=()
    controls=()
    debug draw spline=False
    max stretch=0
    min stretch=0
    primary axis=(X=1,Y=0,Z=0)
    samples=8
    stretch=False
   )
  ]}
 RigVMFunction_Sequence_1_ContainedGraph {nodes [
   n1 = RigVMFunctionEntryNode()
   RigVMFunction_Sequence = &RigVMFunction_Sequence(ExecuteContext=@n1.ExecuteContext)
   n2 = &RigVMFunction_Sequence(ExecuteContext=@RigVMFunction_Sequence.B)
   RigVMFunctionReturnNode(A=@RigVMFunction_Sequence.A, B=@n2.A, C=@n2.B)
  ]}
 RigVMFunction_Sequence_1_ContainedGraph {nodes [
   n1 = RigVMFunctionEntryNode()
   RigVMFunction_Sequence = &RigVMFunction_Sequence(ExecuteContext=@n1.ExecuteContext)
   n2 = &RigVMFunction_Sequence(ExecuteContext=@RigVMFunction_Sequence.B)
   n3 = &RigVMFunction_Sequence(ExecuteContext=@n2.B)
   RigVMFunctionReturnNode(A=@RigVMFunction_Sequence.A, B=@n2.A, C=@n3.A, D=@n3.B)
  ]}
 RigVMFunction_Sequence_1_ContainedGraph {nodes [
   n1 = RigVMFunctionEntryNode()
   RigVMFunction_Sequence = &RigVMFunction_Sequence(ExecuteContext=@n1.ExecuteContext)
   n2 = &RigVMFunction_Sequence(ExecuteContext=@RigVMFunction_Sequence.B)
   n3 = &RigVMFunction_Sequence(ExecuteContext=@n2.B)
   n4 = &RigVMFunction_Sequence(ExecuteContext=@n3.B)
   n5 = &RigVMFunction_Sequence(ExecuteContext=@n4.B)
   n6 = &RigVMFunction_Sequence(ExecuteContext=@n5.B)
   RigVMFunctionReturnNode(A=@RigVMFunction_Sequence.A, B=@n2.A, C=@n3.A, D=@n4.A, E=@n5.A, F=@n6.A, G=@n6.B)
  ]}
 RigVMFunction_Sequence_1_ContainedGraph {nodes [
   n1 = RigVMFunctionEntryNode()
   RigVMFunction_Sequence = &RigVMFunction_Sequence(ExecuteContext=@n1.ExecuteContext)
   n2 = &RigVMFunction_Sequence(ExecuteContext=@RigVMFunction_Sequence.B)
   RigVMFunctionReturnNode(A=@RigVMFunction_Sequence.A, B=@n2.A, C=@n2.B)
  ]}
 RigVMModel {nodes [
   RigUnit_BeginExecution = &RigUnit_BeginExecution()
   RigVMFunction_ControlFlowBranch = &RigVMFunction_ControlFlowBranch(ExecuteContext=@RigUnit_BeginExecution.ExecutePin, Condition=RigVMVariableNode() {variable=RBAN Enabled}.Value)
   Pin Default = RigVMFunctionReferenceNode(ExecuteContext=@RigVMFunction_ControlFlowBranch.False)
   > Pin No RBAN = RigVMFunctionReferenceNode()
   > Solve Bandolier No RBAN = RigVMFunctionReferenceNode()
   Pin Default_1 = RigVMFunctionReferenceNode(ExecuteContext=@RigVMFunction_ControlFlowBranch.True)
   > Solve Bandolier RBAN = RigVMFunctionReferenceNode()
  ]}
 RigVMModel Rig {nodes [
   n1 = &RigUnit_PrepareForExecution()
   Build Bandolier = RigVMFunctionReferenceNode(ExecuteContext=@n1.ExecutePin)
   Build Bandolier Pose Cache = RigVMFunctionReferenceNode(ExecuteContext=())
  ]}
 SequenceAggregate_1_ContainedGraph {nodes [
   n1 = RigVMFunctionEntryNode()
   SequenceAggregate = &RigVMFunction_Sequence(ExecuteContext=@n1.ExecuteContext)
   SequenceAggregate_1 = &RigVMFunction_Sequence(ExecuteContext=@SequenceAggregate.B)
   RigVMFunctionReturnNode(A=@SequenceAggregate.A, B=@SequenceAggregate_1.A, C=@SequenceAggregate_1.B)
  ]}
 set_pose_func_ContainedGraph {nodes [
   n1 = RigVMFunctionEntryNode()
   SequenceAggregate = &RigVMFunction_Sequence(ExecuteContext=@n1.ExecuteContext)
   Subtract = &RigVMFunction_MathDoubleSub(A=1, B=@n1.pose_alpha)
   /Script/ControlRig.RigUnit_HierarchySetPoseItemArray(ElementType=All, ItemsToSet=(), Pose=@n1.bind_pose, ExecutePin=@SequenceAggregate.A, Weight=@Subtract.Result) {is=struct RigVMUnitNode}
   &RigUnit_HierarchySetPoseItemArray(ElementType=All, ItemsToSet=(), Weight=@n1.pose_alpha, Pose=@n1.target_pose, ExecutePin=@SequenceAggregate.B)
   > RigVMFunctionReturnNode()
  ]}
]
hierarchy [
 root {loc=<0 0 0>, rot=<0 0 0 1>}
 pelvis {parent=root, loc=<0 -0.7071 93.1191>, rot=<0.0127 0.707 -0.0127 -0.707>}
 spine_01 {parent=pelvis, loc=<6.7494 -0.6388 0>, rot=<0 0 0.034 -0.9994>}
 spine_02 {parent=spine_01, loc=<7.252 -0.035 0>, rot=<0 0 0.0238 -0.9997>}
 spine_03 {parent=spine_02, loc=<6.0106 0 0>, rot=<0 0 0.0149 -0.9999>}
 spine_04 {parent=spine_03, loc=<8.1023 0 0>, rot=<0 0 -0.0132 -0.9999>}
 spine_05 {parent=spine_04, loc=<10.0142 0 0>, rot=<0 0 -0.0015 -1>}
 def_bandolier_root {parent=spine_05, loc=<10.3801 0.149 -17.0006>, rot=<0.0284 -0.7065 -0.0284 -0.7065>}
 def_bandolier_front_01 {parent=def_bandolier_root, loc=<-0.1809 3.0788 -0.5765>, rot=<-0.3565 -0.7608 -0.2301 -0.4911>}
 def_bandolier_front_02 {parent=def_bandolier_front_01, loc=<7.558 0 0>, rot=<-0.0119 -0.0141 0.1726 -0.9848>}
 def_bandolier_front_03 {parent=def_bandolier_front_02, loc=<6.2016 0 0>, rot=<-0.0095 -0.0218 0.1179 -0.9927>}
 def_bandolier_front_04 {parent=def_bandolier_front_03, loc=<7.9709 0 0>, rot=<-0.0007 -0.003 0.0653 -0.9979>}
 def_bandolier_front_05 {parent=def_bandolier_front_04, loc=<6.5252 0 0>, rot=<-0.0002 -0.0015 0.0301 -0.9995>}
 def_bandolier_front_06 {parent=def_bandolier_front_05, loc=<8.2262 0 0>, rot=<-0.0009 -0.0232 0.0646 -0.9976>}
 def_bandolier_front_07 {parent=def_bandolier_front_06, loc=<6.7783 0 0>, rot=<0.0009 -0.0106 0.0615 -0.998>}
 def_bandolier_front_08 {parent=def_bandolier_front_07, loc=<8.6655 0 0>, rot=<0.0042 -0.0213 0.044 -0.9988>}
 def_bandolier_front_09 {parent=def_bandolier_front_08, loc=<6.3304 0 0>, rot=<0.0118 -0.0375 0.067 -0.997>}
 def_bandolier_front_10 {parent=def_bandolier_front_09, loc=<8.014 0.1924 0.3812>, rot=<0.2619 0.1071 -0.0093 -0.9591>}
 def_bandolier_front_strap {parent=def_bandolier_front_10, loc=<5.7472 0 0.8309>, rot=<0.1351 0.0295 -0.0022 -0.9904>}
 dyn_bag_front_01 {parent=def_bandolier_front_strap, loc=<7.6243 -1.0752 -0.3812>, rot=<0.9114 0.0615 0.1437 -0.3806>}
 dyn_bag_front_02 {parent=dyn_bag_front_01, loc=<14.75 0 -2.601>, rot=<0.96 0.0119 -0.2356 -0.151>}
 dyn_bag_front_03 {parent=dyn_bag_front_02, loc=<4.6074 0 0>, rot=<0.9702 0.0192 -0.189 0.1503>}
 dyn_bag_front_04 {parent=dyn_bag_front_03, loc=<1.1651 -0.0838 -0.1282>, rot=<-0.1258 0 0 -0.9921>}
 def_bandolier_rear_01 {parent=def_bandolier_root, loc=<-0.1001 -2.9962 -0.2535>, rot=<0.3526 -0.7646 0.2259 -0.4899>}
 def_bandolier_rear_02 {parent=def_bandolier_rear_01, loc=<7.2226 0 0>, rot=<0.0077 -0.0113 -0.264 -0.9644>}
 def_bandolier_rear_03 {parent=def_bandolier_rear_02, loc=<6.3069 0 0>, rot=<0.0014 -0.0056 -0.0843 -0.9964>}
 def_bandolier_rear_04 {parent=def_bandolier_rear_03, loc=<8.5101 0 0>, rot=<0.0007 -0.0068 -0.0616 -0.9981>}
 def_bandolier_rear_05 {parent=def_bandolier_rear_04, loc=<7.026 0 0>, rot=<0.0001 -0.0095 -0.0313 -0.9995>}
 def_bandolier_rear_06 {parent=def_bandolier_rear_05, loc=<8.664 0 0>, rot=<-0.0012 -0.0252 -0.0213 -0.9995>}
 def_bandolier_rear_07 {parent=def_bandolier_rear_06, loc=<6.8729 0 0>, rot=<-0.0017 -0.0175 -0.0294 -0.9994>}
 def_bandolier_rear_08 {parent=def_bandolier_rear_07, loc=<8.3785 0 0>, rot=<-0.0015 -0.0098 -0.0263 -0.9996>}
 def_bandolier_rear_09 {parent=def_bandolier_rear_08, loc=<6.0932 0 0>, rot=<-0.0008 -0.0033 -0.0426 -0.9991>}
 def_bandolier_rear_10 {parent=def_bandolier_rear_09, loc=<7.5944 0 0>, rot=<-0.2793 0.1256 0.1348 -0.9423>}
 def_bandolier_rear_strap {parent=def_bandolier_rear_10, loc=<4.9334 0 0.4938>, rot=<-0.2429 0 0 -0.97>}
 dyn_bag_rear_01 {parent=def_bandolier_rear_strap, loc=<6.0234 2.3697 -2.513>, rot=<-0.8227 -0.0753 -0.0688 -0.5593>}
 dyn_bag_rear_02 {parent=dyn_bag_rear_01, loc=<14.8499 -0.6201 -2.7474>, rot=<0.9736 0.0257 -0.1778 0.1405>}
 dyn_bag_rear_03 {parent=dyn_bag_rear_02, loc=<4.0338 0 0>, rot=<0.9736 0.0257 -0.1778 -0.1405>}
 dyn_bag_rear_04 {parent=dyn_bag_rear_03, loc=<1.6286 -0.2173 -0.4046>, rot=<0 -0.067 0 -0.9978>}
 clavicle_l {parent=spine_05, loc=<4.3575 -0.1466 -2.3223>, rot=<0.0465 -0.7289 -0.009 -0.6829>}
 upperarm_l {parent=clavicle_l, loc=<15.5111 0 0>, rot=<0.0226 -0.3234 -0.0088 -0.9459>}
 upperarm_twist_01_l {parent=upperarm_l, loc=<0.0041 0 0>, rot=<0 0 0 1>}
 deform_upperarm_bk_l {parent=upperarm_twist_01_l, loc=<1.0813 -7.2193 -1.9453>, rot=<0 0 0 1>}
 deform_upperarm_fr_l {parent=upperarm_twist_01_l, loc=<-0.3691 7.0506 -2.765>, rot=<0 0 0 1>}
 deform_upperarm_up_l {parent=upperarm_twist_01_l, loc=<-3.2216 -1.1391 3.3311>, rot=<0 0 0 1>}
 upperarm_twist_02_l {parent=upperarm_l, loc=<13.7941 0 0>, rot=<0 0 0 1>}
 clavicle_r {parent=spine_05, loc=<4.3575 -0.1466 2.3223>, rot=<0.7289 0.0465 -0.6829 0.009>}
 attach_backpack {parent=spine_05, loc=<-0.123 -13.473 0>, rot=<-0.0093 -0.707 0.0093 -0.707>}
 neck_01 {parent=spine_05, loc=<11.3138 0.1529 0>, rot=<0 0 -0.2003 -0.9797>}
 neck_02 {parent=neck_01, loc=<4.5627 0 0>, rot=<0 0 0.0504 -0.9987>}
 head {parent=neck_02, loc=<5.3914 0 0>, rot=<0 0 0.115 -0.9934>}
 deform_neck_hair_rear_MSTR {parent=head, loc=<-5.2893 -10.0691 -0.1196>, rot=<0.229 -0.681 0.6596 -0.2209>}
 deform_neck_hair_rear_01 {parent=deform_neck_hair_rear_MSTR, loc=<0 0 0>, rot=<0 -0.0343 0 -0.9994>}
 deform_neck_hair_rear_02 {parent=deform_neck_hair_rear_01, loc=<5.0389 0 0>, rot=<0 0 0 1>}
 deform_neck_hair_d_l_MSTR {parent=head, loc=<-4.8587 -9.2115 -5.2715>, rot=<0.229 -0.681 0.6596 -0.2209>}
 deform_neck_hair_d_l_01 {parent=deform_neck_hair_d_l_MSTR, loc=<0 0 0>, rot=<0 0 0 1>}
 deform_neck_hair_d_l_02 {parent=deform_neck_hair_d_l_01, loc=<3.9434 0 0>, rot=<0 0 0 1>}
 deform_neck_hair_c_r_MSTR {parent=head, loc=<-5.6303 -5.3382 10.1057>, rot=<0.9056 0.2354 -0.2658 -0.2318>}
 deform_neck_hair_c_r_01 {parent=deform_neck_hair_c_r_MSTR, loc=<0 0 0>, rot=<0 0 0 1>}
 deform_neck_hair_c_r_02 {parent=deform_neck_hair_c_r_01, loc=<-2.3088 0.0001 -0.0001>, rot=<0 0 0 1>}
 deform_neck_hair_b_r_MSTR {parent=head, loc=<-4.2051 1.2953 9.5693>, rot=<0.9539 0.124 -0.24 0.1305>}
 deform_neck_hair_b_r_01 {parent=deform_neck_hair_b_r_MSTR, loc=<0 0 0>, rot=<0 0 0 1>}
 deform_neck_hair_b_r_02 {parent=deform_neck_hair_b_r_01, loc=<-4.8599 0.0002 -0.0003>, rot=<0 0.2462 0 -0.9692>}
 deform_neck_hair_a_r_MSTR {parent=head, loc=<-4.009 5.6409 6.2811>, rot=<0.9661 -0.0067 -0.1906 0.1741>}
 deform_neck_hair_a_r_01 {parent=deform_neck_hair_a_r_MSTR, loc=<0 0 0>, rot=<0 0 0 -1>}
 deform_neck_hair_a_r_02 {parent=deform_neck_hair_a_r_01, loc=<-3.92 0 -0.0002>, rot=<0 0 0 -1>}
 deform_neck_hair_c_l_MSTR {parent=head, loc=<-5.6305 -5.3382 -10.1057>, rot=<0.2354 -0.9056 0.2318 -0.2658>}
 deform_neck_hair_c_l_01 {parent=deform_neck_hair_c_l_MSTR, loc=<0 0 0>, rot=<0 0 0 1>}
 deform_neck_hair_c_l_02 {parent=deform_neck_hair_c_l_01, loc=<2.309 0 0>, rot=<0 0 0 1>}
 deform_neck_hair_b_l_MSTR {parent=head, loc=<-4.2046 1.2953 -9.5693>, rot=<0.124 -0.9539 -0.1305 -0.24>}
 deform_neck_hair_b_l_01 {parent=deform_neck_hair_b_l_MSTR, loc=<0 0 0>, rot=<0 0 0 1>}
 deform_neck_hair_b_l_02 {parent=deform_neck_hair_b_l_01, loc=<4.8605 0 0>, rot=<0 0.2462 0 -0.9692>}
 deform_neck_hair_a_l_MSTR {parent=head, loc=<-4.0087 5.6409 -6.2811>, rot=<-0.0067 -0.9661 -0.1741 -0.1906>}
 deform_neck_hair_a_l_01 {parent=deform_neck_hair_a_l_MSTR, loc=<0 0 0>, rot=<0 0 0 1>}
 deform_neck_hair_a_l_02 {parent=deform_neck_hair_a_l_01, loc=<3.9204 0 0>, rot=<0 0 0 1>}
 deform_neck_hair_d_r_MSTR {parent=head, loc=<-4.8583 -9.2115 5.2715>, rot=<0.681 0.229 -0.2209 -0.6596>}
 deform_neck_hair_d_r_01 {parent=deform_neck_hair_d_r_MSTR, loc=<0 0 0>, rot=<0 0 0 1>}
 deform_neck_hair_d_r_02 {parent=deform_neck_hair_d_r_01, loc=<-3.944 0 0.0004>, rot=<0 0 0 1>}
 deform_pec_aim_l {parent=spine_04, loc=<7.8728 12.6388 -2.4123>, rot=<0.0102 -0.4602 0.3979 -0.7936>}
 deform_pec_l {parent=deform_pec_aim_l, loc=<10.8209 0 0>, rot=<0 0 0 1>}
 def_body_fur_root {parent=spine_04, loc=<0 0 0>, rot=<0 0 0 1>}
 def_body_fur_03 {parent=def_body_fur_root, loc=<0.3754 -15.0001 -1.4924>, rot=<0.1264 -0.9876 -0.0907 -0.0201>}
 def_body_fur_05 {parent=def_body_fur_root, loc=<-9.2173 -14.4748 -7.4091>, rot=<0.2342 -0.9129 -0.2553 0.2157>}
 def_body_fur_07 {parent=def_body_fur_root, loc=<-20.3734 -15.5686 -0.5445>, rot=<0.1712 -0.7321 0.6472 -0.1261>}
 def_body_fur_08 {parent=def_body_fur_root, loc=<-5.9464 -14.68 2.9528>, rot=<0.0808 -0.6674 0.6327 -0.3845>}
 def_body_fur_11 {parent=def_body_fur_root, loc=<-26.1955 -15.1248 3.9221>, rot=<0.4197 -0.5934 0.6795 0.0997>}
 def_body_fur_12 {parent=def_body_fur_root, loc=<-24.2984 12.981 2.9626>, rot=<-0.2895 -0.9441 -0.1271 0.0933>}
 def_body_fur_14 {parent=def_body_fur_root, loc=<-13.0217 13.3909 5.8261>, rot=<0.0506 0.5189 0.772 0.3636>}
 def_body_fur_15 {parent=def_body_fur_root, loc=<-16.8918 11.5865 8.433>, rot=<0.2331 0.4728 0.8459 0.0816>}
 def_body_fur_16 {parent=def_body_fur_root, loc=<-19.8712 10.3558 10.2407>, rot=<0.3242 0.3594 0.8675 0.1144>}
 def_body_fur_19 {parent=def_body_fur_root, loc=<-12.0558 13.2174 -5.8644>, rot=<-0.119 -0.7765 -0.6134 -0.0811>}
 def_body_fur_20 {parent=def_body_fur_root, loc=<4.9582 14.9374 -5.7813>, rot=<-0.0353 -0.7321 -0.644 -0.2192>}
 def_body_fur_21 {parent=def_body_fur_root, loc=<-0.4209 15.3878 -2.5777>, rot=<-0.0353 -0.7321 -0.644 -0.2192>}
 dyn_bag_sim_chain_01_MSTR {parent=spine_03, loc=<3.202 -1.3712 -3.9172>, rot=<0.2041 0.0112 0.9774 -0.0536>}
 dyn_bag_sim_chain_01 {parent=dyn_bag_sim_chain_01_MSTR, loc=<0 0 0>, rot=<0 0 0 1>}
 dyn_bag_sim_chain_02 {parent=dyn_bag_sim_chain_01, loc=<14.095 0 0>, rot=<0 0.1146 0 -0.9934>}
 dyn_bag_sim_chain_03 {parent=dyn_bag_sim_chain_02, loc=<20.6474 0 0>, rot=<0.0578 -0.1891 -0.0112 -0.9802>}
 thigh_r {parent=pelvis, loc=<-2.5693 -0.0109 7.7501>, rot=<0.0572 -0.0012 0.9982 0.0155>}
 thigh_twist_01_r {parent=thigh_r, loc=<21.6303 0 0>, rot=<0.0013 0 0 -1>}
 deform_groin_r {parent=thigh_r, loc=<2.1908 -8.1142 1.7442>, rot=<0 0 0 -1>}
 calf_r {parent=thigh_r, loc=<39.0484 0 0>, rot=<0.0006 -0.0161 -0.0391 -0.9991>}
 deform_glute_r {parent=pelvis, loc=<-8.8062 -8.8222 8.4287>, rot=<0.0501 -0.0275 0.8926 -0.4472>}
 thigh_l {parent=pelvis, loc=<-2.5693 -0.0109 -7.7501>, rot=<0.0012 0.0572 0.0155 -0.9982>}
 thigh_twist_01_l {parent=thigh_l, loc=<-21.6303 0 0>, rot=<0.0013 0 0 -1>}
 dyn_bag_sim_chain_03_ctrl_null {type=Null, loc=<-14.0739 -3.5645 86.6956>, rot=<0.6084 0.0467 -0.7915 0.0359>}
 dyn_bag_sim_chain_03_ctrl {
  controlType=EULER_TRANSFORM
  parent=dyn_bag_sim_chain_03_ctrl_null
  type=Control
  loc=<80.1585 0.6805 -36.0701>
  rot=<-0.6084 -0.0467 0.7915 0.0359>
 }
 def_bandolier_front_01_ctrl_null {type=Null, loc=<16.8197 -1.174 140.9411>, rot=<0.3565 0.7608 0.2301 0.4911>}
 def_bandolier_front_01_ctrl {
  controlType=EULER_TRANSFORM
  parent=def_bandolier_front_01_ctrl_null
  type=Control
  loc=<87.5255 -103.2734 42.6915>
  rot=<-0.3565 -0.7608 -0.2301 0.4911>
 }
 def_bandolier_front_02_ctrl_null {type=Null, loc=<14.8287 4.634 136.5335>, rot=<-0.2224 -0.8204 -0.1378 -0.5085>}
 def_bandolier_front_02_ctrl {
  controlType=EULER_TRANSFORM
  parent=def_bandolier_front_02_ctrl_null
  type=Control
  loc=<-7.1048 -2.5716 -0.1781>
  rot=<0.0119 0.0141 -0.1726 -0.9848>
 }
 def_bandolier_front_03_ctrl_null {type=Null, loc=<12.4472 7.7662 131.7401>, rot=<-0.1259 -0.8531 -0.0739 -0.501>}
 def_bandolier_front_03_ctrl {
  controlType=EULER_TRANSFORM
  parent=def_bandolier_front_03_ctrl_null
  type=Control
  loc=<-6.0233 -1.4543 -0.255>
  rot=<0.0095 0.0218 -0.1179 -0.9927>
 }
 def_bandolier_front_04_ctrl_null {type=Null, loc=<8.7301 10.0688 125.0756>, rot=<-0.07 -0.861 -0.0408 -0.5021>}
 def_bandolier_front_04_ctrl {
  controlType=EULER_TRANSFORM
  parent=def_bandolier_front_04_ctrl_null
  type=Control
  loc=<-7.9027 -1.0395 -0.0464>
  rot=<0.0007 0.003 -0.0653 -0.9979>
 }
 def_bandolier_front_05_ctrl_null {type=Null, loc=<5.5593 11.1229 119.4708>, rot=<-0.0441 -0.8635 -0.0256 -0.5018>}
 def_bandolier_front_05_ctrl {
  controlType=EULER_TRANSFORM
  parent=def_bandolier_front_05_ctrl_null
  type=Control
  loc=<-6.5134 -0.3924 -0.0194>
  rot=<0.0002 0.0015 -0.0301 -0.9995>
 }
 def_bandolier_front_06_ctrl_null {type=Null, loc=<1.5083 11.9613 112.3604>, rot=<0.0119 -0.8759 0.0066 -0.4822>}
 def_bandolier_front_06_ctrl {
  controlType=EULER_TRANSFORM
  parent=def_bandolier_front_06_ctrl_null
  type=Control
  loc=<-8.1487 -1.0604 -0.3801>
  rot=<0.0009 0.0232 -0.0646 -0.9976>
 }
 def_bandolier_front_07_ctrl_null {type=Null, loc=<-2.1156 11.7768 106.6352>, rot=<0.0662 -0.8786 0.0355 -0.4716>}
 def_bandolier_front_07_ctrl {
  controlType=EULER_TRANSFORM
  parent=def_bandolier_front_07_ctrl_null
  type=Control
  loc=<-6.7254 -0.8324 -0.1439>
  rot=<-0.0009 0.0106 -0.0615 -0.998>
 }
 def_bandolier_front_08_ctrl_null {type=Null, loc=<-6.8507 10.4788 99.4948>, rot=<0.106 -0.8848 0.054 -0.4504>}
 def_bandolier_front_08_ctrl {
  controlType=EULER_TRANSFORM
  parent=def_bandolier_front_08_ctrl_null
  type=Control
  loc=<-8.624 -0.7608 -0.3726>
  rot=<-0.0042 0.0213 -0.044 -0.9988>
 }
 def_bandolier_front_09_ctrl_null {type=Null, loc=<-10.47 8.9832 94.521>, rot=<0.1683 -0.8926 0.0775 -0.411>}
 def_bandolier_front_09_ctrl {
  controlType=EULER_TRANSFORM
  parent=def_bandolier_front_09_ctrl_null
  type=Control
  loc=<-6.2557 -0.8401 -0.4839>
  rot=<-0.0118 0.0375 -0.067 -0.997>
 }
 def_bandolier_front_10_ctrl_null {type=Null, loc=<-15.0786 6.2443 88.549>, rot=<0.2691 -0.8339 -0.1813 -0.4464>}
 def_bandolier_front_10_ctrl {
  controlType=EULER_TRANSFORM
  parent=def_bandolier_front_10_ctrl_null
  type=Control
  loc=<-7.9194 -0.2808 1.2692>
  rot=<-0.2619 -0.1071 0.0093 -0.9591>
 }
 def_bandolier_rear_01_ctrl_null {type=Null, loc=<16.9005 -7.2489 141.2641>, rot=<-0.3526 0.7646 -0.2259 0.4899>}
 def_bandolier_rear_01_ctrl {
  controlType=EULER_TRANSFORM
  parent=def_bandolier_rear_01_ctrl_null
  type=Control
  loc=<82.4068 107.6845 43.668>
  rot=<0.3526 -0.7646 0.2259 0.4899>
 }
 def_bandolier_rear_02_ctrl_null {type=Null, loc=<14.941 -12.7418 137.0034>, rot=<0.1394 -0.8378 0.0866 -0.5208>}
 def_bandolier_rear_02_ctrl {
  controlType=EULER_TRANSFORM
  parent=def_bandolier_rear_02_ctrl_null
  type=Control
  loc=<-6.2138 3.6793 -0.1284>
  rot=<-0.0077 0.0113 0.264 -0.9644>
 }
 def_bandolier_rear_03_ctrl_null {type=Null, loc=<12.3003 -14.7837 131.6523>, rot=<0.0685 -0.8496 0.042 -0.5213>}
 def_bandolier_rear_03_ctrl {
  controlType=EULER_TRANSFORM
  parent=def_bandolier_rear_03_ctrl_null
  type=Control
  loc=<-6.2168 1.0601 -0.0694>
  rot=<-0.0014 0.0056 0.0843 -0.9964>
 }
 def_bandolier_rear_04_ctrl_null {type=Null, loc=<8.4955 -16.1464 124.1631>, rot=<-0.0161 0.8557 -0.0097 0.5171>}
 def_bandolier_rear_04_ctrl {
  controlType=EULER_TRANSFORM
  parent=def_bandolier_rear_04_ctrl_null
  type=Control
  loc=<-8.4447 1.0462 -0.115>
  rot=<-0.0007 0.0068 0.0616 -0.9981>
 }
 def_bandolier_rear_05_ctrl_null {type=Null, loc=<5.2299 -16.4104 117.9477>, rot=<-0.0107 -0.8607 -0.0063 -0.509>}
 def_bandolier_rear_05_ctrl {
  controlType=EULER_TRANSFORM
  parent=def_bandolier_rear_05_ctrl_null
  type=Control
  loc=<-7.011 0.4389 -0.1331>
  rot=<-0.0001 0.0095 0.0313 -0.9995>
 }
 def_bandolier_rear_06_ctrl_null {type=Null, loc=<1.0568 -16.1944 110.358>, rot=<-0.0295 -0.8728 -0.0164 -0.4869>}
 def_bandolier_rear_06_ctrl {
  controlType=EULER_TRANSFORM
  parent=def_bandolier_rear_06_ctrl_null
  type=Control
  loc=<-8.6452 0.3678 -0.4362>
  rot=<0.0012 0.0252 0.0213 -0.9995>
 }
 def_bandolier_rear_07_ctrl_null {type=Null, loc=<-2.5455 -15.7309 104.5232>, rot=<-0.0556 -0.88 -0.0298 -0.4708>}
 def_bandolier_rear_07_ctrl {
  controlType=EULER_TRANSFORM
  parent=def_bandolier_rear_07_ctrl_null
  type=Control
  loc=<-6.8568 0.403 -0.2413>
  rot=<0.0017 0.0175 0.0294 -0.9994>
 }
 def_bandolier_rear_08_ctrl_null {type=Null, loc=<-7.158 -14.6761 97.6086>, rot=<-0.0792 -0.8828 -0.0414 -0.4611>}
 def_bandolier_rear_08_ctrl {
  controlType=EULER_TRANSFORM
  parent=def_bandolier_rear_08_ctrl_null
  type=Control
  loc=<-8.3653 0.4408 -0.1642>
  rot=<0.0015 0.0098 0.0263 -0.9996>
 }
 def_bandolier_rear_09_ctrl_null {type=Null, loc=<-10.5833 -13.5918 92.6874>, rot=<-0.1169 -0.8802 -0.0606 -0.456>}
 def_bandolier_rear_09_ctrl {
  controlType=EULER_TRANSFORM
  parent=def_bandolier_rear_09_ctrl_null
  type=Control
  loc=<-6.071 0.5186 -0.041>
  rot=<0.0008 0.0033 0.0426 -0.9991>
 }
 def_bandolier_rear_10_ctrl_null {type=Null, loc=<-14.8123 -11.6091 86.699>, rot=<-0.1265 -0.8049 0.2649 -0.5158>}
 def_bandolier_rear_10_ctrl {
  controlType=EULER_TRANSFORM
  parent=def_bandolier_rear_10_ctrl_null
  type=Control
  loc=<-7.0785 -1.3969 2.3704>
  rot=<0.2793 -0.1256 -0.1348 -0.9423>
 }
]
variables [
 BagBindPose {type=FRigPose, default=(Elements=,HierarchyTopologyVersion=-1,PoseHash=-1)}
 BagCtrls {type="TArray<FRigElementKey>", default=((Type=Control,Name="dyn_bag_sim_chain_03_ctrl"))}
 BagThighBackRPose {type=FRigPose, default=(Elements=,HierarchyTopologyVersion=-1,PoseHash=-1)}
 BagThighSideRPose {type=FRigPose, default=(Elements=,HierarchyTopologyVersion=-1,PoseHash=-1)}
 BagThighUpRPose {type=FRigPose, default=(Elements=,HierarchyTopologyVersion=-1,PoseHash=-1)}
 HierarchyGetPoseItemArray_3_1_Pose {type=FRigPose, default=(Elements=,HierarchyTopologyVersion=-1,PoseHash=-1)}
 IsSkydiving {type=bool, default=False}
 NewVar {type=bool, default=False}
 RBAN Enabled {type=bool, default=False}
 ThighRSide SPR {type=float, default=0}
 ThighRUp SPR {type=float, default=0}
 bandolier_front_ctrls {
  type="TArray<FRigElementKey>"
  default="((Type=Control,Name=\"def_bandolier_front_01_ctrl\"),(Type=Control,Name=\"def_bandolier_front_02_ctrl\"),(Type=Control,Name=\"def_bandolier_front_03_ctrl\"),(Type=Control,Name=\"def_bandolier_front_04_ctrl\"),(Type=Control,Name=\"def_bandolier_front_05_ctrl\"),(Type ...(truncated)"
 }
 bandolier_front_ctrls_RBAN {
  type="TArray<FRigElementKey>"
  default=((Type=Control,Name="def_bandolier_front_01_ctrl"),(Type=Control,Name="def_bandolier_front_02_ctrl"),(Type=Control,Name="def_bandolier_front_03_ctrl"),(Type=Control,Name="def_bandolier_front_04_ctrl"))
 }
 bandolier_front_joints {
  type="TArray<FRigElementKey>"
  default="((Type=Bone,Name=\"def_bandolier_front_01\"),(Type=Bone,Name=\"def_bandolier_front_02\"),(Type=Bone,Name=\"def_bandolier_front_03\"),(Type=Bone,Name=\"def_bandolier_front_04\"),(Type=Bone,Name=\"def_bandolier_front_05\"),(Type=Bone,Name=\"def_bandolier_front_06\"),(Ty ...(truncated)"
 }
 bandolier_front_joints_RBAN {
  type="TArray<FRigElementKey>"
  default=((Type=Bone,Name="def_bandolier_front_01"),(Type=Bone,Name="def_bandolier_front_02"),(Type=Bone,Name="def_bandolier_front_03"),(Type=Bone,Name="def_bandolier_front_04"))
 }
 bandolier_rear_ctrls {
  type="TArray<FRigElementKey>"
  default="((Type=Control,Name=\"def_bandolier_rear_01_ctrl\"),(Type=Control,Name=\"def_bandolier_rear_02_ctrl\"),(Type=Control,Name=\"def_bandolier_rear_03_ctrl\"),(Type=Control,Name=\"def_bandolier_rear_04_ctrl\"),(Type=Control,Name=\"def_bandolier_rear_05_ctrl\"),(Type=Cont ...(truncated)"
 }
 bandolier_rear_ctrls_RBAN {
  type="TArray<FRigElementKey>"
  default=((Type=Control,Name="def_bandolier_rear_01_ctrl"),(Type=Control,Name="def_bandolier_rear_02_ctrl"),(Type=Control,Name="def_bandolier_rear_03_ctrl"),(Type=Control,Name="def_bandolier_rear_04_ctrl"),(Type=Control,Name="def_bandolier_rear_05_ctrl"))
 }
 bandolier_rear_joints {
  type="TArray<FRigElementKey>"
  default="((Type=Bone,Name=\"def_bandolier_rear_01\"),(Type=Bone,Name=\"def_bandolier_rear_02\"),(Type=Bone,Name=\"def_bandolier_rear_03\"),(Type=Bone,Name=\"def_bandolier_rear_04\"),(Type=Bone,Name=\"def_bandolier_rear_05\"),(Type=Bone,Name=\"def_bandolier_rear_06\"),(Type=Bon ...(truncated)"
 }
 bandolier_rear_joints_RBAN {
  type="TArray<FRigElementKey>"
  default=((Type=Bone,Name="def_bandolier_rear_01"),(Type=Bone,Name="def_bandolier_rear_02"),(Type=Bone,Name="def_bandolier_rear_03"),(Type=Bone,Name="def_bandolier_rear_04"),(Type=Bone,Name="def_bandolier_rear_05"))
 }
 crouch_pistol {type=bool, default=False}
]
