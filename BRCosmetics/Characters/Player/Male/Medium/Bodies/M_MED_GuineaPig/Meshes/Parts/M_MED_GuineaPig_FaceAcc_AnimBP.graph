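# Animation Blueprint for the M_MED_GuineaPig FaceAcc part, which animates the character's bandolier straps and bags with physics.
# It copies attached-parent and body-mesh poses into a Control Rig, caching the result for two RigidBody simulations that blend on emote state.
# Per-state gravity and RBAN tuning are computed on update.
# In the front end, a Frontend_Idle_Bypass curve weights a DefaultSlot override on the bandolier and bag bones.
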AnimBlueprint /BRCosmetics/Characters/Player/Male/Medium/Bodies/M_MED_GuineaPig/Meshes/Parts/M_MED_GuineaPig_FaceAcc_AnimBP
target_skeleton=&/M_MED_GuineaPig_FaceAcc_Skeleton
graphs [
 AnimGraph {nodes [
   n1 = RigidBody!(ComponentPose=Local To Component(LocalPose=Use cached pose 'PostDeformRig'()), Alpha, OverrideWorldGravity:Vector="0,0,-980", ExternalForce) -> (Pose) {
    bOverrideWorldGravity=true
    overridePhysicsAsset=&/M_MED_GuineaPig_FaceAcc_Bandolier_Emote_Physics
    simulationSpace=BaseBoneSpace
    alphaBoolBlend {blendOption=Linear}
    alphaScaleBiasClamp {bInterpResult=true}
    baseBoneRef {boneName=spine_04}
    componentAppliedLinearAccClamp=<0 0 0>
    overrideWorldGravity {z=-980}
    simSpaceSettings {maxAngularAcceleration=100, maxAngularVelocity=100, maxLinearAcceleration=100, maxLinearVelocity=100}
   }
   n2 = RigidBody!(
    ComponentPose=Local To Component(LocalPose=Use cached pose 'PostDeformRig'())
    Alpha
    OverrideWorldGravity="0,0,-980"
    ExternalForce
    ComponentLinearAccScale:Vector="0,0,0"
    ComponentLinearVelScale:Vector="0,0,0"
    SimSpaceSettings:Sim Space Settings Structure=(WorldAlpha=0,VelocityScaleZ=1,DampingAlpha=1,MaxLinearVelocity=100,MaxAngularVelocity=100,MaxLinearAcceleration=100,MaxAngularAcceleration=100,ExternalLinearDragV=(X=0,Y=0,Z=0),ExternalLinearVelocity=(X=0,Y=0,Z=0),ExternalAngularVelocity=(X=0,Y=0,Z=0))
   ) -> (Pose) {
    bOverrideWorldGravity=true
    overridePhysicsAsset=/Game/Characters/Player/Male/Medium/Bodies/M_MED_GuineaPig/Meshes/Parts/M_MED_GuineaPig_FaceAcc_Bandolier_Physics
    simulationSpace=BaseBoneSpace
    alphaBoolBlend {blendOption=Linear}
    alphaScaleBiasClamp {bInterpResult=true}
    baseBoneRef {boneName=spine_03}
    componentAppliedLinearAccClamp=<200 200 500>
    overrideWorldGravity {z=-980}
    simSpaceSettings {maxAngularAcceleration=100, maxAngularVelocity=100, maxLinearAcceleration=100, maxLinearVelocity=100}
   }
   n3 = Layered blend per bone(
    BasePose=Blend Poses by bool(
     BlendPose_0=Component To Local(ComponentPose=@n1)
     BlendPose_1=Component To Local(ComponentPose=@n2)
     BlendTime_0=0.13
     BlendTime_1=0.13
     bActiveValue=Get bIsPlayingEmote()
    ) {blendType=Sinusoidal}
    BlendPoses_0=Slot 'DefaultSlot'()
    BlendWeights_0=Get Custom Idle Is Playing Float() -> (Custom Idle Is Playing Float:float)
   ) {layerSetup [{branchFilters [
       {boneName=def_bandolier_root}
       {boneName=def_bandolier_rear_01}
       {boneName=def_bandolier_rear_02}
       {boneName=def_bandolier_rear_03}
       {boneName=def_bandolier_rear_04}
       {boneName=def_bandolier_rear_05}
       {boneName=def_bandolier_rear_06}
       {boneName=def_bandolier_rear_07}
       {boneName=def_bandolier_rear_08}
       {boneName=def_bandolier_rear_09}
       {boneName=def_bandolier_rear_10}
       {boneName=def_bandolier_rear_strap}
       {boneName=dyn_bag_rear_01}
       {boneName=dyn_bag_rear_02}
       {boneName=dyn_bag_rear_03}
       {boneName=dyn_bag_rear_04}
       {boneName=def_bandolier_front_01}
       {boneName=def_bandolier_front_02}
       {boneName=def_bandolier_front_03}
       {boneName=def_bandolier_front_04}
       {boneName=def_bandolier_front_05}
       {boneName=def_bandolier_front_06}
       {boneName=def_bandolier_front_07}
       {boneName=def_bandolier_front_08}
       {boneName=def_bandolier_front_09}
       {boneName=def_bandolier_front_10}
       {boneName=def_bandolier_front_strap}
       {boneName=dyn_bag_front_01}
       {boneName=dyn_bag_front_02}
       {boneName=dyn_bag_front_03}
       {boneName=dyn_bag_front_04}
      ]}]}
   Output Pose(Result=@n3) {layerGroup=None}
   n4 = Get BodySkelMesh() -> (BodySkelMesh:Skeletal Mesh Component Budgeted Object Reference)
   n5 = Control Rig!(
    Source=Layered blend per bone(
     BasePose=Copy Pose From Mesh() {bUseAttachedParent=true}
     BlendPoses_0=Copy Pose From Mesh(SourceMeshComponent=@n4) {bUseAttachedParent=true}
    ) {layerSetup [{branchFilters [{boneName=thigh_r}, {boneName=neck_01}]}]}
    Alpha
    RBAN Enabled:bool=Get bRigidBodyEnabled()
    crouch_pistol:bool=Get Crouch Pistol() -> (Crouch Pistol:bool)
    IsSkydiving:bool=Get bIsSkyDiving()
   ) -> (Pose) {
    bTransferInputCurves=false
    alphaBoolBlend {blendOption=Linear}
    controlRigAssetReference {blueprintRigClass=&/M_MED_GuineaPig_FaceAcc_CtrlRig.M_MED_GuineaPig_FaceAcc_CtrlRig_C}
   }
   Save cached pose 'PostDeformRig'(Pose=@n5) {cacheName=PostDeformRig}
  ]}
 EventGraph {nodes [
   Event BlueprintUpdateAnimation()
   > n1 = Sequence()
   > n2 = Gravity Override State Function DEPRECATED(
    CharacterPartAnimInstance=Self-Reference()
    GameTimeInSeconds=GetGameTimeInSeconds()
    GravityOverrideParams=Get Gravity Override Params()
    RBAN Params=Get RBAN Params()
   )
   > Set Gravity Override(Gravity Override=@n2)
   > Set RBAN Settings Out!(execute, RBAN Settings Out:RBANSetting Struct Structure=@n2.RBAN SettingsOut) -> (
    then
    Output_Get_ComponentLinearAccScale_28_#1:Vector=~
    Output_Get_ComponentLinearVelScale_31_#2:Vector=~
    Output_Get_SimSpaceSettings_34_#3:Sim Space Settings Structure
   )
   Branch(execute=@n1.then_1, Condition=Get bIsInFrontEnd())
   > Set Custom Idle Is Playing Float(Custom Idle Is Playing Float:float=GetCurveValue(self=GetOwnerFrontEndAnimInstance(), CurveName=Frontend_Idle_Bypass)) -> (Output_Get:float)
   n3 = Event OnCharacterCustomizationLoaded()
   > Set BodySkelMesh(BodySkelMesh:Skeletal Mesh Component Budgeted Object Reference=Get Array Item with EFort Custom Part Type(Array=@n3.CustomMeshParts)) -> (Output_Get:Skeletal Mesh Component Budgeted Object Reference)
   Event OnBeginPlayingEmote() {is=title K2Node_Event}
   Event OnEndPlayingEmote() {is=title K2Node_Event}
  ]}
]
variables [
 BodySkelMesh {type=SkeletalMeshComponentBudgeted}
 Crouch Pistol {type=boolean}
 Custom Idle Is Playing Float {type=number}
 Gravity Override {type=Vector}
 Gravity Override Params {
  type=GravityOverrideParamsStruct
  default {
   skydive_FloatingDown {gravityOverride=<0 200 200>, windAmplitude=<50 -50 100>, windFrequency=<0.5 2 2>, windOffset=<0 -200 200>}
   ziplining {gravityOverride=<-1000 0 -980>, windAmplitude=<200 0 0>, windFrequency=<2 0 0>, windOffset=<20 0 0>}
  }
 }
 RBAN Alpha {type=number, default=1}
 RBAN Params {
  type=RBANParamsStruct
  default {
   dBNO {simSpaceSettings {
     maxAngularAcceleration=10
     maxAngularVelocity=3
     maxLinearAcceleration=10
     maxLinearVelocity=10
     velocityScaleZ=0
     externalLinearDragV=<0 0.5 0>
    }}
   dBNO_Carried {simSpaceSettings {
     maxAngularAcceleration=10
     maxAngularVelocity=3
     maxLinearAcceleration=10
     maxLinearVelocity=10
     velocityScaleZ=0
     externalLinearDragV=<0 0.5 0>
    }}
   emoteOrMelee {simSpaceSettings {
     maxAngularAcceleration=10
     maxAngularVelocity=3
     maxLinearAcceleration=10
     maxLinearVelocity=10
     velocityScaleZ=0
     worldAlpha=0.1
     externalLinearDragV=<0 0 0>
    }}
   front_End {simSpaceSettings {
     maxAngularAcceleration=10
     maxAngularVelocity=3
     maxLinearAcceleration=10
     maxLinearVelocity=10
     velocityScaleZ=0
     externalLinearDragV=<0 0.5 0>
    }}
   onGround_Crouch_Moving {simSpaceSettings {
     maxAngularAcceleration=10
     maxAngularVelocity=3
     maxLinearAcceleration=10
     maxLinearVelocity=10
     velocityScaleZ=0
     externalLinearDragV=<0 0.5 0>
    }}
   onGround_Falling {simSpaceSettings {
     maxAngularAcceleration=10
     maxAngularVelocity=3
     maxLinearAcceleration=10
     maxLinearVelocity=10
     velocityScaleZ=0
     externalLinearDragV=<0 0.5 0>
    }}
   onGround_Moving {
    componentLinearAccScale=<0.3 0 2>
    componentLinearVelScale=<0.3 0 2>
    simSpaceSettings {
     maxAngularAcceleration=10
     maxAngularVelocity=3
     maxLinearAcceleration=10
     maxLinearVelocity=10
     velocityScaleZ=0
     externalLinearDragV=<0 0.5 0>
    }
   }
   onGround_Standing {
    componentLinearAccScale=<0.3 0.3 2.5>
    componentLinearVelScale=<0.3 0.3 2.5>
    simSpaceSettings {
     maxAngularAcceleration=10
     maxAngularVelocity=3
     maxLinearAcceleration=10
     maxLinearVelocity=10
     velocityScaleZ=0.5
     externalAngularVelocity=<0 0 0.5>
     externalLinearDragV=<0 0.5 0.5>
     externalLinearVelocity=<0 0 0.5>
    }
   }
   rail_Grinding {simSpaceSettings {
     maxAngularAcceleration=10
     maxAngularVelocity=3
     maxLinearAcceleration=10
     maxLinearVelocity=10
     velocityScaleZ=0
     worldAlpha=1
     externalLinearDragV=<0 0.5 0>
    }}
   riding_Moving {simSpaceSettings {
     maxAngularAcceleration=10
     maxAngularVelocity=3
     maxLinearAcceleration=10
     maxLinearVelocity=10
     velocityScaleZ=0
     externalLinearDragV=<0 0.5 0>
    }}
   riding_Moving_Targeting {simSpaceSettings {
     maxAngularAcceleration=10
     maxAngularVelocity=3
     maxLinearAcceleration=10
     maxLinearVelocity=10
     velocityScaleZ=0
     externalLinearDragV=<0 0.5 0>
    }}
   riding_Standing {simSpaceSettings {
     maxAngularAcceleration=10
     maxAngularVelocity=3
     maxLinearAcceleration=10
     maxLinearVelocity=10
     velocityScaleZ=0
     externalLinearDragV=<0 0.5 0>
    }}
   skydive_DivingDown {simSpaceSettings {
     maxAngularAcceleration=10
     maxAngularVelocity=3
     maxLinearAcceleration=10
     maxLinearVelocity=10
     velocityScaleZ=0
     externalLinearDragV=<0 0.5 0>
    }}
   skydive_FloatingDown {
    componentLinearAccScale=<0.1 0.1 0.1>
    componentLinearVelScale=<0.1 0.1 0.1>
    simSpaceSettings {
     maxAngularAcceleration=2
     maxAngularVelocity=2
     maxLinearAcceleration=2
     maxLinearVelocity=2
     velocityScaleZ=0.5
     worldAlpha=0.05
     externalAngularVelocity=<0.05 0.05 0.05>
     externalLinearDragV=<0 0.05 0>
     externalLinearVelocity=<0.05 0.05 0.05>
    }
   }
   skydive_GoingUp {simSpaceSettings {
     maxAngularAcceleration=10
     maxAngularVelocity=3
     maxLinearAcceleration=10
     maxLinearVelocity=10
     velocityScaleZ=0
     externalLinearDragV=<0 0.5 0>
    }}
   skydive_Parachuting {
    componentLinearAccScale=<0 0.1 0.3>
    componentLinearVelScale=<0 0.1 0.3>
    simSpaceSettings {
     maxAngularAcceleration=10
     maxAngularVelocity=3
     maxLinearAcceleration=10
     maxLinearVelocity=10
     velocityScaleZ=0
     externalLinearDragV=<0 0.5 0>
    }
   }
   skydive_Surfing {simSpaceSettings {
     maxAngularAcceleration=10
     maxAngularVelocity=3
     maxLinearAcceleration=10
     maxLinearVelocity=10
     velocityScaleZ=0
     externalLinearDragV=<0 0.5 0>
    }}
   swimming {simSpaceSettings {
     maxAngularAcceleration=10
     maxAngularVelocity=3
     maxLinearAcceleration=10
     maxLinearVelocity=10
     velocityScaleZ=0
     externalLinearDragV=<0 0.5 0>
    }}
   tactical_Sprint {simSpaceSettings {
     maxAngularAcceleration=10
     maxAngularVelocity=3
     maxLinearAcceleration=10
     maxLinearVelocity=10
     velocityScaleZ=0
     externalLinearDragV=<0 0.5 0>
    }}
   ziplining {simSpaceSettings {
     maxAngularAcceleration=10
     maxAngularVelocity=3
     maxLinearAcceleration=10
     maxLinearVelocity=10
     velocityScaleZ=0
     externalLinearDragV=<0 0.5 0>
    }}
  }
 }
 RBAN Settings Out {
  type=RBANSettingStruct
  default {
   componentLinearAccScale=<0 0 0>
   componentLinearVelScale=<0 0 0>
   simSpaceSettings {
    dampingAlpha=1
    maxAngularAcceleration=1000
    maxAngularVelocity=3
    maxLinearAcceleration=300
    maxLinearVelocity=700
    velocityScaleZ=0
    worldAlpha=1
    externalAngularVelocity=<0 0 0>
    externalLinearDragV=<0 0.5 0>
    externalLinearVelocity=<0 0 0>
   }
  }
 }
 RBAN Spline {type=boolean}
]
